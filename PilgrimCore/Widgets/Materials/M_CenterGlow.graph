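# Translucent UI material for a center glow in PilgrimCore widgets, built from two radial gradient backgrounds and two tint colors.
# Each background is an exponential radial gradient with a radius interpolated between start and stop parameters, then faded through lerps.
# Two vector parameters supply the layered color blend, and their alphas become per-background opacities.
# Named reroutes share the Animation scalar, opacities, backgrounds and color across the graph.
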Material /PilgrimCore/Widgets/Materials/M_CenterGlow
graph {
 properties {bAutomaticallySetUsageInEditor=false, blendMode=BLEND_Translucent, materialDomain=MD_UI, usedShadingModels=Default Lit}
 nodes [
  n1 = MaterialFunctionCall(
   Radius=LinearInterpolate(
    A=ScalarParameter() {defaultValue=0.4, parameterName=RadiusStartBackground1}
    B=ScalarParameter() {defaultValue=0.6, parameterName=RadiusStopBackground1}
    Alpha=NamedRerouteUsage()
   )
   Density=Constant() {r=1.5}
  ) {materialFunction=&RadialGradientExponential}
  n2 = NamedRerouteUsage()
  NamedRerouteDeclaration(None=LinearInterpolate(B=LinearInterpolate(B=@n1.RadialGradientExponential, Alpha=@n2), Alpha=NamedRerouteUsage())) {name=Background1, nodeColor=<rgba 0 0.3672 1 1>}
  n3 = LinearInterpolate(
   A=ScalarParameter() {defaultValue=0.2, parameterName=RadiusStartBackground2}
   B=ScalarParameter() {defaultValue=0.3, parameterName=RadiusStopBackground2}
   Alpha=NamedRerouteUsage()
  )
  n4 = Constant() {r=0.8}
  NamedRerouteDeclaration(None=LinearInterpolate(
    B=MaterialFunctionCall(Radius=@n3, Density=@n4) {materialFunction=&RadialGradientExponential}.RadialGradientExponential
    Alpha=NamedRerouteUsage()
   )) {name=Background2, nodeColor=<rgba 0.6875 1 0 1>}
  n5 = VectorParameter() {parameterName=Param_2, defaultValue=<rgba 0 0.0337 1 1>}
  n6 = NamedRerouteUsage()
  n7 = VectorParameter() {parameterName=Param_1, defaultValue=<rgba 1 0.2376 0.2605 1>}
  NamedRerouteDeclaration(None=LinearInterpolate(A=LinearInterpolate(B=@n5.RGB, Alpha=@n6), B=@n7.RGB, Alpha=NamedRerouteUsage())) {name=Color, nodeColor=<rgba 0.9922 0 1 1>}
  NamedRerouteDeclaration(None=ScalarParameter() {defaultValue=1, parameterName=Animation}) {name=Animation, nodeColor=<rgba 0.8281 0 1 1>}
  NamedRerouteDeclaration(None=@n7.A) {name=OpacityBg2, nodeColor=<rgba 0 0.4609 1 1>}
  NamedRerouteDeclaration(None=@n5.A) {name=OpacityBg1, nodeColor=<rgba 0.7812 1 0 1>}
  NamedRerouteUsage()
  NamedRerouteUsage()
 ]
}
