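# Translucent UI material for signed-distance-field text that colors glyphs along a three-color gradient driven by a Range parameter.
# A second three-color lerp starts from that gradient and moves into outline colors, and a reroute-masked lerp blends the two.
# An eased, antialiased glow and a thickness-controlled smoothstep outline, both from font signed distance, combine through Max into a fill mask.
# X/Y offset parameters are scaled by viewport scale and added to world position.
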Material /PilgrimCore/Widgets/Materials/M_Pilgrim_FontRange
graph {
 properties {
  bAutomaticallySetUsageInEditor=false
  bUsedWithStaticMesh=false
  blendMode=BLEND_Translucent
  materialDomain=MD_UI
  usedShadingModels=Default Lit
 }
 nodes [
  n1 = ScalarParameter() {parameterName=X_Offset}
  n2 = ScalarParameter() {parameterName=Y_Offset}
  Add(
   A=WorldPosition().XY
   B=Multiply(
    A=AppendVector(A=@n1, B=@n2)
    B=MaterialFunctionCall() {materialFunction=/FortUILibrary/MaterialFunctions/Basics/MF_UI_ViewportScale}.Viewport Scale Factor
   )
  )
  n3 = MaterialFunctionCall(
   A=VectorParameter() {parameterName=ColorStart, defaultValue=<rgba 1 0.9216 0 0>}.RGB
   B=VectorParameter() {parameterName=ColorMiddle, defaultValue=<rgba 1 0.9216 0 0.25>}.RGB
   C=VectorParameter() {parameterName=ColorEnd, defaultValue=<rgba 1 0 0 1>}.RGB
   Alpha=ScalarParameter() {parameterName=Range}
  ) {materialFunction=&Lerp_3Color}
  n4 = VectorParameter() {parameterName=ColorMiddleOutline, defaultValue=<rgba 0 1 0.4862 0.25>}
  n5 = VectorParameter() {parameterName=ColorEndOutline, defaultValue=<rgba 0.0193 0 1 1>}
  n6 = ScalarParameter() {parameterName=Range}
  NamedRerouteDeclaration(None=LinearInterpolate(
    A=@n3.Result
    B=MaterialFunctionCall(A=@n3.Result, B=@n4.RGB, C=@n5.RGB, Alpha=@n6) {materialFunction=&Lerp_3Color}.Result
    Alpha=NamedRerouteUsage()
   )) {name=Color, nodeColor=<rgba 0 0.5547 1 1>}
  n7 = FontSignedDistance()
  n8 = Saturate(None=MaterialFunctionCall(
    DistanceField=MaterialFunctionCall("Value (0-1)"=@n7.Signed Distance) {materialFunction=&MF_UI_EaseCurves}.EaseInQuint
    Threshold=Constant() {r=0.5}
    Smoothing=Constant() {r=1}
   ) {materialFunction=&MF_UI_AntiAliasing}.Outer)
  n9 = Divide(A=ScalarParameter() {defaultValue=0.75, parameterName=ThiknessFont}, B=ScalarParameter() {defaultValue=30, parameterName=ThiknessGlow})
  n10 = SmoothStep(Min=@n9, Max=Subtract(A=@n9) {constB=0.02}, Value=OneMinus(None=@n7.Signed Distance)) {constMax=0.06, constMin=0.01}
  NamedRerouteDeclaration(None=Max(
    A=NamedRerouteDeclaration(None=@n8) {name=Glow, nodeColor=<rgba 1 0 0.3281 1>}
    B=NamedRerouteDeclaration(None=@n10) {name=FontOutline, nodeColor=<rgba 1 0.8672 0 1>}
   )) {name=Fill, nodeColor=<rgba 0.5 1 0 1>}
  NamedRerouteUsage()
  NamedRerouteUsage()
  VectorParameter() {parameterName=ColorStartOutline, defaultValue=<rgba 1 0 0.7366 1>}
 ]
}
