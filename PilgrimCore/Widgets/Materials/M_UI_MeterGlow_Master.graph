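# Translucent UI master material that draws a soft, glowing circular progress arc for a meter.
# An angular mask, built from a Progress-driven edge plus a start-point falloff, multiplies a blurred SDF circle-outline ring.
# That combined mask blends two colors, with Overdrive crossfading to an alternate color pair.
# Opacity applies TaperGlow to the mask, uses a tiny-exponent power of Progress to hide it at zero, then scales by Alpha_Final.
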Material /PilgrimCore/Widgets/Materials/M_UI_MeterGlow_Master
graph {
 properties {
  bUsedWithStaticMesh=false
  blendMode=BLEND_Translucent
  materialDomain=MD_UI
  pixelDepthOffsetMode=PDOM_Legacy
  usedShadingModels=Default Lit
 }
 nodes [
  n1 = MaterialFunctionCall!(Vector or UVs) {materialFunction=/FortUILibrary/MaterialFunctions/Basics/MF_UI_CheaperVectorToRadial}
  n2 = OneMinus(None=@n1.Vector Converted to Angle)
  n3 = ScalarParameter() {defaultValue=0.25, parameterName=Progress}
  n4 = Multiply(A=Saturate(None=SmoothStep(Value=@n2) {constMin=0.95}), B=Power(Base=LinearInterpolate(Alpha=@n3) {constA=1, constB=0}) {constExponent=0.05})
  n5 = ScalarParameter() {defaultValue=0.05, parameterName=ProgressOffset}
  n6 = SmoothStep(Min=OneMinus(None=Add(A=@n3, B=@n5)), Max=LinearInterpolate(Alpha=@n3) {constA=1, constB=0}, Value=@n1.Vector Converted to Angle) {constMin=0.5}
  n7 = SmoothStep(Value=MaterialFunctionCall(
    Size=ScalarParameter() {defaultValue=0.666, parameterName=RingSize}
    OutlineThickness=ScalarParameter() {defaultValue=0.05, parameterName=ProgressThickness}
   ) {materialFunction=&MF_UI_Circle}.SDFOutline) {constMax=0.25, constMin=-0.25}
  n8 = Multiply(
   A=NamedRerouteDeclaration(None=Add(A=@n4, B=@n6)) {name=Progress Mask, nodeColor=<rgba 0 0 0 1>}
   B=NamedRerouteDeclaration(None=OneMinus(None=@n7)) {name=Progress Ring Mask, nodeColor=<rgba 0 0 0 1>}
  )
  LinearInterpolate(
   A=LinearInterpolate(
    A=VectorParameter() {parameterName=Color_A, defaultValue=<rgba 0 0 1 1>}.RGB
    B=VectorParameter() {parameterName=Color_B, defaultValue=<rgba 1 0 0 1>}.RGB
    Alpha=@n8
   )
   B=LinearInterpolate(
    A=VectorParameter() {parameterName=Color_A_OverDrive, defaultValue=<rgba 0 0 1 1>}.RGB
    B=VectorParameter() {parameterName=Color_B_OverDrive, defaultValue=<rgba 1 0 0 1>}.RGB
    Alpha=@n8
   )
   Alpha=ScalarParameter() {parameterName=Overdrive}
  )
  n9 = Power(Base=@n8, Exp=ScalarParameter() {defaultValue=1, parameterName=TaperGlow})
  n10 = Power(Base=ScalarParameter() {defaultValue=0.25, parameterName=Progress}) {constExponent=0.005}
  Multiply(A=Multiply(A=Saturate(None=@n9), B=Saturate(None=@n10)), B=ScalarParameter() {defaultValue=1, parameterName=Alpha_Final})
 ]
}
