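# Translucent UI master material that draws procedural overdrive particles, colored in red, orange and yellow, around a tip positioned by the Progress parameter.
# Two panning particle grids with separate speed and tiling generate randomly sized circles and boxes, with cell scale varying along V.
# The summed shapes are weighted by V and masked by a high-power sine band along U that Progress shifts, with AlphaCorrection and Alpha_Final scaling opacity.
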Material /PilgrimCore/Widgets/Materials/M_UI_OverdriveParticles_Master
graph {
 properties {
  bUsedWithStaticMesh=false
  blendMode=BLEND_Translucent
  materialDomain=MD_UI
  pixelDepthOffsetMode=PDOM_Legacy
  usedShadingModels=Default Lit
 }
 nodes [
  n1 = ComponentMask(None=TextureCoordinate()) {g=true}
  n2 = Multiply(A=@n1, B=ScalarParameter() {defaultValue=2, parameterName=AlphaCorrection})
  n3 = Multiply(A=@n2, B=NamedRerouteUsage())
  LinearInterpolate(
   A=LinearInterpolate(A=Constant3Vector() {constant=<rgba 1 0 0 0>}, B=Constant3Vector() {constant=<rgba 1 0.25 0 0>}, Alpha=@n2)
   B=LinearInterpolate(A=Constant3Vector() {constant=<rgba 1 1 0 1>}, B=Constant3Vector() {constant=<rgba 1 1 0.25 1>}, Alpha=@n2)
   Alpha=Power(Base=@n3)
  )
  n4 = Constant() {r=0.5}
  n5 = Add(A=TextureCoordinate()) {constB=0.625}
  n6 = Multiply(A=ScalarParameter() {defaultValue=0.5, parameterName=Progress}) {constB=-1}
  n7 = ComponentMask(None=Sine(None=Add(A=@n5, B=@n6)) {period=2}) {r=true}
  n8 = Multiply(
   A=Power(Base=SmoothStep(Min=@n4, Value=@n7), Exp=ScalarParameter() {defaultValue=322, parameterName=OverdriveTip_Power}) {constExponent=16}
   B=ScalarParameter() {defaultValue=2, parameterName=AlphaCorrection}
  )
  Multiply(A=Multiply(A=@n3, B=Saturate(None=@n8)), B=ScalarParameter() {defaultValue=1, parameterName=Alpha_Final})
  n9 = Panner(
   Coordinate=MaterialFunctionCall() {materialFunction=&MF_UI_UVScaling}.ScaleX
   Speed=AppendVector(
    A=ScalarParameter() {defaultValue=0.05, parameterName=Set1_X_PanSpeed}
    B=ScalarParameter() {defaultValue=0.5, parameterName=Set1_Y_PanSpeed}
   )
  ) {speedY=0.5}
  n10 = ScalarParameter() {defaultValue=24, parameterName=Set1_Tile}
  n11 = NamedRerouteUsage()
  n12 = MaterialFunctionCall!(UV=@n9, Global Offset, Global Rotation, Tiling=@n10, Cell Offset=Constant(), Cell Scale=@n11, Cell Rotation, Random Seed=Constant() {r=0.73}) {materialFunction=/FortUILibrary/MaterialFunctions/Basics/MF_UI_ParticleGrid}
  n13 = Constant() {r=0.7}
  n14 = Multiply(
   A=@n12.Random Cell Values
   B=Power(Base=MaterialFunctionCall(UV=@n9, Tiling=@n10, Cell Scale=@n11, Random Seed=@n13) {materialFunction=&MF_UI_ParticleGrid}.Random Cell Values) {constExponent=8}
  )
  n15 = MaterialFunctionCall(UVs=@n12.Cells UV, Size=@n14) {materialFunction=&MF_UI_Circle}
  n16 = Panner(
   Coordinate=MaterialFunctionCall() {materialFunction=&MF_UI_UVScaling}.ScaleX
   Speed=AppendVector(A=ScalarParameter() {defaultValue=-0.06, parameterName=Set2_X_PanSpeed}, B=ScalarParameter() {defaultValue=1, parameterName=Set2_Y_PanSpeed})
  ) {speedY=0.5}
  n17 = ScalarParameter() {defaultValue=32, parameterName=Set2_Tile}
  n18 = NamedRerouteUsage()
  n19 = MaterialFunctionCall(UV=@n16, Tiling=@n17, Cell Offset=Constant(), Cell Scale=@n18, Random Seed=Constant() {r=1.0615}) {materialFunction=&MF_UI_ParticleGrid}
  n20 = Constant() {r=1.7633}
  n21 = Multiply(
   A=@n19.Random Cell Values
   B=Power(Base=MaterialFunctionCall(UV=@n16, Tiling=@n17, Cell Scale=@n18, Random Seed=@n20) {materialFunction=&MF_UI_ParticleGrid}.Random Cell Values) {constExponent=8}
  )
  n22 = MaterialFunctionCall(UVs=@n19.Cells UV, Size=@n21) {materialFunction=&MF_UI_Box}
  NamedRerouteDeclaration(None=Add(A=SmoothStep(Value=@n15.SDF) {constMax=-1}, B=SmoothStep(Value=@n22.SDF) {constMax=-1})) {name=Particles Alpha, nodeColor=<rgba 1 0.1875 0 1>}
  NamedRerouteDeclaration(None=LinearInterpolate(Alpha=@n1) {constA=0.8}) {name=Cell Scale, nodeColor=<rgba 0 1 0.7891 1>}
  SmoothStep(Value=MaterialFunctionCall(UVs=@n19.Cells UV, Size=@n21) {materialFunction=&MF_UI_Circle}.SDF) {constMax=-1}
  SmoothStep(Value=MaterialFunctionCall(UVs=@n12.Cells UV, Size=@n14) {materialFunction=&MF_UI_Box}.SDF) {constMax=-1}
 ]
}
