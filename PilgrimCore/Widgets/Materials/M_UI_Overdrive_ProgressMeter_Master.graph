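# Master translucent UI material for an Overdrive progress meter bar with configurable rounded corners.
# MF_UI_Box builds the bar and Progress-driven fill masks, with a soft SDF mask used for glow variants.
# Sine-based masks shape a pinched core and panning motion, blending falloff, core and empty colors and alpha.
# OverdriveActive swaps palettes and adds a tip highlight and timed flash, while IsPeripheralInstrument overrides the color.
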Material /PilgrimCore/Widgets/Materials/M_UI_Overdrive_ProgressMeter_Master
graph {
 properties {
  bUsedWithStaticMesh=false
  blendMode=BLEND_Translucent
  materialDomain=MD_UI
  pixelDepthOffsetMode=PDOM_Legacy
  usedShadingModels=Default Lit
 }
 nodes [
  Constant() {r=1}
  n1 = LinearInterpolate(B=NamedRerouteUsage(), Alpha=ScalarParameter() {parameterName=OverdriveActive}) {constA=1}
  n2 = LinearInterpolate(
   A=VectorParameter() {parameterName=Color_Progress_Falloff_A, defaultValue=<rgba 1 1 0 1>}.RGB
   B=VectorParameter() {parameterName=Color_Progress_Falloff_B, defaultValue=<rgba 0 1 1 1>}.RGB
   Alpha=@n1
  )
  n3 = LinearInterpolate(
   A=VectorParameter() {parameterName=Color_Progress_Falloff_A_Overdrive, defaultValue=<rgba 1 1 0 1>}.RGB
   B=VectorParameter() {parameterName=Color_Progress_Falloff_B_Overdrive, defaultValue=<rgba 0 1 1 1>}.RGB
   Alpha=@n1
  )
  n4 = ScalarParameter() {parameterName=OverdriveActive}
  n5 = LinearInterpolate(
   A=VectorParameter() {parameterName=Color_Progress_Core, defaultValue=<rgba 0 1 0 1>}.RGB
   B=VectorParameter() {parameterName=Color_Progress_Core_Ready, defaultValue=<rgba 0 1 0 1>}.RGB
   Alpha=ScalarParameter() {parameterName=OverdriveActive}
  )
  n6 = LinearInterpolate(A=LinearInterpolate(A=LinearInterpolate(A=@n2, B=@n3, Alpha=@n4), B=@n5, Alpha=NamedRerouteUsage()), B=@n5, Alpha=NamedRerouteUsage())
  n7 = ScalarParameter() {defaultValue=0.5, parameterName=Progress}
  n8 = ComponentMask(None=Add(A=TextureCoordinate(), B=OneMinus(None=@n7))) {r=true}
  n9 = ScalarParameter() {defaultValue=16, parameterName=OverdriveTip_Power}
  n10 = LinearInterpolate(
   A=@n6
   B=VectorParameter() {parameterName=Color_Tip, defaultValue=<rgba 1 1 0.5 1>}.RGB
   Alpha=Saturate(None=Power(Base=@n8, Exp=@n9) {constExponent=16})
  )
  n11 = ScalarParameter() {parameterName=OverdriveActive}
  n12 = Saturate(None=LinearInterpolate(A=NamedRerouteUsage(), B=NamedRerouteUsage(), Alpha=ScalarParameter() {parameterName=IsGlow}))
  LinearInterpolate(
   A=LinearInterpolate(A=VectorParameter() {parameterName=Color_Empty, defaultValue=<rgba 1 0 0 1>}.RGB, B=LinearInterpolate(A=@n6, B=@n10, Alpha=@n11), Alpha=@n12)
   B=VectorParameter() {parameterName=Color_PeriphalInstrument, defaultValue=<rgba 0.017 0.0184 0.026 1>}.RGB
   Alpha=ScalarParameter() {parameterName=IsPeripheralInstrument}
  )
  n13 = MaterialFunctionCall() {materialFunction=&LinearGradient}
  n14 = StaticSwitchParameter(True=LinearInterpolate(Alpha=OneMinus(None=@n13.UGradient)), False=ScalarParameter() {defaultValue=0.5, parameterName=Alpha_Empty}) {parameterName=BG Gradient}
  n15 = ScalarParameter() {defaultValue=1, parameterName=Alpha_Progress}
  n16 = NamedRerouteUsage()
  n17 = LinearInterpolate(Alpha=NamedRerouteUsage()) {constA=0.666}
  n18 = ScalarParameter() {parameterName=IsGlow}
  Multiply(
   A=Multiply(A=LinearInterpolate(A=@n14, B=@n15, Alpha=@n12), B=LinearInterpolate(A=@n16, B=@n17, Alpha=@n18))
   B=ScalarParameter() {defaultValue=1, parameterName=Alpha_Final}
  )
  n19 = NamedRerouteUsage()
  n20 = NamedRerouteUsage()
  n21 = ScalarParameter() {defaultValue=0.5, parameterName=Progress}
  n22 = Add(A=LinearInterpolate(B=LinearInterpolate(A=@n20, Alpha=@n21) {constB=0}, Alpha=ScalarParameter() {parameterName=IsGlow}), B=@n21)
  n23 = AppendVector(A=Multiply(A=OneMinus(None=@n22)) {constB=0.5}, B=Constant())
  n24 = StaticSwitch(
   True=ScalarParameter() {defaultValue=0.5, parameterName=CornerRadius}
   False=MaterialFunctionCall(
    R=ScalarParameter() {defaultValue=0.25, parameterName=1_Corner}
    G=ScalarParameter() {defaultValue=0.25, parameterName=2_Corner}
    B=ScalarParameter() {defaultValue=0.5, parameterName=3_Corner}
    A=ScalarParameter() {defaultValue=0.5, parameterName=4_Corner}
   ) {materialFunction=&AppendMany}.RGBA
   Value=StaticBoolParameter() {defaultValue=true, parameterName=UniformCorners}
  )
  n25 = MaterialFunctionCall(
   UVs=NamedRerouteDeclaration(None=Add(A=@n19, B=@n23)) {name=BoxUVs, nodeColor=<rgba 0 0 0 1>}
   Size=AppendVector(A=@n22, B=Constant() {r=1})
   CornerRoundness=@n24
   KeepAspectRatio=StaticBool() {value=true}
  ) {materialFunction=&MF_UI_Box}
  NamedRerouteDeclaration(None=@n25.Fill) {name=RoundedProgressMask, nodeColor=<rgba 0 0 0 1>}
  n26 = NamedRerouteUsage()
  n27 = Sine(None=Panner(Coordinate=@n26) {speedX=-0.666}) {period=1.5}
  n28 = ComponentMask(None=Sine(None=Panner(Coordinate=@n26) {speedX=-1}) {period=2}) {r=true}
  n29 = OneMinus(None=Multiply(A=ComponentMask(None=@n27) {r=true}, B=@n28))
  n30 = OneMinus(None=@n28)
  NamedRerouteDeclaration(None=Saturate(None=Multiply(A=@n29, B=@n30))) {name=MotionMask, nodeColor=<rgba 0 0 0 1>}
  NamedRerouteDeclaration(None=MaterialFunctionCall(Size=Constant() {r=1}, CornerRoundness=@n24, KeepAspectRatio=StaticBool() {value=true}) {materialFunction=&MF_UI_Box}.Fill) {name=BarMask, nodeColor=<rgba 0 0 0 1>}
  n31 = MaterialFunctionCall() {materialFunction=&GetUserInterfaceUV}
  n32 = MaterialFunctionCall(Float2=@n31.Pixel Size) {materialFunction=&BreakOutFloat2Components}
  n33 = NamedRerouteDeclaration(None=@n32.G) {name=PixelSize.Y, nodeColor=<rgba 0 0 0 1>}
  n34 = NamedRerouteDeclaration(None=@n32.R) {name=PixelSize.X, nodeColor=<rgba 0 0 0 1>}
  NamedRerouteDeclaration(None=Divide(A=Multiply(A=@n33) {constB=0.5}, B=Max(A=@n34) {constB=0.001})) {name=Ratio, nodeColor=<rgba 0 0 0 1>}
  NamedRerouteDeclaration(None=@n31.9-Slice UV) {name=9sliceuv, nodeColor=<rgba 0 0 0 1>}
  NamedRerouteDeclaration(None=@n31.Normalized UV) {name=NormalizedUV, nodeColor=<rgba 0 0 0 1>}
  NamedRerouteDeclaration(None=SmoothStep(Value=@n25.SDF) {constMax=-1}) {name=Soft Mask, nodeColor=<rgba 0 0 0 1>}
  n35 = Sine(None=Add(A=NamedRerouteUsage(), B=ScalarParameter() {parameterName=OffsetCore})) {period=2}
  n36 = ScalarParameter() {defaultValue=0.333, parameterName=Flash_Min}
  n37 = MaterialFunctionCall(Speed=ScalarParameter() {defaultValue=3, parameterName=ReadyFlashSpeed}) {materialFunction=&DebugTimeSine}
  n38 = TextureCoordinate()
  n39 = ScalarParameter() {defaultValue=0.5, parameterName=Progress}
  n40 = Sine(None=Add(A=Add(A=@n38) {constB=0.5}, B=Multiply(A=@n39) {constB=-1})) {period=2}
  n41 = LinearInterpolate(
   A=LinearInterpolate(
    A=ScalarParameter() {defaultValue=0.333, parameterName=CoreFalloffAmount}
    B=LinearInterpolate(A=@n36, Alpha=@n37.Result) {constA=0.333}
    Alpha=ScalarParameter() {parameterName=OverdriveActive}
   )
   Alpha=Power(Base=ComponentMask(None=@n40) {r=true}, Exp=ScalarParameter() {defaultValue=16, parameterName=OverdriveTip_Power}) {constExponent=16}
  ) {constA=0.333}
  n42 = Multiply(A=Multiply(A=ComponentMask(None=@n35) {g=true}, B=Saturate(None=@n41)), B=NamedRerouteUsage())
  n43 = ScalarParameter() {defaultValue=16, parameterName=CorePinch}
  NamedRerouteDeclaration(None=Saturate(None=Power(Base=@n42, Exp=@n43) {constExponent=16})) {name=PinchedCore, nodeColor=<rgba 0 0 0 1>}
  NamedRerouteDeclaration(None=@n42) {name=Color Mask, nodeColor=<rgba 0 0 0 1>}
 ]
}
