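# Master translucent UI material for Pilgrim box portraits, showing a hero portrait texture with adjustable scale and final alpha.
# Portrait UVs are scaled to fill and zoomed about center, and the texture alpha is clamped to a transparent border.
# An optional rounded-box SDF mask with anti-aliasing multiplies that alpha.
# Widget pixel size drives aspect-aware box sizing, and a static switch selects pixel-based or normalized corner radius.
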Material /PilgrimCore/Widgets/Materials/M_UI_PIligrim_BoxPortrait_Master
graph {
 properties {
  bUsedWithStaticMesh=false
  blendMode=BLEND_Translucent
  materialDomain=MD_UI
  pixelDepthOffsetMode=PDOM_Legacy
  usedShadingModels=Default Lit
 }
 nodes [
  n1 = NamedRerouteUsage()
  n2 = NamedRerouteUsage()
  Max(A=@n1, B=@n2)
  n3 = MaterialFunctionCall(
   UVs=MaterialFunctionCall() {materialFunction=&MF_UI_UVScaling}.Scaled To Fill
   Texture Scale=ScalarParameter() {defaultValue=1, parameterName=Scale}
  ) {materialFunction=&ScaleUVsByCenter}
  n4 = MaterialFunctionCall!(
   UVs=@n3.UVs
   Alpha=TextureSampleParameter2D(UVs=@n3.UVs) {parameterName=Portrait, texture=/Game/UI/Foundation/Textures/Icons/Heroes/Athena/Soldier/T-Soldier-HID-005-Athena-Commando-M}.A
  ) {materialFunction=/Game/UI/Foundation/Items/ClampAlphaToTransparentBorder}
  n5 = NamedRerouteUsage()
  Multiply(
   A=StaticSwitchParameter(True=Multiply(A=@n4.Result, B=@n5), False=@n4.Result) {parameterName=RoundedBoxMask}
   B=ScalarParameter() {defaultValue=1, parameterName=Alpha_Final}
  )
  n6 = MaterialFunctionCall(Float2=MaterialFunctionCall() {materialFunction=&GetUserInterfaceUV}.Pixel Size) {materialFunction=&BreakOutFloat2Components}
  n7 = NamedRerouteDeclaration(None=@n6.R) {name=PixelX, nodeColor=<rgba 1 0.0469 0 1>}
  n8 = NamedRerouteDeclaration(None=@n6.G) {name=PixelY, nodeColor=<rgba 0 0.1562 1 1>}
  n9 = Max(A=@n7, B=@n8)
  NamedRerouteDeclaration(None=Divide(A=Min(A=@n7, B=@n8), B=Max(A=@n9) {constB=0.001})) {name=Ratio, nodeColor=<rgba 1 0.0234 0 1>}
  n10 = ScalarParameter() {defaultValue=0.025, parameterName=Threshold}
  n11 = Multiply(A=@n10, B=NamedRerouteUsage())
  n12 = Subtract(A=Constant() {r=1}, B=LinearInterpolate(A=AppendVector(A=@n10, B=@n11), B=AppendVector(A=@n11, B=@n10), Alpha=NamedRerouteUsage()))
  n13 = ScalarParameter() {defaultValue=16, parameterName=CornerRadiusPixels}
  n14 = Min(A=@n1, B=@n2)
  n15 = StaticSwitchParameter(
   True=Divide(A=Multiply(A=@n13) {constA=-0.9746, constB=2}, B=Max(A=@n14) {constB=0.001})
   False=ScalarParameter() {defaultValue=0.25, parameterName=Corner_Radius}
  ) {parameterName=PixelRadius}
  n16 = StaticBool() {value=true}
  NamedRerouteDeclaration(None=MaterialFunctionCall(
    DistanceField=MaterialFunctionCall(Size=@n12, CornerRoundness=@n15, KeepAspectRatio=@n16) {materialFunction=&MF_UI_Box}.SDF
    Threshold=ScalarParameter() {defaultValue=0.025, parameterName=Threshold}
   ) {materialFunction=&MF_UI_AntiAliasing}.Inner) {name=Box Mask, nodeColor=<rgba 0 0 0 1>}
  NamedRerouteDeclaration(None=MaterialFunctionCall!(A=NamedRerouteUsage(), B=NamedRerouteUsage()) {materialFunction=/FortUILibrary/MaterialFunctions/Basics/MF_UI_BooleanOperations}.IsGreater) {name=YisLarger, nodeColor=<rgba 0.2422 0 1 1>}
  NamedRerouteUsage()
 ]
}
