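# Translucent UI master material for Pilgrim widget backing panels, with static switches for border, glow, fades, scan lines, dots and swipe.
# Rounded-box and gradient functions build the shape and fade masks, which combine with effect layers into Final Color and Final Alpha named reroutes.
# Sine waves on square-to-screen-space UVs generate dots and time-scrolled scan lines procedurally, without textures.
# Anim scalar parameters blend colors, alphas and scan-line amount between start and target values.
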Material /PilgrimCore/Widgets/Materials/M_UI_Pilgrim_BackingPanel_Master
graph {
 properties {
  bUsedWithStaticMesh=false
  blendMode=BLEND_Translucent
  materialDomain=MD_UI
  pixelDepthOffsetMode=PDOM_Legacy
  usedShadingModels=Default Lit
 }
 nodes [
  n1 = ScalarParameter() {defaultValue=0.5, group=FadeTreatment, parameterName=Fade_Rotation_CenterPoint_X, sliderMin=0.5, sortPriority=1}
  n2 = ScalarParameter() {defaultValue=0.5, group=FadeTreatment, parameterName=Fade_Rotation_CenterPoint_Y, sliderMin=0.5, sortPriority=1}
  n3 = ScalarParameter() {defaultValue=1, group=FadeTreatment, parameterName=Fade_Width, sortPriority=1}
  n4 = Subtract(
   A=MaterialFunctionCall!(
    UVs=MaterialFunctionCall() {materialFunction=&GetUserInterfaceUV}.Normalized UV
    Offset=ScalarParameter() {group=FadeTreatment, parameterName=Fade_Offset, sortPriority=1}
    RotationCenter=AppendVector(A=@n1, B=@n2)
    RotationAngle=ScalarParameter() {group=FadeTreatment, parameterName=Fade_Rotation, sortPriority=1}
   ) {materialFunction=/FortUILibrary/MaterialFunctions/Gradients/MF_UI_ReflectedGradient}.Result
   B=OneMinus(None=@n3)
  ) {constB=0.25}
  n5 = Divide(B=@n3) {constB=0.5}
  n6 = Saturate(None=Multiply(A=Multiply(A=@n4, B=@n5), B=ScalarParameter() {defaultValue=1, group=FadeTreatment, parameterName=Fade_Density, sortPriority=1}))
  n7 = ScalarParameter() {defaultValue=1, group=FadeTreatment, parameterName=Fade_Rate, sortPriority=1}
  NamedRerouteDeclaration(None=Saturate(None=Power(Base=@n6, Exp=@n7))) {name=Fade/Feather, nodeColor=<rgba 0.5703 0 1 1>}
  n8 = Saturate(None=MaterialFunctionCall(
    Size=Constant() {r=1}
    CornerRoundness=ScalarParameter() {defaultValue=0.125, group=BackingShape, parameterName=Corner_Radius, sortPriority=0}
   ) {materialFunction=&MF_UI_Box}.Fill)
  n9 = DotProduct(A=@n8, B=NamedRerouteUsage())
  NamedRerouteDeclaration(None=StaticSwitchParameter(True=Saturate(None=@n9), False=@n8) {parameterName=UseLinearFade}) {name=Backing Panel Shape, nodeColor=<rgba 0.2187 1 0 1>}
  n10 = NamedRerouteUsage()
  n11 = NamedRerouteUsage()
  NamedRerouteDeclaration(None=Saturate(None=DotProduct(A=@n10, B=@n11))) {name=Final Border Shape, nodeColor=<rgba 1 0.375 0 1>}
  NamedRerouteDeclaration(None=LinearInterpolate(
    A=ScalarParameter() {defaultValue=0.25, group=Alpha, parameterName=Border_Alpha_Start, sortPriority=2}
    B=ScalarParameter() {defaultValue=0.5, group=Alpha, parameterName=Border_Alpha}
    Alpha=ScalarParameter() {group=Anim, parameterName=Anim_BorderAlpha, sliderMax=1}
   )) {name=Border Alpha, nodeColor=<rgba 1 0.9141 0 1>}
  n12 = VectorParameter() {group=Color, parameterName=Base_Color, sortPriority=1, defaultValue=<rgba 1 0 0 1>}
  n13 = StaticSwitchParameter(
   True=LinearInterpolate(
    A=@n12.RGB
    B=VectorParameter() {group=Color, parameterName=Base_Color_Stop, sortPriority=1, defaultValue=<rgba 1 0 0 1>}.RGB
    Alpha=ScalarParameter() {parameterName=Anim_Color}
   )
   False=@n12.RGB
  ) {parameterName=AnimateColor}
  n14 = VectorParameter() {group=Color, parameterName=Scanlines_Color_Start, sortPriority=1, defaultValue=<rgba 0 1 0 1>}
  n15 = VectorParameter() {group=Color, parameterName=Scanlines_Color, sortPriority=1, defaultValue=<rgba 0 1 0 1>}
  n16 = ScalarParameter() {parameterName=Anim_Color}
  NamedRerouteDeclaration(None=LinearInterpolate(A=@n13, B=LinearInterpolate(A=@n14.RGB, B=@n15.RGB, Alpha=@n16), Alpha=NamedRerouteUsage())) {name=Scan Lines Color, nodeColor=<rgba 0.6406 1 0 1>}
  n17 = MaterialFunctionCall() {materialFunction=/FortUILibrary/Foundation/Art_ToolKit/Tools/MF_UI_SquareToScreenSpace}
  n18 = ScalarParameter() {defaultValue=512, group=Dots, parameterName=Dots_Size}
  n19 = Sine(None=Multiply(A=@n17.Y_Priority, B=@n18))
  n20 = ScalarParameter() {defaultValue=1.5, group=Dots, parameterName=Dots_Softness}
  n21 = Saturate(None=Multiply(A=ComponentMask(None=@n19) {g=true}, B=@n20))
  n22 = Sine(None=Multiply(A=@n17.X_Priority, B=@n18))
  n23 = Saturate(None=Multiply(A=ComponentMask(None=@n22) {r=true}, B=@n20))
  NamedRerouteDeclaration(None=Saturate(None=DotProduct(A=@n21, B=@n23))) {name=Dot Texture, nodeColor=<rgba 1 0 0.1641 1>}
  n24 = NamedRerouteUsage()
  n25 = ScalarParameter() {defaultValue=0.25, group=Dots, parameterName=DotTexture_Amount}
  NamedRerouteDeclaration(None=Saturate(None=Multiply(A=@n24, B=@n25))) {name=Dot Texture Alpha, nodeColor=<rgba 0 1 1 1>}
  n26 = StaticSwitchParameter(True=NamedRerouteUsage(), False=NamedRerouteUsage()) {group=Border, parameterName=IsBorder}
  n27 = Multiply(A=NamedRerouteUsage(), B=NamedRerouteUsage())
  n28 = NamedRerouteUsage()
  n29 = StaticSwitchParameter(True=Add(A=@n26, B=Add(A=@n27, B=@n28)), False=@n26) {group=ScanLines, parameterName=UseScanLinesFX}
  n30 = NamedRerouteUsage()
  n31 = StaticSwitchParameter(True=Saturate(None=Add(A=@n29, B=@n30)), False=@n29) {defaultValue=true, group=Dots, parameterName=UseDotTexture}
  n32 = Multiply(A=NamedRerouteUsage(), B=ScalarParameter() {group=FadeTreatment, parameterName=SwipeAmount})
  n33 = StaticSwitchParameter(True=Saturate(None=Add(A=@n31, B=@n32)), False=@n31) {group=Anim, parameterName=HasSwipe}
  n34 = Saturate(None=MaterialFunctionCall(RotationAngle=Constant() {r=0.25}) {materialFunction=&MF_UI_LinearGradient}.Result)
  n35 = StaticSwitchParameter(True=NamedRerouteUsage(), False=NamedRerouteUsage()) {group=Border, parameterName=IsBorder}
  n36 = DotProduct(A=NamedRerouteUsage(), B=@n35)
  n37 = StaticSwitchParameter(True=NamedRerouteDeclaration(None=Saturate(None=@n36)) {name=FadedEdgeShape, nodeColor=<rgba 0.1719 0 1 1>}, False=@n35) {group=FadeTreatment, parameterName=UseFadedEdges}
  n38 = DotProduct(A=@n37, B=OneMinus(None=@n34))
  n39 = StaticSwitchParameter(True=Saturate(None=DotProduct(A=@n37, B=@n34)), False=StaticSwitchParameter(True=Saturate(None=@n38), False=@n37) {parameterName=FadeTop}) {parameterName=FadeBottom}
  NamedRerouteDeclaration(None=Saturate(None=DotProduct(A=@n33, B=@n39))) {name=Final Alpha, nodeColor=<rgba 0.2188 0 1 1>}
  n40 = MaterialFunctionCall(
   UVs=MaterialFunctionCall() {materialFunction=&MF_UI_SquareToScreenSpace}.X_Priority
   "Rotation Angle (0-1)"=Constant() {r=0.25}
  ) {materialFunction=&CustomRotator}
  n41 = Multiply(A=Time(), B=ScalarParameter() {defaultValue=0.25, parameterName=SS_ScanLine_Speed})
  n42 = Multiply(A=ComponentMask(None=Subtract(A=@n40.Rotated Values, B=@n41)) {r=true}) {constB=0.5}
  n43 = ScalarParameter() {defaultValue=0.25, parameterName=SS_ScanLine_Width}
  n44 = Multiply(A=Subtract(A=Sine(None=@n42), B=OneMinus(None=@n43)) {constB=0.25}, B=Divide(B=@n43) {constB=0.5})
  n45 = ScalarParameter() {defaultValue=0.25, parameterName=SS_ScanLine_Amount}
  NamedRerouteDeclaration(None=Saturate(None=Multiply(A=@n44, B=@n45))) {name=ScreenSpace ScanLine, nodeColor=<rgba 0.0781 0 1 1>}
  n46 = Multiply(
   A=MaterialFunctionCall() {materialFunction=&MF_UI_SquareToScreenSpace}.Y_Priority
   B=ScalarParameter() {defaultValue=3, group=ScanLines, parameterName=ScanLines_SimpleScale}
  )
  n47 = Multiply(
   A=LinearInterpolate(Alpha=ScalarParameter() {defaultValue=1, group=ScanLines, parameterName=ScanLines_SwitchAnimDirection}) {constA=-1}
   B=ScalarParameter() {defaultValue=0.8, group=ScanLines, parameterName=ScanLines_Speed}
  ) {constB=-1}
  n48 = Multiply(A=Time(), B=@n47)
  n49 = ComponentMask(None=Add(A=@n46, B=@n48)) {g=true}
  n50 = ScalarParameter() {defaultValue=1, group=ScanLines, parameterName=ScanLines_Softness}
  n51 = Clamp(None=Multiply(A=Sine(None=@n49) {period=0.333}, B=@n50)) {minDefault=-1}
  n52 = Clamp(None=Multiply(A=Sine(None=@n49) {period=0.1}, B=@n50)) {minDefault=-1}
  n53 = Multiply(A=Sine(None=@n49) {period=0.1625}, B=@n50)
  n54 = Multiply(A=Sine(None=@n49) {period=0.2625}, B=@n50)
  n55 = DotProduct(
   A=LinearInterpolate(A=LinearInterpolate(A=@n51, Alpha=@n52), Alpha=Clamp(None=@n53) {minDefault=-1})
   B=OneMinus(None=Clamp(None=@n54) {minDefault=-1})
  )
  n56 = Multiply(A=@n47) {constB=-1}
  n57 = ComponentMask(None=Add(A=@n46, B=Multiply(A=@n48, B=@n56))) {g=true}
  n58 = Clamp(None=Multiply(A=Sine(None=@n57) {period=0.1}, B=@n50)) {minDefault=-1}
  NamedRerouteDeclaration(None=Saturate(None=DotProduct(A=@n55, B=@n58))) {name=ScanLines, nodeColor=<rgba 1 0 0.8672 1>}
  NamedRerouteDeclaration(None=LinearInterpolate(
    A=ScalarParameter() {defaultValue=0.25, group=Alpha, parameterName=Backing_Alpha_Start, sortPriority=2}
    B=ScalarParameter() {defaultValue=0.25, group=Alpha, parameterName=Backing_Alpha, sortPriority=2}
    Alpha=ScalarParameter() {defaultValue=1, group=Anim, parameterName=Anim_BackingAlpha, sliderMax=1}
   )) {name=Backing Alpha, nodeColor=<rgba 1 0.4453 0 1>}
  n59 = ScalarParameter() {defaultValue=0.25, group=Alpha, parameterName=ScanLines_Amount_Start}
  n60 = ScalarParameter() {defaultValue=0.5, group=Alpha, parameterName=ScanLines_Amount_Stop}
  n61 = ScalarParameter() {group=Anim, parameterName=Anim_ScanLines_Amount}
  NamedRerouteDeclaration(None=Saturate(None=LinearInterpolate(A=@n59, B=@n60, Alpha=@n61))) {name=Amount of ScanLines FX, nodeColor=<rgba 0 1 0.8125 1>}
  n62 = StaticSwitchParameter(
   True=LinearInterpolate(A=VectorParameter() {group=Color, parameterName=Glow_Falloff_Color, defaultValue=<rgba 0 0 1 1>}.RGB, B=@n13, Alpha=NamedRerouteUsage())
   False=@n13
  ) {group=BackingShape, parameterName=IsGlow}
  n63 = LinearInterpolate(
   A=VectorParameter() {group=Color, parameterName=EdgeFade_Color, sortPriority=1, defaultValue=<rgba 1 0 0 1>}.RGB
   B=@n13
   Alpha=NamedRerouteUsage()
  )
  NamedRerouteDeclaration(None=StaticSwitchParameter(True=NamedRerouteUsage(), False=StaticSwitchParameter(True=@n63, False=@n62) {group=FadeTreatment, parameterName=UseFadedEdges}) {group=ScanLines, parameterName=UseScanLinesFX}) {name=Final Color, nodeColor=<rgba 0.5469 1 0 1>}
  n64 = ScalarParameter() {defaultValue=1, group=LinearFade, parameterName=LinearFade_Start}
  n65 = ScalarParameter() {group=LinearFade, parameterName=LinearFade_Stop}
  n66 = LinearInterpolate(Alpha=ScalarParameter() {group=LinearFade, parameterName=LinearFade_Flip}) {constB=0.5}
  n67 = ScalarParameter() {group=LinearFade, parameterName=LinearFade_Offset}
  n68 = Saturate(None=Multiply(
    A=MaterialFunctionCall(GradientStart=@n64, GradientEnd=@n65, RotationAngle=@n66, GradientOffset=@n67) {materialFunction=&MF_UI_LinearGradient}.Result
    B=ScalarParameter() {defaultValue=1, group=LinearFade, parameterName=LinearFade_EdgeHardness}
   ))
  n69 = ScalarParameter() {defaultValue=1, group=LinearFade, parameterName=LinearFade_Rate}
  NamedRerouteDeclaration(None=Saturate(None=Power(Base=@n68, Exp=@n69))) {name=LinearFade, nodeColor=<rgba 0.3359 0 1 1>}
  n70 = ScalarParameter() {defaultValue=0.125, group=Border, parameterName=Border_Corner_Radius, sortPriority=0}
  n71 = MaterialFunctionCall(Size=OneMinus(None=@n70), CornerRoundness=@n70) {materialFunction=&MF_UI_Box}
  NamedRerouteDeclaration(None=Saturate(None=OneMinus(None=@n71.Fill))) {name=Inversed Base Border Shape, nodeColor=<rgba 0.7813 0 1 1>}
  NamedRerouteUsage()
  NamedRerouteUsage()
  Time()
 ]
}
