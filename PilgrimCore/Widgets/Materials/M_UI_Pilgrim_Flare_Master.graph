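# Translucent UI master material for a horizontal flare with scrolling noise, configurable colors and a final alpha control.
# Color blends parameter tints, with a static switch swapping one for a time-driven cosine rainbow.
# Two panning layers of blurred noise, one distorting the other's UVs, form left- and right-moving streaks.
# Static switches select simple, one-direction or right-only gradient variants, shaped by edge masks and a vertical falloff into the output mask.
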Material /PilgrimCore/Widgets/Materials/M_UI_Pilgrim_Flare_Master
graph {
 properties {
  bUsedWithStaticMesh=false
  blendMode=BLEND_Translucent
  materialDomain=MD_UI
  pixelDepthOffsetMode=PDOM_Legacy
  usedShadingModels=Default Lit
 }
 nodes [
  n1 = Constant() {r=0.5}
  n2 = MaterialFunctionCall() {materialFunction=&GetUserInterfaceUV}
  n3 = MaterialFunctionCall(Float2=Divide(A=Multiply(A=@n2.Pixel Size, B=@n2.Pixel Size), B=@n2.Pixel Size)) {materialFunction=&BreakOutFloat2Components}
  n4 = MaterialFunctionCall(X=@n3.R, Y=@n3.R, Z=@n3.R) {materialFunction=&MakeFloat3}
  n5 = Constant3Vector() {constant=<rgba 0 2 4 0>}
  n6 = Time()
  n7 = Cosine(None=Add(A=Add(A=@n4.Result, B=@n5), B=Multiply(A=@n6) {constB=3})) {desc=2 PI, period=6.2832}
  n8 = Add(A=Add(A=@n1, B=Multiply(A=@n1, B=@n7))) {constB=0.5}
  LinearInterpolate(
   A=LinearInterpolate(
    A=VectorParameter() {parameterName=Color_2, defaultValue=<rgba 0 0 1 1>}.RGB
    B=VectorParameter() {parameterName=Color_B, defaultValue=<rgba 0.75 1 1 1>}.RGB
    Alpha=NamedRerouteUsage()
   )
   B=StaticSwitchParameter(True=Saturate(None=@n8), False=VectorParameter() {parameterName=Color_A, defaultValue=<rgba 0 0 0 1>}.RGB) {defaultValue=true, parameterName=UseOscColors}
   Alpha=NamedRerouteUsage()
  )
  n9 = TextureCoordinate()
  n10 = SmoothStep(Max=ScalarParameter() {defaultValue=0.5, parameterName=VerticalFalloff}, Value=ComponentMask(None=Sine(None=@n9) {period=2}) {g=true}) {constMax=0.25}
  n11 = MaterialFunctionCall!(UVs, Offset, RotationCenter, RotationAngle=ScalarParameter() {defaultValue=0.5}) {materialFunction=/FortUILibrary/MaterialFunctions/Gradients/MF_UI_ReflectedGradient}
  n12 = Constant()
  n13 = Constant() {r=3}
  n14 = Saturate(None=Power(Base=MaterialFunctionCall!(X=@n11.Result, Tangent 0=@n12, Tangent 1=@n13) {materialFunction=/Engine/Functions/Engine_MaterialFunctions02/Gradients/SmoothCurve}.Result) {constExponent=2.5})
  n15 = LinearInterpolate(
   A=StaticSwitchParameter(True=Constant(), False=NamedRerouteUsage()) {parameterName=IsSimpleGradient}
   Alpha=StaticSwitchParameter(True=NamedRerouteUsage(), False=NamedRerouteDeclaration(None=@n14) {name=Gradient, nodeColor=<rgba 1 0 0.8672 1>}) {parameterName=IsOneDirection}
  )
  n16 = Saturate(None=Multiply(A=Saturate(None=@n10), B=NamedRerouteDeclaration(None=@n15) {name=GradientType, nodeColor=<rgba 0 1 0.9766 1>}))
  Multiply(
   A=NamedRerouteDeclaration(None=NamedRerouteDeclaration(None=@n16) {name=FinalMask, nodeColor=<rgba 0.8516 1 0 1>}) {name=Output, nodeColor=<rgba 0 1 0.8828 1>}
   B=ScalarParameter() {defaultValue=1, parameterName=Alpha_Final}
  )
  n17 = Multiply(A=TextureCoordinate() {uTiling=0.05, vTiling=0.5}, B=ScalarParameter() {defaultValue=1, parameterName=NoiseSet1_Tile_Alter})
  n18 = Panner(Coordinate=@n17) {speedX=0.5, speedY=-0.025}
  n19 = Multiply(A=TextureCoordinate() {uTiling=0.25, vTiling=1.5}, B=ScalarParameter() {defaultValue=1, parameterName=NoiseSet2_Tile_Alter})
  n20 = Panner(Coordinate=@n19) {speedX=0.3, speedY=0.013}
  n21 = NamedRerouteUsage()
  n22 = Constant() {r=0.125}
  n23 = AppendVector(
   A=Multiply(
    A=TextureSample(UVs=@n20, Tex=@n21) {samplerType=SAMPLERTYPE_Grayscale, texture=/FortUILibrary/UIKit/Buttons/Textures/LowResBlurredNoise}.R
    B=@n22
   )
   B=Constant()
  )
  NamedRerouteDeclaration(None=TextureSample(UVs=Add(A=@n18, B=@n23), Tex=@n21) {samplerType=SAMPLERTYPE_Grayscale, texture=&LowResBlurredNoise}.R) {name=LeftDirectionNoise, nodeColor=<rgba 1 0 0.7734 1>}
  n24 = NamedRerouteUsage()
  n25 = TextureSample(UVs=Panner(Coordinate=@n19) {speedX=-0.3, speedY=0.013}, Tex=@n24) {samplerType=SAMPLERTYPE_Grayscale, texture=&LowResBlurredNoise}
  n26 = Add(A=Panner(Coordinate=@n17) {speedX=-0.5, speedY=-0.025}, B=AppendVector(A=Multiply(A=@n25.R, B=@n22), B=Constant()))
  NamedRerouteDeclaration(None=Multiply(
    A=TextureSample(UVs=@n26, Tex=@n24) {samplerType=SAMPLERTYPE_Grayscale, texture=&LowResBlurredNoise}.R
    B=OneMinus(None=@n25.R)
   )) {name=RightDirectionNoise, nodeColor=<rgba 0 0.0625 1 1>}
  n27 = NamedRerouteUsage()
  n28 = StaticSwitchParameter(
   True=Power(Base=@n27, Exp=ScalarParameter() {defaultValue=1, parameterName=SimpleGradientFalloffRate})
   False=Multiply(A=@n27, B=NamedRerouteUsage())
  ) {parameterName=IsSimpleGradient}
  n29 = NamedRerouteUsage()
  n30 = NamedRerouteUsage()
  n31 = StaticSwitchParameter(
   True=Power(Base=@n30, Exp=ScalarParameter() {defaultValue=1, parameterName=SimpleGradientFalloffRate})
   False=Multiply(A=@n30, B=NamedRerouteUsage())
  ) {parameterName=IsSimpleGradient}
  n32 = NamedRerouteUsage()
  NamedRerouteDeclaration(None=StaticSwitchParameter(True=Multiply(A=@n28, B=@n29), False=Multiply(A=@n31, B=@n32)) {parameterName=IsRightOnly}) {name=DirectionLogic, nodeColor=<rgba 0 1 0.0625 1>}
  n33 = ComponentMask(None=TextureCoordinate()) {r=true}
  NamedRerouteDeclaration(None=SmoothStep(Value=@n33) {constMax=0.125}) {name=RightMaskOD_EdgeMask, nodeColor=<rgba 0 1 0.25 1>}
  n34 = OneMinus(None=@n33)
  NamedRerouteDeclaration(None=SmoothStep(Value=@n34) {constMax=0.125}) {name=LeftMaskOD_EdgeMask, nodeColor=<rgba 0 1 0.6719 1>}
  n35 = SmoothStep(Value=@n33) {constMax=0.5}
  n36 = Constant() {r=1}
  NamedRerouteDeclaration(None=LinearInterpolate(
    A=@n35
    Alpha=MaterialFunctionCall!(A=@n35, B=@n36) {materialFunction=/FortUILibrary/MaterialFunctions/Basics/MF_UI_BooleanOperations}.IsGreaterOrEqual
   ) {constB=0}) {name=LeftDirectionMask, nodeColor=<rgba 1 0 0.8438 1>}
  n37 = SmoothStep(Value=@n34) {constMax=0.5}
  NamedRerouteDeclaration(None=LinearInterpolate(
    A=@n37
    Alpha=MaterialFunctionCall(A=@n37, B=@n36) {materialFunction=&MF_UI_BooleanOperations}.IsGreaterOrEqual
   ) {constB=0}) {name=RightDirectionMask, nodeColor=<rgba 0 1 0.8359 1>}
  n38 = NamedRerouteUsage()
  n39 = Add(A=Multiply(A=NamedRerouteUsage(), B=NamedRerouteUsage()), B=Multiply(A=NamedRerouteUsage(), B=NamedRerouteUsage()))
  NamedRerouteDeclaration(None=MaterialFunctionCall(X=StaticSwitchParameter(True=@n38, False=@n39) {parameterName=IsOneDirection}, Tangent 0=Constant() {r=0.125}, Tangent 1=Constant() {r=6}) {materialFunction=&SmoothCurve}.Result) {name=Noise, nodeColor=<rgba 0.2656 0 1 1>}
  NamedRerouteDeclaration(None=TextureObject() {samplerType=SAMPLERTYPE_Grayscale, texture=&LowResBlurredNoise}) {name=Texture, nodeColor=<rgba 1 0.3281 0 1>}
  NamedRerouteDeclaration(None=LinearInterpolate(
    A=@n34
    B=MaterialFunctionCall("Value (0-1)"=@n34) {materialFunction=&MF_UI_EaseCurves}.EaseInQuint
    Alpha=ScalarParameter() {parameterName=UseEaseCurve}
   )) {name=RightDireciton, nodeColor=<rgba 1 0.7031 0 1>}
  NamedRerouteDeclaration(None=@n33) {name=LeftDirection, nodeColor=<rgba 0.8516 0 1 1>}
 ]
}
