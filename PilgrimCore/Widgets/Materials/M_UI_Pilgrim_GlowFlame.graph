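# Translucent UI material for an animated glow and flame effect, with a GlowToFire scalar that by name blends between the two looks.
# Glow layers counter-panning radial-mapped noise to warp a radial gradient, plus a soft box falloff, tinted by color parameters.
# Flame thresholds upward-gradient, noise-distorted scrolling noise, with a rounded MF_UI_Box mask and three-color gradient.
# Named reroutes carry each layer's color and opacity into final Color and Opacity lerps.
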Material /PilgrimCore/Widgets/Materials/M_UI_Pilgrim_GlowFlame
graph {
 properties {
  bAutomaticallySetUsageInEditor=false
  bUsedWithStaticMesh=false
  blendMode=BLEND_Translucent
  materialDomain=MD_UI
  usedShadingModels=Default Lit
 }
 nodes [
  n1 = NamedRerouteUsage()
  n2 = NamedRerouteUsage()
  NamedRerouteDeclaration(None=Saturate(None=Min(A=@n1, B=@n2))) {name=FlameOpacity, nodeColor=<rgba 1 0 0.8203 1>}
  n3 = Constant() {r=0.5}
  n4 = Constant() {r=3.4567}
  n5 = ComponentMask(None=MaterialFunctionCall!(In) {materialFunction=/Game/UI/Frontend/Rewards/materials/VectortoRadialFix}.Vector to Radial Reversed) {r=true}
  n6 = Add(A=Multiply(A=Time(), B=Constant() {r=1}) {constB=8}, B=Constant())
  n7 = MaterialFunctionCall() {materialFunction=&VectorToRadialValue}
  n8 = LinearInterpolate(
   A=TextureSample(UVs=Panner(Coordinate=@n5, Time=@n6) {speedX=0.025, speedY=0.05}) {samplerType=SAMPLERTYPE_Grayscale, texture=&LowResBlurredNoise}.R
   B=TextureSample(UVs=Panner(Coordinate=@n7.Vector Converted to Angle, Time=@n6) {speedX=-0.025, speedY=-0.05}) {samplerType=SAMPLERTYPE_Grayscale, texture=&LowResBlurredNoise}.R
   Alpha=MaterialFunctionCall() {materialFunction=&VectortoRadialFix}.Vector to Reversed Mask
  )
  n9 = ComponentMask(None=MaterialFunctionCall() {materialFunction=&VectortoRadialFix}.Vector to Radial Reversed) {r=true}
  n10 = MaterialFunctionCall() {materialFunction=&VectorToRadialValue}
  n11 = LinearInterpolate(
   A=TextureSample(UVs=Panner(Coordinate=@n9, Time=@n6) {speedX=-0.025, speedY=-0.05}) {samplerType=SAMPLERTYPE_Grayscale, texture=&LowResBlurredNoise}.R
   B=TextureSample(UVs=Panner(Coordinate=@n10.Vector Converted to Angle, Time=@n6) {speedX=0.025, speedY=0.05}) {samplerType=SAMPLERTYPE_Grayscale, texture=&LowResBlurredNoise}.R
   Alpha=MaterialFunctionCall() {materialFunction=&VectortoRadialFix}.Vector to Reversed Mask
  )
  n12 = Multiply(A=ComponentMask(None=Add(A=@n8, B=@n11)) {r=true}, B=Constant() {r=1}) {constB=0.5}
  n13 = Abs(None=Add(A=TextureCoordinate() {uTiling=1.125}, B=Constant2Vector() {g=-0.5, r=-0.5625}) {constB=2})
  n14 = Constant() {r=0.05}
  n15 = MaterialFunctionCall(Vector 2=Max(A=Subtract(A=@n13, B=@n14)) {constB=0}) {materialFunction=/Engine/Functions/Engine_MaterialFunctions02/Utility/VectorLength}
  n16 = Constant()
  n17 = Constant() {r=0.4}
  n18 = Add(
   A=LinearInterpolate(
    B=MaterialFunctionCall(Radius=@n3, Density=@n4) {materialFunction=&RadialGradientExponential}.RadialGradientExponential
    Alpha=TextureSample(UVs=@n12) {samplerType=SAMPLERTYPE_Grayscale, texture=&LowResBlurredNoise}.R
   )
   B=OneMinus(None=MaterialFunctionCall(Alpha=@n15.V2 Length, Min=@n16, Max=@n17) {materialFunction=&SmoothStep}.Result)
  )
  NamedRerouteDeclaration(None=Multiply(A=@n18, B=Constant() {r=1})) {name=GlowOpacity, nodeColor=<rgba 0 1 0.8594 1>}
  n19 = VectorParameter() {parameterName=ColorGlow1, defaultValue=<rgba 0.8388 0.0552 0 1>}
  n20 = VectorParameter() {parameterName=GlowColor2, defaultValue=<rgba 0.8388 0.4564 0.0723 1>}
  NamedRerouteDeclaration(None=Add(A=LinearInterpolate(A=@n19.RGB, B=@n20.RGB), B=Multiply(A=@n20.RGB, B=@n18))) {name=GlowColor, nodeColor=<rgba 1 0.5391 0 1>}
  NamedRerouteDeclaration(None=ScalarParameter() {defaultValue=1, parameterName=GlowToFire}) {name=GlowToFire, nodeColor=<rgba 0.1016 1 0 1>}
  n21 = MaterialFunctionCall(Base=NamedRerouteUsage(), Blend=Multiply(A=NamedRerouteUsage())) {materialFunction=&Blend_Overlay}
  n22 = LinearInterpolate(
   A=VectorParameter() {parameterName=ColorGlow3, defaultValue=<rgba 1 0 0.2307 1>}.RGB
   B=Constant3Vector() {constant=<rgba 1 0.8986 0 1>}
   Alpha=MaterialFunctionCall(Radius=Constant() {r=0.35}) {materialFunction=&RadialGradientExponential}.RadialGradientExponential
  )
  NamedRerouteDeclaration(None=LinearInterpolate(A=NamedRerouteUsage(), B=LinearInterpolate(A=@n21.Result, B=@n22), Alpha=NamedRerouteUsage())) {name=Color, nodeColor=<rgba 1 0 0.4219 1>}
  n23 = VectorParameter() {parameterName=ColorFlame1, defaultValue=<rgba 1 0.3066 0 1>}
  n24 = VectorParameter() {parameterName=ColorFlame2, defaultValue=<rgba 1 0.9446 0.0885 1>}
  n25 = Constant()
  n26 = Constant() {r=-0.5}
  n27 = MaterialFunctionCall(UVs=Add(A=TextureCoordinate(), B=AppendVector(A=@n25, B=@n26)), Radius=Constant() {r=0.4}) {materialFunction=&RadialGradientExponential}
  n28 = MaterialFunctionCall() {materialFunction=&LinearGradient}
  NamedRerouteDeclaration(None=LinearInterpolate(
    A=VectorParameter() {parameterName=ColorFlame3, defaultValue=<rgba 0.9219 0.054 0 1>}.RGB
    B=LinearInterpolate(A=@n23.RGB, B=@n24.RGB, Alpha=@n27.RadialGradientExponential)
    Alpha=SmoothStep(Value=@n28.VGradient) {constMin=0.3}
   )) {name=ColorFlame, nodeColor=<rgba 1 0.4688 0 1>}
  n29 = NamedRerouteUsage()
  n30 = NamedRerouteUsage()
  NamedRerouteDeclaration(None=LinearInterpolate(A=NamedRerouteUsage(), B=LinearInterpolate(A=@n29, Alpha=@n30), Alpha=NamedRerouteUsage())) {name=Opacity, nodeColor=<rgba 0.8984 1 0 1>}
  NamedRerouteDeclaration(None=TextureCoordinate()) {name=TexCoordinate, nodeColor=<rgba 0.7813 0 1 1>}
  n31 = NamedRerouteUsage()
  n32 = NamedRerouteUsage()
  n33 = AppendVector(A=Constant() {r=1.2}, B=Constant() {r=0.8})
  n34 = Constant()
  n35 = Constant() {r=0.52}
  n36 = TextureSample(UVs=Panner(Coordinate=Multiply(A=@n32, B=@n33), Speed=AppendVector(A=@n34, B=@n35))) {samplerType=SAMPLERTYPE_Grayscale, texture=&LowResBlurredNoise}
  n37 = Multiply(A=@n36.R) {constB=0.4}
  NamedRerouteDeclaration(None=TextureSample(UVs=Add(A=@n31, B=@n37)) {samplerType=SAMPLERTYPE_Grayscale, texture=&LowResBlurredNoise}.R) {name=Plasma, nodeColor=<rgba 0 0.3906 1 1>}
  n38 = Add(A=NamedRerouteUsage(), B=AppendVector(A=Constant(), B=Constant() {r=0.03}))
  n39 = Multiply(A=NamedRerouteUsage(), B=Constant() {r=0.1})
  n40 = Constant() {r=0.5}
  n41 = Constant() {r=0.6}
  n42 = Constant() {r=0.1}
  n43 = Constant() {r=0.5}
  NamedRerouteDeclaration(None=MaterialFunctionCall(
    UVs=Add(A=@n38, B=@n39)
    Size=AppendVector(A=@n40, B=@n41)
    CornerRoundness=MaterialFunctionCall(R=@n42, G=@n43, B=@n42, A=@n43) {materialFunction=&AppendMany}.RGBA
    GlowMax=Constant() {r=0.2}
   ) {materialFunction=&MF_UI_Box}.Fill) {name=FlameMask, nodeColor=<rgba 0.7109 1 0 1>}
  n44 = MaterialFunctionCall() {materialFunction=&LinearGradient}
  n45 = OneMinus(None=LinearInterpolate(Alpha=SmoothStep(Value=@n44.VGradient) {constMax=0.9}) {constA=0.95, constB=1.25})
  n46 = Constant() {r=2.3}
  n47 = Constant() {r=0.5}
  n48 = Panner(Coordinate=Multiply(A=NamedRerouteUsage(), B=AppendVector(A=@n46, B=@n47)), Speed=AppendVector(A=Constant(), B=Constant() {r=0.73}))
  n49 = Multiply(A=NamedRerouteUsage()) {constB=0.2}
  n50 = Multiply(
   A=@n36.RGB
   B=TextureSample(UVs=Add(A=@n48, B=@n49)) {samplerType=SAMPLERTYPE_Grayscale, texture=&LowResBlurredNoise}.RGB
  )
  NamedRerouteDeclaration(None=SmoothStep(Value=SmoothStep(Min=@n45, Value=@n50)) {constMax=0.2}) {name=Flame, nodeColor=<rgba 1 0.3984 0 1>}
  NamedRerouteUsage()
  NamedRerouteUsage()
 ]
}
