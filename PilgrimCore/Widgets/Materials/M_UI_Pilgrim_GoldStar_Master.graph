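# Translucent UI master material for the Pilgrim gold star, with Progress, Completed and BandStarTreatment parameters driving its look.
# A channel-packed distance-field texture supplies anti-aliased star, outline and glow masks, combined with procedural circle masks.
# Cosine-gradient color uses a panning, distorted sine shine, while Progress positions gradient masks for scrolling screen-aligned noise alongside grid-based particles.
# The IsGlow static switch swaps color and opacity for a simpler two-color glow.
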Material /PilgrimCore/Widgets/Materials/M_UI_Pilgrim_GoldStar_Master
graph {
 properties {
  bUsedWithStaticMesh=false
  blendMode=BLEND_Translucent
  materialDomain=MD_UI
  pixelDepthOffsetMode=PDOM_Legacy
  usedShadingModels=Default Lit
 }
 nodes [
  n1 = AppendVector(
   A=ScalarParameter() {defaultValue=0.5, parameterName=Gradient Center X, sliderMax=1}
   B=ScalarParameter() {defaultValue=0.5, parameterName=Gradient Center Y, sliderMax=1}
  )
  n2 = ScalarParameter() {defaultValue=0.025, parameterName=Gradient Rotation, sliderMax=1}
  n3 = MaterialFunctionCall(UVs=TextureCoordinate(), Rotation Center=@n1, "Rotation Angle (0-1)"=@n2) {materialFunction=&CustomRotator}
  ComponentMask(None=Ceil(None=Add(A=@n3.Rotated Values) {constB=-0.5})) {r=true}
  n4 = TextureCoordinate()
  n5 = ScalarParameter() {defaultValue=1.25, parameterName=StarScale}
  n6 = LinearInterpolate(A=Constant(), Alpha=ScalarParameter() {parameterName=BandStarTreatment}) {constA=0.35, constB=-0.01}
  n7 = LinearInterpolate(A=Constant() {r=0.025}, Alpha=ScalarParameter() {parameterName=BandStarTreatment}) {constA=0.35, constB=0}
  n8 = TextureSample(
   UVs=Add(
    A=MaterialFunctionCall(UVs=@n4, Texture Scale=@n5) {materialFunction=&ScaleUVsByCenter}.UVs
    B=AppendVector(A=@n6, B=@n7)
   )
   Tex=TextureObjectParameter() {parameterName=Icon, texture=/PilgrimCore/Widgets/Textures/T_UI_Piligrim_StarDFChannelPack_64x}
  ) {texture=&T_UI_Piligrim_StarDFChannelPack_64x}
  n9 = NamedRerouteDeclaration(None=@n8.B) {name=Star, nodeColor=<rgba 0.3359 0 1 1>}
  n10 = LinearInterpolate(Alpha=ScalarParameter() {parameterName=BandStarTreatment}) {constA=0.35, constB=0.475}
  n11 = Saturate(None=MaterialFunctionCall(DistanceField=Add(A=@n8.R, B=@n10) {constB=0.35}, Threshold=Constant() {r=0.5}) {materialFunction=&MF_UI_AntiAliasing}.Outer)
  n12 = Min(A=@n9, B=OneMinus(None=@n11))
  n13 = SmoothStep(Value=@n8.R)
  LinearInterpolate(A=@n12, B=Min(A=@n9, B=OneMinus(None=@n13)), Alpha=ScalarParameter() {parameterName=BandStarTreatment}) {constA=0.35, constB=0}
  n14 = NamedRerouteUsage()
  n15 = LinearInterpolate(A=NamedRerouteUsage(), B=NamedRerouteUsage(), Alpha=NamedRerouteUsage())
  n16 = Saturate(None=MaterialFunctionCall("Value (0-1)"=Multiply(A=@n14, B=@n15)) {materialFunction=&MF_UI_EaseCurves}.EaseInSine)
  n17 = LinearInterpolate(A=Multiply(A=NamedRerouteUsage(), B=ScalarParameter() {defaultValue=0.75, parameterName=OuterCircle_Alpha}), Alpha=NamedRerouteUsage()) {constAlpha=0}
  n18 = NamedRerouteUsage()
  Multiply(
   A=StaticSwitchParameter(True=@n16, False=Multiply(A=@n17, B=@n18)) {parameterName=IsGlow}
   B=ScalarParameter() {defaultValue=1, parameterName=Final_Alpha}
  )
  n19 = OneMinus(None=NamedRerouteUsage())
  n20 = MaterialFunctionCall(Size=Constant() {r=1}) {materialFunction=&MF_UI_Circle}
  n21 = NamedRerouteUsage()
  n22 = OneMinus(None=Add(A=LinearInterpolate(A=@n19, B=@n20.SDF, Alpha=@n21)) {constB=0.5})
  n23 = SmoothStep(Max=LinearInterpolate(Alpha=ScalarParameter() {parameterName=Completed}) {constA=0.8, constB=0.85}, Value=@n22) {constMax=0.8, constMin=0.5}
  NamedRerouteDeclaration(None=@n23) {name=OuterCircleMask, nodeColor=<rgba 0 0 0 1>}
  n24 = OneMinus(None=NamedRerouteUsage())
  n25 = MaterialFunctionCall(
   GradientStart=Subtract(A=@n24, B=ScalarParameter() {defaultValue=0.1, parameterName=NoiseBeamSize}) {constB=0.2}
   GradientEnd=@n24
   RotationAngle=Constant() {r=-0.25}
  ) {materialFunction=&MF_UI_LinearGradient}
  NamedRerouteDeclaration(None=@n25.Result) {name=TopGlowmask, nodeColor=<rgba 1 0 0.9375 1>}
  NamedRerouteDeclaration(None=@n8.G) {name=StarGlow, nodeColor=<rgba 0.2188 0 1 1>}
  NamedRerouteDeclaration(None=@n11) {name=StarDf, nodeColor=<rgba 0.2188 0 1 1>}
  NamedRerouteDeclaration(None=@n12) {name=Star outline, nodeColor=<rgba 0.6641 1 0 1>}
  n26 = MaterialFunctionCall("Value (0-1)"=Min(A=@n25.Result, B=NamedRerouteUsage())) {materialFunction=&MF_UI_EaseCurves}
  NamedRerouteDeclaration(None=LinearInterpolate(A=@n26.EaseInQuint, Alpha=NamedRerouteUsage())) {name=FinalAlpha, nodeColor=<rgba 0 1 0.2266 1>}
  NamedRerouteDeclaration(None=@n13) {name=innerGlow, nodeColor=<rgba 0 0.9766 1 1>}
  NamedRerouteDeclaration(None=ScalarParameter() {defaultValue=1, parameterName=UseOuterCircle}) {name=UseOuterCircle, nodeColor=<rgba 0.1016 1 0 1>}
  n27 = Constant()
  n28 = SmoothStep(Value=MaterialFunctionCall!(
    UV
    Global Offset
    Global Rotation
    Tiling=ScalarParameter() {defaultValue=5.5, parameterName=ParticleTiling}
    Cell Offset
    Cell Scale
    Cell Rotation
    Random Seed
   ) {materialFunction=/FortUILibrary/MaterialFunctions/Basics/MF_UI_ParticleGrid}.Random Columns) {constMax=0.75, constMin=-0.5}
  n29 = Constant()
  n30 = MaterialFunctionCall(
   UV=Panner(Coordinate=TextureCoordinate(), Time=Time(), Speed=AppendVector(A=@n27, B=@n28))
   Tiling=ScalarParameter() {defaultValue=5.5, parameterName=ParticleTiling}
   Cell Offset=@n29
   Cell Rotation=@n29
  ) {materialFunction=&MF_UI_ParticleGrid}
  n31 = MaterialFunctionCall(
   UVs=@n30.Cells UV
   Size=Min(A=@n30.Random Cell Values, B=NamedRerouteUsage())
   CornerRoundness=Constant() {r=0.25}
   KeepAspectRatio=StaticBool() {value=true}
  ) {materialFunction=&MF_UI_Box}
  n32 = Saturate(None=MaterialFunctionCall(DistanceField=Add(A=@n31.SDFOutline) {constB=0.333}, Threshold=Constant() {r=0.5}) {materialFunction=&MF_UI_AntiAliasing}.Inner)
  NamedRerouteDeclaration(None=Multiply(A=@n32, B=ScalarParameter() {defaultValue=0.5, parameterName=ParticlesIntensity})) {name=Particles, nodeColor=<rgba 1 0 0.8672 1>}
  n33 = ScalarParameter() {defaultValue=0.85, parameterName=InnerCircle_Size}
  NamedRerouteDeclaration(None=LinearInterpolate(
    A=NamedRerouteUsage()
    B=MaterialFunctionCall(Size=@n33) {materialFunction=&MF_UI_Circle}.Fill
    Alpha=NamedRerouteUsage()
   )) {name=InnerCircleMask, nodeColor=<rgba 0 0 0 1>}
  NamedRerouteDeclaration(None=@n32) {name=ParticlesALpha, nodeColor=<rgba 0.5469 0 1 1>}
  NamedRerouteDeclaration(None=OneMinus(None=@n23)) {name=OuterEdge, nodeColor=<rgba 0 0.1562 1 1>}
  NamedRerouteDeclaration(None=SmoothStep(Value=@n22) {constMax=1.5, constMin=0.5}) {name=RadialGradiantMask, nodeColor=<rgba 0 0.8125 1 1>}
  NamedRerouteDeclaration(None=Multiply(A=NamedRerouteUsage()) {constB=50}) {name=InnerGlowMask, nodeColor=<rgba 0 1 0.2031 1>}
  n34 = ScalarParameter() {defaultValue=-0.1, parameterName=RemappedValue_Bottom}
  n35 = ScalarParameter() {defaultValue=0.9, parameterName=RemappedValue_Top}
  n36 = ScalarParameter() {defaultValue=1, parameterName=Progress}
  NamedRerouteDeclaration(None=Saturate(None=LinearInterpolate(A=@n34, B=@n35, Alpha=@n36) {constB=0.9})) {name=RemappedProgress, nodeColor=<rgba 1 0.1641 0 1>}
  n37 = NamedRerouteUsage()
  n38 = Subtract(A=@n37, B=ScalarParameter() {defaultValue=0.1, parameterName=NoiseBeamSize}) {constB=0.2}
  n39 = Constant() {r=0.25}
  n40 = Add(A=TextureCoordinate()) {constB=-1}
  n41 = NamedRerouteUsage()
  n42 = Ceil(None=ComponentMask(None=Add(A=@n40, B=@n41)) {g=true})
  n43 = MaterialFunctionCall("Value (0-1)"=Min(
    A=MaterialFunctionCall(GradientStart=@n38, GradientEnd=@n37, RotationAngle=@n39) {materialFunction=&MF_UI_LinearGradient}.Result
    B=NamedRerouteDeclaration(None=@n42) {name=ProgressMask, nodeColor=<rgba 0.2891 0 1 1>}
   )) {materialFunction=&MF_UI_EaseCurves}
  NamedRerouteDeclaration(None=@n43.EaseInQuint) {name=NoiseFlowDownMask, nodeColor=<rgba 0 0 0 1>}
  NamedRerouteDeclaration(None=TextureObject() {samplerType=SAMPLERTYPE_Grayscale, texture=/FortUILibrary/UIKit/Buttons/Textures/LowResBlurredNoise}) {name=NoiseTexture, nodeColor=<rgba 1 1 1 1>}
  NamedRerouteDeclaration(None=OneMinus(None=@n42)) {name=UpHardMask, nodeColor=<rgba 0 0 0 1>}
  NamedRerouteDeclaration(None=@n26.EaseInCubic) {name=NoiseFlowUpMask, nodeColor=<rgba 0 0 0 1>}
  n44 = MaterialFunctionCall!(TextureResolution) {materialFunction=/Engine/Functions/Engine_MaterialFunctions02/Texturing/ScreenAlignedPixelToPixelUVs}
  n45 = Constant() {r=0.25}
  n46 = Constant() {r=-0.125}
  n47 = Constant() {r=-0.5}
  n48 = TextureSample(UVs=Panner(Coordinate=Multiply(A=@n44.Result, B=@n45), Time=Time(), Speed=AppendVector(A=@n46, B=@n47)), Tex=NamedRerouteUsage()) {samplerType=SAMPLERTYPE_Grayscale, texture=&LowResBlurredNoise}
  n49 = MaterialFunctionCall() {materialFunction=&ScreenAlignedPixelToPixelUVs}
  n50 = Constant() {r=0.3}
  n51 = Constant() {r=0.125}
  n52 = Constant() {r=-0.3333}
  n53 = TextureSample(UVs=Panner(Coordinate=Multiply(A=@n49.Result, B=@n50), Time=Time(), Speed=AppendVector(A=@n51, B=@n52)), Tex=NamedRerouteUsage()) {samplerType=SAMPLERTYPE_Grayscale, texture=&LowResBlurredNoise}
  NamedRerouteDeclaration(None=Multiply(A=Max(A=@n48.R, B=@n53.R), B=NamedRerouteUsage())) {name=NoiseDown, nodeColor=<rgba 0 0 0 1>}
  n54 = MaterialFunctionCall(DistanceField=Add(A=@n8.R) {constB=-0.005}) {materialFunction=&MF_UI_AntiAliasing}
  n55 = MaterialFunctionCall(DistanceField=Add(A=@n8.R) {constB=-0.15}) {materialFunction=&MF_UI_AntiAliasing}
  NamedRerouteDeclaration(None=Min(A=Saturate(None=@n54.Outer), B=Saturate(None=@n55.Inner))) {name=ThickerBandOutline, nodeColor=<rgba 0 1 0.625 1>}
  n56 = Multiply(A=Add(A=NamedRerouteUsage(), B=@n26.EaseInQuint), B=ScalarParameter() {defaultValue=3, parameterName=NoiseIntensity}) {constB=3}
  NamedRerouteDeclaration(None=LinearInterpolate(A=Saturate(None=@n56), Alpha=Add(A=@n26.EaseInQuint, B=@n43.EaseInQuint))) {name=FinalNoiseDivderMask, nodeColor=<rgba 0 0 0 1>}
  n57 = ScalarParameter() {parameterName=Completed}
  n58 = LinearInterpolate(A=Constant() {r=0.9}, B=Constant() {r=1}, Alpha=@n57)
  n59 = LinearInterpolate(A=Constant() {r=0.85}, B=Constant() {r=0.8}, Alpha=@n57)
  n60 = LinearInterpolate(A=Constant() {r=-0.2}, B=Constant() {r=0.05}, Alpha=@n57)
  n61 = Constant() {r=0.1}
  n62 = Constant() {r=0.43}
  n63 = Constant()
  n64 = Constant() {r=1}
  n65 = Constant() {r=0.8}
  n66 = Constant() {r=0.5}
  n67 = Constant()
  n68 = Constant() {r=0.3333}
  n69 = Constant() {r=0.6667}
  n70 = MaterialFunctionCall() {materialFunction=&ScreenAlignedUVs}
  n71 = Divide(B=MaterialFunctionCall(Float2=MaterialFunctionCall() {materialFunction=&GetUserInterfaceUV}.Pixel Size) {materialFunction=&BreakOutFloat2Components}.R) {constA=64}
  n72 = Panner(
   Coordinate=MaterialFunctionCall(UVs=Multiply(A=@n70."X 100%, Y 100%", B=@n71), Rotation Center=@n1, "Rotation Angle (0-1)"=@n2) {materialFunction=&CustomRotator}.Rotated Values
   Time=Time()
   Speed=AppendVector(A=ScalarParameter() {defaultValue=-0.025, parameterName=PanSpeedX}, B=Constant())
  )
  n73 = Add(A=ComponentMask(None=Sine(None=@n72) {period=0.125}) {r=true})
  n74 = LinearInterpolate(A=OneMinus(None=NamedRerouteUsage()), Alpha=ComponentMask(None=Sine(None=@n3.Rotated Values) {period=2}) {r=true})
  n75 = ScalarParameter() {defaultValue=0.5, parameterName=DistortionAmount}
  n76 = MaterialFunctionCall("Value (0-1)"=Add(A=Multiply(A=@n73) {constB=0.5}, B=Multiply(A=@n74, B=@n75))) {materialFunction=&MF_UI_EaseCurves}
  n77 = Multiply(A=Saturate(None=OneMinus(None=@n76.EaseOutSine)), B=ScalarParameter() {defaultValue=1, parameterName=ShineBoost})
  n78 = NamedRerouteUsage()
  n79 = OneMinus(None=NamedRerouteUsage())
  n80 = LinearInterpolate(
   A=MaterialFunctionCall!(
    Gradient=NamedRerouteDeclaration(None=@n77) {name=GradientMask, nodeColor=<rgba 0 0 0 1>}
    a=MaterialFunctionCall(R=@n58, G=@n59, B=@n60) {desc=a, materialFunction=&AppendMany}.RGB
    b=MaterialFunctionCall(R=@n61, G=@n62, B=@n63) {desc=b, materialFunction=&AppendMany}.RGB
    c=MaterialFunctionCall(R=@n64, G=@n65, B=@n66) {desc=c, materialFunction=&AppendMany}.RGB
    d=MaterialFunctionCall(R=@n67, G=@n68, B=@n69) {desc=d, materialFunction=&AppendMany}.RGB
   ) {materialFunction=/FortUILibrary/MaterialFunctions/Color/MF_UI_CosineGradient}.Result
   B=VectorParameter() {parameterName=StarCoreColor, defaultValue=<rgba 0.7292 0.1645 0.0771 1>}.RGB
   Alpha=Multiply(A=Multiply(A=@n78, B=@n79), B=ScalarParameter() {defaultValue=0.5, parameterName=StarFillStartAmount})
  )
  n81 = LinearInterpolate(
   A=VectorParameter() {parameterName=StarCoreColor, defaultValue=<rgba 0.7292 0.1645 0.0771 1>}.RGB
   B=VectorParameter() {parameterName=StarCoreColor_Complete, defaultValue=<rgba 0.95 1 0.9917 1>}.RGB
   Alpha=ScalarParameter() {parameterName=Completed}
  )
  n82 = Multiply(
   A=NamedRerouteUsage()
   B=LinearInterpolate(A=ScalarParameter() {defaultValue=0.5, parameterName=StarFillStartAmount}, Alpha=ScalarParameter() {parameterName=Completed}) {constA=0.5}
  )
  n83 = LinearInterpolate(A=NamedRerouteUsage(), B=NamedRerouteUsage(), Alpha=ScalarParameter() {parameterName=BandStarTreatment})
  n84 = ScalarParameter() {defaultValue=0.5, parameterName=StarOutlineBlendAmount}
  n85 = LinearInterpolate(
   A=LinearInterpolate(
    A=LinearInterpolate(A=@n80, B=@n81, Alpha=@n82)
    B=VectorParameter() {parameterName=StarOutlineColor, defaultValue=<rgba 0.9844 0.4037 0 1>}.RGB
    Alpha=Multiply(A=@n83, B=@n84)
   )
   B=LinearInterpolate(
    A=VectorParameter() {parameterName=OuterEdgeColorBase, defaultValue=<rgba 1 0.7225 0.667 1>}.RGB
    Alpha=ScalarParameter() {parameterName=Completed}
   ) {constA=0.5}
   Alpha=NamedRerouteUsage()
  )
  n86 = NamedRerouteUsage()
  StaticSwitchParameter(
   True=LinearInterpolate(
    A=VectorParameter() {parameterName=GlowColorB, defaultValue=<rgba 1 0.05 0.05 1>}.RGB
    B=VectorParameter() {parameterName=GlowColorA, defaultValue=<rgba 1 0.1667 0 1>}.RGB
    Alpha=@n16
   )
   False=Saturate(None=LinearInterpolate(A=@n85, Alpha=@n86))
  ) {parameterName=IsGlow}
 ]
}
