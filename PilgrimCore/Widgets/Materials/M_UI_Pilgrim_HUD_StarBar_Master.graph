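# Translucent UI master material for the Pilgrim HUD star bar, drawing a circular progress ring driven by a Progress scalar.
# An SDF circle outline with anti-aliasing shapes the ring, and a radial angle compared against Progress masks the filled arc.
# Rotated radial gradients form rounded start and moving end caps, zeroed at zero progress.
# Two fill colors blend through Progress power curves and a gradient mask, with separate empty color and opacity.
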Material /PilgrimCore/Widgets/Materials/M_UI_Pilgrim_HUD_StarBar_Master
graph {
 properties {
  bUsedWithStaticMesh=false
  blendMode=BLEND_Translucent
  materialDomain=MD_UI
  pixelDepthOffsetMode=PDOM_Legacy
  usedShadingModels=Default Lit
 }
 nodes [
  n1 = VectorParameter() {parameterName=Color_Fill_B, defaultValue=<rgba 1 0 0 1>}
  n2 = VectorParameter() {parameterName=Color_Fill_A, defaultValue=<rgba 0 1 0 1>}
  n3 = Saturate(None=Power(Base=ScalarParameter() {defaultValue=0.25, parameterName=Progress}) {constExponent=0.5})
  n4 = Saturate(None=Power(Base=ScalarParameter() {defaultValue=0.25, parameterName=Progress}) {constExponent=5})
  n5 = ComponentMask(None=Add(A=TextureCoordinate()) {constB=-0.5}) {r=true}
  n6 = SmoothStep(
   Min=OneMinus(None=ScalarParameter() {defaultValue=0.25, parameterName=Progress})
   Value=MaterialFunctionCall!(Vector or UVs) {materialFunction=/FortUILibrary/MaterialFunctions/Basics/MF_UI_CheaperVectorToRadial}.Vector Converted to Angle
  )
  n7 = LinearInterpolate(
   A=LinearInterpolate(A=NamedRerouteUsage(), Alpha=Ceil(None=@n5)) {constB=0}
   Alpha=NamedRerouteDeclaration(None=@n6) {name=GradientMask, nodeColor=<rgba 0 0 0 1>}
  )
  n8 = Ceil(None=@n6)
  n9 = NamedRerouteUsage()
  n10 = NamedRerouteDeclaration(None=Saturate(None=Add(A=@n8, B=@n9))) {name=Progress Mask, nodeColor=<rgba 0 0 0 1>}
  LinearInterpolate(
   A=VectorParameter() {parameterName=Color_Empty, defaultValue=<rgba 0 0 1 1>}.RGB
   B=LinearInterpolate(A=LinearInterpolate(A=@n1.RGB, B=@n2.RGB, Alpha=@n3), B=LinearInterpolate(A=@n1.RGB, B=@n2.RGB, Alpha=@n4), Alpha=Saturate(None=@n7))
   Alpha=@n10
  )
  MaterialFunctionCall() {materialFunction=&CustomRotator}
  n11 = ScalarParameter() {defaultValue=0.1, parameterName=ProgressBar_Padding}
  n12 = ScalarParameter() {defaultValue=0.05, parameterName=RingSize}
  n13 = Subtract(A=OneMinus(None=Add(A=@n11, B=@n12)), B=ScalarParameter() {defaultValue=0.05, parameterName=AA_Threshold})
  n14 = ScalarParameter() {defaultValue=0.05, parameterName=ProgressThickness}
  n15 = NamedRerouteDeclaration(None=MaterialFunctionCall(
    DistanceField=MaterialFunctionCall(Size=@n13, OutlineThickness=@n14) {materialFunction=&MF_UI_Circle}.SDFOutline
    Threshold=ScalarParameter() {defaultValue=0.05, parameterName=AA_Threshold}
   ) {materialFunction=&MF_UI_AntiAliasing}.Inner) {name=Progress Ring Mask, nodeColor=<rgba 0 0 0 1>}
  Multiply(
   A=LinearInterpolate(
    A=ScalarParameter() {defaultValue=0.5, parameterName=Alpha_Empty}
    B=ScalarParameter() {defaultValue=1, parameterName=Alpha_Fill}
    Alpha=Multiply(A=@n10, B=@n15)
   ) {constB=-0.68}
   B=@n15
  )
  n16 = TextureCoordinate()
  n17 = AppendVector(
   A=Constant() {r=0.5}
   B=Add(A=ScalarParameter() {defaultValue=0.05, parameterName=EndCapSize}, B=ScalarParameter() {defaultValue=0.05, parameterName=EndCapPadding})
  )
  n18 = ScalarParameter() {defaultValue=0.05, parameterName=EndCapSize}
  n19 = ScalarParameter() {defaultValue=16, parameterName=EndCapHardness}
  n20 = NamedRerouteDeclaration(None=Saturate(None=MaterialFunctionCall(UVs=@n16, CenterPosition=@n17, Radius=@n18, Density=@n19) {materialFunction=&RadialGradientExponential}.RadialGradientExponential)) {name=StartEndCap, nodeColor=<rgba 0 1 0.8359 1>}
  n21 = MaterialFunctionCall(UVs=@n16, "Rotation Angle (0-1)"=Multiply(A=ScalarParameter() {defaultValue=0.25, parameterName=Progress}) {constB=-1}) {materialFunction=&CustomRotator}
  n22 = NamedRerouteDeclaration(None=Saturate(None=MaterialFunctionCall(UVs=@n21.Rotated Values, CenterPosition=@n17, Radius=@n18, Density=@n19) {materialFunction=&RadialGradientExponential}.RadialGradientExponential)) {name=DeltEndCap, nodeColor=<rgba 0 1 0.8359 1>}
  n23 = MaterialFunctionCall!(A=ScalarParameter() {defaultValue=0.25, parameterName=Progress}, B=Constant()) {materialFunction=/FortUILibrary/MaterialFunctions/Basics/MF_UI_BooleanOperations}
  NamedRerouteDeclaration(None=Multiply(A=Add(A=@n20, B=@n22), B=LinearInterpolate(Alpha=@n23.IsEqual) {constA=1, constB=0})) {name=RoundedEndCaps, nodeColor=<rgba 0.5234 1 0 1>}
  n24 = ScalarParameter() {defaultValue=0.1}
  Saturate(None=MaterialFunctionCall(
    DistanceField=MaterialFunctionCall(Size=@n24) {materialFunction=&MF_UI_Circle}.SDF
    Threshold=ScalarParameter() {defaultValue=0.05, parameterName=AA_Threshold}
   ) {materialFunction=&MF_UI_AntiAliasing}.Inner)
 ]
}
