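# UI material for the Pilgrim multiplier background, drawing a circle with a gradient fill and colored outline.
# A linear gradient blends Start, Middle and End colors for the fill, while MF_UI_Circle supplies signed-distance fill and outline masks.
# Masked fill and outline colors are built through named reroutes, with a Max node for combining.
# A Filled static switch picks either the solid disc or outline-only mask.
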Material /PilgrimCore/Widgets/Materials/M_UI_Pilgrim_MultiplierBackground
graph {
 properties {
  bUsedWithStaticMesh=false
  blendMode=BLEND_Translucent
  materialDomain=MD_UI
  pixelDepthOffsetMode=PDOM_Legacy
  usedShadingModels=Default Lit
 }
 nodes [
  Max(A=NamedRerouteUsage(), B=NamedRerouteUsage())
  n1 = Saturate(None=MaterialFunctionCall(GradientStart=Constant(), GradientEnd=Constant() {r=1}, RotationAngle=Constant() {r=0.25}) {materialFunction=&MF_UI_LinearGradient}.Result)
  n2 = LinearInterpolate(
   A=VectorParameter() {parameterName=End Color, defaultValue=<rgba 0 0 0 1>}.RGB
   B=VectorParameter() {parameterName=Middle Color, defaultValue=<rgba 1 1 1 1>}.RGB
   Alpha=@n1
  )
  n3 = VectorParameter() {parameterName=Start Color, defaultValue=<rgba 1 1 1 1>}
  n4 = NamedRerouteUsage()
  n5 = NamedRerouteUsage()
  NamedRerouteDeclaration(None=LinearInterpolate(B=LinearInterpolate(A=@n2, B=@n3.RGB, Alpha=@n1), Alpha=Subtract(A=@n4, B=@n5))) {name=Fill Color, nodeColor=<rgba 0 0 0 1>}
  n6 = MaterialFunctionCall(Size=Constant() {r=0.9}, OutlineThickness=ScalarParameter() {defaultValue=0.05, parameterName=Outline Thickness}) {materialFunction=&MF_UI_Circle}
  n7 = SmoothStep(Value=Saturate(None=OneMinus(None=@n6.SDFOutline))) {constMin=0.975}
  NamedRerouteDeclaration(None=@n7) {name=Outline Alpha, nodeColor=<rgba 0 0 0 1>}
  n8 = Saturate(None=OneMinus(None=@n6.SDF))
  NamedRerouteDeclaration(None=@n8) {name=Fill Alpha, nodeColor=<rgba 0 0 0 1>}
  NamedRerouteDeclaration(None=LinearInterpolate(B=VectorParameter() {parameterName=Outline Color, defaultValue=<rgba 0.6771 0.6771 0.6771 1>}.RGB, Alpha=NamedRerouteUsage())) {name=Outline Color, nodeColor=<rgba 0 0 0 1>}
  StaticSwitchParameter(True=SmoothStep(Value=@n8) {constMin=0.975}, False=@n7) {defaultValue=true, parameterName=Filled}
 ]
}
