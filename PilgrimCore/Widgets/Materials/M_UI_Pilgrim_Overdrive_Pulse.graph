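# Translucent UI material for an Overdrive pulse, blending two color parameters (red to orange by default) through animated bands on a progress-sized shape.
# Tile and Speed drive panning sine waves that scroll in opposite directions on either vertical half.
# A static vertical sine envelope shapes and fades the bands.
# MF_UI_Box builds a soft SDF mask from CornerRadius and Progress, which sets box width and horizontal UV offset.
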Material /PilgrimCore/Widgets/Materials/M_UI_Pilgrim_Overdrive_Pulse
graph {
 properties {
  bUsedWithStaticMesh=false
  blendMode=BLEND_Translucent
  materialDomain=MD_UI
  pixelDepthOffsetMode=PDOM_Legacy
  usedShadingModels=Default Lit
 }
 nodes [
  n1 = ComponentMask(None=Sine(None=TextureCoordinate()) {period=2}) {g=true}
  n2 = Multiply(A=TextureCoordinate(), B=Add(A=ScalarParameter() {defaultValue=6, parameterName=Tile}) {constB=0.5}) {constB=5.5}
  n3 = ScalarParameter() {defaultValue=3, parameterName=Speed}
  n4 = Sine(None=Panner(Coordinate=@n2, Speed=@n3) {speedY=2})
  n5 = Sine(None=Panner(Coordinate=@n2, Speed=Multiply(A=@n3) {constB=-1}) {speedY=-2})
  n6 = ComponentMask(None=Add(A=TextureCoordinate()) {constB=-0.5}) {g=true}
  n7 = SmoothStep(
   Min=LinearInterpolate(Alpha=@n1) {constA=1, constB=0.5}
   Value=LinearInterpolate(A=ComponentMask(None=@n4) {g=true}, B=ComponentMask(None=@n5) {g=true}, Alpha=Ceil(None=@n6))
  )
  LinearInterpolate(
   A=VectorParameter() {parameterName=Color_A, defaultValue=<rgba 1 0 0 1>}.RGB
   B=VectorParameter() {parameterName=Color_B, defaultValue=<rgba 1 0.5 0 1>}.RGB
   Alpha=Multiply(A=Multiply(A=@n7, B=@n1), B=NamedRerouteUsage())
  )
  n8 = MaterialFunctionCall() {materialFunction=&GetUserInterfaceUV}
  NamedRerouteDeclaration(None=@n8.9-Slice UV) {name=9sliceuv, nodeColor=<rgba 0 0 0 1>}
  NamedRerouteDeclaration(None=@n8.Normalized UV) {name=NormalizedUV, nodeColor=<rgba 0 0 0 1>}
  n9 = ScalarParameter() {defaultValue=0.5, parameterName=Progress}
  n10 = Multiply(A=OneMinus(None=@n9)) {constB=0.5}
  n11 = Constant()
  n12 = NamedRerouteDeclaration(None=Add(A=NamedRerouteUsage(), B=AppendVector(A=@n10, B=@n11))) {name=BoxUVs, nodeColor=<rgba 0 0 0 1>}
  n13 = AppendVector(A=@n9, B=Constant() {r=1})
  n14 = ScalarParameter() {defaultValue=1, parameterName=CornerRadius}
  n15 = StaticBool() {value=true}
  NamedRerouteDeclaration(None=SmoothStep(Value=MaterialFunctionCall(UVs=@n12, Size=@n13, CornerRoundness=@n14, KeepAspectRatio=@n15) {materialFunction=&MF_UI_Box}.SDF) {constMax=-1}) {name=Soft Mask, nodeColor=<rgba 0 0 0 1>}
 ]
}
