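# Translucent UI master material that draws the Pilgrim star icon with colored borders, core and pulsing glow.
# A channel-packed distance-field texture, sampled at separate centered scales, feeds anti-aliased star and border masks plus a sine-animated glow.
# Named reroutes carry the masks into layered color lerps and alpha compositing.
# Static switches IsJustBorder and IsSoftGlow select border-only or soft-glow variants.
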Material /PilgrimCore/Widgets/Materials/M_UI_Pilgrim_Star_Master
graph {
 properties {
  bUsedWithStaticMesh=false
  blendMode=BLEND_Translucent
  materialDomain=MD_UI
  pixelDepthOffsetMode=PDOM_Legacy
  usedShadingModels=Default Lit
 }
 nodes [
  n1 = NamedRerouteUsage()
  n2 = Multiply(A=NamedRerouteUsage(), B=ScalarParameter() {defaultValue=16, parameterName=Alpha_GlowCompensation})
  n3 = NamedRerouteUsage()
  n4 = ScalarParameter() {defaultValue=0.25, parameterName=Alpha_Border}
  n5 = LinearInterpolate(A=LinearInterpolate(A=Saturate(None=@n2), Alpha=Multiply(A=@n3, B=@n4)), Alpha=NamedRerouteUsage())
  Multiply(
   A=ScalarParameter() {defaultValue=1, parameterName=Alpha_Final}
   B=StaticSwitchParameter(True=NamedRerouteUsage(), False=StaticSwitchParameter(True=@n1, False=@n5) {parameterName=IsSoftGlow}) {parameterName=IsJustBorder}
  )
  n6 = MaterialFunctionCall(UVs=TextureCoordinate(), Texture Scale=ScalarParameter() {defaultValue=1, parameterName=Scale_Glow}) {materialFunction=&ScaleUVsByCenter}
  n7 = TextureObjectParameter() {parameterName=Icon, texture=/PilgrimCore/Widgets/Textures/T_UI_Piligrim_StarDFChannelPack_64x}
  NamedRerouteDeclaration(None=MaterialFunctionCall!(
    UVs=@n6.UVs
    Alpha=TextureSample(UVs=@n6.UVs, Tex=@n7) {samplerType=SAMPLERTYPE_LinearColor, texture=&T_UI_Piligrim_StarDFChannelPack_64x}.G
   ) {materialFunction=/Game/UI/Foundation/Items/ClampAlphaToTransparentBorder}.Result) {name=Glow_Channel, nodeColor=<rgba 0 0.9734 0 1>}
  n8 = MaterialFunctionCall(UVs=TextureCoordinate(), Texture Scale=ScalarParameter() {defaultValue=1.1, parameterName=Scale_Icon}) {materialFunction=&ScaleUVsByCenter}
  n9 = TextureSample(UVs=@n8.UVs, Tex=@n7) {samplerType=SAMPLERTYPE_LinearColor, texture=&T_UI_Piligrim_StarDFChannelPack_64x}
  NamedRerouteDeclaration(None=MaterialFunctionCall(UVs=@n8.UVs, Alpha=@n9.R) {materialFunction=&ClampAlphaToTransparentBorder}.Result) {name=DF_Channel, nodeColor=<rgba 0.9647 0 0 1>}
  n10 = TextureCoordinate()
  n11 = ScalarParameter() {defaultValue=1, parameterName=Scale_Icon_Border}
  n12 = Constant()
  n13 = Constant() {r=-0.005}
  n14 = TextureSample(
   UVs=Add(
    A=MaterialFunctionCall(UVs=@n10, Texture Scale=@n11) {materialFunction=&ScaleUVsByCenter}.UVs
    B=AppendVector(A=@n12, B=@n13)
   )
   Tex=@n7
  ) {texture=&T_UI_Piligrim_StarDFChannelPack_64x}
  n15 = MaterialFunctionCall(DistanceField=@n14.R, Threshold=ScalarParameter() {defaultValue=0.1, parameterName=Threshold_Borde_Inner}) {materialFunction=&MF_UI_AntiAliasing}
  NamedRerouteDeclaration(None=Multiply(A=NamedRerouteUsage(), B=OneMinus(None=@n15.Outer))) {name=InnerBorderMask, nodeColor=<rgba 1 0.0469 0 1>}
  NamedRerouteDeclaration(None=@n9.B) {name=Regular_Channel, nodeColor=<rgba 0 0 0.9301 1>}
  NamedRerouteDeclaration(None=MaterialFunctionCall(DistanceField=NamedRerouteUsage(), Threshold=Constant() {r=0.05}) {materialFunction=&MF_UI_AntiAliasing}.Outer) {name=SimpleStarMask, nodeColor=<rgba 0 0 0 1>}
  n16 = ScalarParameter() {defaultValue=1, parameterName=GlowAmount}
  n17 = ScalarParameter() {defaultValue=0.1, parameterName=GlowAnimateAmount}
  n18 = Add(A=@n16, B=@n17)
  n19 = Subtract(A=@n16, B=@n17)
  n20 = MaterialFunctionCall(Speed=ScalarParameter() {defaultValue=0.5, parameterName=GlowSpeed}) {materialFunction=&DebugTimeSine}
  NamedRerouteDeclaration(None=Multiply(A=LinearInterpolate(A=@n18, B=@n19, Alpha=@n20.Result), B=NamedRerouteUsage())) {name=Glow Mask, nodeColor=<rgba 0 0 0 1>}
  n21 = VectorParameter() {parameterName=Color_InnerBorder, defaultValue=<rgba 1 1 0 1>}
  n22 = VectorParameter() {parameterName=Color_Glow_A, defaultValue=<rgba 0 0 1 1>}
  n23 = VectorParameter() {parameterName=Color_Glow_B, defaultValue=<rgba 1 0 0 1>}
  n24 = NamedRerouteUsage()
  n25 = VectorParameter() {parameterName=Color_Border_A, defaultValue=<rgba 0 0 1 1>}
  n26 = VectorParameter() {parameterName=Color_Border_B, defaultValue=<rgba 1 0 0 1>}
  n27 = NamedRerouteUsage()
  n28 = LinearInterpolate(
   B=LinearInterpolate(B=NamedRerouteUsage(), Alpha=ScalarParameter() {defaultValue=0.5, parameterName=HotCore_Amount})
   Alpha=NamedRerouteUsage()
  )
  n29 = LinearInterpolate(
   A=LinearInterpolate(A=LinearInterpolate(A=@n22.RGB, B=@n23.RGB, Alpha=@n24), B=LinearInterpolate(A=@n25.RGB, B=@n26.RGB, Alpha=@n27), Alpha=NamedRerouteUsage())
   B=LinearInterpolate(
    A=VectorParameter() {parameterName=Color_Star_Core, defaultValue=<rgba 0 1 1 1>}.RGB
    B=VectorParameter() {parameterName=Color_Star, defaultValue=<rgba 0 1 0 1>}.RGB
    Alpha=OneMinus(None=@n28)
   )
   Alpha=NamedRerouteUsage()
  )
  n30 = NamedRerouteUsage()
  NamedRerouteDeclaration(None=StaticSwitchParameter(True=@n21.RGB, False=LinearInterpolate(A=@n29, B=@n21.RGB, Alpha=@n30)) {parameterName=IsJustBorder}) {name=Final Color, nodeColor=<rgba 0 0 0 1>}
  n31 = NamedRerouteUsage()
  n32 = OneMinus(None=SmoothStep(Value=@n14.R) {constMax=0.5})
  NamedRerouteDeclaration(None=LinearInterpolate(A=Multiply(A=@n31, B=@n32), B=@n31, Alpha=ScalarParameter() {parameterName=BandStar})) {name=SoftGlowStar, nodeColor=<rgba 0 0.7891 1 1>}
  n33 = Add(
   A=MaterialFunctionCall(UVs=TextureCoordinate(), Texture Scale=ScalarParameter() {defaultValue=1.05, parameterName=Scale_Border_Outer}) {materialFunction=&ScaleUVsByCenter}.UVs
   B=AppendVector(A=Constant(), B=Constant())
  )
  NamedRerouteDeclaration(None=MaterialFunctionCall(
    DistanceField=TextureSample(UVs=@n33, Tex=@n7) {samplerType=SAMPLERTYPE_LinearColor, texture=&T_UI_Piligrim_StarDFChannelPack_64x}.R
    Threshold=ScalarParameter() {defaultValue=0.005, parameterName=Threshold_Borde_Outer}
   ) {materialFunction=&MF_UI_AntiAliasing}.Outer) {name=Border Mask, nodeColor=<rgba 1 0.0469 0 1>}
  OneMinus(None=@n14.R)
 ]
}
