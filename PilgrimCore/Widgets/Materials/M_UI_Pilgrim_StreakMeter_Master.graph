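# Translucent UI master material for the Pilgrim streak meter, drawing a ringed disc with a radial progress bar.
# Circle SDFs with anti-aliasing build procedural masks for rings, shadow, hot core and ornaments.
# An angle gradient and rotated end cap turn direct or lerped progress into the fill, recolored by Overdrive and Maxed.
# Static switches select meter, combo indicator (crown-textured gloss with perfect-streak colors) or ornament mask output, plus gradient-ring and attacked variants.
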Material /PilgrimCore/Widgets/Materials/M_UI_Pilgrim_StreakMeter_Master
graph {
 properties {
  bUsedWithStaticMesh=false
  blendMode=BLEND_Translucent
  materialDomain=MD_UI
  pixelDepthOffsetMode=PDOM_Legacy
  usedShadingModels=Default Lit
 }
 nodes [
  n1 = Subtract(B=ScalarParameter() {defaultValue=0.05, parameterName=AA_Threshold})
  n2 = NamedRerouteDeclaration(None=MaterialFunctionCall(
    DistanceField=MaterialFunctionCall(Size=@n1) {materialFunction=&MF_UI_Circle}.SDF
    Threshold=ScalarParameter() {defaultValue=0.05, parameterName=AA_Threshold}
   ) {materialFunction=&MF_UI_AntiAliasing}.Inner) {name=Outer Mask, nodeColor=<rgba 0 0 0 1>}
  LinearInterpolate(A=NamedRerouteUsage(), Alpha=@n2)
  MaterialFunctionCall() {materialFunction=&CustomRotator}
  n3 = NamedRerouteUsage()
  n4 = NamedRerouteUsage()
  n5 = LinearInterpolate(
   B=LinearInterpolate(
    A=ScalarParameter() {defaultValue=0.5, parameterName=GlossAmount}
    B=ScalarParameter() {defaultValue=0.5, parameterName=GlossAmount_Max}
    Alpha=ScalarParameter() {parameterName=IsMax}
   )
   Alpha=NamedRerouteUsage()
  )
  n6 = ScalarParameter() {defaultValue=0.75, parameterName=AlphaOutline}
  n7 = ScalarParameter() {defaultValue=0.5, parameterName=Alpha_Fill}
  n8 = NamedRerouteUsage()
  n9 = Multiply(
   A=LinearInterpolate(
    A=ScalarParameter() {defaultValue=0.5, parameterName=Alpha_Progress_Empty}
    B=ScalarParameter() {defaultValue=1, parameterName=Alpha_Progress_Fill}
    Alpha=NamedRerouteUsage()
   )
   B=NamedRerouteUsage()
  )
  n10 = LinearInterpolate(
   A=Multiply(A=@n2, B=LinearInterpolate(A=@n6, B=@n7, Alpha=@n8) {constA=1})
   Alpha=NamedRerouteDeclaration(None=Saturate(None=@n9)) {name=Progress Alpha Mask, nodeColor=<rgba 0 0 0 1>}
  )
  Multiply(
   A=StaticSwitchParameter(True=LinearInterpolate(A=@n3, Alpha=@n4) {constB=0}, False=StaticSwitchParameter(True=@n5, False=@n10) {parameterName=IsComboIndicator}) {parameterName=IsOrnamentMask}
   B=ScalarParameter() {defaultValue=1, parameterName=Alpha_All}
  )
  n11 = Subtract(
   A=OneMinus(None=ScalarParameter() {defaultValue=0.05, parameterName=RingSize})
   B=ScalarParameter() {defaultValue=0.05, parameterName=AA_Threshold}
  )
  NamedRerouteDeclaration(None=MaterialFunctionCall(
    DistanceField=MaterialFunctionCall(Size=@n11) {materialFunction=&MF_UI_Circle}.SDF
    Threshold=ScalarParameter() {defaultValue=0.05, parameterName=AA_Threshold}
   ) {materialFunction=&MF_UI_AntiAliasing}.Inner) {name=Inner Mask, nodeColor=<rgba 0 0 0 1>}
  n12 = ScalarParameter() {parameterName=Overdrive}
  n13 = LinearInterpolate(
   A=LinearInterpolate(
    A=VectorParameter() {parameterName=Color_Progress_Fill_Falloff, defaultValue=<rgba 0 0 0.25 1>}.RGB
    B=VectorParameter() {parameterName=Color_Progress_Fill_Falloff_OverDrive, defaultValue=<rgba 0 0 0.25 1>}.RGB
    Alpha=@n12
   )
   B=LinearInterpolate(
    A=VectorParameter() {parameterName=Color_Progress_Fill, defaultValue=<rgba 0 0 1 1>}.RGB
    B=VectorParameter() {parameterName=Color_Progress_Fill_OverDrive, defaultValue=<rgba 0 0 1 1>}.RGB
    Alpha=@n12
   )
   Alpha=NamedRerouteUsage()
  ) {constA=0.333, constB=0.5}
  n14 = LinearInterpolate(
   A=VectorParameter() {parameterName=HotCoreColor_Normal, defaultValue=<rgba 1 1 1 1>}.RGB
   B=VectorParameter() {parameterName=HotCoreColor_OverDrive, defaultValue=<rgba 1 1 1 1>}.RGB
   Alpha=ScalarParameter() {parameterName=Overdrive}
  )
  n15 = LinearInterpolate(B=Multiply(A=NamedRerouteUsage()) {constB=2}, Alpha=ScalarParameter() {parameterName=Maxed})
  n16 = LinearInterpolate(
   A=ScalarParameter() {parameterName=Progress_Start}
   B=ScalarParameter() {defaultValue=1, parameterName=Progress_Stop}
   Alpha=ScalarParameter() {parameterName=Anim_ProgressLerp}
  )
  n17 = ScalarParameter() {defaultValue=0.25, parameterName=Progress}
  n18 = SmoothStep(
   Min=StaticSwitchParameter(True=OneMinus(None=@n16), False=OneMinus(None=@n17)) {parameterName=LerpedProgress}
   Value=MaterialFunctionCall!(Vector or UVs) {materialFunction=/FortUILibrary/MaterialFunctions/Basics/MF_UI_CheaperVectorToRadial}.Vector Converted to Angle
  )
  n19 = NamedRerouteUsage()
  n20 = ScalarParameter() {defaultValue=1, parameterName=B_RaundedendCaps}
  n21 = Saturate(None=Add(A=Ceil(None=@n18), B=LinearInterpolate(B=@n19, Alpha=@n20)))
  NamedRerouteDeclaration(None=LinearInterpolate(
    A=VectorParameter() {parameterName=Color_Progress_Empty, defaultValue=<rgba 0.5 0.5 0.5 1>}.RGB
    B=LinearInterpolate(A=@n13, B=@n14, Alpha=@n15)
    Alpha=NamedRerouteDeclaration(None=@n21) {name=Progress Mask, nodeColor=<rgba 0 0 0 1>}
   )) {name=Progress Color, nodeColor=<rgba 0 0 0 1>}
  n22 = TextureCoordinate()
  n23 = AppendVector(A=Constant(), B=Constant() {r=0.025})
  n24 = ScalarParameter() {defaultValue=0.1, parameterName=ProgressBar_Padding}
  n25 = ScalarParameter() {defaultValue=0.05, parameterName=RingSize}
  n26 = Subtract(A=OneMinus(None=Add(A=@n24, B=@n25)), B=ScalarParameter() {defaultValue=0.05, parameterName=AA_Threshold})
  n27 = ScalarParameter() {defaultValue=0.05, parameterName=ProgressThickness}
  n28 = SmoothStep(
   Min=ScalarParameter() {defaultValue=-0.025, parameterName=ShadowMin}
   Max=ScalarParameter() {defaultValue=0.05, parameterName=ShadowMax}
   Value=MaterialFunctionCall(UVs=Add(A=@n22, B=@n23), Size=@n26, OutlineThickness=@n27) {materialFunction=&MF_UI_Circle}.SDFOutline
  ) {constMax=0.05, constMin=-0.025}
  NamedRerouteDeclaration(None=Saturate(None=OneMinus(None=@n28))) {name=Shadow Mask, nodeColor=<rgba 0 0 0 1>}
  n29 = MaterialFunctionCall(CenterPosition=AppendVector(A=Constant() {r=0.5}, B=Constant() {r=0.425}), Radius=Constant() {r=0.75}, Density=Constant() {r=3.75}) {materialFunction=&RadialGradientExponential}
  n30 = MaterialFunctionCall(CenterPosition=AppendVector(A=Constant() {r=0.5}, B=Constant() {r=0.5}), Radius=Constant() {r=0.4}, Density=Constant() {r=20}) {materialFunction=&RadialGradientExponential}
  n31 = Multiply(A=Multiply(A=OneMinus(None=@n29.RadialGradientExponential), B=Saturate(None=@n30.RadialGradientExponential))) {constB=10}
  n32 = ScalarParameter() {parameterName=CrownPostionX}
  n33 = ScalarParameter() {defaultValue=-0.35, parameterName=CrownPostionY}
  n34 = MaterialFunctionCall(UVs=Add(A=TextureCoordinate(), B=AppendVector(A=@n32, B=@n33)), Texture Scale=Constant() {r=0.15}) {materialFunction=&ScaleUVsByCenter}
  n35 = MaterialFunctionCall!(UVs=@n34.UVs, Alpha=TextureSample(UVs=@n34.UVs) {texture=/Game/UI/Foundation/Shell/Textures/Shell-Top-Crown}.A) {materialFunction=/Game/UI/Foundation/Items/ClampAlphaToTransparentBorder}
  n36 = LinearInterpolate(
   A=Saturate(None=Power(Base=@n31) {constExponent=1.25})
   Alpha=LinearInterpolate(B=Multiply(A=@n30.RadialGradientExponential, B=@n35.Result), Alpha=ScalarParameter() {defaultValue=1, parameterName=IsCombo})
  )
  n37 = Add(
   A=@n36
   B=LinearInterpolate(B=ScalarParameter() {defaultValue=0.333, parameterName=PerfectStreakBoost}, Alpha=ScalarParameter() {parameterName=PerfectStreakActive}) {constB=0.333}
  )
  NamedRerouteDeclaration(None=Saturate(None=LinearInterpolate(A=@n36, B=@n37, Alpha=@n36))) {name=Gloss Mask, nodeColor=<rgba 0 0 0 1>}
  n38 = TextureCoordinate()
  n39 = AppendVector(
   A=Constant() {r=0.5}
   B=Add(A=ScalarParameter() {defaultValue=0.05, parameterName=EndCapSize}, B=ScalarParameter() {defaultValue=0.05, parameterName=EndCapPadding})
  )
  n40 = ScalarParameter() {defaultValue=0.05, parameterName=EndCapSize}
  n41 = ScalarParameter() {defaultValue=16, parameterName=EndCapHardness}
  n42 = NamedRerouteDeclaration(None=Saturate(None=MaterialFunctionCall(UVs=@n38, CenterPosition=@n39, Radius=@n40, Density=@n41) {materialFunction=&RadialGradientExponential}.RadialGradientExponential)) {name=StartEndCap, nodeColor=<rgba 0 1 0.8359 1>}
  n43 = LinearInterpolate(
   A=ScalarParameter() {parameterName=Progress_Start}
   B=ScalarParameter() {defaultValue=1, parameterName=Progress_Stop}
   Alpha=ScalarParameter() {parameterName=Anim_ProgressLerp}
  )
  n44 = ScalarParameter() {defaultValue=0.25, parameterName=Progress}
  n45 = MaterialFunctionCall(UVs=@n38, "Rotation Angle (0-1)"=Multiply(A=StaticSwitchParameter(True=@n43, False=@n44) {parameterName=LerpedProgress}) {constB=-1}) {materialFunction=&CustomRotator}
  n46 = NamedRerouteDeclaration(None=Saturate(None=MaterialFunctionCall(UVs=@n45.Rotated Values, CenterPosition=@n39, Radius=@n40, Density=@n41) {materialFunction=&RadialGradientExponential}.RadialGradientExponential)) {name=DeltEndCap, nodeColor=<rgba 0 1 0.8359 1>}
  n47 = ScalarParameter() {parameterName=Progress_Start}
  n48 = ScalarParameter() {defaultValue=1, parameterName=Progress_Stop}
  n49 = ScalarParameter() {parameterName=Anim_ProgressLerp}
  n50 = MaterialFunctionCall!(
   A=StaticSwitchParameter(True=LinearInterpolate(A=@n47, B=@n48, Alpha=@n49), False=ScalarParameter() {defaultValue=0.25, parameterName=Progress}) {parameterName=LerpedProgress}
   B=Constant()
  ) {materialFunction=/FortUILibrary/MaterialFunctions/Basics/MF_UI_BooleanOperations}
  NamedRerouteDeclaration(None=Multiply(A=Add(A=@n42, B=@n46), B=LinearInterpolate(Alpha=@n50.IsEqual) {constA=1, constB=0})) {name=RoundedEndCaps, nodeColor=<rgba 0.5234 1 0 1>}
  NamedRerouteDeclaration(None=@n36) {name=GlossC olorMask, nodeColor=<rgba 0 0.3906 1 1>}
  n51 = MaterialFunctionCall(Size=@n26, OutlineThickness=@n27) {materialFunction=&MF_UI_Circle}
  n52 = SmoothStep(Value=@n51.SDFOutline) {constMax=-0.75, constMin=0.025}
  NamedRerouteDeclaration(None=Saturate(None=Multiply(A=@n52) {constB=32})) {name=HotCore, nodeColor=<rgba 0 0 0 1>}
  n53 = ScalarParameter() {defaultValue=0.4825, parameterName=Ornanment Offset X}
  n54 = ScalarParameter() {defaultValue=0.35, parameterName=Ornament Offset Y}
  n55 = MaterialFunctionCall(
   UVs=Add(A=TextureCoordinate(), B=AppendVector(A=@n53, B=@n54))
   Size=Subtract(
    A=ScalarParameter() {defaultValue=0.2, parameterName=Ornament_Size}
    B=ScalarParameter() {defaultValue=0.025, parameterName=AA_Threshold_Ornament}
   )
  ) {materialFunction=&MF_UI_Circle}
  n56 = ScalarParameter() {defaultValue=0.025, parameterName=AA_Threshold_Ornament}
  NamedRerouteDeclaration(None=OneMinus(None=MaterialFunctionCall(DistanceField=@n55.SDF, Threshold=@n56) {materialFunction=&MF_UI_AntiAliasing}.Inner)) {name=Ornament Mask Left, nodeColor=<rgba 0 0 0 1>}
  n57 = Multiply(A=ScalarParameter() {defaultValue=0.4825, parameterName=Ornanment Offset X}) {constB=-1}
  n58 = ScalarParameter() {defaultValue=0.35, parameterName=Ornament Offset Y}
  n59 = MaterialFunctionCall(
   UVs=Add(A=TextureCoordinate(), B=AppendVector(A=@n57, B=@n58))
   Size=Subtract(
    A=ScalarParameter() {defaultValue=0.2, parameterName=Ornament_Size}
    B=ScalarParameter() {defaultValue=0.025, parameterName=AA_Threshold_Ornament}
   )
  ) {materialFunction=&MF_UI_Circle}
  n60 = ScalarParameter() {defaultValue=0.025, parameterName=AA_Threshold_Ornament}
  NamedRerouteDeclaration(None=OneMinus(None=MaterialFunctionCall(DistanceField=@n59.SDF, Threshold=@n60) {materialFunction=&MF_UI_AntiAliasing}.Inner)) {name=Ornament Mask Right, nodeColor=<rgba 0 0 0 1>}
  n61 = MaterialFunctionCall(GradientStart=ScalarParameter() {defaultValue=0.785, parameterName=OrnamentBottomCutoff}, RotationAngle=Constant() {r=0.25}) {materialFunction=&MF_UI_LinearGradient}
  n62 = MaterialFunctionCall(
   GradientStart=ScalarParameter() {defaultValue=0.9, parameterName=OrnamentTopCutoff}
   GradientEnd=Constant()
   RotationAngle=Constant() {r=0.25}
  ) {materialFunction=&MF_UI_LinearGradient}
  n63 = Multiply(A=Saturate(None=Multiply(A=@n61.Result) {constB=256}), B=Saturate(None=Multiply(A=@n62.Result) {constB=256}))
  n64 = NamedRerouteUsage()
  n65 = NamedRerouteUsage()
  NamedRerouteDeclaration(None=Multiply(A=NamedRerouteDeclaration(None=@n63) {name=Bar Mask, nodeColor=<rgba 0 0 0 1>}, B=Multiply(A=@n64, B=@n65))) {name=Ornamet Mask, nodeColor=<rgba 0 0 0 1>}
  NamedRerouteDeclaration(None=MaterialFunctionCall(DistanceField=@n51.SDFOutline, Threshold=ScalarParameter() {defaultValue=0.05, parameterName=AA_Threshold}) {materialFunction=&MF_UI_AntiAliasing}.Inner) {name=Progress Ring Mask, nodeColor=<rgba 0 0 0 1>}
  n66 = VectorParameter() {parameterName=GlossColor, defaultValue=<rgba 1 0.75 1 1>}
  n67 = LinearInterpolate(
   A=VectorParameter() {parameterName=PerfectColor_A, defaultValue=<rgba 1 0.3333 0 1>}.RGB
   B=VectorParameter() {parameterName=PerfectColor_B, defaultValue=<rgba 1 1 0 1>}.RGB
   Alpha=NamedRerouteUsage()
  )
  n68 = ScalarParameter() {parameterName=PerfectStreakActive}
  n69 = VectorParameter() {parameterName=Color_Outline, defaultValue=<rgba 0 1 0 1>}
  n70 = VectorParameter() {parameterName=OuterRingColor, defaultValue=<rgba 1 1 0 1>}
  n71 = Add(A=TextureCoordinate()) {constB=-0.05}
  n72 = Saturate(None=Power(Base=ComponentMask(None=@n71) {g=true}) {constExponent=2.2})
  n73 = LinearInterpolate(
   A=@n69.RGB
   B=VectorParameter() {parameterName=Color_PeriphalInstrument, defaultValue=<rgba 0.0168 0.0185 0.0262 1>}.RGB
   Alpha=ScalarParameter() {parameterName=IsPeripheralInstrument}
  )
  n74 = VectorParameter() {parameterName=Color_Fill, defaultValue=<rgba 1 0 0 1>}
  n75 = Constant3Vector() {constant=<rgba 1 0 0 0>}
  n76 = ScalarParameter() {parameterName=IsAttackedAmount}
  n77 = LinearInterpolate(
   A=StaticSwitchParameter(
    True=LinearInterpolate(A=@n69.RGB, B=@n70.RGB, Alpha=@n72)
    False=NamedRerouteDeclaration(None=@n73) {name=OutlineColor, nodeColor=<rgba 0 0.6484 1 1>}
   ) {parameterName=GradientRing}
   B=StaticSwitchParameter(True=LinearInterpolate(A=@n74.RGB, B=@n75, Alpha=@n76) {constB=0}, False=@n74.RGB) {parameterName=CanBeAttacked}
   Alpha=NamedRerouteUsage()
  )
  n78 = NamedRerouteUsage()
  n79 = NamedRerouteUsage()
  StaticSwitchParameter(
   True=NamedRerouteUsage()
   False=StaticSwitchParameter(True=LinearInterpolate(A=@n66.RGB, B=@n67, Alpha=@n68), False=LinearInterpolate(A=@n77, B=@n78, Alpha=@n79)) {parameterName=IsComboIndicator}
  ) {parameterName=IsOrnamentMask}
 ]
}
