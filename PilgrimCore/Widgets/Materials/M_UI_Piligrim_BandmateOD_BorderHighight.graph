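# Translucent UI material for a border highlight on the Pilgrim bandmate OD widget, drawn procedurally as part of a rounded box.
# A vertical SmoothStep gradient blends two color parameters.
# The mask subtracts a second MF_UI_Box SDF, offset and resized by Height and Width, from an anti-aliased base box, then scales by Alpha_Final.
# Threshold insets are corrected for aspect ratio using the widget's pixel size.
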Material /PilgrimCore/Widgets/Materials/M_UI_Piligrim_BandmateOD_BorderHighight
graph {
 properties {
  bUsedWithStaticMesh=false
  blendMode=BLEND_Translucent
  materialDomain=MD_UI
  pixelDepthOffsetMode=PDOM_Legacy
  usedShadingModels=Default Lit
 }
 nodes [
  Constant() {r=1}
  Constant() {r=1}
  n1 = TextureCoordinate()
  LinearInterpolate(
   A=VectorParameter() {parameterName=ColorA, defaultValue=<rgba 0 1 0 1>}.RGB
   B=VectorParameter() {parameterName=ColorB, defaultValue=<rgba 1 0 0 1>}.RGB
   Alpha=SmoothStep(Min=ScalarParameter() {defaultValue=0.5, parameterName=AdjustGradient}, Value=ComponentMask(None=@n1) {g=true}) {constMin=0.5}
  )
  n2 = ScalarParameter() {defaultValue=0.05, parameterName=Threshold}
  n3 = MaterialFunctionCall() {materialFunction=&GetUserInterfaceUV}
  n4 = MaterialFunctionCall(Float2=@n3.Pixel Size) {materialFunction=&BreakOutFloat2Components}
  n5 = Divide(
   A=NamedRerouteDeclaration(None=@n4.G) {name=PixelSize.Y, nodeColor=<rgba 0 0 0 1>}
   B=Max(A=NamedRerouteDeclaration(None=@n4.R) {name=PixelSize.X, nodeColor=<rgba 0 0 0 1>}) {constB=0.001}
  )
  n6 = AppendVector(A=Multiply(A=@n2, B=Multiply(A=@n5)), B=@n2)
  n7 = Subtract(B=@n6)
  n8 = ScalarParameter() {defaultValue=0.5, parameterName=CornerRadius}
  n9 = StaticBool() {value=true}
  n10 = Saturate(None=MaterialFunctionCall(
    DistanceField=MaterialFunctionCall(Size=@n7, CornerRoundness=@n8, KeepAspectRatio=@n9) {materialFunction=&MF_UI_Box}.SDF
    Threshold=ScalarParameter() {defaultValue=0.05, parameterName=Threshold}
   ) {materialFunction=&MF_UI_AntiAliasing}.Inner)
  n11 = Constant()
  n12 = ScalarParameter() {defaultValue=0.1, parameterName=Height}
  n13 = AppendVector(A=ScalarParameter() {defaultValue=0.75, parameterName=Width}, B=Constant() {r=1})
  n14 = MaterialFunctionCall(
   UVs=Add(A=TextureCoordinate(), B=AppendVector(A=@n11, B=@n12))
   Size=Subtract(B=Multiply(A=@n6, B=@n13))
   CornerRoundness=ScalarParameter() {defaultValue=0.5, parameterName=CornerRadius}
   KeepAspectRatio=StaticBool() {value=true}
  ) {materialFunction=&MF_UI_Box}
  n15 = ScalarParameter() {defaultValue=0.05, parameterName=Threshold}
  n16 = NamedRerouteDeclaration(None=Saturate(None=MaterialFunctionCall(DistanceField=@n14.SDF, Threshold=@n15) {materialFunction=&MF_UI_AntiAliasing}.Inner)) {name=Box Mask_1, nodeColor=<rgba 0 0 0 1>}
  Multiply(
   A=Multiply(A=NamedRerouteDeclaration(None=@n10) {name=Box Mask, nodeColor=<rgba 0 0 0 1>}, B=OneMinus(None=@n16))
   B=ScalarParameter() {defaultValue=1, parameterName=Alpha_Final}
  )
  NamedRerouteDeclaration(None=@n3.9-Slice UV) {name=9sliceuv, nodeColor=<rgba 0 0 0 1>}
  NamedRerouteDeclaration(None=@n3.Normalized UV) {name=NormalizedUV, nodeColor=<rgba 0 0 0 1>}
 ]
}
