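# Translucent UI master material for font coloring, going by its name, in the PilgrimCore widgets.
# It linearly interpolates between two color parameters, Color and Color_B, using the scalar Anim_Color as the blend, so the tint can be animated.
# A separate scalar Alpha parameter, opaque by default, stands apart from the color blend; no textures are sampled.
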Material /PilgrimCore/Widgets/Materials/M_UI_Piligrim_FontColor_Master
graph {
 properties {
  bUsedWithStaticMesh=false
  blendMode=BLEND_Translucent
  materialDomain=MD_UI
  pixelDepthOffsetMode=PDOM_Legacy
  usedShadingModels=Default Lit
 }
 nodes [
  LinearInterpolate(
   A=VectorParameter() {parameterName=Color, defaultValue=<rgba 1 0 0 1>}.RGB
   B=VectorParameter() {parameterName=Color_B, defaultValue=<rgba 1 0 0 1>}.RGB
   Alpha=ScalarParameter() {parameterName=Anim_Color}
  )
  ScalarParameter() {defaultValue=1, parameterName=Alpha}
 ]
}
