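# Translucent UI material that draws a rounded box with independently set corners and optional horizontal skew.
# An MF_UI_Box SDF, sized using pixel-aspect terms, becomes a mask through anti-aliasing or smoothstep, multiplied by animated alphas.
# Static switches pick color from selectable-tab button states, a gradient, animated colors or a player-darkened color; tab mode also multiplies state opacity.
# The PixelRadius switch is unconnected, as the box takes per-corner values directly.
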Material /PilgrimCore/Widgets/Materials/M_UI_Piligrim_NonUnifrom_RoundedBox
graph {
 properties {
  bUsedWithStaticMesh=false
  blendMode=BLEND_Translucent
  materialDomain=MD_UI
  pixelDepthOffsetMode=PDOM_Legacy
  usedShadingModels=Default Lit
 }
 nodes [
  n1 = NamedRerouteUsage()
  n2 = NamedRerouteUsage()
  Max(A=@n1, B=@n2)
  n3 = VectorParameter() {parameterName=Color_Start, defaultValue=<rgba 1 0 0 1>}
  n4 = VectorParameter() {parameterName=Color_Stop, defaultValue=<rgba 0 1 0 1>}
  n5 = ScalarParameter() {parameterName=Anim_Color}
  NamedRerouteDeclaration(None=LinearInterpolate(
    A=LinearInterpolate(A=@n3.RGB, B=@n4.RGB, Alpha=@n5)
    B=VectorParameter() {parameterName=Color_SelectedHover, defaultValue=<rgba 1 0 0 1>}.RGB
    Alpha=ScalarParameter() {parameterName=Anim_Color_SelectedHover}
   )) {name=AnimatedColor, nodeColor=<rgba 0.9688 0 1 1>}
  n6 = MaterialFunctionCall(Float2=MaterialFunctionCall() {materialFunction=&GetUserInterfaceUV}.Pixel Size) {materialFunction=&BreakOutFloat2Components}
  n7 = NamedRerouteDeclaration(None=@n6.R) {name=PixelX, nodeColor=<rgba 1 0.0469 0 1>}
  n8 = NamedRerouteDeclaration(None=@n6.G) {name=PixelY, nodeColor=<rgba 0 0.1562 1 1>}
  n9 = Max(A=@n7, B=@n8)
  NamedRerouteDeclaration(None=Divide(A=Min(A=@n7, B=@n8), B=Max(A=@n9) {constB=0.001})) {name=Ratio, nodeColor=<rgba 1 0.0234 0 1>}
  NamedRerouteDeclaration(None=MaterialFunctionCall!(A=NamedRerouteUsage(), B=NamedRerouteUsage()) {materialFunction=/FortUILibrary/MaterialFunctions/Basics/MF_UI_BooleanOperations}.IsGreater) {name=YisLarger, nodeColor=<rgba 0.2422 0 1 1>}
  ScalarParameter() {defaultValue=0.25, parameterName=Corner_Radius}
  StaticBool()
  n10 = ScalarParameter() {defaultValue=16, parameterName=CornerRadiusPixels}
  n11 = Min(A=@n1, B=@n2)
  n12 = MaterialFunctionCall(
   X=ScalarParameter() {defaultValue=0.25, parameterName=Corner_TopLeft}
   Y=ScalarParameter() {defaultValue=-0.1, parameterName=Corner_BottomLeft}
   Z=ScalarParameter() {defaultValue=0.25, parameterName=Corner_TopRight}
   A=ScalarParameter() {defaultValue=-0.1, parameterName=Corner_BottomRight}
  ) {materialFunction=&MakeFloat4}
  StaticSwitchParameter(True=Divide(A=Multiply(A=@n10) {constA=-0.9746, constB=2}, B=Max(A=@n11) {constB=0.001}), False=@n12.Result) {parameterName=PixelRadius}
  n13 = MaterialFunctionCall!(Idle, Idle Opacity, Hovered, Hovered Opacity, Focused, Focused Opacity, Pressed, Pressed Opacity, Disabled, Disabled Opacity) {materialFunction=/FortUILibrary/Systems/ModularButton/MaterialFunctions/MF_UIKit_ButtonStatesBase_Lerp}
  n14 = MaterialFunctionCall!(
   Base States=@n13.Color
   Base States Opacity=@n13.Color
   Selected_Idle
   Selected_Idle Opacity
   Selected_Hovered
   Selected_Hovered Opacity
   Selected_Focused
   Selected_Focused Opacity
   Selected_Pressed
   Selected_Pressed Opacity
   Selected_Disabled
   Selected_Disabled Opacity
  ) {materialFunction=/FortUILibrary/Systems/ModularButton/MaterialFunctions/MF_UIKit_ButtonStatesSelectable_Lerp}
  n15 = VectorParameter() {parameterName=Color, defaultValue=<rgba 1 0 0 1>}
  n16 = VectorParameter() {parameterName=Color_A, defaultValue=<rgba 0 1 0 1>}
  n17 = VectorParameter() {parameterName=Color_B, defaultValue=<rgba 1 0 0 1>}
  n18 = ScalarParameter() {parameterName=OffsetGradient}
  n19 = ComponentMask(None=TextureCoordinate()) {r=true}
  n20 = Saturate(None=Power(Base=SmoothStep(Min=@n18, Value=@n19) {constMin=0.5}, Exp=ScalarParameter() {defaultValue=1, parameterName=TaperGradient}))
  n21 = NamedRerouteUsage()
  n22 = VectorParameter() {parameterName=Color, defaultValue=<rgba 1 0 0 1>}
  n23 = LinearInterpolate(A=@n22.RGB, B=Multiply(A=@n22.RGB) {constB=0.1}, Alpha=ScalarParameter() {parameterName=IsPlayerColor})
  StaticSwitchParameter(
   True=LinearInterpolate(A=@n14.Color, B=Multiply(A=@n15.RGB) {constB=0.8}, Alpha=ScalarParameter() {parameterName=IsPlayerColor})
   False=StaticSwitchParameter(
    True=LinearInterpolate(A=@n16.RGB, B=@n17.RGB, Alpha=@n20) {constA=-2.5904, constAlpha=-0.21}
    False=StaticSwitchParameter(True=@n21, False=@n23) {parameterName=AnimatedColor}
   ) {parameterName=IsGradient}
  ) {parameterName=IsSelectableTabIcon}
  n24 = Multiply(A=ScalarParameter() {parameterName=SkewAmount}, B=NamedRerouteUsage())
  n25 = LinearInterpolate(A=Multiply(A=@n24) {constB=-1}, B=@n24, Alpha=ComponentMask(None=TextureCoordinate()) {g=true}) {constA=-0.5, constB=0.5}
  n26 = Constant()
  n27 = ScalarParameter() {parameterName=SkewAmount}
  n28 = OneMinus(None=Multiply(A=Multiply(A=@n27) {constB=1.5}, B=NamedRerouteUsage()))
  n29 = Constant() {r=1}
  n30 = ScalarParameter() {defaultValue=0.025, parameterName=Threshold}
  n31 = Multiply(A=@n30, B=NamedRerouteUsage())
  n32 = AppendVector(A=@n30, B=@n31)
  n33 = AppendVector(A=@n31, B=@n30)
  n34 = NamedRerouteUsage()
  n35 = MaterialFunctionCall(
   UVs=Add(A=TextureCoordinate(), B=AppendVector(A=@n25, B=@n26))
   Size=Subtract(A=AppendVector(A=@n28, B=@n29), B=LinearInterpolate(A=@n32, B=@n33, Alpha=@n34))
   CornerRoundness=@n12.Result
   KeepAspectRatio=StaticBool() {value=true}
  ) {materialFunction=&MF_UI_Box}
  n36 = ScalarParameter() {defaultValue=0.05, parameterName=Smoothness}
  n37 = StaticSwitchParameter(
   True=MaterialFunctionCall(DistanceField=@n35.SDF, Threshold=ScalarParameter() {defaultValue=0.025, parameterName=Threshold}) {materialFunction=&MF_UI_AntiAliasing}.Inner
   False=SmoothStep(Max=Multiply(A=@n36) {constB=-1}, Value=@n35.SDF) {constMax=-0.05}
  ) {defaultValue=true, parameterName=AARoute}
  n38 = LinearInterpolate(
   A=ScalarParameter() {defaultValue=1, parameterName=Alpha_Start}
   B=ScalarParameter() {defaultValue=1, parameterName=Alpha_Hover_Stop}
   Alpha=ScalarParameter() {parameterName=Anim_Alpha_Hover}
  )
  n39 = ScalarParameter() {defaultValue=1, parameterName=Alpha_Stop}
  n40 = ScalarParameter() {parameterName=Anim_Alpha_Selected}
  n41 = Multiply(
   A=Multiply(A=NamedRerouteDeclaration(None=@n37) {name=Box Mask, nodeColor=<rgba 0 0 0 1>}, B=LinearInterpolate(A=@n38, B=@n39, Alpha=@n40))
   B=ScalarParameter() {defaultValue=1, parameterName=Alpha_Final}
  )
  StaticSwitchParameter(True=Multiply(A=@n14.Opacity, B=@n41), False=@n41) {parameterName=IsSelectableTabIcon}
 ]
}
