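# Translucent UI master material that draws two layers of drifting rounded-square particles, each with fill and outline.
# Per layer, MF_UI_ParticleGrid randomizes row speed, cell visibility and size, while IsHorizontal, Left and Up set panning direction.
# A directional gradient thresholded by Anim_Scale scales the MF_UI_Box shapes, and the combined size masks blend Color-A to Color-B.
# Opacity combines both shape masks with a sine edge falloff, vertical gradient and Alpha_Final.
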Material /PilgrimCore/Widgets/Materials/M_UI_Piligrim_SquareParticles_Master
graph {
 properties {
  bUsedWithStaticMesh=false
  blendMode=BLEND_Translucent
  materialDomain=MD_UI
  pixelDepthOffsetMode=PDOM_Legacy
  usedShadingModels=Default Lit
 }
 nodes [
  n1 = TextureCoordinate()
  NamedRerouteDeclaration(None=ComponentMask(None=@n1) {r=true}) {name=HGradient, nodeColor=<rgba 1 0.6797 0 1>}
  n2 = MaterialFunctionCall() {materialFunction=&MF_UI_UVScaling}
  n3 = ScalarParameter() {defaultValue=2, parameterName=Set2_X_PanSpeed}
  n4 = Constant()
  n5 = MaterialFunctionCall!(
   UV
   Global Offset
   Global Rotation
   Tiling=ScalarParameter() {defaultValue=5, parameterName=Set2_Tile}
   Cell Offset=@n4
   Cell Scale
   Cell Rotation=@n4
   Random Seed=Constant() {r=1.0769}
  ) {materialFunction=/FortUILibrary/MaterialFunctions/Basics/MF_UI_ParticleGrid}
  n6 = SmoothStep(Value=StaticSwitchParameter(True=@n5.Random Rows, False=@n5.Random Rows) {parameterName=IsHorizontal}) {constMax=0.25, constMin=-0.25}
  n7 = Multiply(A=LinearInterpolate(A=Multiply(A=@n3) {constB=-1}, B=@n3, Alpha=ScalarParameter() {defaultValue=1, parameterName=Left}), B=@n6)
  n8 = Constant()
  n9 = ScalarParameter() {defaultValue=3, parameterName=Set2_Y_PanSpeed}
  n10 = Multiply(A=LinearInterpolate(A=Multiply(A=@n9) {constB=-1}, B=@n9, Alpha=NamedRerouteUsage()), B=@n6)
  n11 = Panner(
   Coordinate=StaticSwitchParameter(True=@n2.ScaleY, False=@n2.ScaleY) {parameterName=IsHorizontal}
   Speed=StaticSwitchParameter(True=AppendVector(A=@n7, B=@n8), False=AppendVector(A=@n8, B=@n10)) {parameterName=IsHorizontal}
  ) {speedY=0.25}
  n12 = Constant()
  n13 = MaterialFunctionCall(
   UV=@n11
   Tiling=ScalarParameter() {defaultValue=5, parameterName=Set2_Tile}
   Cell Offset=@n12
   Cell Rotation=@n12
   Random Seed=ScalarParameter() {defaultValue=3.5095, parameterName=Set2_RandomSeed}
  ) {materialFunction=&MF_UI_ParticleGrid}
  n14 = Constant()
  n15 = MaterialFunctionCall(
   UV=@n11
   Tiling=ScalarParameter() {defaultValue=5, parameterName=Set2_Tile}
   Cell Offset=@n14
   Cell Rotation=@n14
   Random Seed=Multiply(A=Constant() {r=5.7646}, B=ScalarParameter() {defaultValue=1, parameterName=Set2_SeedAdjustment})
  ) {materialFunction=&MF_UI_ParticleGrid}
  n16 = Multiply(A=Ceil(None=SmoothStep(Value=@n15.Random Cell Values) {constMin=0.5}), B=@n13.Random Cell Values)
  n17 = NamedRerouteUsage()
  n18 = LinearInterpolate(A=OneMinus(None=@n17), B=@n17, Alpha=ScalarParameter() {defaultValue=1, parameterName=Left})
  n19 = NamedRerouteUsage()
  n20 = LinearInterpolate(A=OneMinus(None=@n19), B=@n19, Alpha=NamedRerouteUsage())
  n21 = LinearInterpolate(
   B=@n16
   Alpha=SmoothStep(Min=ScalarParameter() {parameterName=Anim_Scale}, Value=StaticSwitchParameter(True=@n18, False=@n20) {parameterName=IsHorizontal})
  )
  n22 = MaterialFunctionCall(UVs=@n13.Cells UV, Size=@n21, CornerRoundness=Multiply(A=@n16) {constB=0.5}, OutlineThickness=Constant() {r=0.125}) {materialFunction=&MF_UI_Box}
  NamedRerouteDeclaration(None=Add(A=Multiply(A=@n22.Fill) {constB=0.5}, B=@n22.Outline)) {name=Set2 Mask, nodeColor=<rgba 0 0 0 1>}
  NamedRerouteDeclaration(None=ComponentMask(None=@n1) {g=true}) {name=VGradient, nodeColor=<rgba 1 0.6797 0 1>}
  n23 = Constant()
  n24 = MaterialFunctionCall() {materialFunction=&MF_UI_UVScaling}
  n25 = ScalarParameter() {defaultValue=2, parameterName=Set1_X_PanSpeed}
  n26 = Constant()
  n27 = MaterialFunctionCall(Tiling=ScalarParameter() {defaultValue=7, parameterName=Set1_Tile}, Cell Offset=@n26, Cell Rotation=@n26, Random Seed=Constant() {r=2.9415}) {materialFunction=&MF_UI_ParticleGrid}
  n28 = SmoothStep(Value=StaticSwitchParameter(True=@n27.Random Rows, False=@n27.Random Rows) {parameterName=IsHorizontal}) {constMax=0.5, constMin=-0.25}
  n29 = Multiply(A=LinearInterpolate(A=Multiply(A=@n25) {constB=-1}, B=@n25, Alpha=ScalarParameter() {defaultValue=1, parameterName=Left}), B=@n28)
  n30 = Constant()
  n31 = ScalarParameter() {defaultValue=2, parameterName=Set1_Y_PanSpeed}
  n32 = Multiply(A=LinearInterpolate(A=Multiply(A=@n31) {constB=-1}, B=@n31, Alpha=NamedRerouteUsage()), B=@n28)
  n33 = Panner(
   Coordinate=StaticSwitchParameter(True=@n24.ScaleY, False=@n24.ScaleY) {parameterName=IsHorizontal}
   Speed=StaticSwitchParameter(True=AppendVector(A=@n29, B=@n30), False=AppendVector(A=@n30, B=@n32)) {parameterName=IsHorizontal}
  ) {speedY=0.25}
  n34 = MaterialFunctionCall(
   UV=@n33
   Tiling=ScalarParameter() {defaultValue=7, parameterName=Set1_Tile}
   Cell Offset=@n23
   Cell Rotation=@n23
   Random Seed=Multiply(A=Constant() {r=5.7646}, B=ScalarParameter() {defaultValue=1, parameterName=Set1_SeedAdjustment})
  ) {materialFunction=&MF_UI_ParticleGrid}
  n35 = Constant()
  n36 = MaterialFunctionCall(
   UV=@n33
   Tiling=ScalarParameter() {defaultValue=7, parameterName=Set1_Tile}
   Cell Offset=@n35
   Cell Rotation=@n35
   Random Seed=ScalarParameter() {defaultValue=4.4388, parameterName=Set1_RandomSeed}
  ) {materialFunction=&MF_UI_ParticleGrid}
  n37 = Multiply(A=Ceil(None=SmoothStep(Value=@n34.Random Cell Values) {constMax=1.25, constMin=0.75}), B=@n36.Random Cell Values)
  n38 = NamedRerouteUsage()
  n39 = LinearInterpolate(A=OneMinus(None=@n38), B=@n38, Alpha=ScalarParameter() {defaultValue=1, parameterName=Left})
  n40 = NamedRerouteUsage()
  n41 = LinearInterpolate(A=OneMinus(None=@n40), B=@n40, Alpha=NamedRerouteUsage())
  n42 = LinearInterpolate(
   B=@n37
   Alpha=SmoothStep(Min=ScalarParameter() {parameterName=Anim_Scale}, Value=StaticSwitchParameter(True=@n39, False=@n41) {parameterName=IsHorizontal})
  )
  NamedRerouteDeclaration(None=@n42) {name=Set1 UV Mask, nodeColor=<rgba 0.7578 0 1 1>}
  NamedRerouteDeclaration(None=@n21) {name=Set2 UV Mask, nodeColor=<rgba 0.875 1 0 1>}
  n43 = Add(A=NamedRerouteUsage(), B=NamedRerouteUsage())
  NamedRerouteDeclaration(None=LinearInterpolate(
    A=VectorParameter() {parameterName=Color-A, defaultValue=<rgba 0 1 0 1>}.RGB
    B=VectorParameter() {parameterName=Color-B, defaultValue=<rgba 1 0 0 1>}.RGB
    Alpha=Saturate(None=@n43)
   )) {name=Color, nodeColor=<rgba 0 1 0.0391 1>}
  n44 = MaterialFunctionCall(UVs=@n36.Cells UV, Size=@n42, CornerRoundness=Multiply(A=@n37) {constB=0.5}, OutlineThickness=Constant() {r=0.25}) {materialFunction=&MF_UI_Box}
  n45 = Multiply(A=@n44.Fill) {constB=0.5}
  n46 = Add(A=NamedRerouteDeclaration(None=Add(A=@n45, B=@n44.Outline)) {name=Set1 Mask, nodeColor=<rgba 0 0 0 1>}, B=NamedRerouteUsage())
  n47 = ScalarParameter() {defaultValue=0.5, parameterName=EdgeFalloff}
  n48 = Sine(None=TextureCoordinate()) {period=2}
  n49 = StaticSwitchParameter(True=ComponentMask(None=@n48) {g=true}, False=ComponentMask(None=@n48) {r=true}) {parameterName=IsHorizontal}
  n50 = Multiply(
   A=Multiply(A=Saturate(None=@n46), B=SmoothStep(Max=@n47, Value=@n49) {constMax=0.5})
   B=ScalarParameter() {defaultValue=1, parameterName=Alpha_Final}
  )
  n51 = ScalarParameter() {defaultValue=1, parameterName=Alpha_VerticalLinearGradient}
  n52 = LinearInterpolate(Alpha=NamedRerouteUsage()) {constA=1, constB=0}
  n53 = SmoothStep(
   Min=OneMinus(None=@n52)
   Max=@n52
   Value=MaterialFunctionCall() {materialFunction=&LinearGradient}.VGradient
  ) {constMax=0, constMin=1}
  NamedRerouteDeclaration(None=Multiply(A=@n50, B=LinearInterpolate(A=@n50, B=@n51, Alpha=@n53))) {name=Fill, nodeColor=<rgba 0.5234 1 0 1>}
  NamedRerouteDeclaration(None=ScalarParameter() {defaultValue=1, parameterName=Up}) {name=b_Up, nodeColor=<rgba 0.4531 1 0 1>}
  NamedRerouteUsage()
  NamedRerouteUsage()
 ]
}
