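# Translucent UI material for a song vote background: a rounded box with an eased glow, gradient colors and an animated triangle pattern.
# MF_UI_Box builds the fill and glow masks, and the Vote Color State parameter blends both shape and color between no-vote and voted looks.
# The pattern pans two screen-aligned triangle textures at different tilings, using time sines to cycle their RGB channels and crossfade the layers.
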Material /PilgrimCore/Widgets/Materials/M_UI_SongVote_BackgroundM
graph {
 properties {
  bAutomaticallySetUsageInEditor=false
  bUsedWithStaticMesh=false
  blendMode=BLEND_Translucent
  materialDomain=MD_UI
  usedShadingModels=Default Lit
 }
 nodes [
  n1 = MaterialFunctionCall(Size=Constant() {r=0.95}, CornerRoundness=ScalarParameter() {defaultValue=0.04, parameterName=CornerRoundness}) {materialFunction=&MF_UI_Box}
  n2 = Constant() {r=0.95}
  n3 = ScalarParameter() {defaultValue=0.04, parameterName=CornerRoundness}
  n4 = ScalarParameter() {defaultValue=0.05, parameterName=Glow}
  n5 = LinearInterpolate(
   B=MaterialFunctionCall("Value (0-1)"=MaterialFunctionCall(Size=@n2, CornerRoundness=@n3, GlowMax=@n4) {materialFunction=&MF_UI_Box}.Fill) {materialFunction=&MF_UI_EaseCurves}.EaseInCubic
   Alpha=ScalarParameter() {defaultValue=0.8, parameterName=Glow OP}
  )
  NamedRerouteDeclaration(None=LinearInterpolate(
    A=NamedRerouteDeclaration(None=@n1.Fill) {name=Mask_NonGlow, nodeColor=<rgba 0 0.7187 1 1>}
    B=Max(A=@n1.Fill, B=@n5)
    Alpha=ScalarParameter() {defaultValue=1, parameterName=Vote Color State}
   )) {name=Base, nodeColor=<rgba 1 0.1641 0 1>}
  n6 = MaterialFunctionCall() {materialFunction=&LinearGradient}
  n7 = LinearInterpolate(A=NamedRerouteUsage(), B=NamedRerouteUsage(), Alpha=@n6.VGradient)
  n8 = MaterialFunctionCall() {materialFunction=&LinearGradient}
  n9 = LinearInterpolate(
   A=LinearInterpolate(A=NamedRerouteUsage(), B=NamedRerouteUsage(), Alpha=@n6.VGradient)
   B=NamedRerouteUsage()
   Alpha=LinearInterpolate(B=NamedRerouteUsage(), Alpha=SmoothStep(Value=@n8.VGradient) {constMin=0.1})
  )
  n10 = NamedRerouteUsage()
  NamedRerouteDeclaration(None=LinearInterpolate(
    A=NamedRerouteUsage()
    B=LinearInterpolate(A=@n7, B=@n9, Alpha=@n10)
    Alpha=ScalarParameter() {defaultValue=1, parameterName=Vote Color State}
   )) {name=Color, nodeColor=<rgba 0.4766 1 0 1>}
  n11 = NamedRerouteDeclaration(None=MaterialFunctionCall() {materialFunction=&ScreenAlignedUVs}."X 100%, Y 100%") {name=ScreenSizeUV, nodeColor=<rgba 0 0 0 1>}
  n12 = AppendVector(A=Constant() {r=5}, B=Constant() {r=5})
  n13 = Constant()
  n14 = Constant() {r=0.015}
  n15 = TextureSample(UVs=Panner(Coordinate=Multiply(A=@n11, B=@n12), Speed=AppendVector(A=@n13, B=@n14))) {samplerType=SAMPLERTYPE_LinearColor, texture=/Game/UI/Foundation/Textures/Pattern/T_Triangle_BattlePassSmall}
  n16 = MaterialFunctionCall(
   1=SmoothStep(Value=@n15.R) {constMax=0.2, constMin=0.08}
   2=SmoothStep(Value=@n15.G) {constMax=0.2, constMin=0.08}
   3=SmoothStep(Value=@n15.B) {constMax=0.2, constMin=0.08}
   A=MaterialFunctionCall(Speed=Constant() {r=0.1}) {materialFunction=&DebugTimeSine}.Result
  ) {materialFunction=&Lerp_Multiple_Float3}
  n17 = NamedRerouteUsage()
  n18 = AppendVector(A=Constant() {r=6}, B=Constant() {r=6})
  n19 = Constant()
  n20 = Constant() {r=0.015}
  n21 = TextureSample(UVs=Panner(Coordinate=Multiply(A=@n17, B=@n18), Speed=AppendVector(A=@n19, B=@n20))) {samplerType=SAMPLERTYPE_LinearColor, texture=&T_Triangle_BattlePassSmall}
  n22 = MaterialFunctionCall(
   1=SmoothStep(Value=@n21.R) {constMax=0.2, constMin=0.08}
   2=SmoothStep(Value=@n21.G) {constMax=0.2, constMin=0.08}
   3=SmoothStep(Value=@n21.B) {constMax=0.2, constMin=0.08}
   A=MaterialFunctionCall(Speed=Constant() {r=0.08}) {materialFunction=&DebugTimeSine}.Result
  ) {materialFunction=&Lerp_Multiple_Float3}
  n23 = MaterialFunctionCall(Speed=Constant() {r=0.05}) {materialFunction=&DebugTimeSine}
  NamedRerouteDeclaration(None=Multiply(A=LinearInterpolate(A=@n16.Lerp 3 Inputs, B=@n22.Lerp 3 Inputs, Alpha=@n23.Result))) {name=TriangelePattern, nodeColor=<rgba 0 0.0625 1 1>}
  NamedRerouteDeclaration(None=@n16.Lerp 3 Inputs) {name=TriangelePattern1, nodeColor=<rgba 0 0.1094 1 1>}
  NamedRerouteDeclaration(None=@n22.Lerp 3 Inputs) {name=TrianglePattern2, nodeColor=<rgba 0.8984 1 0 1>}
  NamedRerouteDeclaration(None=@n5) {name=MaskGlow, nodeColor=<rgba 0.0078 1 0 1>}
  NamedRerouteDeclaration(None=VectorParameter() {parameterName=Color 1, defaultValue=<rgba 0.0423 0.0437 0.4287 1>}.RGB) {name=CL1, nodeColor=<rgba 0 0.8828 1 1>}
  NamedRerouteDeclaration(None=VectorParameter() {parameterName=Color 2, defaultValue=<rgba 0.0823 0.0545 0.3095 1>}.RGB) {name=CL2, nodeColor=<rgba 1 0.7969 0 1>}
  NamedRerouteDeclaration(None=VectorParameter() {parameterName=Glow_CL 2, defaultValue=<rgba 0 0.2297 1 1>}.RGB) {name=Glow_CL 2, nodeColor=<rgba 0 1 0.5547 1>}
  NamedRerouteDeclaration(None=VectorParameter() {parameterName=Glow_CL 1, defaultValue=<rgba 0 0.5033 1 1>}.RGB) {name=Glow_CL 1, nodeColor=<rgba 1 0.2344 0 1>}
  NamedRerouteDeclaration(None=VectorParameter() {parameterName=Pattern Color, defaultValue=<rgba 0.0905 0.0925 0.599 1>}.RGB) {name=PaternCl, nodeColor=<rgba 0 1 0.0391 1>}
  NamedRerouteDeclaration(None=VectorParameter() {parameterName=NoVote Color, defaultValue=<rgba 0.0319 0.0262 0.2122 1>}.RGB) {name=NoVote_CL, nodeColor=<rgba 1 0 0.7734 1>}
  NamedRerouteUsage()
  NamedRerouteUsage()
 ]
}
