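# Translucent UI master material for streak meter center effects, with an IsParticles static switch choosing between two looks.
# The default look multiplies panning radial sine waves by a radial gradient, blending two colors, with optional 9-slice UVs.
# The particle look combines thresholded noise sparks panning radially with a noisy, color-cycling circle flare, animated by scale and alpha parameters.
# Noise samples use explicit UV derivatives for mip selection on radial coordinates.
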Material /PilgrimCore/Widgets/Materials/M_UI_StreakMeter_CenterFX_Master
graph {
 properties {
  bUsedWithStaticMesh=false
  blendMode=BLEND_Translucent
  materialDomain=MD_UI
  pixelDepthOffsetMode=PDOM_Legacy
  usedShadingModels=Default Lit
 }
 nodes [
  n1 = TextureCoordinate()
  n2 = MaterialFunctionCall() {materialFunction=/FortUILibrary/MaterialFunctions/Basics/MF_UI_WIP_9SliceUV}
  n3 = ScalarParameter() {parameterName=9Slice}
  n4 = MaterialFunctionCall!(Vector or UVs=LinearInterpolate(
    A=TextureCoordinate()
    B=MaterialFunctionCall() {materialFunction=&MF_UI_WIP_9SliceUV}.Result
    Alpha=ScalarParameter() {parameterName=9Slice}
   )) {materialFunction=/FortUILibrary/MaterialFunctions/Basics/MF_UI_CheaperVectorToRadial}
  n5 = Multiply(A=Time(), B=ScalarParameter() {defaultValue=1, parameterName=PanSpeed})
  n6 = Sine(None=Panner(Coordinate=Multiply(A=@n4.Radial Coordinates) {constB=3}, Time=@n5) {speedY=-0.75})
  n7 = Sine(None=Panner(Coordinate=Multiply(A=@n4.Radial Coordinates) {constB=4}, Time=@n5) {speedY=-0.75})
  n8 = Multiply(
   A=MaterialFunctionCall(UVs=LinearInterpolate(A=@n1, B=@n2.Result, Alpha=@n3)) {materialFunction=&RadialGradientExponential}.RadialGradientExponential
   B=Multiply(A=ComponentMask(None=@n6) {g=true}, B=ComponentMask(None=@n7) {g=true})
  )
  Multiply(
   A=StaticSwitchParameter(True=NamedRerouteUsage(), False=@n8) {parameterName=IsParticles}
   B=ScalarParameter() {defaultValue=1, parameterName=Alpha_Final}
  )
  n9 = OneMinus(None=MaterialFunctionCall(Size=Constant()) {materialFunction=&MF_UI_Circle}.SDF)
  NamedRerouteDeclaration(None=@n9) {name=CircleMask, nodeColor=<rgba 0.8047 0 1 1>}
  n10 = TextureObject() {samplerType=SAMPLERTYPE_LinearColor, texture=&HighContrastNoise}
  NamedRerouteDeclaration(None=@n10) {name=Tex, nodeColor=<rgba 1 0 0.8203 1>}
  n11 = NamedRerouteUsage()
  n12 = Saturate(None=Subtract(A=Multiply(A=@n11) {constB=3}))
  n13 = Saturate(None=Multiply(A=@n9))
  n14 = LinearInterpolate(Alpha=@n13) {constA=0.9, constB=0.8}
  n15 = TextureCoordinate()
  n16 = MaterialFunctionCall(Vector or UVs=@n15) {materialFunction=&MF_UI_CheaperVectorToRadial}
  n17 = NamedRerouteUsage()
  n18 = NamedRerouteUsage()
  n19 = TextureSample!(
   UVs=Panner(Coordinate=Multiply(A=@n16.Radial Coordinates) {constB=2}) {speedX=0.25, speedY=-2}
   Tex=@n10
   "DDX(UVs)"=@n17
   "DDY(UVs)"=@n18
   Apply View MipBias
  ) {mipValueMode=TMVM_Derivative, samplerType=SAMPLERTYPE_LinearColor, texture=&HighContrastNoise}
  n20 = LinearInterpolate(Alpha=@n13) {constA=0.925, constB=0.8}
  n21 = TextureSample(UVs=Panner(Coordinate=Multiply(A=@n16.Radial Coordinates) {constB=5}) {speedX=0.125, speedY=-3}, Tex=@n10, "DDX(UVs)"=@n17, "DDY(UVs)"=@n18) {mipValueMode=TMVM_Derivative, samplerType=SAMPLERTYPE_LinearColor, texture=&HighContrastNoise}
  n22 = Add(
   A=Multiply(A=SmoothStep(Min=@n14, Value=@n19.G) {constMin=0.8}) {constB=20}
   B=Multiply(A=SmoothStep(Min=@n20, Value=@n21.B) {constMin=0.8}) {constB=20}
  )
  n23 = NamedRerouteDeclaration(None=Saturate(None=Multiply(A=@n12, B=@n22))) {name=particles mask, nodeColor=<rgba 1 0.5625 0 1>}
  n24 = NamedRerouteUsage()
  n25 = ScalarParameter() {parameterName=AlphaStart}
  n26 = ScalarParameter() {defaultValue=1, parameterName=AlphaStop}
  n27 = ScalarParameter() {defaultValue=1, parameterName=Anim_Alpha}
  NamedRerouteDeclaration(None=Multiply(A=Add(A=@n23, B=@n24), B=LinearInterpolate(A=@n25, B=@n26, Alpha=@n27))) {name=ParticlesAlpha, nodeColor=<rgba 0 0 0 1>}
  NamedRerouteDeclaration(None=DDX(Value=@n15)) {name=UVddx, nodeColor=<rgba 0 1 0.9766 1>}
  NamedRerouteDeclaration(None=DDY(Value=@n15)) {name=UVddy, nodeColor=<rgba 0 1 0.0156 1>}
  n28 = Multiply(A=MaterialFunctionCall(Vector or UVs=TextureCoordinate()) {materialFunction=&MF_UI_CheaperVectorToRadial}.Vector Converted to Angle) {constB=1.8}
  n29 = Constant()
  n30 = MaterialFunctionCall(Size=LinearInterpolate(Alpha=ScalarParameter() {parameterName=Anim_Scale}) {constA=0.25, constB=0.333}) {materialFunction=&MF_UI_Circle}
  n31 = Multiply(A=@n30.SDFOutline) {constB=0.125}
  n32 = TextureSample(
   UVs=Add(A=Panner(Coordinate=@n28) {speedX=-0.25, speedY=0.666}, B=AppendVector(A=@n29, B=@n31))
   Tex=NamedRerouteUsage()
   "DDX(UVs)"=NamedRerouteUsage()
   "DDY(UVs)"=NamedRerouteUsage()
  ) {mipValueMode=TMVM_Derivative, samplerType=SAMPLERTYPE_LinearColor, texture=&HighContrastNoise}
  n33 = Multiply(A=Multiply(A=@n32.R, B=Multiply(A=@n32.B, B=@n32.B))) {constB=2}
  n34 = OneMinus(None=@n30.SDFOutline)
  n35 = SmoothStep(Value=@n34) {constMax=1.5, constMin=1}
  n36 = Add(A=SmoothStep(Min=Constant() {r=0.8}, Max=Constant() {r=1.25}, Value=@n34), B=Multiply(A=@n35) {constB=4})
  n37 = Multiply(A=Add(A=Multiply(A=@n33, B=@n36), B=@n35)) {constB=3}
  n38 = Saturate(None=Multiply(A=Power(Base=@n37) {constExponent=4}, B=NamedRerouteUsage()))
  NamedRerouteDeclaration(None=@n38) {name=FlareMask, nodeColor=<rgba 0.9219 1 0 1>}
  n39 = Constant() {r=0.5}
  n40 = MaterialFunctionCall() {materialFunction=&GetUserInterfaceUV}
  n41 = MaterialFunctionCall(Float2=Divide(A=Multiply(A=@n40.Pixel Size, B=@n40.Pixel Size), B=@n40.Pixel Size)) {materialFunction=&BreakOutFloat2Components}
  n42 = Add(
   A=Add(
    A=MaterialFunctionCall(X=@n41.R, Y=@n41.R, Z=@n41.R) {materialFunction=&MakeFloat3}.Result
    B=Constant3Vector() {constant=<rgba 0 2 4 0>}
   )
   B=Multiply(A=Time()) {constB=3}
  )
  n43 = Add(A=@n39, B=Multiply(A=@n39, B=Cosine(None=@n42) {desc=2 PI, period=6.2832}))
  NamedRerouteDeclaration(None=LinearInterpolate(A=Saturate(None=@n43), B=Constant3Vector() {constant=<rgba 0.4411 0.4219 1 1>}, Alpha=@n38)) {name=FlareColor, nodeColor=<rgba 1 0.8438 0 1>}
  n44 = Constant3Vector() {constant=<rgba 1 0.25 0 1>}
  n45 = Constant3Vector() {constant=<rgba 1 1 0.5 0>}
  NamedRerouteDeclaration(None=LinearInterpolate(A=LinearInterpolate(A=@n44, B=@n45, Alpha=@n23), B=NamedRerouteUsage(), Alpha=NamedRerouteUsage())) {name=ParticlesColor, nodeColor=<rgba 0 0 0 1>}
  StaticSwitchParameter(
   True=NamedRerouteUsage()
   False=LinearInterpolate(
    A=VectorParameter() {parameterName=Color_A, defaultValue=<rgba 1 0 0 1>}.RGB
    B=VectorParameter() {parameterName=ColorB, defaultValue=<rgba 1 0.3333 0 1>}.RGB
    Alpha=Saturate(None=@n8)
   )
  ) {parameterName=IsParticles}
 ]
}
