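# UI master material for a streak meter progress indicator, drawn procedurally as an arc ring split into pips, with active pips filled.
# Circle SDFs, rotated radial gradients and sector functions build named masks for progress, sections, borders, offset shadow and glow.
# These masks drive opacity and blend empty, fill and border colors, with Overdrive shifting the fill palette and a peripheral-instrument border tint.
# The IsGlow static switch turns the same material into a glow layer.
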Material /PilgrimCore/Widgets/Materials/M_UI_StreakMeter_ProgressIndicator_Master
graph {
 properties {
  bUsedWithStaticMesh=false
  blendMode=BLEND_Translucent
  materialDomain=MD_UI
  pixelDepthOffsetMode=PDOM_Legacy
  usedShadingModels=Default Lit
 }
 nodes [
  n1 = ScalarParameter() {defaultValue=0.75, parameterName=Alpha_Outline}
  n2 = NamedRerouteUsage()
  n3 = NamedRerouteUsage()
  n4 = Subtract(
   A=OneMinus(None=ScalarParameter() {defaultValue=0.125, parameterName=ProgressThickness})
   B=ScalarParameter() {defaultValue=0.05, parameterName=AA_Threshold}
  )
  n5 = ScalarParameter() {defaultValue=0.125, parameterName=ProgressThickness}
  n6 = MaterialFunctionCall(Size=@n4, OutlineThickness=@n5) {materialFunction=&MF_UI_Circle}
  n7 = NamedRerouteDeclaration(None=MaterialFunctionCall(DistanceField=@n6.SDFOutline, Threshold=ScalarParameter() {defaultValue=0.05, parameterName=AA_Threshold}) {materialFunction=&MF_UI_AntiAliasing}.Inner) {name=Progress Border Mask, nodeColor=<rgba 0 0 0 1>}
  n8 = NamedRerouteDeclaration(None=SmoothStep(Value=NamedRerouteDeclaration(None=@n6.SDFOutline) {name=SDF Outline Mask, nodeColor=<rgba 0 0 0 1>}) {constMax=0.0125}) {name=Inset Border Mask, nodeColor=<rgba 0 0 0 1>}
  n9 = Saturate(None=Add(A=Multiply(A=@n2, B=@n3), B=Multiply(A=@n7, B=@n8)))
  Multiply(A=@n1, B=@n9)
  Multiply(A=OneMinus(None=ScalarParameter() {defaultValue=0.6575, parameterName=ArcLength_ExtraPad})) {constB=-1}
  n10 = LinearInterpolate(A=ScalarParameter() {defaultValue=0.5, parameterName=Alpha_Progress_Empty}, Alpha=NamedRerouteUsage())
  n11 = NamedRerouteUsage()
  n12 = Multiply(
   A=NamedRerouteUsage()
   B=LinearInterpolate(A=Multiply(A=@n10, B=@n11), B=@n1, Alpha=NamedRerouteDeclaration(None=@n9) {name=Border Mask, nodeColor=<rgba 0 0 0 1>})
  )
  Multiply(
   A=StaticSwitchParameter(True=NamedRerouteUsage(), False=NamedRerouteDeclaration(None=@n12) {name=Final Mask, nodeColor=<rgba 0 0 0 1>}) {parameterName=IsGlow}
   B=ScalarParameter() {defaultValue=1, parameterName=Alpha_Final}
  )
  n13 = TextureCoordinate()
  n14 = Multiply(A=OneMinus(None=ScalarParameter() {defaultValue=0.6667, parameterName=ArcLength})) {constB=-1}
  n15 = MaterialFunctionCall(
   Input=MaterialFunctionCall!(Vector or UVs=MaterialFunctionCall(UVs=@n13, "Rotation Angle (0-1)"=@n14) {materialFunction=&CustomRotator}.Rotated Values) {materialFunction=/FortUILibrary/MaterialFunctions/Basics/MF_UI_CheaperVectorToRadial}.Vector Converted to Angle
   Input Low=ScalarParameter() {defaultValue=0.6667, parameterName=ArcLength}
   Input High=Constant() {r=1}
  ) {materialFunction=&RemapValue_0-1}
  n16 = ScalarParameter() {defaultValue=3, parameterName=NumberofPips}
  n17 = Constant() {r=0.9}
  NamedRerouteDeclaration(None=MaterialFunctionCall(
    DistanceField=MaterialFunctionCall!(UVs, Gradient Map=@n15.Result, Sector Count=@n16, Sector Spacing=@n17, Fixed Spacing) {materialFunction=/FortUILibrary/MaterialFunctions/Basics/MF_UI_Sectors}.Parallel Sectors SDF
    Threshold=Constant() {r=0.333}
   ) {materialFunction=&MF_UI_AntiAliasing}.Outer) {name=Sections Mask, nodeColor=<rgba 0 0 0 1>}
  n18 = TextureCoordinate()
  n19 = AppendVector(A=Constant(), B=ScalarParameter() {defaultValue=0.025, parameterName=YShadowOffset})
  n20 = SmoothStep(Value=MaterialFunctionCall(UVs=Add(A=@n18, B=@n19), Size=@n4, OutlineThickness=@n5) {materialFunction=&MF_UI_Circle}.SDFOutline) {constMax=0.05, constMin=-0.025}
  NamedRerouteDeclaration(None=Saturate(None=OneMinus(None=@n20))) {name=Shadow Mask, nodeColor=<rgba 0 0 0 1>}
  n21 = Power(Base=OneMinus(None=MaterialFunctionCall(Gradient Map=@n15.Result, Sector Count=@n16) {materialFunction=&MF_UI_Sectors}.Parallel Sectors SDF)) {constExponent=0.5}
  n22 = OneMinus(None=@n6.SDFOutline)
  NamedRerouteDeclaration(None=Multiply(A=LinearInterpolate(Alpha=@n21) {constA=-1}, B=SmoothStep(Value=@n22) {constMax=1.25, constMin=0.95})) {name=Glow Mask, nodeColor=<rgba 0 0 0 1>}
  n23 = MaterialFunctionCall(UVs=TextureCoordinate(), "Rotation Angle (0-1)"=Constant() {r=-0.3285}) {materialFunction=&CustomRotator}
  n24 = OneMinus(None=MaterialFunctionCall(
    Input=MaterialFunctionCall(Vector or UVs=@n23.Rotated Values) {materialFunction=&MF_UI_CheaperVectorToRadial}.Vector Converted to Angle
    Input Low=ScalarParameter() {defaultValue=0.6575, parameterName=ArcLength_ExtraPad}
    Input High=Constant() {r=1}
   ) {materialFunction=&RemapValue_0-1}.Result)
  NamedRerouteDeclaration(None=OneMinus(None=Floor(None=@n24))) {name=MaskAll, nodeColor=<rgba 0 0 0 1>}
  n25 = ScalarParameter() {defaultValue=1, parameterName=NumberofActivePips}
  n26 = Max(A=ScalarParameter() {defaultValue=3, parameterName=NumberofPips}) {constB=0.0001}
  n27 = Ceil(None=SmoothStep(Min=Divide(A=@n25, B=@n26), Value=@n15.Result))
  NamedRerouteDeclaration(None=Multiply(A=NamedRerouteUsage(), B=OneMinus(None=@n27))) {name=Progress Mask, nodeColor=<rgba 0 0 0 1>}
  NamedRerouteDeclaration(None=Ceil(None=@n15.Result)) {name=Progress Area, nodeColor=<rgba 0 0 0 1>}
  OneMinus(None=ScalarParameter() {defaultValue=0.25, parameterName=Progress})
  n28 = LinearInterpolate(B=NamedRerouteUsage(), Alpha=NamedRerouteUsage())
  n29 = VectorParameter() {parameterName=Color_Fill_Falloff, defaultValue=<rgba 0.25 0 0 1>}
  n30 = VectorParameter() {parameterName=Color_Fill, defaultValue=<rgba 1 0 0 1>}
  n31 = NamedRerouteUsage()
  n32 = VectorParameter() {parameterName=Color_Fill_Falloff_Overdrive, defaultValue=<rgba 0.25 0 0 1>}
  n33 = VectorParameter() {parameterName=Color_Fill_Overdrive, defaultValue=<rgba 1 0 0 1>}
  n34 = NamedRerouteUsage()
  n35 = LinearInterpolate(
   A=VectorParameter() {parameterName=Color_Empty, defaultValue=<rgba 0 1 0 1>}.RGB
   B=LinearInterpolate(
    A=LinearInterpolate(A=@n29.RGB, B=@n30.RGB, Alpha=@n31) {constA=0.333, constB=0.5}
    B=LinearInterpolate(A=@n32.RGB, B=@n33.RGB, Alpha=@n34) {constA=0.333, constB=0.5}
    Alpha=ScalarParameter() {parameterName=Overdrive}
   )
   Alpha=NamedRerouteUsage()
  )
  n36 = ScalarParameter() {defaultValue=0.75}
  n37 = TextureCoordinate()
  n38 = LinearInterpolate(
   A=VectorParameter() {parameterName=Border_Tint, defaultValue=<rgba 0 0 1 1>}.RGB
   B=VectorParameter() {parameterName=Color_Border, defaultValue=<rgba 0 0 1 1>}.RGB
   Alpha=ComponentMask(None=SmoothStep(Min=@n36, Value=@n37)) {g=true}
  )
  n39 = VectorParameter() {parameterName=Color_PeriphalInstrument, defaultValue=<rgba 0.0168 0.0185 0.0262 1>}
  n40 = ScalarParameter() {parameterName=IsPeripheralInstrument}
  n41 = NamedRerouteUsage()
  n42 = NamedRerouteUsage()
  StaticSwitchParameter(
   True=LinearInterpolate(
    A=VectorParameter() {parameterName=Color_Glow_A, defaultValue=<rgba 0 0 0 1>}.RGB
    B=VectorParameter() {parameterName=Color_Glow_B, defaultValue=<rgba 0 0 0 1>}.RGB
    Alpha=NamedRerouteDeclaration(None=@n28) {name=Glow Alpha, nodeColor=<rgba 0 0 0 1>}
   )
   False=LinearInterpolate(
    A=NamedRerouteDeclaration(None=@n35) {name=Progress Color, nodeColor=<rgba 0 0 0 1>}
    B=LinearInterpolate(A=@n38, B=@n39.RGB, Alpha=@n40)
    Alpha=LinearInterpolate(A=@n41, Alpha=@n42)
   )
  ) {parameterName=IsGlow}
 ]
}
