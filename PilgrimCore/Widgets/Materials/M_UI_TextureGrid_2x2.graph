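# UI material that draws four album textures as a 2x2 grid with a rounded-corner mask, likely for album art in PilgrimCore widgets.
# Doubled, wrapped UVs tile each cell, and MF_UI_UVScaling fits each texture parameter to fill it.
# Floored horizontal and vertical UV masks drive nested lerps that select the image per quadrant, avoiding branches.
# A CornerRadius parameter feeds MF_UI_CoreUI_RoundedMask to produce the shape mask.
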Material /PilgrimCore/Widgets/Materials/M_UI_TextureGrid_2x2
graph {
 properties {
  bAutomaticallySetUsageInEditor=false
  bUsedWithStaticMesh=false
  blendMode=BLEND_Translucent
  materialDomain=MD_UI
  usedShadingModels=Default Lit
 }
 nodes [
  n1 = TextureObjectParameter() {parameterName=Album1, texture=/FortniteCoreUI/UI/Creative/Materials/ServerImageAssets/T_UI_Creative_ServerImage_1}
  n2 = MaterialFunctionCall(UV=NamedRerouteUsage(), TextureSize=TextureProperty(Tex=@n1)) {materialFunction=&MF_UI_UVScaling}
  Multiply(A=@n2.Scaled To Fill) {constB=2}
  n3 = MaterialFunctionCall() {materialFunction=&GetUserInterfaceUV}
  NamedRerouteDeclaration(None=Frac(None=Multiply(A=@n3.Normalized UV) {constB=2})) {name=TiledUV, nodeColor=<rgba 0.4062 1 0 1>}
  n4 = MaterialFunctionCall(Float2=@n3.Normalized UV) {materialFunction=&BreakOutFloat2Components}
  NamedRerouteDeclaration(None=Floor(None=Add(A=@n4.R) {constB=0.5})) {name=HorizontalGridLerp, nodeColor=<rgba 0 1 0.9531 1>}
  n5 = MaterialFunctionCall(Float2=@n3.Normalized UV) {materialFunction=&BreakOutFloat2Components}
  NamedRerouteDeclaration(None=Floor(None=Add(A=@n5.G) {constB=0.5})) {name=VerticalGridLerp, nodeColor=<rgba 1 0.6328 0 1>}
  n6 = TextureSample(UVs=@n2.Scaled To Fill, Tex=@n1)
  n7 = TextureObjectParameter() {parameterName=Album2, texture=/FortniteCoreUI/UI/Creative/Materials/ServerImageAssets/T_UI_Creative_ServerImage_2}
  n8 = TextureSample(
   UVs=MaterialFunctionCall(UV=NamedRerouteUsage(), TextureSize=TextureProperty(Tex=@n7)) {materialFunction=&MF_UI_UVScaling}.Scaled To Fill
   Tex=@n7
  )
  n9 = NamedRerouteUsage()
  n10 = TextureObjectParameter() {parameterName=Album3, texture=/FortniteCoreUI/UI/Creative/Materials/ServerImageAssets/T_UI_Creative_ServerImage_3}
  n11 = TextureSample(
   UVs=MaterialFunctionCall(UV=NamedRerouteUsage(), TextureSize=TextureProperty(Tex=@n10)) {materialFunction=&MF_UI_UVScaling}.Scaled To Fill
   Tex=@n10
  )
  n12 = TextureObjectParameter() {parameterName=Album4, texture=/FortniteCoreUI/UI/Creative/Materials/ServerImageAssets/T_UI_Creative_ServerImage_8}
  n13 = TextureSample(
   UVs=MaterialFunctionCall(UV=NamedRerouteUsage(), TextureSize=TextureProperty(Tex=@n12)) {materialFunction=&MF_UI_UVScaling}.Scaled To Fill
   Tex=@n12
  )
  n14 = NamedRerouteUsage()
  NamedRerouteDeclaration(None=LinearInterpolate(A=LinearInterpolate(A=@n6.RGB, B=@n8.RGB, Alpha=@n9), B=LinearInterpolate(A=@n11.RGB, B=@n13.RGB, Alpha=@n14), Alpha=NamedRerouteUsage())) {name=AlbumGridColor, nodeColor=<rgba 0 1 0.3906 1>}
  NamedRerouteDeclaration(None=MaterialFunctionCall(Corner Radius=ScalarParameter() {defaultValue=0.05, parameterName=CornerRadius}) {materialFunction=&MF_UI_CoreUI_RoundedMask}.Shape Mask) {name=Shape, nodeColor=<rgba 1 0.0703 0 1>}
  NamedRerouteUsage()
  NamedRerouteUsage()
 ]
}
