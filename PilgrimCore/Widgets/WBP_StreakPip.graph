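# WBP_StreakPip is a UMG widget for a single pip in a streak meter, layering a padded Pip image over an empty backing image.
# SetActive and SetOverdriveColor turn booleans into 0 or 1 for the Progress and OverdriveActive parameters on the Pip's dynamic material.
# SetAvailable toggles the whole widget between Collapsed and HitTestInvisible.
# It holds no variables or event logic, so its state is set only through these functions.
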WidgetBlueprint /PilgrimCore/Widgets/WBP_StreakPip
graphs [
 EventGraph {nodes [
   Event PreConstruct()
   Event Construct()
  ]}
 SetActive {nodes [
   n1 = SetActive!() -> (then, isActive:bool=~) {is=title K2Node_FunctionEntry}
   > n2 = GetDynamicMaterial(self=Get Pip() -> (Pip:Image Object Reference))
   > SetScalarParameterValue(self=@n2, ParameterName=Progress, Value=Select(Option 0:float=0, Option 1:float=1, Index=@n1) -> (ReturnValue:float))
  ]}
 SetAvailable {nodes [
   n1 = SetAvailable!() -> (then, isAvailable:bool=~) {is=title K2Node_FunctionEntry}
   > SetVisibility(InVisibility=Select(Option 0:ESlateVisibility Enum=Collapsed, Option 1:ESlateVisibility Enum=HitTestInvisible, Index=@n1) -> (ReturnValue:ESlateVisibility Enum))
  ]}
 SetOverdriveColor {nodes [
   n1 = SetOverdriveColor!() -> (then, OverdriveActive:bool=~) {is=title K2Node_FunctionEntry}
   > n2 = GetDynamicMaterial(self=Get Pip())
   > SetScalarParameterValue(self=@n2, ParameterName=OverdriveActive, Value=Select(Option 0:float=0, Option 1:float=1, Index=@n1) -> (ReturnValue:float))
  ]}
]
variables []
widgets {
 info {inheritedWidgetCount=0, parentClass=&UserWidget, rootWidgetClass=&Overlay, widgetCount=4}
 widgets [
  Overlay_0 {widgetClassPath=&Overlay, widget=Overlay_0}
  Spacer_61 {
   widgetClassPath=&Spacer
   parent=Overlay_0
   widget=Spacer_61
   properties {size=<12 12>}
   slot {class=OverlaySlot, properties {horizontalAlignment=HAlign_Fill, verticalAlignment=VAlign_Fill}}
  }
  Pip_Backing {
   widgetClassPath=&Image
   bIsVariable=true
   parent=Overlay_0
   widget=Pip_Backing
   properties {brush {
     bIsDynamicallyLoaded=false
     drawAs=Image
     imageType=NoImage
     mirroring=NoMirror
     resourceName=None
     resourceObject=/PilgrimCore/Widgets/Materials/M_UI_StreakMeter_Pip_Empty
     tiling=NoTile
     imageSize=<8 8>
     margin=<ltrb 0 0 0 0>
     outlineSettings {
      bUseBrushTransparency=false
      roundingType=HalfHeightRadius
      width=0
      color {colorUseRule=UseColor_Specified, specifiedColor=<rgba 0 0 0 0>}
      cornerRadii=<0 0 0 0>
     }
     tintColor {colorUseRule=UseColor_Specified, specifiedColor=<rgba 1 1 1 1>}
     uVRegion {bIsValid=false, max=<0 0>, min=<0 0>}
    }}
   slot {class=OverlaySlot, properties {horizontalAlignment=HAlign_Fill, verticalAlignment=VAlign_Fill}}
  }
  Pip {
   widgetClassPath=&Image
   bIsVariable=true
   parent=Overlay_0
   widget=Pip
   properties {brush {
     bIsDynamicallyLoaded=false
     drawAs=Image
     imageType=NoImage
     mirroring=NoMirror
     resourceName=None
     resourceObject=/PilgrimCore/Widgets/Materials/M_UI_StreakMeter_Pip
     tiling=NoTile
     imageSize=<8 8>
     margin=<ltrb 0 0 0 0>
     outlineSettings {
      bUseBrushTransparency=false
      roundingType=HalfHeightRadius
      width=0
      color {colorUseRule=UseColor_Specified, specifiedColor=<rgba 0 0 0 0>}
      cornerRadii=<0 0 0 0>
     }
     tintColor {colorUseRule=UseColor_Specified, specifiedColor=<rgba 1 1 1 1>}
     uVRegion {bIsValid=false, max=<0 0>, min=<0 0>}
    }}
   slot {class=OverlaySlot, properties {horizontalAlignment=HAlign_Fill, verticalAlignment=VAlign_Fill, padding=<ltrb 2 2 2 2>}}
  }
 ]
}
