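# Common UI widget that displays a streak level as a horizontal row of five WBP_StreakPip children.
# Construct caches the pips in an array, which functions loop over to set availability up to Max Pips, active state up to a given count, and overdrive color.
# Updating the active count also calls UpdateMaxEffect with an at-max flag, but both visibility options there are Hidden, so the Max_Glow overlay image stays hidden.
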WidgetBlueprint /PilgrimCore/Widgets/WBP_StreakLevelPips
graphs [
 EventGraph {nodes [
   Event Construct()
   > Set PipArray(PipArray:Array of WBP Streak Pip Object References=Make Array!(
     "[0]":WBP Streak Pip Object Reference=Get WBP_StreakPip_1() -> (WBP_StreakPip_1:WBP Streak Pip Object Reference)
     "[1]":WBP Streak Pip Object Reference=Get WBP_StreakPip_2() -> (WBP_StreakPip_2:WBP Streak Pip Object Reference)
     "[2]":WBP Streak Pip Object Reference=Get WBP_StreakPip_3() -> (WBP_StreakPip_3:WBP Streak Pip Object Reference)
     "[3]":WBP Streak Pip Object Reference=Get WBP_StreakPip_4() -> (WBP_StreakPip_4:WBP Streak Pip Object Reference)
     "[4]":WBP Streak Pip Object Reference=Get WBP_StreakPip_5() -> (WBP_StreakPip_5:WBP Streak Pip Object Reference)
    ) -> (Array:Array of WBP Streak Pip Object References)) -> (Output_Get:Array of WBP Streak Pip Object References)
   Event PreConstruct()
  ]}
 Set Number Of Available Pips {nodes [
   n1 = Set Number Of Available Pips!() -> (then, NumberOfPipsToShow:int) {is=title K2Node_FunctionEntry}
   > Set Max Pips(Max Pips:int=@n1) -> (Output_Get:int=0)
   > n2 = For Each Loop(Array:Array of WBP Streak Pip Object References=Get PipArray() -> (PipArray:Array of WBP Streak Pip Object References)) -> (Array Element:WBP Streak Pip Object Reference)
   > SetAvailable!(
    execute
    self:WBP Streak Pip Object Reference=@n2
    isAvailable:bool="integer < integer"(A=@n2.Array Index, B=Get Max Pips() -> (Max Pips:int=0))
   ) -> (then)
  ]}
 SetOverdriveState {nodes [
   n1 = SetOverdriveState!() -> (then, OverdriveActive:bool=~) {is=title K2Node_FunctionEntry}
   > n2 = For Each Loop(Array=Get PipArray())
   > SetOverdriveColor!(execute, self:WBP Streak Pip Object Reference=@n2, OverdriveActive:bool=@n1) -> (then)
  ]}
 UpdateMaxEffect {nodes [
   n1 = UpdateMaxEffect!() -> (then, isAtMax:bool=~) {is=title K2Node_FunctionEntry}
   > SetVisibility(self=Get Max_Glow() -> (Max_Glow:Image Object Reference), InVisibility=Select(Option 0=Hidden, Option 1=Hidden, Index=@n1))
  ]}
 UpdateNumActivePips {nodes [
   n1 = UpdateNumActivePips!() -> (then, NumActivePips:int) {is=title K2Node_FunctionEntry}
   > n2 = For Each Loop(Array=Get PipArray())
   UpdateMaxEffect!(execute=@n2.Completed, self, isAtMax:bool="integer >= integer"(A=@n1, B=Get Max Pips() -> (Max Pips=0))) -> (then) {is=title K2Node_CallFunction}
   SetActive!(execute=@n2.LoopBody, self:WBP Streak Pip Object Reference=@n2, isActive:bool="integer < integer"(A=@n2.Array Index, B=@n1)) -> (then)
  ]}
]
variables [
 Max Pips {type=integer}
 PipArray {type="array<WBP_StreakPip_C>", default []}
]
widgets {
 info {inheritedWidgetCount=0, parentClass=&CommonUserWidget, rootWidgetClass=&Overlay, widgetCount=8}
 widgets [
  Overlay_0 {widgetClassPath=&Overlay, widget=Overlay_0}
  HorizontalBox_22 {
   widgetClassPath=&HorizontalBox
   parent=Overlay_0
   widget=HorizontalBox_22
   slot {class=OverlaySlot, properties {horizontalAlignment=HAlign_Fill, verticalAlignment=VAlign_Fill}}
  }
  WBP_StreakPip_1 {
   widgetClassPath=&/WBP_StreakPip
   bIsVariable=true
   parent=HorizontalBox_22
   widget=WBP_StreakPip_1
   slot {class=HorizontalBoxSlot, properties {size {sizeRule=Fill, value=1}}}
  }
  WBP_StreakPip_2 {
   widgetClassPath=&WBP_StreakPip
   bIsVariable=true
   parent=HorizontalBox_22
   widget=WBP_StreakPip_2
   slot {class=HorizontalBoxSlot, properties {padding=<ltrb 1 0 0 0>, size {sizeRule=Fill, value=1}}}
  }
  WBP_StreakPip_3 {
   widgetClassPath=&WBP_StreakPip
   bIsVariable=true
   parent=HorizontalBox_22
   widget=WBP_StreakPip_3
   slot {class=HorizontalBoxSlot, properties {padding=<ltrb 1 0 0 0>, size {sizeRule=Fill, value=1}}}
  }
  WBP_StreakPip_4 {
   widgetClassPath=&WBP_StreakPip
   bIsVariable=true
   parent=HorizontalBox_22
   widget=WBP_StreakPip_4
   slot {class=HorizontalBoxSlot, properties {padding=<ltrb 1 0 0 0>, size {sizeRule=Fill, value=1}}}
  }
  WBP_StreakPip_5 {
   widgetClassPath=&WBP_StreakPip
   bIsVariable=true
   parent=HorizontalBox_22
   widget=WBP_StreakPip_5
   slot {class=HorizontalBoxSlot, properties {padding=<ltrb 1 0 0 0>, size {sizeRule=Fill, value=1}}}
  }
  Max_Glow {
   widgetClassPath=&Image
   bIsVariable=true
   parent=Overlay_0
   widget=Max_Glow
   properties {
    visibility=Hidden
    brush {
     bIsDynamicallyLoaded=false
     drawAs=Image
     imageType=NoImage
     mirroring=NoMirror
     resourceName=None
     resourceObject=/PilgrimCore/Widgets/Materials/MI_UI_Pilgrim_Device_RadialGlow
     tiling=NoTile
     imageSize=<112 80>
     margin=<ltrb 0 0 0 0>
     outlineSettings {
      bUseBrushTransparency=false
      roundingType=HalfHeightRadius
      width=0
      color {colorUseRule=UseColor_Specified, specifiedColor=<rgba 0 0 0 0>}
      cornerRadii=<0 0 0 0>
     }
     tintColor {colorUseRule=UseColor_Specified, specifiedColor=<rgba 1 1 1 1>}
     uVRegion {bIsValid=false, max=<0 0>, min=<0 0>}
    }
   }
   slot {class=OverlaySlot, properties {horizontalAlignment=HAlign_Fill, verticalAlignment=VAlign_Fill}}
  }
 ]
}
