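# Bandmate HUD frame for Pilgrim, showing a player's avatar, overdrive meter and streak pips.
# Update functions set the meter's material progress, avatar and active pips, and size pip count by track, with bass and vocals using the higher max multiplier.
# Overdrive changes toggle FX, bracket, pip state and backing material, playing a burst animation on activation.
# PreConstruct seeds particle and flare materials from variables and resets meter, overdrive and pips.
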WidgetBlueprint /PilgrimCore/Widgets/WBP_BandmateFrame
animations [Anim_OD_Activated {
  binding_count=5
  length_seconds=0.75
  object=&self:Anim_OD_Activated
  sequence {bindings [
    OD_Burst {
     id=#1
     possessed_class=Image
     tracks [
      {
       type=MovieSceneWidgetMaterialTrack
       sections [{type=MovieSceneParameterSection, channels [Alpha_Final {type=MovieSceneScriptingFloatChannel, keys [[0,1],[2,1.5],[10,1]]}]}]
      }
      {
       type=MovieScene2DTransformTrack
       sections [{
         type=MovieScene2DTransformSection
         channels [
          Scale.X {type=MovieSceneScriptingFloatChannel, keys [[0,1],[2,1.333],[10,1]]}
          Scale.Y {type=MovieSceneScriptingFloatChannel, const [1.5,1]}
         ]
        }]
      }
     ]
    }
    WBP_BandmateFrame_C_43 {
     id=#2
     possessed_class=WBP_BandmateFrame_C
     tracks [{
       type=MovieScene2DTransformTrack
       sections [{
         type=MovieScene2DTransformSection
         channels [
          Scale.X {type=MovieSceneScriptingFloatChannel, keys [[0,1],[1,0.975],[2,1.05],[6,1]]}
          Scale.Y {type=MovieSceneScriptingFloatChannel, keys [[0,1],[1,0.975],[2,1.05],[6,1]]}
         ]
        }]
      }]
    }
    PanelBacking {
     id=#3
     possessed_class=Image
     tracks [{
       type=MovieSceneWidgetMaterialTrack
       sections [{type=MovieSceneParameterSection, channels [OD_ColorIntensity {type=MovieSceneScriptingFloatChannel, keys [[0,1.0693],[4,16],[15,1.666]]}]}]
      }]
    }
    OD_Flare {
     id=#4
     possessed_class=Image
     tracks [{
       type=MovieSceneWidgetMaterialTrack
       sections [{type=MovieSceneParameterSection, channels [UseEaseCurve {type=MovieSceneScriptingFloatChannel, keys [[0,2],[4,1]]}]}]
      }]
    }
    FX {
     id=#5
     possessed_class=Overlay
     tracks [{
       type=MovieSceneFloatTrack
       display_name=Render Opacity
       sections [{type=MovieSceneFloatSection, channels [None {type=MovieSceneScriptingFloatChannel, keys [[0,0],[2,1]]}]}]
      }]
    }
   ]}
 }]
graphs [
 EventGraph {nodes [
   n1 = OnOverdriveStateChanged!() -> (OutputDelegate, then, IsActive:bool=~) {is=title K2Node_CustomEvent}
   > SetVisibility(
    self=[Get FX() -> (FX:Overlay Object Reference), Get OD_BracketIndicator() -> (OD_BracketIndicator:Image Object Reference)]
    InVisibility=Select(Option 0=Hidden, Option 1=SelfHitTestInvisible, Index=@n1.IsActive)
   )
   > n2 = GetDynamicMaterial(self=Get PanelBacking() -> (PanelBacking:Image Object Reference))
   > SetScalarParameterValue(self=@n2, ParameterName=OverdriveActive, Value=Select(Option 0:float=0, Option 1:float=1, Index=@n1.IsActive) -> (ReturnValue:float))
   > Branch(Condition=@n1.IsActive)
   > PlayAnimation(InAnimation=Get Anim_OD_Activated() -> (Anim_OD_Activated:Widget Animation Object Reference))
   Event PreConstruct()
   > n3 = GetDynamicMaterial(self=Get OD_Particles() -> (OD_Particles:Image Object Reference))
   > SetScalarParameterValue(self=@n3, ParameterName=Set1_RandomSeed, Value=Get RandomSeed_1() -> (RandomSeed_1:float))
   > SetScalarParameterValue(self=@n3, ParameterName=Set2_RandomSeed, Value=Get RandomSeed_2() -> (RandomSeed_2:float))
   > SetScalarParameterValue(self=@n3, ParameterName=Set1_SeedAdjustment, Value=Get Set1_SeedAdjustment() -> (Set1_SeedAdjustment:float))
   > SetScalarParameterValue(self=@n3, ParameterName=Set2_SeedAdjustment, Value=Get Set2_SeedAdjustment() -> (Set2_SeedAdjustment:float))
   > n4 = GetDynamicMaterial(self=Get OD_Flare() -> (OD_Flare:Image Object Reference))
   > SetScalarParameterValue(self=@n4, ParameterName=NoiseSet1_Tile_Alter, Value=Get FlareNoise1_Adjustment() -> (FlareNoise1_Adjustment:float))
   > SetScalarParameterValue(self=@n4, ParameterName=NoiseSet2_Tile_Alter, Value=Get FlareNoise2_Adjustment() -> (FlareNoise2_Adjustment:float))
   > UpdateOverdriveProgress!(execute, self, Overdrive:double=0) -> (then)
   > UpdateOverdriveActive!(execute, self, IsActive:bool=false) -> (then)
   > UpdateNumActiveStreakPips!(execute, self, NumPips:int=0) -> (then)
  ]}
 SetShearRenderTransform {nodes [
   n1 = SetShearRenderTransform!() -> (then, Shear:Vector 2D Structure) {is=title K2Node_FunctionEntry}
   > SetRenderShear!(
    execute
    self:Widget Object Reference=[
     Get SizeBox_ODBar() -> (SizeBox_ODBar:Size Box Object Reference)
     Get SizeBox_StreakPips() -> (SizeBox_StreakPips:Size Box Object Reference)
     Get AvatarOverlay() -> (AvatarOverlay:Overlay Object Reference)
    ]
    Shear:Vector 2D Structure=@n1
   ) -> (then)
  ]}
 UpdateNumActiveStreakPips {nodes [
   n1 = UpdateNumActiveStreakPips!() -> (then, NumPips:int) {is=title K2Node_FunctionEntry}
   > UpdateNumActivePips!(
    execute
    self:WBP Streak Level Pips Object Reference=Get WBP_StreakLevelPips() -> (WBP_StreakLevelPips:WBP Streak Level Pips Object Reference)
    NumActivePips:int=int - int(A=@n1)
   ) -> (then)
  ]}
 UpdateOverdriveActive {nodes [
   n1 = UpdateOverdriveActive!() -> (then, IsActive:bool=~) {is=title K2Node_FunctionEntry}
   > Branch(Condition="Not Equal (Boolean)"(A=@n1, B=Get OverdriveActive() -> (OverdriveActive:bool)) -> (ReturnValue=~))
   > n2 = Set OverdriveActive(OverdriveActive:bool=@n1) -> (Output_Get:bool)
   > SetOverdriveState!(execute, self:WBP Streak Level Pips Object Reference=Get WBP_StreakLevelPips(), OverdriveActive:bool=@n2) -> (then)
   > OnOverdriveStateChanged!(execute, self, IsActive:bool=@n2) -> (then) {is=title K2Node_CallFunction}
  ]}
 UpdateOverdriveProgress {nodes [
   n1 = UpdateOverdriveProgress!() -> (then, Overdrive:double=~) {is=title K2Node_FunctionEntry}
   > n2 = GetDynamicMaterial(self=Get Fill() -> (Fill:Image Object Reference))
   > SetScalarParameterValue(self=@n2, ParameterName=Progress, Value=@n1)
  ]}
 UpdatePlayerIcon {nodes [
   n1 = UpdatePlayerIcon!() -> (then, PlayerState:Fort Player State Athena Object Reference) {is=title K2Node_FunctionEntry}
   > SetPlayerImage!(
    execute
    self:WBP Pilgrim Avatar Icon Object Reference=Get WBP_Pilgrim_AvatarIcon() -> (WBP_Pilgrim_AvatarIcon:WBP Pilgrim Avatar Icon Object Reference)
    FortPlayerStateAthena:Fort Player State Athena Object Reference=@n1
   ) -> (then)
  ]}
 UpdateTrack {nodes [
   n1 = UpdateTrack!() -> (then, Track Type:EPilgrimTrackType Enum=TrackGuitar) {is=title K2Node_FunctionEntry}
   n2 = "Equal (Enum)"(A:EPilgrimTrackType Enum=@n1, B:EPilgrimTrackType Enum=TrackBass)
   n3 = "Equal (Enum)"(A=@n1, B=TrackVocals)
   n4 = Get 4xMaxStreakScoreMultiplier() -> (4xMaxStreakScoreMultiplier:int=0)
   n5 = Get 6xMaxStreakScoreMultiplier() -> (6xMaxStreakScoreMultiplier:int=0)
   > Set Number Of Available Pips!(
    execute
    self:WBP Streak Level Pips Object Reference=Get WBP_StreakLevelPips()
    NumberOfPipsToShow:int=SelectInt(A=int - int(A=@n5), B=int - int(A=@n4), bPickA=OR Boolean(A=@n2, B=@n3))
   ) -> (then) {is=title K2Node_CallFunction}
  ]}
]
variables [
 4xMaxStreakScoreMultiplier {type=integer, default=4}
 6xMaxStreakScoreMultiplier {type=integer, default=6}
 Background Color {type=Color, default=<rgba 0.1882 0 0.6039 1>}
 FlareNoise1_Adjustment {type=number, default=1}
 FlareNoise2_Adjustment {type=number, default=1}
 OverdriveActive {type=boolean}
 RandomSeed_1 {type=number, default=2.2389}
 RandomSeed_2 {type=number, default=1.491}
 Set1_SeedAdjustment {type=number, default=1.491}
 Set2_SeedAdjustment {type=number, default=1.491}
]
widgets {
 info {inheritedWidgetCount=0, parentClass=&UserWidget, rootWidgetClass=&SizeBox, widgetCount=28}
 widgets [
  SizeBox_108 {
   widgetClassPath=&SizeBox
   widget=SizeBox_108
   properties {bOverride_HeightOverride=true, bOverride_WidthOverride=true, heightOverride=56, widthOverride=176}
  }
  Overlay_2 {widgetClassPath=&Overlay, parent=SizeBox_108, widget=Overlay_2, slot {class=SizeBoxSlot}}
  FX {
   widgetClassPath=&Overlay
   bIsVariable=true
   parent=Overlay_2
   widget=FX
   properties {visibility=Collapsed}
   slot {class=OverlaySlot, properties {horizontalAlignment=HAlign_Fill, verticalAlignment=VAlign_Fill}}
  }
  OD_Glow {
   widgetClassPath=&Image
   bIsVariable=true
   parent=FX
   widget=OD_Glow
   properties {
    visibility=SelfHitTestInvisible
    brush {
     bIsDynamicallyLoaded=false
     drawAs=Image
     imageType=NoImage
     mirroring=NoMirror
     resourceName=None
     resourceObject=/PilgrimCore/Widgets/Materials/M_UI_Pilgrim_Bandmate_OD_Glow
     tiling=NoTile
     imageSize=<32 32>
     margin=<ltrb 0 0 0 0>
     outlineSettings {
      bUseBrushTransparency=false
      roundingType=HalfHeightRadius
      width=0
      color {colorUseRule=UseColor_Specified, specifiedColor=<rgba 0 0 0 0>}
      cornerRadii=<0 0 0 0>
     }
     tintColor {colorUseRule=UseColor_Specified, specifiedColor=<rgba 1 1 1 1>}
     uVRegion {bIsValid=false, max=<0 0>, min=<0 0>}
    }
   }
   slot {class=OverlaySlot, properties {horizontalAlignment=HAlign_Fill, verticalAlignment=VAlign_Fill, padding=<ltrb -12 -12 -12 -14>}}
  }
  FortVisualAttachment_1 {
   widgetClassPath=&FortVisualAttachment
   parent=FX
   widget=FortVisualAttachment_1
   properties {bOverride_HeightOverride=true, bOverride_WidthOverride=true, heightOverride=56, widthOverride=512}
   slot {class=OverlaySlot, properties {horizontalAlignment=HAlign_Fill, verticalAlignment=VAlign_Fill}}
  }
  Overlay_3 {
   widgetClassPath=&Overlay
   parent=FortVisualAttachment_1
   widget=Overlay_3
   slot {class=SizeBoxSlot, properties {verticalAlignment=VAlign_Center}}
  }
  OD_Burst {
   widgetClassPath=&Image
   bIsVariable=true
   parent=Overlay_3
   widget=OD_Burst
   properties {
    visibility=SelfHitTestInvisible
    brush {
     bIsDynamicallyLoaded=false
     drawAs=Image
     imageType=NoImage
     mirroring=NoMirror
     resourceName=None
     resourceObject=/PilgrimCore/Widgets/Materials/M_UI_Pilgrim_OD_FlareGradient
     tiling=NoTile
     imageSize=<32 64>
     margin=<ltrb 0 0 0 0>
     outlineSettings {
      bUseBrushTransparency=false
      roundingType=HalfHeightRadius
      width=0
      color {colorUseRule=UseColor_Specified, specifiedColor=<rgba 0 0 0 0>}
      cornerRadii=<0 0 0 0>
     }
     tintColor {colorUseRule=UseColor_Specified, specifiedColor=<rgba 1 1 1 1>}
     uVRegion {bIsValid=false, max=<0 0>, min=<0 0>}
    }
    renderTransform {angle=0, scale=<1 1.5>, shear=<0 0>, translation=<-24 0>}
    renderTransformPivot=<0 0.5>
   }
   slot {class=OverlaySlot, properties {horizontalAlignment=HAlign_Fill, verticalAlignment=VAlign_Center}}
  }
  OD_Flare {
   widgetClassPath=&Image
   bIsVariable=true
   parent=Overlay_3
   widget=OD_Flare
   properties {
    visibility=SelfHitTestInvisible
    brush {
     bIsDynamicallyLoaded=false
     drawAs=Image
     imageType=NoImage
     mirroring=NoMirror
     resourceName=None
     resourceObject=/PilgrimCore/Widgets/Materials/M_UI_Pilgrim_OD_Bandmate_Flare
     tiling=NoTile
     imageSize=<32 40>
     margin=<ltrb 0 0 0 0>
     outlineSettings {
      bUseBrushTransparency=false
      roundingType=HalfHeightRadius
      width=0
      color {colorUseRule=UseColor_Specified, specifiedColor=<rgba 0 0 0 0>}
      cornerRadii=<0 0 0 0>
     }
     tintColor {colorUseRule=UseColor_Specified, specifiedColor=<rgba 1 1 1 1>}
     uVRegion {bIsValid=false, max=<0 0>, min=<0 0>}
    }
    renderTransformPivot=<0 0.5>
   }
   slot {class=OverlaySlot, properties {horizontalAlignment=HAlign_Fill, verticalAlignment=VAlign_Center}}
  }
  OD_Particles {
   widgetClassPath=&Image
   bIsVariable=true
   parent=Overlay_3
   widget=OD_Particles
   properties {
    visibility=Collapsed
    brush {
     bIsDynamicallyLoaded=false
     drawAs=Image
     imageType=NoImage
     mirroring=NoMirror
     resourceName=None
     resourceObject=/PilgrimCore/Widgets/Materials/MI_UI_Piligrim_SquareParticles_OD_Right
     tiling=NoTile
     imageSize=<32 64>
     margin=<ltrb 0 0 0 0>
     outlineSettings {
      bUseBrushTransparency=false
      roundingType=HalfHeightRadius
      width=0
      color {colorUseRule=UseColor_Specified, specifiedColor=<rgba 0 0 0 0>}
      cornerRadii=<0 0 0 0>
     }
     tintColor {colorUseRule=UseColor_Specified, specifiedColor=<rgba 1 1 1 1>}
     uVRegion {bIsValid=false, max=<0 0>, min=<0 0>}
    }
    renderTransformPivot=<0 0.5>
   }
   slot {class=OverlaySlot, properties {horizontalAlignment=HAlign_Fill, verticalAlignment=VAlign_Center, padding=<ltrb 64 0 0 0>}}
  }
  PanelBacking {
   widgetClassPath=&Image
   bIsVariable=true
   parent=Overlay_2
   widget=PanelBacking
   properties {brush {
     bIsDynamicallyLoaded=false
     drawAs=Image
     imageType=NoImage
     mirroring=NoMirror
     resourceName=None
     resourceObject=/PilgrimCore/Widgets/Materials/M_UI_Pilgrim_Bandmate_Backing
     tiling=NoTile
     imageSize=<32 32>
     margin=<ltrb 0 0 0 0>
     outlineSettings {
      bUseBrushTransparency=false
      roundingType=HalfHeightRadius
      width=0
      color {colorUseRule=UseColor_Specified, specifiedColor=<rgba 0 0 0 0>}
      cornerRadii=<0 0 0 0>
     }
     tintColor {colorUseRule=UseColor_Specified, specifiedColor=<rgba 1 1 1 1>}
     uVRegion {bIsValid=false, max=<0 0>, min=<0 0>}
    }}
   slot {class=OverlaySlot, properties {horizontalAlignment=HAlign_Fill, verticalAlignment=VAlign_Fill}}
  }
  OD_BracketIndicator {
   widgetClassPath=&Image
   bIsVariable=true
   parent=Overlay_2
   widget=OD_BracketIndicator
   properties {
    visibility=Collapsed
    brush {
     bIsDynamicallyLoaded=false
     drawAs=Image
     imageType=NoImage
     mirroring=NoMirror
     resourceName=None
     resourceObject=/PilgrimCore/Widgets/Materials/MI_Piligrim_Bandmate_ODBracket
     tiling=NoTile
     imageSize=<32 32>
     margin=<ltrb 0 0 0 0>
     outlineSettings {
      bUseBrushTransparency=false
      roundingType=HalfHeightRadius
      width=0
      color {colorUseRule=UseColor_Specified, specifiedColor=<rgba 0 0 0 0>}
      cornerRadii=<0 0 0 0>
     }
     tintColor {colorUseRule=UseColor_Specified, specifiedColor=<rgba 1 1 1 1>}
     uVRegion {bIsValid=false, max=<0 0>, min=<0 0>}
    }
    renderTransform {angle=0, scale=<1 1>, shear=<0 0>, translation=<0 2>}
   }
   slot {class=OverlaySlot, properties {horizontalAlignment=HAlign_Fill, verticalAlignment=VAlign_Fill}}
  }
  HorizontalBox_0 {
   widgetClassPath=&HorizontalBox
   parent=Overlay_2
   widget=HorizontalBox_0
   slot {class=OverlaySlot, properties {horizontalAlignment=HAlign_Fill, verticalAlignment=VAlign_Center, padding=<ltrb 8 4 12 4>}}
  }
  Overlay_1 {
   widgetClassPath=&Overlay
   parent=HorizontalBox_0
   widget=Overlay_1
   slot {class=HorizontalBoxSlot, properties {verticalAlignment=VAlign_Center, size {sizeRule=Fill, value=1}}}
  }
  Bars_WidthControl {
   widgetClassPath=&Spacer
   parent=Overlay_1
   widget=Bars_WidthControl
   properties {size=<96 1>}
   slot {class=OverlaySlot, properties {horizontalAlignment=HAlign_Fill, verticalAlignment=VAlign_Fill}}
  }
  VerticalBox_0 {
   widgetClassPath=&VerticalBox
   parent=Overlay_1
   widget=VerticalBox_0
   slot {class=OverlaySlot, properties {horizontalAlignment=HAlign_Fill}}
  }
  SizeBox_ODBar {
   widgetClassPath=&SizeBox
   bIsVariable=true
   parent=VerticalBox_0
   widget=SizeBox_ODBar
   properties {bOverride_HeightOverride=true, heightOverride=16, renderTransform {angle=0, scale=<1 1>, shear=<-19 0>, translation=<0 0>}}
   slot {class=VerticalBoxSlot}
  }
  OverdriveMeterAndFills_Overlay {widgetClassPath=&Overlay, parent=SizeBox_ODBar, widget=OverdriveMeterAndFills_Overlay, slot {class=SizeBoxSlot}}
  OverdriveHeightControl {
   widgetClassPath=&Spacer
   parent=OverdriveMeterAndFills_Overlay
   widget=OverdriveHeightControl
   properties {size=<1 14>}
   slot {class=OverlaySlot}
  }
  Border {
   widgetClassPath=&Image
   bIsVariable=true
   parent=OverdriveMeterAndFills_Overlay
   widget=Border
   properties {
    visibility=Collapsed
    brush {
     bIsDynamicallyLoaded=false
     drawAs=Image
     imageType=NoImage
     mirroring=NoMirror
     resourceName=None
     resourceObject=/PilgrimCore/Widgets/Materials/M_UI_Overdrive_ProgressMeter_Backing
     tiling=NoTile
     imageSize=<4 4>
     margin=<ltrb 0 0 0 0>
     outlineSettings {
      bUseBrushTransparency=false
      roundingType=HalfHeightRadius
      width=0
      color {colorUseRule=UseColor_Specified, specifiedColor=<rgba 0 0 0 0>}
      cornerRadii=<0 0 0 0>
     }
     tintColor {colorUseRule=UseColor_Specified, specifiedColor=<rgba 1 1 1 1>}
     uVRegion {bIsValid=false, max=<0 0>, min=<0 0>}
    }
   }
   slot {class=OverlaySlot, properties {horizontalAlignment=HAlign_Fill, verticalAlignment=VAlign_Fill}}
  }
  Backing {
   widgetClassPath=&Image
   bIsVariable=true
   parent=OverdriveMeterAndFills_Overlay
   widget=Backing
   properties {brush {
     bIsDynamicallyLoaded=false
     drawAs=Image
     imageType=NoImage
     mirroring=NoMirror
     resourceName=None
     resourceObject=/PilgrimCore/Widgets/Materials/M_UI_Overdrive_ProgressMeter_Backing_Empty
     tiling=NoTile
     imageSize=<4 4>
     margin=<ltrb 0 0 0 0>
     outlineSettings {
      bUseBrushTransparency=false
      roundingType=HalfHeightRadius
      width=0
      color {colorUseRule=UseColor_Specified, specifiedColor=<rgba 0 0 0 0>}
      cornerRadii=<0 0 0 0>
     }
     tintColor {colorUseRule=UseColor_Specified, specifiedColor=<rgba 1 1 1 1>}
     uVRegion {bIsValid=false, max=<0 0>, min=<0 0>}
    }}
   slot {class=OverlaySlot, properties {horizontalAlignment=HAlign_Fill, verticalAlignment=VAlign_Fill, padding=<ltrb 2 2 2 2>}}
  }
  Fill {
   widgetClassPath=&Image
   bIsVariable=true
   parent=OverdriveMeterAndFills_Overlay
   widget=Fill
   properties {brush {
     bIsDynamicallyLoaded=false
     drawAs=Image
     imageType=NoImage
     mirroring=NoMirror
     resourceName=None
     resourceObject=/PilgrimCore/Widgets/Materials/M_UI_Overdrive_ProgressMeter
     tiling=NoTile
     imageSize=<4 6>
     margin=<ltrb 0 0 0 0>
     outlineSettings {
      bUseBrushTransparency=false
      roundingType=HalfHeightRadius
      width=0
      color {colorUseRule=UseColor_Specified, specifiedColor=<rgba 0 0 0 0>}
      cornerRadii=<0 0 0 0>
     }
     tintColor {colorUseRule=UseColor_Specified, specifiedColor=<rgba 1 1 1 1>}
     uVRegion {bIsValid=false, max=<0 0>, min=<0 0>}
    }}
   slot {class=OverlaySlot, properties {horizontalAlignment=HAlign_Fill, verticalAlignment=VAlign_Fill, padding=<ltrb 4 4 4 4>}}
  }
  SizeBox_StreakPips {
   widgetClassPath=&SizeBox
   bIsVariable=true
   parent=VerticalBox_0
   widget=SizeBox_StreakPips
   properties {bOverride_HeightOverride=true, heightOverride=16, renderTransform {angle=0, scale=<1 1>, shear=<-19 0>, translation=<0 0>}}
   slot {class=VerticalBoxSlot, properties {verticalAlignment=VAlign_Top, padding=<ltrb 0 4 0 0>}}
  }
  Overlay_0 {widgetClassPath=&Overlay, parent=SizeBox_StreakPips, widget=Overlay_0, slot {class=SizeBoxSlot}}
  Image {
   widgetClassPath=&Image
   bIsVariable=true
   parent=Overlay_0
   widget=Image
   properties {
    visibility=Collapsed
    brush {
     bIsDynamicallyLoaded=false
     drawAs=Image
     imageType=NoImage
     mirroring=NoMirror
     resourceName=None
     resourceObject=/PilgrimCore/Widgets/Materials/M_UI_StreakMeter_Pips_Backing
     tiling=NoTile
     imageSize=<8 8>
     margin=<ltrb 0 0 0 0>
     outlineSettings {
      bUseBrushTransparency=false
      roundingType=HalfHeightRadius
      width=0
      color {colorUseRule=UseColor_Specified, specifiedColor=<rgba 0 0 0 0>}
      cornerRadii=<0 0 0 0>
     }
     tintColor {colorUseRule=UseColor_Specified, specifiedColor=<rgba 1 1 1 1>}
     uVRegion {bIsValid=false, max=<0 0>, min=<0 0>}
    }
   }
   slot {class=OverlaySlot, properties {horizontalAlignment=HAlign_Fill, verticalAlignment=VAlign_Fill}}
  }
  WBP_StreakLevelPips {
   widgetClassPath=&/WBP_StreakLevelPips
   bIsVariable=true
   parent=Overlay_0
   widget=WBP_StreakLevelPips
   slot {class=OverlaySlot, properties {horizontalAlignment=HAlign_Fill, verticalAlignment=VAlign_Fill, padding=<ltrb 2 2 2 2>}}
  }
  AvatarOverlay {
   widgetClassPath=&Overlay
   bIsVariable=true
   parent=HorizontalBox_0
   widget=AvatarOverlay
   properties {renderTransform {angle=0, scale=<1 1>, shear=<-19 0>, translation=<0 0>}}
   slot {class=HorizontalBoxSlot, properties {horizontalAlignment=HAlign_Center, verticalAlignment=VAlign_Center, padding=<ltrb 4 0 0 0>}}
  }
  Image_340 {
   widgetClassPath=&Image
   bIsVariable=true
   parent=AvatarOverlay
   widget=Image_340
   properties {brush {
     bIsDynamicallyLoaded=false
     drawAs=Image
     imageType=NoImage
     mirroring=NoMirror
     resourceName=None
     resourceObject=/PilgrimCore/Widgets/Materials/MI_Pilgrim_OD_BandmatePortrait_Backing
     tiling=NoTile
     imageSize=<32 32>
     margin=<ltrb 0 0 0 0>
     outlineSettings {
      bUseBrushTransparency=false
      roundingType=HalfHeightRadius
      width=0
      color {colorUseRule=UseColor_Specified, specifiedColor=<rgba 0 0 0 0>}
      cornerRadii=<0 0 0 0>
     }
     tintColor {colorUseRule=UseColor_Specified, specifiedColor=<rgba 1 1 1 1>}
     uVRegion {bIsValid=false, max=<0 0>, min=<0 0>}
    }}
   slot {class=OverlaySlot, properties {horizontalAlignment=HAlign_Fill, verticalAlignment=VAlign_Fill}}
  }
  WBP_Pilgrim_AvatarIcon {
   widgetClassPath=/PilgrimCore/Widgets/Player/WBP_Pilgrim_AvatarIcon
   bIsVariable=true
   parent=AvatarOverlay
   widget=WBP_Pilgrim_AvatarIcon
   slot {class=OverlaySlot, properties {padding=<ltrb 2 2 2 2>}}
  }
 ]
}
