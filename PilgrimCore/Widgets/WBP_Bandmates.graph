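# WBP_Bandmates is the Pilgrim HUD list of bandmates, binding each entry frame's score view model and icon to that player's state.
# SheerForTrack shears the widget to match the track, using lane count, track type and mobile expanded-track state, and counter-shears displayed frames.
# It runs on pre-construct, entry generation and battle initialization.
# The expanded-track settings listener passes zero lanes, which only acts when RemoveShearing is set.
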WidgetBlueprint /PilgrimCore/Widgets/WBP_Bandmates
graphs [
 EventGraph {nodes [
   n1 = "On Entry Initialized (ListView)"!() -> (OutputDelegate, then, Item:Object Reference, Widget:User Widget Object Reference) {is=title K2Node_ComponentBoundEvent}
   > n2 = Cast To WBP_BandmateFrame!(execute, Object=@n1.Widget) -> (then, CastFailed, AsWBP Bandmate Frame:WBP Bandmate Frame Object Reference) {is=title K2Node_DynamicCast}
   > n3 = Cast To PilgrimPlayerInfoVM!(execute, Object=@n1.Item) -> (then, CastFailed, AsPilgrim Player Info VM:Pilgrim Player Info VM Object Reference) {is=title K2Node_DynamicCast}
   > n4 = Cast To FortPlayerStateAthena!(
    execute
    Object=Break Pilgrim Player Info!(PilgrimPlayerInfo:"Pilgrim Player Info Structure (by ref)"=Get PlayerInfo!(self:Pilgrim Player Info VM Object Reference=@n3) -> (PlayerInfo:Pilgrim Player Info Structure)) -> (PlayerState:Fort Player State Weak object, PlayerController:Fort Player Controller Weak object, bIsLocalPlayer:bool=~) {is=title K2Node_BreakStruct}
   ) -> (then, CastFailed, AsFort Player State Athena:Fort Player State Athena Object Reference) {is=title K2Node_DynamicCast}
   > SetOwningPlayerState!(
    execute
    self:Pilgrim Player Score VM Object Reference=Get PilgrimPlayerScoreVM!(self:WBP Bandmate Frame Object Reference=@n2) -> (PilgrimPlayerScoreVM:Pilgrim Player Score VM Object Reference)
    PlayerState:Fort Player State Athena Object Reference=@n4
   ) -> (then)
   > UpdatePlayerIcon!(execute, self:WBP Bandmate Frame Object Reference=@n2, PlayerState:Fort Player State Athena Object Reference=@n4) -> (then)
   "On Entries Generated (ListView)"!() -> (OutputDelegate, then, NumEntries:int=0) {is=title K2Node_ComponentBoundEvent}
   > SheerForTrack!(execute, self, NumLanes:int=GetNumLanes!(self) -> (NumLanes:int=0)) -> (then)
   n5 = Event PreConstruct()
   > Branch(Condition=@n5.IsDesignTime)
   > SheerForTrack(NumLanes=GetNumItems!(self:List View Object Reference=Get ListView() -> (ListView:Common List View Object Reference)) -> (ReturnValue:int=0))
   Event OnMusicBattleInstanceInitializeEvent() {is=title K2Node_Event}
   > SheerForTrack(NumLanes=GetNumLanes() -> (NumLanes=0))
   n6 = GetOwningPlayer()
   > n7 = StartListeningToEvent!(
    execute:Exec
    Target:Gameplay Event Router Component Object Reference=GetGameplayEventRouter!(ContextActor:Actor Object Reference=GetSparksMusicPlayspaceOfClass!(
      PlayspaceType:Sparks Music Playspace Class Reference=/Script/PilgrimCoreRuntime.PilgrimPlayspace
      Actor:Actor Object Reference=@n6
      bMatchChildTypes:bool=false
     ) -> (ReturnValue:Pilgrim Playspace Object Reference)) -> (ReturnValue:Gameplay Event Router Component Object Reference)
    EventType:Script Struct Object Reference=/Script/PilgrimSettingsRuntime.PilgrimEvent_PilgrimSettings_EnableMobileExpandedTrackChanged
    Context:Object Reference
    EventBubblingRule:EEventBubblingRule Enum
   ) -> (
    then:Exec
    AsyncTaskProxy:Async Action Start Listening To Event Object Reference
    OnEventReceived:Exec
    Payload:Pilgrim Event Pilgrim Settings Enable Mobile Expanded Track Changed Structure
   )
   SheerForTrack(execute=@n7.OnEventReceived, NumLanes=0)
  ]}
 GetNumLanes {nodes [
   GetNumLanes() {is=title K2Node_FunctionEntry}
   > n1 = Get!(execute:Exec) -> (then:Exec, else:Exec, MusicBattleInstance:Pilgrim Music Battle Instance Object Reference)
   n2 = GetLocalPilgrimGameForPlayer!(self:Pilgrim Music Battle Instance Object Reference=@n1, InPC:Player Controller Object Reference=GetOwningPlayer()) -> (ReturnValue:Pilgrim Game Object Reference)
   > n3 = Is Valid(InputObject=@n2)
   > Return Node!(execute, NumLanes:int=GetNumLanes!(self:Pilgrim Game Object Reference=@n2) -> (ReturnValue:int=0) {is=title K2Node_CallFunction})
   Return Node(execute=[@n3.Is Not Valid, @n1.else], NumLanes=4)
  ]}
 PivotForTrack {nodes [
   PivotForTrack() {is=title K2Node_FunctionEntry}
   > SetRenderTransformPivot!(execute, self:Widget Object Reference, Pivot:Vector 2D Structure=Get RenderPivot() -> (RenderPivot:Vector 2D Structure)) -> (then)
  ]}
 SetSheerRenderTransform {nodes [
   n1 = SetSheerRenderTransform!() -> (then, Sheer:Vector 2D Structure) {is=title K2Node_FunctionEntry}
   > n2 = GetDisplayedEntryWidgets!(execute, self:List View Base Object Reference=Get ListView()) -> (then, ReturnValue:Array of User Widget Object References)
   > n3 = For Each Loop(Array:Array of Object References=@n2) -> (Array Element:Object Reference)
   > n4 = Cast To WBP_BandmateFrame(Object=@n3)
   > SetShearRenderTransform!(execute, self:WBP Bandmate Frame Object Reference=@n4, Shear:Vector 2D Structure=@n1) -> (then)
  ]}
 SheerForTrack {nodes [
   n1 = SheerForTrack!() -> (then, NumLanes:int) {is=title K2Node_FunctionEntry}
   > n2 = Branch(Condition=Get RemoveShearing() -> (RemoveShearing:bool))
   > SetRenderShear!(execute, self:Widget Object Reference, Shear:Vector 2D Structure=(X=0,Y=0)) -> (then)
   > SetSheerRenderTransform!(execute, self, Sheer:Vector 2D Structure) -> (then) {is=title K2Node_CallFunction}
   n3 = Branch(execute=@n2.else, Condition="Equal (Integer)"(A=@n1))
   n4 = Get TrackType() -> (TrackType:EPilgrimTrackType Enum=TrackGuitar)
   n5 = Branch(
    execute=@n3.else
    Condition=OR Boolean!(
     A="Equal (Integer)"(A=@n1, B=4)
     B="Equal (Enum)"(A:EPilgrimTrackType Enum=@n4, B:EPilgrimTrackType Enum=TrackPlasticDrum)
     C:bool="Equal (Enum)"(A=@n4, B=TrackProCymbals)
    ) -> (ReturnValue)
   )
   n6 = Select(
    Option 0:Vector 2D Structure=(X=19,Y=0)
    Option 1:Vector 2D Structure=(X=35,Y=0)
    Index=IsMobileAndShouldUseExpandedTrack!() -> (ReturnValue:bool)
   ) -> (ReturnValue:Vector 2D Structure)
   > SetRenderShear(Shear=@n6)
   > SetSheerRenderTransform(Sheer=vector2d * vector2d(A:Vector 2D Structure=@n6, B:Vector 2D Structure=(X=-1,Y=1)) -> (ReturnValue:Vector 2D Structure) {is=title K2Node_PromotableOperator})
   n7 = Select(Option 0=(X=23,Y=0), Option 1=(X=40,Y=0), Index=IsMobileAndShouldUseExpandedTrack())
   SetRenderShear(execute=@n5.else, Shear=@n7)
   > SetSheerRenderTransform(Sheer=vector2d * vector2d(A=@n7, B=(X=-1,Y=1)))
  ]}
]
variables [
 RemoveShearing {type=boolean}
 RenderPivot {type=Vector2D, default=<0 0>}
 TrackType {type=EPilgrimTrackType, default=TrackNone}
]
widgets {
 info {inheritedWidgetCount=0, parentClass=/Script/PilgrimCoreUI.PilgrimGameWidgetBase, rootWidgetClass=&VerticalBox, widgetCount=2}
 widgets [
  VerticalBox_15 {widgetClassPath=&VerticalBox, widget=VerticalBox_15, properties {renderTransformPivot=<0.5 0>}}
  ListView {
   widgetClassPath=&CommonListView
   bIsVariable=true
   parent=VerticalBox_15
   widget=ListView
   properties {
    clipping=Inherit
    entryWidgetClass=&/WBP_BandmateFrame.WBP_BandmateFrame_C
    horizontalEntrySpacing=4
    numDesignerPreviewEntries=3
    verticalEntrySpacing=4
   }
   slot {class=VerticalBoxSlot}
  }
 ]
}
