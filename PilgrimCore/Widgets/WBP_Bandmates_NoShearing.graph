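# Wrapper widget in PilgrimCore that hosts WBP_Bandmates as its only widget, with that widget's removeShearing property turned on.
# On Construct it plays a half-second intro animation that holds the bandmates widget offset along X, then slides it into place while fading it in.
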WidgetBlueprint /PilgrimCore/Widgets/WBP_Bandmates_NoShearing
animations [Anim_Intro {
  binding_count=1
  length_seconds=0.5
  object=&self:Anim_Intro
  sequence {bindings [WBP_Bandmates {
     id=#1
     possessed_class=WBP_Bandmates_C
     tracks [
      {
       type=MovieSceneFloatTrack
       display_name=Render Opacity
       sections [{type=MovieSceneFloatSection, channels [None {type=MovieSceneScriptingFloatChannel, keys [[9,0],[15,1]]}]}]
      }
      {
       type=MovieScene2DTransformTrack
       sections [{type=MovieScene2DTransformSection, channels [Translation.X {type=MovieSceneScriptingFloatChannel, keys [[5,384],[7,384],[15,0]]}]}]
      }
     ]
    }]}
 }]
graphs [EventGraph {nodes [
   Event Construct()
   > PlayAnimation(InAnimation=Get Anim_Intro() -> (Anim_Intro:Widget Animation Object Reference))
  ]}]
variables []
widgets {
 info {
  inheritedWidgetCount=0
  parentClass=/Script/PilgrimCoreUI.PilgrimGameWidgetBase
  rootWidgetClass=&/WBP_Bandmates.WBP_Bandmates_C
  widgetCount=1
 }
 widgets [WBP_Bandmates {widgetClassPath=&/WBP_Bandmates, bIsVariable=true, widget=WBP_Bandmates, properties {removeShearing=true}}]
}
