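# Debug CommonActivatableWidget meant to show a "Guitar Peripheral Enabled" label for plastic (guitar peripheral) selection.
# A right-aligned CommonTextBlock sits in an Overlay root; the text is collapsed, and the root is both collapsed and fully transparent.
# It declares player state component, event router and player controller variables, but the listed graph events have no logic, so nothing here reveals the label.
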WidgetBlueprint /PilgrimCore/Widgets/WBP_DebugPlasticSelectionDisplay
graphs [EventGraph {nodes [
   Event PreConstruct()
   Event Construct()
   Event Tick()
   Event OnInitialized()
  ]}]
variables [
 CorePlayerStateComponent {type=PilgrimCorePlayerStateComponent, default=null}
 Event Router {type=GameplayEventRouterComponent, default=null}
 OwningPlayerController {type=PlayerController, default=null}
]
widgets {
 info {inheritedWidgetCount=0, parentClass=&CommonActivatableWidget, rootWidgetClass=&Overlay, widgetCount=2}
 widgets [
  Overlay_0 {widgetClassPath=&Overlay, widget=Overlay_0, properties {renderOpacity=0, visibility=Collapsed}}
  PlasticModeEnabledText {
   widgetClassPath=&CommonTextBlock
   bIsVariable=true
   parent=Overlay_0
   widget=PlasticModeEnabledText
   properties {
    justification=Right
    style=&TextStyle-Base_C
    text=Guitar Peripheral Enabled
    visibility=Collapsed
    font {
     bForceMonospaced=false
     bMaterialIsStencil=false
     fontMaterial=null
     fontObject=&NotoSans
     letterSpacing=0
     monospacedWidth=1
     size=24
     skewAmount=0
     typefaceFontName=NotoSans-Regular
     outlineSettings {
      bApplyOutlineToDropShadows=false
      bMiteredCorners=false
      bSeparateFillAlpha=false
      outlineMaterial=null
      outlineSize=0
      outlineColor=<rgba 0 0 0 1>
     }
    }
    shadowOffset=<0 0>
   }
   slot {class=OverlaySlot, properties {horizontalAlignment=HAlign_Right, verticalAlignment=VAlign_Center}}
  }
 ]
}
