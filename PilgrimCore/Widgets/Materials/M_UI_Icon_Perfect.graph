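# Translucent UI material for a "Perfect" icon, layering the icon texture with two star highlights.
# The icon scales around its center by a Size parameter, then Max-combines with two offset, differently scaled samples of T_Ui_PerfectTexture's green channel.
# Both highlight samples pass through an anti-aliasing distance-field threshold.
# Named reroutes hold star scale and a debug time-sine animation controller, alongside apparently leftover easing, lerp and smoothstep nodes.
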Material /PilgrimCore/Widgets/Materials/M_UI_Icon_Perfect
graph {
 properties {
  bAutomaticallySetUsageInEditor=false
  bUsedWithStaticMesh=false
  blendMode=BLEND_Translucent
  materialDomain=MD_UI
  usedShadingModels=Default Lit
 }
 nodes [
  Constant3Vector() {constant=<rgba 1 1 1 0>}
  Constant() {r=0.9}
  LinearInterpolate() {constA=0.84}
  LinearInterpolate()
  MaterialFunctionCall() {materialFunction=&MF_UI_EaseCurves}
  n1 = Constant() {r=0.7}
  MaterialFunctionCall(A=Constant(), C=@n1, Alpha=NamedRerouteUsage()) {materialFunction=&Lerp_3Color}
  n2 = TextureCoordinate()
  n3 = ScalarParameter() {defaultValue=1, parameterName=Size}
  n4 = AppendVector(A=Constant() {r=-0.18}, B=Constant())
  n5 = TextureCoordinate()
  n6 = TextureSample(UVs=MaterialFunctionCall(UVs=Add(A=@n5, B=@n4), Texture Scale=NamedRerouteUsage()) {materialFunction=&ScaleUVsByCenter}.UVs) {samplerSource=SSM_Clamp_WorldGroupSettings, texture=/PilgrimCore/Widgets/Textures/T_Ui_PerfectTexture}
  n7 = MaterialFunctionCall(DistanceField=@n6.G, Threshold=Constant() {r=0.6}, Smoothing=Constant() {r=0.4}) {materialFunction=&MF_UI_AntiAliasing}
  n8 = MaterialFunctionCall(UVs=@n5, Texture Scale=Constant() {r=0.5}) {materialFunction=&ScaleUVsByCenter}
  n9 = AppendVector(A=Constant() {r=0.54}, B=Constant() {r=0.2})
  n10 = MaterialFunctionCall(
   DistanceField=TextureSample(UVs=Add(A=@n8.UVs, B=@n9)) {samplerSource=SSM_Clamp_WorldGroupSettings, texture=&T_Ui_PerfectTexture}.G
   Threshold=Constant() {r=0.6}
   Smoothing=Constant() {r=0.4}
  ) {materialFunction=&MF_UI_AntiAliasing}
  Max(
   A=TextureSample(UVs=MaterialFunctionCall(UVs=@n2, Texture Scale=@n3) {materialFunction=&ScaleUVsByCenter}.UVs) {samplerType=SAMPLERTYPE_LinearColor, texture=/PilgrimCore/Widgets/Textures/T_Icon_Perfect}.RGB
   B=Max(A=Saturate(None=@n7.Outer), B=Saturate(None=@n10.Outer))
  )
  NamedRerouteDeclaration(None=@n1) {name=BigStarScale, nodeColor=<rgba 0 1 0.0625 1>}
  NamedRerouteDeclaration(None=MaterialFunctionCall() {materialFunction=&DebugTimeSine}.Result) {name=AnimationController, nodeColor=<rgba 1 0 0.2578 1>}
  ScalarParameter() {defaultValue=0.84, parameterName=AnimationController}
  SmoothStep(Value=@n6.G) {constMin=0.9}
 ]
}
