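# Widget that flashes a "perfect" accuracy icon when the owning player scores an accurately hit gem in the Pilgrim music game.
# On construct it waits for the gameplay event router and listens for GemScored events, stopping those listeners on destruct.
# Hits with a tier index within the music battle config's displayed maximum (-1 without an instance) play AccuracyAnimation.
# The animation nudges and fades the icon over skewed, mirrored material backgrounds.
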WidgetBlueprint /PilgrimCore/Widgets/WBP_PilgrimAccuracyIcon
animations [AccuracyAnimation {
  binding_count=4
  length_seconds=1.1
  object=&self:AccuracyAnimation
  sequence {bindings [
    Overlay_0 {
     id=#1
     possessed_class=Overlay
     tracks [{
       type=MovieScene2DTransformTrack
       sections [{
         type=MovieScene2DTransformSection
         end_frame=20
         start_frame=0
         channels [
          Translation.X {type=MovieSceneScriptingFloatChannel, keys [[0,-4],[10,0],[20,0]]}
          Translation.Y {type=MovieSceneScriptingFloatChannel}
         ]
        }]
      }]
    }
    PerfectIcon {
     id=#2
     possessed_class=FortLazyImage
     tracks [
      {
       type=MovieSceneColorTrack
       display_name=Color and Opacity
       sections [{
         type=MovieSceneColorSection
         end_frame=20
         start_frame=0
         channels [
          Color.R {type=MovieSceneScriptingFloatChannel}
          Color.G {type=MovieSceneScriptingFloatChannel}
          Color.B {type=MovieSceneScriptingFloatChannel}
          Color.A {type=MovieSceneScriptingFloatChannel, keys [[0,1],[5,1],[20,0]]}
         ]
        }]
      }
      {
       type=MovieScene2DTransformTrack
       sections [{
         type=MovieScene2DTransformSection
         end_frame=20
         start_frame=0
         channels [
          Scale.X {type=MovieSceneScriptingFloatChannel, keys [[0,1.1],[10,1],[20,1]]}
          Scale.Y {type=MovieSceneScriptingFloatChannel, keys [[0,1.1],[10,1],[20,1]]}
         ]
        }]
      }
     ]
    }
    BG {
     id=#3
     possessed_class=Image
     tracks [
      {
       type=MovieSceneWidgetMaterialTrack
       sections [{
         type=MovieSceneParameterSection
         end_frame=22
         start_frame=0
         channels [
          Gradient_Color_BR {type=MovieSceneScriptingFloatChannel, keys [[0,0.8],[5,1],[20,1]]}
          Gradient_Color_BG {type=MovieSceneScriptingFloatChannel, keys [[0,0.8],[5,0.1],[20,0.1]]}
          Gradient_Color_BB {type=MovieSceneScriptingFloatChannel, keys [[0,0.4],[5,0],[20,0]]}
          Gradient_Color_BA {type=MovieSceneScriptingFloatChannel, const [1,2]}
         ]
        }]
      }
      {
       type=MovieSceneFloatTrack
       display_name=Render Opacity
       sections [{type=MovieSceneFloatSection, end_frame=20, start_frame=0, channels [None {type=MovieSceneScriptingFloatChannel, keys [[0,1],[15,0],[20,0]]}]}]
      }
      {
       type=MovieSceneByteTrack
       display_name=Visibility
       sections [{type=MovieSceneByteSection, end_frame=0, start_frame=0, channels [None {type=MovieSceneScriptingByteChannel, const [3,1]}]}]
      }
     ]
    }
    BG_Echo {
     id=#4
     possessed_class=Image
     tracks [
      {
       type=MovieSceneFloatTrack
       display_name=Render Opacity
       sections [{type=MovieSceneFloatSection, end_frame=20, start_frame=0, channels [None {type=MovieSceneScriptingFloatChannel, keys [[0,1],[10,0],[20,0]]}]}]
      }
      {
       type=MovieSceneByteTrack
       display_name=Visibility
       sections [{type=MovieSceneByteSection, end_frame=0, start_frame=0, channels [None {type=MovieSceneScriptingByteChannel, const [3,1]}]}]
      }
     ]
    }
   ]}
 }]
graphs [
 EventGraph {nodes [
   Event Construct()
   > Parent: Construct()
   > n1 = PilgrimWaitForEventRouter!(execute:Exec, ContextActor:Actor Object Reference=GetOwningPlayer()) -> (then:Exec, OnEventRouterAvailable:Exec, EventRouter:Gameplay Event Router Component Object Reference) {is=title K2Node_AsyncAction}
   n2 = StartListeningToEvent!(
    execute:Exec=@n1.OnEventRouterAvailable
    Target:Gameplay Event Router Component Object Reference=@n1
    EventType:Script Struct Object Reference=/Script/PilgrimCoreRuntime.PilgrimEvent_Scoring_GemScored
    Context:Object Reference=Get PlayerState!(self:Controller Object Reference=GetOwningPlayer()) -> (PlayerState:Player State Object Reference)
    EventBubblingRule:EEventBubblingRule Enum
   ) -> (
    then:Exec
    AsyncTaskProxy:Async Action Start Listening To Event Object Reference
    OnEventReceived:Exec
    Payload:Pilgrim Event Scoring Gem Scored Structure
   )
   > Add(
    TargetArray:Array of Async Action Start Listening To Event Object References=Get GameplayEventHandles() -> (GameplayEventHandles:Array of Async Action Start Listening To Event Object References)
    NewItem:"Async Action Start Listening To Event Object Reference (by ref)"=@n2
   ) {is=title K2Node_CallArrayFunction}
   n3 = Break Pilgrim Event Scoring Gem Scored!(PilgrimEvent_Scoring_GemScored:"Pilgrim Event Scoring Gem Scored Structure (by ref)"=@n2.Payload) -> (
    ForPlayer:Fort Player State Object Reference
    Accuracy:Pilgrim Gem Hit Accuracy Structure
    Gem:Track Item Info Structure
    BaseScore:float=~
    BonusScore:float=~
   ) {is=title K2Node_BreakStruct}
   Branch(
    execute=@n2.OnEventReceived
    Condition="Equal (Object)"(A=GetPlayerController!(self:Player State Object Reference=@n3) -> (ReturnValue:Player Controller Object Reference), B=GetOwningPlayer())
   )
   > ShowAccuracyIcon!(execute, self, PilgrimGemHitAccuracy:"Pilgrim Gem Hit Accuracy Structure (by ref)"=@n3.Accuracy) -> (then)
   Event Destruct()
   n4 = Get GameplayEventHandles()
   > n5 = For Each Loop(Array:Array of Async Action Start Listening To Event Object References=@n4) -> (Array Element:Async Action Start Listening To Event Object Reference)
   Clear(execute=@n5.Completed, TargetArray:Array of Async Action Start Listening To Event Object References=@n4)
   Is Valid(exec=@n5.LoopBody, InputObject=@n5)
   > StopListeningToEvent(self:Async Action Start Listening To Event Object Reference=@n5)
  ]}
 GetMaximumDisplayedAccuracyTier {nodes [
   GetMaximumDisplayedAccuracyTier() {is=title K2Node_FunctionEntry}
   n1 = Get MusicBattleInstance() -> (MusicBattleInstance:Pilgrim Music Battle Instance Object Reference)
   > n2 = Branch(Condition=IsValid(Object=@n1))
   n3 = GetConfig!(self:Pilgrim Music Battle Instance Object Reference=@n1) -> (ReturnValue:Pilgrim Settings Object Reference)
   > Return Node!(
    execute
    TierIndex:int=Break Pilgrim Game Config Settings!(PilgrimGameConfigSettings:"Pilgrim Game Config Settings Structure (by ref)"=Get ConfigSettings!(self:Pilgrim Settings Object Reference=@n3) -> (ConfigSettings:Pilgrim Game Config Settings Structure)) -> (
     SlopMs:float=~
     GemBreakEndInputUnlockMs:float=~
     EndOfTrackBuffer:float=~
     MsToNowBar:float=~
     BackgroundTrackDb:float=~
     BackgroundTrackGain:float=~
     SongAudioResponceCueSettingsMap:Map of EPilgrimSongAudioResponseCue Enums to Pilgrim Song Audio Response Cue Settings Structures
     FireMidiEventTicksEarly:float=~
     BaseGemScoreValue:"Map of EPilgrimSongDifficulty Enums to Float (single-precision)s"
     BaseCymbalGemScoreValue:"Map of EPilgrimSongDifficulty Enums to Float (single-precision)s"
     SustainScorePerBeat:int
     bNormalizeScores:bool=~
     NormalizedScaleTowardsHighestInDifficulty:Scalable Float Structure
     NormalizedScaleTowardsHighestInExpert:Scalable Float Structure
     OutroPreAlertSeconds:float=~
     ScoreMultiplierDefinitions:Array of Pilgrim Score Multiplier Info Structures
     PlayerHealthModifiers:Map of EPilgrimSongDifficulty Enums to Player Health Modifiers Structures
     PilgrimHealthCrowdResponseTiers:"Array of Float (single-precision)s"
     IndividualStarPercentageThresholds:Map of EPilgrimSongDifficulty Enums to Star Percentage Thresholds Structures
     BandStarPercentageThresholds:"Array of Float (single-precision)s"
     IsBandMultiplierBasedOnUniqueInstruments:bool=~
     IsBandVarietyBonusScoredDuringPlay:bool=~
     BandOverdriveScoreMultipliers:"Array of Float (single-precision)s"
     BandVarietyScoreMultipliers:"Array of Float (single-precision)s"
     AccuracyTiers:Array of Pilgrim Accuracy Tier Structures
     MaximumDisplayedAccuracyTierIndex:int
     TeamQuickplayInitialHealth:float=~
     TeamQuickplayOverdriveDecayRate:float=~
     TeamQuickplayInitialOverdrive:float=~
     TeamQuickplayOverdrive:float=~
     TeamQuickplayOverdriveScoreMultiplier:float=~
     TeamQuickplayOverdriveHealthGainMultiplier:float=~
     TeamQuickplayRequiredOverdriveToActivate:float=~
     TeamQuickplayOverdriveAddedOnSuccesfulSection:float=~
     bOverdriveClearsGems:bool=~
     bOverdriveClearAutoHoldSustain:bool=~
     GemBreakPendingBars:float=~
     bEnableInputSpecificGemsAndSmashers:bool=~
     InputConfig:Pilgrim Input Settings Object Reference
     bEnableOverdriveDeploys:bool=~
     bEnableOverdrivePhrases:bool=~
     bEnableCaptureSystem:bool=~
     bEnableBandPerformanceScoring:bool=~
     ArenaInitialShields:float=~
     ArenaInitialHealth:float=~
     ArenaInitialAmmo:bool=~
     DifficultyCaptureFactors:"Map of EPilgrimSongDifficulty Enums to Float (single-precision)s"
     PerInstrumentConfigs:Map of EPilgrimTrackType Enums to Pilgrim Settings Instrument Overrides Object References
     StreakToastThresholds:Array of Integers
     InstrumentTypeMissCues:Map of ESparksInstrumentType Enums to Sound Cue Object References
     DefaultInstrumentTypeMissCue:Sound Cue Object Reference
     CreateStandinBandmembers:bool=~
     SongIntroEmbargoMs:float=~
     DeployOverdriveGem_BarsUntilGemSpawn:float=~
     DeployOverdriveGem_TailDurationBar:float=~
     RollsMinInputForgivenessMs:float=~
     SwellsMinInputForgivenessMs:float=~
     ProVocalPitchThresholds:Map of EPilgrimSongDifficulty Enums to Pro Vocal Scoring Pitch Range Structures
     ProVocalSectionCompletionThresholds:"Map of EPilgrimSongDifficulty Enums to Float (single-precision)s"
     ProVocalsSectionMaxValue:"Map of EPilgrimSongDifficulty Enums to Float (single-precision)s"
     ProVocalsSectionRatings:Array of Pro Vocals Section Rating Structures
     ProVocalsScoreMultiplierDefinitions:Array of Pilgrim Score Multiplier Info Structures
     ScoreWeightInSection:Curve Float Object Reference
    ) {is=title K2Node_BreakStruct}.MaximumDisplayedAccuracyTierIndex
   )
   Return Node(execute=@n2.else, TierIndex=-1)
  ]}
 ShowAccuracyIcon {nodes [
   n1 = ShowAccuracyIcon!() -> (then, PilgrimGemHitAccuracy:"Pilgrim Gem Hit Accuracy Structure (by ref)") {is=title K2Node_FunctionEntry}
   > Branch(Condition="integer <= integer"(
     A=Break Pilgrim Gem Hit Accuracy!(PilgrimGemHitAccuracy:"Pilgrim Gem Hit Accuracy Structure (by ref)"=@n1) -> (AccuracyMs:float=~, AccuracyTier:Pilgrim Accuracy Tier Structure, TierIndex:int) {is=title K2Node_BreakStruct}.TierIndex
     B=GetMaximumDisplayedAccuracyTier!() -> (TierIndex:int=0) {is=title K2Node_CallFunction}
    ))
   > SetVisibility(self=Get PerfectIcon() -> (PerfectIcon:Fort Lazy Image Object Reference), InVisibility=SelfHitTestInvisible)
   > PlayAnimation(InAnimation=Get AccuracyAnimation() -> (AccuracyAnimation:Widget Animation Object Reference))
  ]}
]
variables []
widgets {
 info {inheritedWidgetCount=0, parentClass=/Script/PilgrimCoreUI.PilgrimGameWidgetBase, rootWidgetClass=&Overlay, widgetCount=4}
 widgets [
  Overlay_0 {widgetClassPath=&Overlay, widget=Overlay_0, properties {pixelSnapping=Disabled}}
  BG_Echo {
   widgetClassPath=&Image
   bIsVariable=true
   parent=Overlay_0
   widget=BG_Echo
   properties {
    visibility=Hidden
    brush {
     bIsDynamicallyLoaded=false
     drawAs=Image
     imageType=NoImage
     mirroring=NoMirror
     resourceName=None
     resourceObject=/PilgrimCore/Widgets/Materials/MI_UI_Pilgrim_PerfectFeedback_Echo
     tiling=NoTile
     imageSize=<180 72>
     margin=<ltrb 0 0 0 0>
     outlineSettings {
      bUseBrushTransparency=false
      roundingType=HalfHeightRadius
      width=0
      color {colorUseRule=UseColor_Specified, specifiedColor=<rgba 0 0 0 0>}
      cornerRadii=<0 0 0 0>
     }
     tintColor {colorUseRule=UseColor_Specified, specifiedColor=<rgba 1 1 1 1>}
     uVRegion {bIsValid=false, max=<0 0>, min=<0 0>}
    }
    renderTransform {angle=0, scale=<-1 1>, shear=<-15 0>, translation=<5 0>}
   }
   slot {class=OverlaySlot, properties {horizontalAlignment=HAlign_Fill, verticalAlignment=VAlign_Center}}
  }
  BG {
   widgetClassPath=&Image
   bIsVariable=true
   parent=Overlay_0
   widget=BG
   properties {
    visibility=Hidden
    brush {
     bIsDynamicallyLoaded=false
     drawAs=Image
     imageType=NoImage
     mirroring=NoMirror
     resourceName=None
     resourceObject=/PilgrimCore/Widgets/Materials/MI_UI_Pilgrim_PerfectFeedback
     tiling=NoTile
     imageSize=<180 36>
     margin=<ltrb 0 0 0 0>
     outlineSettings {
      bUseBrushTransparency=false
      roundingType=HalfHeightRadius
      width=0
      color {colorUseRule=UseColor_Specified, specifiedColor=<rgba 0 0 0 0>}
      cornerRadii=<0 0 0 0>
     }
     tintColor {colorUseRule=UseColor_Specified, specifiedColor=<rgba 1 1 1 1>}
     uVRegion {bIsValid=false, max=<0 0>, min=<0 0>}
    }
    renderTransform {angle=0, scale=<-1 1>, shear=<-15 0>, translation=<0 0>}
   }
   slot {class=OverlaySlot, properties {horizontalAlignment=HAlign_Fill, verticalAlignment=VAlign_Center}}
  }
  PerfectIcon {
   widgetClassPath=&FortLazyImage
   bIsVariable=true
   parent=Overlay_0
   widget=PerfectIcon
   properties {
    visibility=Hidden
    brush {
     bIsDynamicallyLoaded=false
     drawAs=Image
     imageType=FullColor
     mirroring=NoMirror
     resourceName=None
     resourceObject=/PilgrimCore/Widgets/Materials/M_UI_Icon_Perfect
     tiling=NoTile
     imageSize=<64 64>
     margin=<ltrb 0 0 0 0>
     outlineSettings {
      bUseBrushTransparency=false
      roundingType=HalfHeightRadius
      width=0
      color {colorUseRule=UseColor_Specified, specifiedColor=<rgba 0 0 0 0>}
      cornerRadii=<0 0 0 0>
     }
     tintColor {colorUseRule=UseColor_Specified, specifiedColor=<rgba 1 1 1 1>}
     uVRegion {bIsValid=false, max=<0 0>, min=<0 0>}
    }
   }
   slot {class=OverlaySlot, properties {horizontalAlignment=HAlign_Right, verticalAlignment=VAlign_Center, padding=<ltrb 0 0 10 0>}}
  }
 ]
}
