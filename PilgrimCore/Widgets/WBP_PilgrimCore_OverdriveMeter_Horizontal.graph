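# Horizontal Overdrive meter widget showing the local player's Overdrive charge and active state in Pilgrim songs.
# Event-router listeners size it by lane count, tint the border, cache dynamic materials and fade it with the track's secondary-elements phase.
# Progress events, plus Tick polling while active or rising, feed UpdateLocalOverdrive.
# That function smooths the value with FInterpTo, drives material parameters and shows emission and particles during Overdrive.
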WidgetBlueprint /PilgrimCore/Widgets/WBP_PilgrimCore_OverdriveMeter_Horizontal
graphs [
 EventGraph {nodes [
   Event On Owner Set() {is=title K2Node_CustomEvent}
   > n1 = WaitForDiscoverablePlayspaceOfTypeForActor!(
    execute:Exec
    PlayspaceClass:Playspace Soft Class Reference=/Script/PilgrimCoreRuntime.PilgrimPlayspace
    Actor:Actor Object Reference=GetOwningPlayer()
    Requester:Object Reference
   ) -> (
    then:Exec
    AsyncTaskProxy:Wait For Discoverable Playspace With Filter Action Object Reference
    OnFoundPlayspace:Exec
    Playspace:Playspace Object Reference
   ) {is=title K2Node_AsyncAction}
   n2 = PilgrimWaitForEventRouter!(execute:Exec=@n1.OnFoundPlayspace, ContextActor:Actor Object Reference=GetOwningPlayer()) -> (then:Exec, OnEventRouterAvailable:Exec, EventRouter:Gameplay Event Router Component Object Reference) {is=title K2Node_AsyncAction}
   n3 = Sequence(execute=@n2.OnEventRouterAvailable)
   n4 = GetOwningPlayer()
   > n5 = StartListeningToEvent!(
    execute:Exec
    Target:Gameplay Event Router Component Object Reference=@n2
    EventType:Script Struct Object Reference=/Script/PilgrimCoreRuntime.PilgrimEvent_PilgrimGame_PlayingSong
    Context:Object Reference=@n4
    EventBubblingRule:EEventBubblingRule Enum
   ) -> (
    then:Exec
    AsyncTaskProxy:Async Action Start Listening To Event Object Reference
    OnEventReceived:Exec
    Payload:Pilgrim Event Pilgrim Game Playing Song Structure
   )
   > Add(
    TargetArray:Array of Async Action Start Listening To Event Object References=Get GameplayEventHandles() -> (GameplayEventHandles:Array of Async Action Start Listening To Event Object References)
    NewItem:"Async Action Start Listening To Event Object Reference (by ref)"=@n5
   ) {is=title K2Node_CallArrayFunction}
   n6 = Break Pilgrim Event Pilgrim Game Playing Song!(PilgrimEvent_PilgrimGame_PlayingSong:"Pilgrim Event Pilgrim Game Playing Song Structure (by ref)"=@n5.Payload) -> (PilgrimGame:Pilgrim Game Object Reference) {is=title K2Node_BreakStruct}
   n7 = GetNumLanes!(self:Pilgrim Game Object Reference=@n6) -> (ReturnValue:int=0)
   SetWidthOverride(
    execute=@n5.OnEventReceived
    self=Get SizeBox_Adjustable() -> (SizeBox_Adjustable:Size Box Object Reference)
    InWidthOverride=SelectFloat(A=110, B=134, bPickA="Equal (Integer)"(A=@n7, B=4))
   )
   > UpdateOverdriveUIColor!(execute, self, IsPeripheral:bool=IsTrackUsingProColors!(self:Pilgrim Game Object Reference=@n6) -> (ReturnValue:bool)) -> (then)
   > n8 = GetDynamicMaterial(self=Get Fill() -> (Fill:Image Object Reference))
   > Set Fill Material Instance(Fill Material Instance:Material Instance Dynamic Object Reference=@n8) -> (Output_Get:Material Instance Dynamic Object Reference)
   > n9 = GetDynamicMaterial(self=Get Glow() -> (Glow:Image Object Reference))
   > Set Glow Material Instance(Glow Material Instance:Material Instance Dynamic Object Reference=@n9) -> (Output_Get:Material Instance Dynamic Object Reference)
   > n10 = GetDynamicMaterial(self=Get Particles() -> (Particles:Image Object Reference))
   > Set Particles Material Instance(Particles Material Instance:Material Instance Dynamic Object Reference=@n10) -> (Output_Get:Material Instance Dynamic Object Reference)
   > n11 = GetDynamicMaterial(self=Get Emission() -> (Emission:Image Object Reference))
   > Set Emission Material Instance(Emission Material Instance:Material Instance Dynamic Object Reference=@n11) -> (Output_Get:Material Instance Dynamic Object Reference)
   > Set UpdatedOverdrive(UpdatedOverdrive:double=0) -> (Output_Get:double)
   > UpdateLocalOverdrive!(execute, self, IsOverdriveActive:bool=false, CurrentOverdrive:double=0) -> (then)
   n12 = StartListeningToEvent(execute=@n3.then_1, Target=@n2, EventType=/Script/PilgrimCoreRuntime.PilgrimEvent_PilgrimTrack_SecondaryElementsPhaseChanged, Context=@n4) -> (Payload:Pilgrim Event Pilgrim Track Secondary Elements Phase Changed Structure)
   > Add(TargetArray=Get GameplayEventHandles(), NewItem=@n12)
   SetRenderOpacity(
    execute=@n12.OnEventReceived
    InOpacity=Break Pilgrim Event Pilgrim Track Secondary Elements Phase Changed!(PilgrimEvent_PilgrimTrack_SecondaryElementsPhaseChanged:"Pilgrim Event Pilgrim Track Secondary Elements Phase Changed Structure (by ref)"=@n12.Payload) -> (Track:Pilgrim Track Object Reference, SecondaryElementsPhase:float=~) {is=title K2Node_BreakStruct}.SecondaryElementsPhase
   )
   n13 = Event OnOverdriveDataReceivedLocalPlayer!() -> (OutputDelegate, then, GameProgress:"Pilgrim Game Progress Structure (by ref)") {is=title K2Node_Event}
   > n14 = Sequence()
   n15 = Break Pilgrim Game Progress!(PilgrimGameProgress:"Pilgrim Game Progress Structure (by ref)"=@n13.GameProgress) -> (
    NumStreakBreakingMisses:int
    bHasPerfectCombo:bool=~
    RawCurrentScore:float=~
    CurrentModifierBonus:float=~
    IsOverdriveActive:bool=~
    CurrentOverdrive:float=~
    BandModifierBonusContribution:float=~
    RawMaxScore:float=~
    MaxStreakLength:int
    AverageStreakMultiplierPerHitOrPass:float=~
    CurrentSectionCompletionPercentage:float=~
   ) {is=title K2Node_BreakStruct}
   > UpdateLocalOverdrive(IsOverdriveActive=@n15.IsOverdriveActive, CurrentOverdrive=@n15.CurrentOverdrive)
   Set LocalPlayerLastUpdate(execute=@n14.then_1, LocalPlayerLastUpdate:Pilgrim Game Progress Structure) -> (Output_Get:Pilgrim Game Progress Structure)
   n16 = Event Tick()
   > n17 = Set In Delta Time(In Delta Time:float=@n16.InDeltaTime) -> (Output_Get:float)
   > Parent: Tick!(execute, MyGeometry:Geometry Structure=@n16.MyGeometry, InDeltaTime:float=@n17) -> (then) {is=title K2Node_CallParentFunction}
   > n18 = Sequence()
   n19 = Break Pilgrim Game Progress!(PilgrimGameProgress=Get LocalPlayerLastUpdate() -> (LocalPlayerLastUpdate:Pilgrim Game Progress Structure)) -> (
    NumStreakBreakingMisses
    bHasPerfectCombo=~
    RawCurrentScore=~
    CurrentModifierBonus=~
    IsOverdriveActive=~
    BandScoreContribution:float=~
    BandModifierBonusContribution=~
    RawMaxScore=~
    MaxStreakLength
    AverageStreakMultiplierPerHitOrPass=~
    CurrentSectionCompletionPercentage=~
   )
   > n20 = Branch(Condition=@n19.IsOverdriveActive)
   n21 = Get!(execute:Exec=[@n20.then, @n23.then]) -> (then:Exec, else:Exec, MusicBattleInstance:Pilgrim Music Battle Instance Object Reference)
   n22 = GetLocalPilgrimGameForPlayer!(self:Pilgrim Music Battle Instance Object Reference=@n21, InPC:Player Controller Object Reference=GetOwningPlayer()) -> (ReturnValue:Pilgrim Game Object Reference)
   > Is Valid(InputObject=@n22)
   > UpdateLocalOverdrive(
    IsOverdriveActive=@n19.IsOverdriveActive
    CurrentOverdrive=GetCurrentODValue!(self:Pilgrim Game Object Reference=@n22) -> (ReturnValue:float)
   )
   n23 = Branch(
    execute=@n18.then_1
    Condition="float < float"(A=Get UpdatedOverdrive() -> (UpdatedOverdrive:double), B=Get CurrentOverdrive() -> (CurrentOverdrive:float))
   )
   Event Construct()
   > SetRenderOpacity(InOpacity=0)
  ]}
 UpdateLocalOverdrive {nodes [
   n1 = UpdateLocalOverdrive!() -> (then, IsOverdriveActive:bool=~, CurrentOverdrive:double=~) {is=title K2Node_FunctionEntry}
   > n2 = Set CurrentOverdrive(CurrentOverdrive:float=@n1.CurrentOverdrive) -> (Output_Get:float)
   > n3 = Set UpdatedOverdrive(UpdatedOverdrive=FInterpTo(Current=Get UpdatedOverdrive(), Target=@n2, DeltaTime=Get In Delta Time() -> (In Delta Time:float), InterpSpeed=10))
   > SetScalarParameterValue(
    self=[
     Get Emission Material Instance() -> (Emission Material Instance:Material Instance Dynamic Object Reference)
     Get Glow Material Instance() -> (Glow Material Instance:Material Instance Dynamic Object Reference)
     Get GlowActiveMaterial Instance() -> (GlowActiveMaterial Instance:Material Instance Dynamic Object Reference)
     Get FillActiveMaterial Instance() -> (FillActiveMaterial Instance:Material Instance Dynamic Object Reference)
     Get Fill Material Instance() -> (Fill Material Instance:Material Instance Dynamic Object Reference)
     Get Particles Material Instance() -> (Particles Material Instance:Material Instance Dynamic Object Reference)
    ]
    ParameterName=Progress
    Value=@n3
   )
   > SetScalarParameterValue(
    self=[Get Glow Material Instance(), Get Fill Material Instance()]
    ParameterName=Alpha_Final
    Value=Select(Option 0:float=1, Option 1:float=0, Index="Equal (Float)"(A=@n2)) -> (ReturnValue:float)
   )
   > SetScalarParameterValue(
    self=Get Glow Material Instance()
    ParameterName=Alpha_Progress
    Value=Select(Option 0=float * float(A=@n2, B=0.5) {is=title K2Node_PromotableOperator}, Option 1=1, Index=@n1)
   )
   > SetScalarParameterValue(self=Get Fill Material Instance(), ParameterName=OverdriveReady, Value=Select(Option 0=0, Option 1=1, Index=@n1))
   > SetScalarParameterValue(self=Get Fill Material Instance(), ParameterName=OverdriveActive, Value=Select(Option 0=0, Option 1=1, Index=@n1))
   > SetVisibility(
    self=[Get Emission(), Get Particles()]
    InVisibility=Select(Option 0:ESlateVisibility Enum=Collapsed, Option 1:ESlateVisibility Enum=SelfHitTestInvisible, Index=@n1) -> (ReturnValue:ESlateVisibility Enum)
   )
   AND Boolean(A="float >= float"(A=@n2, B=0.01), B=NOT Boolean(A=@n1))
  ]}
 UpdateOverdriveUIColor {nodes [
   n1 = UpdateOverdriveUIColor!() -> (then, IsPeripheral:bool=~) {is=title K2Node_FunctionEntry}
   > n2 = GetDynamicMaterial(self=Get Border() -> (Border:Image Object Reference))
   > SetScalarParameterValue(self=@n2, ParameterName=IsPeripheralInstrument, Value="To Float (Boolean)"(InBool=@n1))
  ]}
]
variables [
 AmountHorizontalMovement_BandmateIcons {type=integer, default=32}
 CurrentOverdrive {type=number}
 CurrentOverdrive_AllowableDeviation {type=number, default=0.1}
 CurrentOverdrive_BandmateIcons {type="array<number>", default []}
 Emission Material Instance {type=MaterialInstanceDynamic}
 Fill Material Instance {type=MaterialInstanceDynamic}
 FillActiveMaterial Instance {type=MaterialInstanceDynamic}
 Glow Material Instance {type=MaterialInstanceDynamic}
 GlowActiveMaterial Instance {type=MaterialInstanceDynamic}
 In Delta Time {type=number}
 IsInnermostBandmateIcon_Bool {type="array<boolean>", default []}
 IsLocalPlayer_Bool {type=boolean}
 IsLocalPlayer_Int {type=integer}
 LabelGemMap {default {}}
 Local Music Track {type=BP_PilgrimTrack_C, default=null}
 LocalPlayerLastUpdate {
  type=PilgrimGameProgress
  default {
   averageStreakMultiplierPerHitOrPass=0
   bHasFullCombo=true
   bHasPerfectCombo=true
   bIsAtMaxStreak=false
   bandModifierBonusContribution=0
   bandScoreContribution=0
   currentChordScore=0
   currentHealth=0
   currentModifierBonus=0
   currentModulation=0
   currentOverdrive=0
   currentOverdriveSectionNum=-1
   currentScore=0
   currentSectionCompletionPercentage=0
   currentSectionValue=0
   currentStar=0
   currentSustainScore=0
   currentTick=0
   isOverdriveActive=false
   maxScore=0
   maxStreakLength=0
   nextStarProgress=0
   noteStreak=0
   notesPastLastMultiplier=-1
   notesToNextMultiplier=-1
   numHits=0
   numMisses=0
   numPasses=0
   numPerfects=0
   numStreakBreakingMisses=0
   overdriveScoringState=None
   overdriveTime=0
   rawCurrentScore=0
   rawMaxScore=0
   streakScoreMultiplier=1
   sectionsCompletedByTier []
  }
 }
 Particles Material Instance {type=MaterialInstanceDynamic}
 TotalPossibleInstrumentIconTravelDistance {type=integer, default=-300}
 UnequalOverdrive_BandmateIcons {type=integer}
 UpdatedOverdrive {type=number}
]
widgets {
 info {inheritedWidgetCount=0, parentClass=/Script/PilgrimCoreUI.PilgrimGameWidgetBase, rootWidgetClass=&ScaleBox, widgetCount=11}
 widgets [
  ScaleBox_1 {widgetClassPath=&ScaleBox, widget=ScaleBox_1, properties {stretch=UserSpecified, userSpecifiedScale=7}}
  SizeBox_Fixed {
   widgetClassPath=&SizeBox
   bIsVariable=true
   parent=ScaleBox_1
   widget=SizeBox_Fixed
   properties {bOverride_HeightOverride=true, bOverride_WidthOverride=true, heightOverride=64, widthOverride=134}
   slot {class=ScaleBoxSlot, properties {verticalAlignment=VAlign_Fill}}
  }
  Overlay_0 {
   widgetClassPath=&Overlay
   parent=SizeBox_Fixed
   widget=Overlay_0
   slot {class=SizeBoxSlot, properties {verticalAlignment=VAlign_Center}}
  }
  SizeBox_Adjustable {
   widgetClassPath=&SizeBox
   bIsVariable=true
   parent=Overlay_0
   widget=SizeBox_Adjustable
   properties {bOverride_HeightOverride=true, bOverride_WidthOverride=true, heightOverride=64, widthOverride=110}
   slot {class=OverlaySlot, properties {horizontalAlignment=HAlign_Center, verticalAlignment=VAlign_Center}}
  }
  OverdriveMeterAndFills_Overlay {
   widgetClassPath=&Overlay
   parent=SizeBox_Adjustable
   widget=OverdriveMeterAndFills_Overlay
   slot {class=SizeBoxSlot, properties {padding=<ltrb 4 0 4 0>}}
  }
  Border {
   widgetClassPath=&Image
   bIsVariable=true
   parent=OverdriveMeterAndFills_Overlay
   widget=Border
   properties {
    bIsVolatile=true
    brush {
     bIsDynamicallyLoaded=false
     drawAs=Image
     imageType=NoImage
     mirroring=NoMirror
     resourceName=None
     resourceObject=/PilgrimCore/Widgets/Materials/M_UI_Overdrive_ProgressMeter_Backing
     tiling=NoTile
     imageSize=<32 10>
     margin=<ltrb 0 0 0 0>
     outlineSettings {
      bUseBrushTransparency=false
      roundingType=HalfHeightRadius
      width=0
      color {colorUseRule=UseColor_Specified, specifiedColor=<rgba 0 0 0 0>}
      cornerRadii=<0 0 0 0>
     }
     tintColor {colorUseRule=UseColor_Specified, specifiedColor=<rgba 1 1 1 1>}
     uVRegion {bIsValid=false, max=<0 0>, min=<0 0>}
    }
   }
   slot {class=OverlaySlot, properties {horizontalAlignment=HAlign_Fill, verticalAlignment=VAlign_Center}}
  }
  Empty {
   widgetClassPath=&Image
   bIsVariable=true
   parent=OverdriveMeterAndFills_Overlay
   widget=Empty
   properties {brush {
     bIsDynamicallyLoaded=false
     drawAs=Image
     imageType=NoImage
     mirroring=NoMirror
     resourceName=None
     resourceObject=/PilgrimCore/Widgets/Materials/M_UI_Overdrive_ProgressMeter_Backing_Empty
     tiling=NoTile
     imageSize=<32 6>
     margin=<ltrb 0 0 0 0>
     outlineSettings {
      bUseBrushTransparency=false
      roundingType=HalfHeightRadius
      width=0
      color {colorUseRule=UseColor_Specified, specifiedColor=<rgba 0 0 0 0>}
      cornerRadii=<0 0 0 0>
     }
     tintColor {colorUseRule=UseColor_Specified, specifiedColor=<rgba 1 1 1 1>}
     uVRegion {bIsValid=false, max=<0 0>, min=<0 0>}
    }}
   slot {class=OverlaySlot, properties {horizontalAlignment=HAlign_Fill, verticalAlignment=VAlign_Center, padding=<ltrb 1 2 1 2>}}
  }
  Glow {
   widgetClassPath=&Image
   bIsVariable=true
   parent=OverdriveMeterAndFills_Overlay
   widget=Glow
   properties {
    bIsVolatile=true
    brush {
     bIsDynamicallyLoaded=false
     drawAs=Image
     imageType=NoImage
     mirroring=NoMirror
     resourceName=None
     resourceObject=/PilgrimCore/Widgets/Materials/M_UI_Overdrive_ProgressMeter_Glow
     tiling=NoTile
     imageSize=<32 10>
     margin=<ltrb 0 0 0 0>
     outlineSettings {
      bUseBrushTransparency=false
      roundingType=HalfHeightRadius
      width=0
      color {colorUseRule=UseColor_Specified, specifiedColor=<rgba 0 0 0 0>}
      cornerRadii=<0 0 0 0>
     }
     tintColor {colorUseRule=UseColor_Specified, specifiedColor=<rgba 1 1 1 1>}
     uVRegion {bIsValid=false, max=<0 0>, min=<0 0>}
    }
    renderTransform {angle=0, scale=<1 1>, shear=<0 0>, translation=<0 -2>}
   }
   slot {class=OverlaySlot, properties {horizontalAlignment=HAlign_Fill, verticalAlignment=VAlign_Center}}
  }
  Emission {
   widgetClassPath=&Image
   bIsVariable=true
   parent=OverdriveMeterAndFills_Overlay
   widget=Emission
   properties {
    bIsVolatile=true
    visibility=Collapsed
    brush {
     bIsDynamicallyLoaded=false
     drawAs=Image
     imageType=NoImage
     mirroring=NoMirror
     resourceName=None
     resourceObject=/PilgrimCore/Widgets/Materials/MI_UI_Pilgrim_Overdrive_Pulse
     tiling=NoTile
     imageSize=<24 32>
     margin=<ltrb 0.5 0.5 0.5 0.5>
     outlineSettings {
      bUseBrushTransparency=false
      roundingType=HalfHeightRadius
      width=0
      color {colorUseRule=UseColor_Specified, specifiedColor=<rgba 0 0 0 0>}
      cornerRadii=<0 0 0 0>
     }
     tintColor {colorUseRule=UseColor_Specified, specifiedColor=<rgba 1 1 1 1>}
     uVRegion {bIsValid=false, max=<0 0>, min=<0 0>}
    }
    renderTransform {angle=0, scale=<1 1>, shear=<0 0>, translation=<0 -1>}
    renderTransformPivot=<0 0.5>
   }
   slot {class=OverlaySlot, properties {horizontalAlignment=HAlign_Fill, verticalAlignment=VAlign_Center}}
  }
  Particles {
   widgetClassPath=&Image
   bIsVariable=true
   parent=OverdriveMeterAndFills_Overlay
   widget=Particles
   properties {
    bIsVolatile=true
    visibility=Collapsed
    brush {
     bIsDynamicallyLoaded=false
     drawAs=Image
     imageType=NoImage
     mirroring=NoMirror
     resourceName=None
     resourceObject=/PilgrimCore/Widgets/Materials/M_UI_OverdriveParticles
     tiling=NoTile
     imageSize=<32 8>
     margin=<ltrb 0 0 0 0>
     outlineSettings {
      bUseBrushTransparency=false
      roundingType=HalfHeightRadius
      width=0
      color {colorUseRule=UseColor_Specified, specifiedColor=<rgba 0 0 0 0>}
      cornerRadii=<0 0 0 0>
     }
     tintColor {colorUseRule=UseColor_Specified, specifiedColor=<rgba 1 1 1 1>}
     uVRegion {bIsValid=false, max=<0 0>, min=<0 0>}
    }
    renderTransform {angle=0, scale=<1 3>, shear=<0 0>, translation=<0 -4>}
    renderTransformPivot=<0.5 1>
   }
   slot {class=OverlaySlot, properties {horizontalAlignment=HAlign_Fill, verticalAlignment=VAlign_Center}}
  }
  Fill {
   widgetClassPath=&Image
   bIsVariable=true
   parent=OverdriveMeterAndFills_Overlay
   widget=Fill
   properties {
    bIsVolatile=true
    brush {
     bIsDynamicallyLoaded=false
     drawAs=Image
     imageType=NoImage
     mirroring=NoMirror
     resourceName=None
     resourceObject=/PilgrimCore/Widgets/Materials/M_UI_Overdrive_ProgressMeter
     tiling=NoTile
     imageSize=<32 5>
     margin=<ltrb 0 0 0 0>
     outlineSettings {
      bUseBrushTransparency=false
      roundingType=HalfHeightRadius
      width=0
      color {colorUseRule=UseColor_Specified, specifiedColor=<rgba 0 0 0 0>}
      cornerRadii=<0 0 0 0>
     }
     tintColor {colorUseRule=UseColor_Specified, specifiedColor=<rgba 1 1 1 1>}
     uVRegion {bIsValid=false, max=<0 0>, min=<0 0>}
    }
   }
   slot {class=OverlaySlot, properties {horizontalAlignment=HAlign_Fill, verticalAlignment=VAlign_Center, padding=<ltrb 2 2 2 3>}}
  }
 ]
}
