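# Pilgrim debug settings widget, collapsed by default, with combo boxes for input-binding and camera presets.
# Selections broadcast through the player's Gameplay Event Router, with a flag suppressing echoes from programmatic keybind changes.
# Camera picks, or track layout events mapped through LayoutCameraMap, set SelectedCamera and send a tagged gameplay event to the owning player, passing the widget as payload.
# Tick counts down a camera transition timer.
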WidgetBlueprint /PilgrimCore/Widgets/WBP_PilgrimDebugSettings
graphs [
 EventGraph {nodes [
   n1 = "On Selection Changed (ComboBox_InputBinding)"!() -> (OutputDelegate, then, SelectedItem:String, SelectionType:ESelectInfo Enum) {is=title K2Node_ComponentBoundEvent}
   > n2 = Branch(Condition=Get IgnoreSelectedPresetChange() -> (IgnoreSelectedPresetChange:bool))
   n3 = PilgrimWaitForEventRouter!(execute:Exec=@n2.else, ContextActor:Actor Object Reference=GetOwningPlayer()) -> (then:Exec, OnEventRouterAvailable:Exec, EventRouter:Gameplay Event Router Component Object Reference) {is=title K2Node_AsyncAction}
   Broadcast Event!(
    execute=@n3.OnEventRouterAvailable
    self:Gameplay Event Router Component Object Reference=@n3
    EventData:Pilgrim Event Keybind Preset Changed Structure=Make Pilgrim Event Keybind Preset Changed!(Name:String=@n1.SelectedItem, InputMappingContext:Input Mapping Context Object Reference) -> (PilgrimEvent_KeybindPresetChanged:Pilgrim Event Keybind Preset Changed Structure) {is=title K2Node_MakeStruct}
    EventContext:Object Reference
   ) -> (then, ReturnValue:bool)
   n4 = "On Selection Changed (ComboBox_CameraPresets)"!() -> (OutputDelegate, then, SelectedItem:String, SelectionType:ESelectInfo Enum) {is=title K2Node_ComponentBoundEvent}
   > SelectCamera!(execute, self, Selected Camera:"String (by ref)"=@n4.SelectedItem) -> (then)
   > n5 = PilgrimWaitForEventRouter(ContextActor=GetOwningPlayer())
   Broadcast Event(
    execute=@n5.OnEventRouterAvailable
    self=@n5
    EventData:Pilgrim Event Camera Preset Changed Structure=Make Pilgrim Event Camera Preset Changed!(Name:String=@n4.SelectedItem) -> (PilgrimEvent_CameraPresetChanged:Pilgrim Event Camera Preset Changed Structure) {is=title K2Node_MakeStruct}
   )
   Event PreConstruct()
   > n6 = Keys!(
    execute
    TargetMap:Map of Strings to Pilgrim Battle Instance Camera Layout Base Class References=Get CameraPresets() -> (CameraPresets:Map of Strings to Pilgrim Battle Instance Camera Layout Base Class References)
   ) -> (then, Keys:Array of Strings)
   > n7 = For Each Loop(Array=@n6)
   > AddOption!(
    execute
    self:"Combo Box (String) Object Reference"=Get ComboBox_CameraPresets() -> (ComboBox_CameraPresets:"Combo Box (String) Object Reference")
    Option:String=@n7
   ) -> (then)
   Event Construct()
   > n8 = PilgrimWaitForEventRouter(ContextActor=GetOwningPlayer())
   n9 = StartListeningToEvent!(
    execute:Exec=@n8.OnEventRouterAvailable
    Target:Gameplay Event Router Component Object Reference=@n8
    EventType:Script Struct Object Reference=/Script/PilgrimCoreRuntime.PilgrimEvent_TrackLayoutPresetChanged
    Context:Object Reference
    EventBubblingRule:EEventBubblingRule Enum
   ) -> (
    then:Exec
    AsyncTaskProxy:Async Action Start Listening To Event Object Reference
    OnEventReceived:Exec
    Payload:Pilgrim Event Track Layout Preset Changed Structure
   )
   > Add(
    TargetArray:Array of Async Action Start Listening To Event Object References=Get GameplayEventHandles() -> (GameplayEventHandles:Array of Async Action Start Listening To Event Object References)
    NewItem:"Async Action Start Listening To Event Object Reference (by ref)"=@n9
   ) {is=title K2Node_CallArrayFunction}
   > Reroute Node()
   On Music Battle Layout Changed!(
    execute=@n9.OnEventReceived
    self
    Selected Layout:String=Break Pilgrim Event Track Layout Preset Changed!(PilgrimEvent_TrackLayoutPresetChanged:"Pilgrim Event Track Layout Preset Changed Structure (by ref)"=@n9.Payload) -> (PresetName:String) {is=title K2Node_BreakStruct}
   ) -> (then)
   n10 = Event Tick()
   > On Tick UpdateMusicBattleOptions!(execute, self, B:double=@n10.InDeltaTime) -> (then)
   Reroute Node()
   > SetVisibility(
    self=Get VerticalBox_Settings() -> (VerticalBox_Settings:Vertical Box Object Reference)
    InVisibility=Reroute Node(InputPin:ESlateVisibility Enum) -> (OutputPin:ESlateVisibility Enum)
   )
   Reroute Node(InputPin:Gameplay Event Router Component Object Reference=@n8) -> (OutputPin:Gameplay Event Router Component Object Reference)
  ]}
 On Music Battle Layout Changed {nodes [
   n1 = On Music Battle Layout Changed!() -> (then, Selected Layout:String) {is=title K2Node_FunctionEntry}
   n2 = Find(TargetMap=Get LayoutCameraMap() -> (LayoutCameraMap:Map of Strings to Strings), Key=@n1)
   > Branch(Condition=@n2.ReturnValue)
   > SelectCamera(Selected Camera=@n2)
  ]}
 On Tick UpdateMusicBattleOptions {nodes [
   n1 = On Tick UpdateMusicBattleOptions!() -> (then, B:double=~) {is=title K2Node_FunctionEntry}
   > Branch(Condition="float > float"(A=Get CameraTransitionTimeRemaining() -> (CameraTransitionTimeRemaining:double)))
   n2 = Get CameraTransitionTimeRemaining()
   > Set CameraTransitionTimeRemaining(CameraTransitionTimeRemaining:double="Max (Float)"(A=float - float(A=@n2, B=@n1), B=0)) -> (Output_Get:double)
  ]}
 SelectCamera {nodes [
   n1 = SelectCamera!() -> (then, Selected Camera:"String (by ref)") {is=title K2Node_FunctionEntry}
   n2 = Find(TargetMap:Map of Strings to Pilgrim Battle Instance Camera Layout Base Class References=Get CameraPresets(), Key=@n1) -> (Value:Pilgrim Battle Instance Camera Layout Base Class Reference)
   > Branch(Condition=@n2.ReturnValue)
   > Set SelectedCamera(SelectedCamera:Pilgrim Battle Instance Camera Layout Base Class Reference=@n2) -> (Output_Get:Pilgrim Battle Instance Camera Layout Base Class Reference)
   > SendCameraUpdateEventMessage()
  ]}
 SendCameraUpdateEventMessage {nodes [
   SendCameraUpdateEventMessage() {is=title K2Node_FunctionEntry}
   n1 = GetOwningPlayer()
   > SendGameplayEventToActor!(
    execute
    Actor:Actor Object Reference=@n1
    EventTag:Gameplay Tag Structure=Get CameraChangedEventTag() -> (CameraChangedEventTag:Gameplay Tag Structure)
    Payload:Gameplay Event Data Structure=Make Gameplay Event Data!(
     EventTag:Gameplay Tag Structure=Get CameraChangedEventTag()
     Instigator:Actor Object Reference=@n1
     Target:Actor Object Reference
     OptionalObject:Object Reference=Self-Reference()
     OptionalObject2:Object Reference
     ContextHandle:Gameplay Effect Context Handle Structure=()
     InstigatorTags:Gameplay Tag Container Structure=(GameplayTags=)
     TargetTags:Gameplay Tag Container Structure=(GameplayTags=)
     EventMagnitude:float=0
     TargetData:Gameplay Ability Target Data Handle Structure=()
    ) -> (GameplayEventData:Gameplay Event Data Structure) {is=title K2Node_MakeStruct}
   ) -> (then)
  ]}
 SetSelectedPresetName {nodes [
   n1 = SetSelectedPresetName!() -> (then, PresetName:Name) {is=title K2Node_FunctionEntry}
   > Set IgnoreSelectedPresetChange(IgnoreSelectedPresetChange:bool=true) -> (Output_Get:bool)
   > SetSelectedOption!(
    execute
    self:"Combo Box (String) Object Reference"=Get ComboBox_InputBinding() -> (ComboBox_InputBinding:"Combo Box (String) Object Reference")
    Option:String="To String (Name)"(InName=@n1)
   ) -> (then)
   > Set IgnoreSelectedPresetChange(IgnoreSelectedPresetChange=false)
  ]}
 StartCameraTransition {nodes [
   n1 = StartCameraTransition!() -> (then, TransitionTime:double=~) {is=title K2Node_FunctionEntry}
   > Set CameraTransitionTimeRemaining(CameraTransitionTimeRemaining=@n1)
  ]}
]
variables [
 CameraChangedEventTag {type=GameplayTag, default {tagName=EventMessage.Pilgrim.Camera.Layout}}
 CameraPresets {
  type="map<Class>"
  default {
   Camera 1=/PilgrimCore/Cameras/Pilgrim_BattleInstance_Camera_Layout1.Pilgrim_BattleInstance_Camera_Layout1_C
   Camera 2=/PilgrimCore/Cameras/Pilgrim_BattleInstance_Camera_Layout2.Pilgrim_BattleInstance_Camera_Layout2_C
   Camera 3=/PilgrimCore/Cameras/Pilgrim_BattleInstance_Camera_Layout3.Pilgrim_BattleInstance_Camera_Layout3_C
   Camera 4=/PilgrimCore/Cameras/Pilgrim_BattleInstance_Camera_Layout4.Pilgrim_BattleInstance_Camera_Layout4_C
   Camera 5=/PilgrimCore/Cameras/Pilgrim_BattleInstance_Camera_Layout5.Pilgrim_BattleInstance_Camera_Layout5_C
   Camera 6=/PilgrimCore/Cameras/Pilgrim_BattleInstance_Camera_Layout6.Pilgrim_BattleInstance_Camera_Layout6_C
  }
 }
 CameraTransitionTimeRemaining {type=number}
 IgnoreSelectedPresetChange {type=boolean}
 LayoutCameraAbility {type=Class, default=/PilgrimCore/Player/Abilities/GA_Pilgrim_LayoutCamera.GA_Pilgrim_LayoutCamera_C}
 LayoutCameraMap {type="map<string>", default {Layout1=Camera 1, Layout2=Camera 2, Layout3=Camera 1}}
 PresetBindingTable {type=DataTable, default=null}
 SelectedCamera {type=Class, default=null}
]
widgets {
 info {inheritedWidgetCount=0, parentClass=/Script/PilgrimCoreUI.PilgrimGameWidgetBase, rootWidgetClass=&CanvasPanel, widgetCount=10}
 widgets [
  CanvasPanel_0 {widgetClassPath=&CanvasPanel, widget=CanvasPanel_0}
  CanvasPanel-ScoreboardAreaDontUse {
   widgetClassPath=&CanvasPanel
   parent=CanvasPanel_0
   widget=CanvasPanel-ScoreboardAreaDontUse
   slot {
    class=CanvasPanelSlot
    properties {bAutoSize=true, layoutData {alignment=<0 0>, anchors {maximum=<0.3 0.6>, minimum=<0.05 0.0759>}, offsets=<ltrb 0 0 0 0>}}
   }
  }
  VerticalBox-TrackInfoAreaDontUse {
   widgetClassPath=&VerticalBox
   parent=CanvasPanel_0
   widget=VerticalBox-TrackInfoAreaDontUse
   slot {
    class=CanvasPanelSlot
    properties {layoutData {alignment=<0 0>, anchors {maximum=<1 0>, minimum=<1 0>}, offsets=<ltrb -461.9219 89.7738 300.5363 674.9678>}}
   }
  }
  VerticalBox_Settings {
   widgetClassPath=&VerticalBox
   bIsVariable=true
   parent=CanvasPanel_0
   widget=VerticalBox_Settings
   properties {visibility=Collapsed}
   slot {
    class=CanvasPanelSlot
    properties {layoutData {alignment=<0 0>, anchors {maximum=<1 0>, minimum=<1 0>}, offsets=<ltrb -461.9219 790.1101 300.5363 259.108>}}
   }
  }
  HorizontalBox_0 {widgetClassPath=&HorizontalBox, parent=VerticalBox_Settings, widget=HorizontalBox_0, slot {class=VerticalBoxSlot}}
  ComboBox_InputBinding {
   widgetClassPath=/Script/UMG.ComboBoxString
   bIsVariable=true
   parent=HorizontalBox_0
   widget=ComboBox_InputBinding
   properties {
    foregroundColor {colorUseRule=UseColor_Specified, specifiedColor=<rgba 1 1 1 1>}
    itemStyle {
     bUseParentRowBrush=false
     activeBrush {
      bIsDynamicallyLoaded=false
      drawAs=Image
      imageType=FullColor
      mirroring=NoMirror
      resourceName=../../../Engine/Content/Slate/Common/Selection.png
      resourceObject=null
      tiling=NoTile
      imageSize=<8 8>
      margin=<ltrb 0 0 0 0>
      outlineSettings {
       bUseBrushTransparency=false
       roundingType=HalfHeightRadius
       width=0
       color {colorUseRule=UseColor_Specified, specifiedColor=<rgba 0 0 0 0>}
       cornerRadii=<0 0 0 0>
      }
      tintColor {colorUseRule=UseColor_Specified, specifiedColor=<rgba 0.0091 0.3515 0.9647 1>}
      uVRegion {bIsValid=false, max=<0 0>, min=<0 0>}
     }
     activeHighlightedBrush {
      bIsDynamicallyLoaded=false
      drawAs=Image
      imageType=FullColor
      mirroring=NoMirror
      resourceName=../../../Engine/Content/Slate/Common/Selection.png
      resourceObject=null
      tiling=NoTile
      imageSize=<8 8>
      margin=<ltrb 0 0 0 0>
      outlineSettings {
       bUseBrushTransparency=false
       roundingType=HalfHeightRadius
       width=0
       color {colorUseRule=UseColor_Specified, specifiedColor=<rgba 0 0 0 0>}
       cornerRadii=<0 0 0 0>
      }
      tintColor {colorUseRule=UseColor_Specified, specifiedColor=<rgba 0 0.1878 0.6795 1>}
      uVRegion {bIsValid=false, max=<0 0>, min=<0 0>}
     }
     activeHoveredBrush {
      bIsDynamicallyLoaded=false
      drawAs=Image
      imageType=FullColor
      mirroring=NoMirror
      resourceName=../../../Engine/Content/Slate/Common/Selection.png
      resourceObject=null
      tiling=NoTile
      imageSize=<8 8>
      margin=<ltrb 0 0 0 0>
      outlineSettings {
       bUseBrushTransparency=false
       roundingType=HalfHeightRadius
       width=0
       color {colorUseRule=UseColor_Specified, specifiedColor=<rgba 0 0 0 0>}
       cornerRadii=<0 0 0 0>
      }
      tintColor {colorUseRule=UseColor_Specified, specifiedColor=<rgba 0.0091 0.3515 0.9647 1>}
      uVRegion {bIsValid=false, max=<0 0>, min=<0 0>}
     }
     dropIndicator_Above {
      bIsDynamicallyLoaded=false
      drawAs=Box
      imageType=FullColor
      mirroring=NoMirror
      resourceName=../../../Engine/Content/Slate/Common/DropZoneIndicator_Above.png
      resourceObject=null
      tiling=NoTile
      imageSize=<0 0>
      margin=<ltrb 0.625 0.625 0 0>
      outlineSettings {
       bUseBrushTransparency=false
       roundingType=HalfHeightRadius
       width=0
       color {colorUseRule=UseColor_Specified, specifiedColor=<rgba 0 0 0 0>}
       cornerRadii=<0 0 0 0>
      }
      tintColor {colorUseRule=UseColor_Specified, specifiedColor=<rgba 0.0091 0.3515 0.9647 1>}
      uVRegion {bIsValid=false, max=<0 0>, min=<0 0>}
     }
     dropIndicator_Below {
      bIsDynamicallyLoaded=false
      drawAs=Box
      imageType=FullColor
      mirroring=NoMirror
      resourceName=../../../Engine/Content/Slate/Common/DropZoneIndicator_Below.png
      resourceObject=null
      tiling=NoTile
      imageSize=<0 0>
      margin=<ltrb 0.625 0 0 0.625>
      outlineSettings {
       bUseBrushTransparency=false
       roundingType=HalfHeightRadius
       width=0
       color {colorUseRule=UseColor_Specified, specifiedColor=<rgba 0 0 0 0>}
       cornerRadii=<0 0 0 0>
      }
      tintColor {colorUseRule=UseColor_Specified, specifiedColor=<rgba 0.0091 0.3515 0.9647 1>}
      uVRegion {bIsValid=false, max=<0 0>, min=<0 0>}
     }
     dropIndicator_Onto {
      bIsDynamicallyLoaded=false
      drawAs=Box
      imageType=FullColor
      mirroring=NoMirror
      resourceName=../../../Engine/Content/Slate/Common/DropZoneIndicator_Onto.png
      resourceObject=null
      tiling=NoTile
      imageSize=<0 0>
      margin=<ltrb 0.25 0.25 0.25 0.25>
      outlineSettings {
       bUseBrushTransparency=false
       roundingType=HalfHeightRadius
       width=0
       color {colorUseRule=UseColor_Specified, specifiedColor=<rgba 0 0 0 0>}
       cornerRadii=<0 0 0 0>
      }
      tintColor {colorUseRule=UseColor_Specified, specifiedColor=<rgba 0.0091 0.3515 0.9647 1>}
      uVRegion {bIsValid=false, max=<0 0>, min=<0 0>}
     }
     evenRowBackgroundBrush {
      bIsDynamicallyLoaded=false
      drawAs=NoDrawType
      imageType=NoImage
      mirroring=NoMirror
      resourceName=None
      resourceObject=null
      tiling=NoTile
      imageSize=<0 0>
      margin=<ltrb 0 0 0 0>
      outlineSettings {
       bUseBrushTransparency=false
       roundingType=HalfHeightRadius
       width=0
       color {colorUseRule=UseColor_Specified, specifiedColor=<rgba 0 0 0 0>}
       cornerRadii=<0 0 0 0>
      }
      tintColor {colorUseRule=UseColor_Specified, specifiedColor=<rgba 1 1 1 1>}
      uVRegion {bIsValid=false, max=<0 0>, min=<0 0>}
     }
     evenRowBackgroundHoveredBrush {
      bIsDynamicallyLoaded=false
      drawAs=Image
      imageType=FullColor
      mirroring=NoMirror
      resourceName=../../../Engine/Content/Slate/Common/Selection.png
      resourceObject=null
      tiling=NoTile
      imageSize=<8 8>
      margin=<ltrb 0 0 0 0>
      outlineSettings {
       bUseBrushTransparency=false
       roundingType=HalfHeightRadius
       width=0
       color {colorUseRule=UseColor_Specified, specifiedColor=<rgba 0 0 0 0>}
       cornerRadii=<0 0 0 0>
      }
      tintColor {colorUseRule=UseColor_Specified, specifiedColor=<rgba 1 1 1 0.1>}
      uVRegion {bIsValid=false, max=<0 0>, min=<0 0>}
     }
     inactiveBrush {
      bIsDynamicallyLoaded=false
      drawAs=Image
      imageType=FullColor
      mirroring=NoMirror
      resourceName=../../../Engine/Content/Slate/Common/Selection.png
      resourceObject=null
      tiling=NoTile
      imageSize=<8 8>
      margin=<ltrb 0 0 0 0>
      outlineSettings {
       bUseBrushTransparency=false
       roundingType=HalfHeightRadius
       width=0
       color {colorUseRule=UseColor_Specified, specifiedColor=<rgba 0 0 0 0>}
       cornerRadii=<0 0 0 0>
      }
      tintColor {colorUseRule=UseColor_Specified, specifiedColor=<rgba 0.25 0.25 0.25 1>}
      uVRegion {bIsValid=false, max=<0 0>, min=<0 0>}
     }
     inactiveHighlightedBrush {
      bIsDynamicallyLoaded=false
      drawAs=Image
      imageType=FullColor
      mirroring=NoMirror
      resourceName=../../../Engine/Content/Slate/Common/Selection.png
      resourceObject=null
      tiling=NoTile
      imageSize=<8 8>
      margin=<ltrb 0 0 0 0>
      outlineSettings {
       bUseBrushTransparency=false
       roundingType=HalfHeightRadius
       width=0
       color {colorUseRule=UseColor_Specified, specifiedColor=<rgba 0 0 0 0>}
       cornerRadii=<0 0 0 0>
      }
      tintColor {colorUseRule=UseColor_Specified, specifiedColor=<rgba 0.1 0.1 0.1 1>}
      uVRegion {bIsValid=false, max=<0 0>, min=<0 0>}
     }
     inactiveHoveredBrush {
      bIsDynamicallyLoaded=false
      drawAs=Image
      imageType=FullColor
      mirroring=NoMirror
      resourceName=../../../Engine/Content/Slate/Common/Selection.png
      resourceObject=null
      tiling=NoTile
      imageSize=<8 8>
      margin=<ltrb 0 0 0 0>
      outlineSettings {
       bUseBrushTransparency=false
       roundingType=HalfHeightRadius
       width=0
       color {colorUseRule=UseColor_Specified, specifiedColor=<rgba 0 0 0 0>}
       cornerRadii=<0 0 0 0>
      }
      tintColor {colorUseRule=UseColor_Specified, specifiedColor=<rgba 0.25 0.25 0.25 1>}
      uVRegion {bIsValid=false, max=<0 0>, min=<0 0>}
     }
     oddRowBackgroundBrush {
      bIsDynamicallyLoaded=false
      drawAs=NoDrawType
      imageType=NoImage
      mirroring=NoMirror
      resourceName=None
      resourceObject=null
      tiling=NoTile
      imageSize=<0 0>
      margin=<ltrb 0 0 0 0>
      outlineSettings {
       bUseBrushTransparency=false
       roundingType=HalfHeightRadius
       width=0
       color {colorUseRule=UseColor_Specified, specifiedColor=<rgba 0 0 0 0>}
       cornerRadii=<0 0 0 0>
      }
      tintColor {colorUseRule=UseColor_Specified, specifiedColor=<rgba 1 1 1 1>}
      uVRegion {bIsValid=false, max=<0 0>, min=<0 0>}
     }
     oddRowBackgroundHoveredBrush {
      bIsDynamicallyLoaded=false
      drawAs=Image
      imageType=FullColor
      mirroring=NoMirror
      resourceName=../../../Engine/Content/Slate/Common/Selection.png
      resourceObject=null
      tiling=NoTile
      imageSize=<8 8>
      margin=<ltrb 0 0 0 0>
      outlineSettings {
       bUseBrushTransparency=false
       roundingType=HalfHeightRadius
       width=0
       color {colorUseRule=UseColor_Specified, specifiedColor=<rgba 0 0 0 0>}
       cornerRadii=<0 0 0 0>
      }
      tintColor {colorUseRule=UseColor_Specified, specifiedColor=<rgba 1 1 1 0.1>}
      uVRegion {bIsValid=false, max=<0 0>, min=<0 0>}
     }
     parentRowBackgroundBrush {
      bIsDynamicallyLoaded=false
      drawAs=Image
      imageType=NoImage
      mirroring=NoMirror
      resourceName=None
      resourceObject=null
      tiling=NoTile
      imageSize=<32 32>
      margin=<ltrb 0 0 0 0>
      outlineSettings {
       bUseBrushTransparency=false
       roundingType=HalfHeightRadius
       width=0
       color {colorUseRule=UseColor_Specified, specifiedColor=<rgba 0 0 0 0>}
       cornerRadii=<0 0 0 0>
      }
      tintColor {colorUseRule=UseColor_Specified, specifiedColor=<rgba 1 1 1 1>}
      uVRegion {bIsValid=false, max=<0 0>, min=<0 0>}
     }
     parentRowBackgroundHoveredBrush {
      bIsDynamicallyLoaded=false
      drawAs=Image
      imageType=NoImage
      mirroring=NoMirror
      resourceName=None
      resourceObject=null
      tiling=NoTile
      imageSize=<32 32>
      margin=<ltrb 0 0 0 0>
      outlineSettings {
       bUseBrushTransparency=false
       roundingType=HalfHeightRadius
       width=0
       color {colorUseRule=UseColor_Specified, specifiedColor=<rgba 0 0 0 0>}
       cornerRadii=<0 0 0 0>
      }
      tintColor {colorUseRule=UseColor_Specified, specifiedColor=<rgba 1 1 1 1>}
      uVRegion {bIsValid=false, max=<0 0>, min=<0 0>}
     }
     selectedTextColor {colorUseRule=UseColor_Specified, specifiedColor=<rgba 0 0.2231 0.4729 1>}
     selectorFocusedBrush {
      bIsDynamicallyLoaded=false
      drawAs=Border
      imageType=FullColor
      mirroring=NoMirror
      resourceName=../../../Engine/Content/Slate/Common/Selector.png
      resourceObject=null
      tiling=Both
      imageSize=<0 0>
      margin=<ltrb 0.25 0.25 0.25 0.25>
      outlineSettings {
       bUseBrushTransparency=false
       roundingType=HalfHeightRadius
       width=0
       color {colorUseRule=UseColor_Specified, specifiedColor=<rgba 0 0 0 0>}
       cornerRadii=<0 0 0 0>
      }
      tintColor {colorUseRule=UseColor_Specified, specifiedColor=<rgba 0.701 0.225 0.003 1>}
      uVRegion {bIsValid=false, max=<0 0>, min=<0 0>}
     }
     textColor {colorUseRule=UseColor_Specified, specifiedColor=<rgba 0.7726 0.7726 0.7726 1>}
    }
   }
   slot {class=HorizontalBoxSlot}
  }
  CommonTextBlock_1 {
   widgetClassPath=&CommonTextBlock
   parent=HorizontalBox_0
   widget=CommonTextBlock_1
   properties {
    style=/FortUILibrary/Styles/Typography/Foundation/TextStyle-Header-S-outline.TextStyle-Header-S-outline_C
    text=Input Binding
    font {
     bForceMonospaced=false
     bMaterialIsStencil=false
     fontMaterial=null
     fontObject=&Burbank
     letterSpacing=0
     monospacedWidth=1
     size=24
     skewAmount=0
     typefaceFontName=Burbank-Black
     outlineSettings {
      bApplyOutlineToDropShadows=false
      bMiteredCorners=false
      bSeparateFillAlpha=false
      outlineMaterial=null
      outlineSize=0
      outlineColor=<rgba 0 0 0 1>
     }
    }
    shadowColorAndOpacity=<rgba 0 0 0 1>
    shadowOffset=<1.6 1.6>
   }
   slot {class=HorizontalBoxSlot, properties {horizontalAlignment=HAlign_Right, size {sizeRule=Fill, value=1}}}
  }
  HorizontalBox {widgetClassPath=&HorizontalBox, parent=VerticalBox_Settings, widget=HorizontalBox, slot {class=VerticalBoxSlot}}
  ComboBox_CameraPresets {
   widgetClassPath=&ComboBoxString
   bIsVariable=true
   parent=HorizontalBox
   widget=ComboBox_CameraPresets
   properties {
    foregroundColor {colorUseRule=UseColor_Specified, specifiedColor=<rgba 1 1 1 1>}
    itemStyle {
     bUseParentRowBrush=false
     activeBrush {
      bIsDynamicallyLoaded=false
      drawAs=Image
      imageType=FullColor
      mirroring=NoMirror
      resourceName=../../../Engine/Content/Slate/Common/Selection.png
      resourceObject=null
      tiling=NoTile
      imageSize=<8 8>
      margin=<ltrb 0 0 0 0>
      outlineSettings {
       bUseBrushTransparency=false
       roundingType=HalfHeightRadius
       width=0
       color {colorUseRule=UseColor_Specified, specifiedColor=<rgba 0 0 0 0>}
       cornerRadii=<0 0 0 0>
      }
      tintColor {colorUseRule=UseColor_Specified, specifiedColor=<rgba 0.0091 0.3515 0.9647 1>}
      uVRegion {bIsValid=false, max=<0 0>, min=<0 0>}
     }
     activeHighlightedBrush {
      bIsDynamicallyLoaded=false
      drawAs=Image
      imageType=FullColor
      mirroring=NoMirror
      resourceName=../../../Engine/Content/Slate/Common/Selection.png
      resourceObject=null
      tiling=NoTile
      imageSize=<8 8>
      margin=<ltrb 0 0 0 0>
      outlineSettings {
       bUseBrushTransparency=false
       roundingType=HalfHeightRadius
       width=0
       color {colorUseRule=UseColor_Specified, specifiedColor=<rgba 0 0 0 0>}
       cornerRadii=<0 0 0 0>
      }
      tintColor {colorUseRule=UseColor_Specified, specifiedColor=<rgba 0 0.1878 0.6795 1>}
      uVRegion {bIsValid=false, max=<0 0>, min=<0 0>}
     }
     activeHoveredBrush {
      bIsDynamicallyLoaded=false
      drawAs=Image
      imageType=FullColor
      mirroring=NoMirror
      resourceName=../../../Engine/Content/Slate/Common/Selection.png
      resourceObject=null
      tiling=NoTile
      imageSize=<8 8>
      margin=<ltrb 0 0 0 0>
      outlineSettings {
       bUseBrushTransparency=false
       roundingType=HalfHeightRadius
       width=0
       color {colorUseRule=UseColor_Specified, specifiedColor=<rgba 0 0 0 0>}
       cornerRadii=<0 0 0 0>
      }
      tintColor {colorUseRule=UseColor_Specified, specifiedColor=<rgba 0.0091 0.3515 0.9647 1>}
      uVRegion {bIsValid=false, max=<0 0>, min=<0 0>}
     }
     dropIndicator_Above {
      bIsDynamicallyLoaded=false
      drawAs=Box
      imageType=FullColor
      mirroring=NoMirror
      resourceName=../../../Engine/Content/Slate/Common/DropZoneIndicator_Above.png
      resourceObject=null
      tiling=NoTile
      imageSize=<0 0>
      margin=<ltrb 0.625 0.625 0 0>
      outlineSettings {
       bUseBrushTransparency=false
       roundingType=HalfHeightRadius
       width=0
       color {colorUseRule=UseColor_Specified, specifiedColor=<rgba 0 0 0 0>}
       cornerRadii=<0 0 0 0>
      }
      tintColor {colorUseRule=UseColor_Specified, specifiedColor=<rgba 0.0091 0.3515 0.9647 1>}
      uVRegion {bIsValid=false, max=<0 0>, min=<0 0>}
     }
     dropIndicator_Below {
      bIsDynamicallyLoaded=false
      drawAs=Box
      imageType=FullColor
      mirroring=NoMirror
      resourceName=../../../Engine/Content/Slate/Common/DropZoneIndicator_Below.png
      resourceObject=null
      tiling=NoTile
      imageSize=<0 0>
      margin=<ltrb 0.625 0 0 0.625>
      outlineSettings {
       bUseBrushTransparency=false
       roundingType=HalfHeightRadius
       width=0
       color {colorUseRule=UseColor_Specified, specifiedColor=<rgba 0 0 0 0>}
       cornerRadii=<0 0 0 0>
      }
      tintColor {colorUseRule=UseColor_Specified, specifiedColor=<rgba 0.0091 0.3515 0.9647 1>}
      uVRegion {bIsValid=false, max=<0 0>, min=<0 0>}
     }
     dropIndicator_Onto {
      bIsDynamicallyLoaded=false
      drawAs=Box
      imageType=FullColor
      mirroring=NoMirror
      resourceName=../../../Engine/Content/Slate/Common/DropZoneIndicator_Onto.png
      resourceObject=null
      tiling=NoTile
      imageSize=<0 0>
      margin=<ltrb 0.25 0.25 0.25 0.25>
      outlineSettings {
       bUseBrushTransparency=false
       roundingType=HalfHeightRadius
       width=0
       color {colorUseRule=UseColor_Specified, specifiedColor=<rgba 0 0 0 0>}
       cornerRadii=<0 0 0 0>
      }
      tintColor {colorUseRule=UseColor_Specified, specifiedColor=<rgba 0.0091 0.3515 0.9647 1>}
      uVRegion {bIsValid=false, max=<0 0>, min=<0 0>}
     }
     evenRowBackgroundBrush {
      bIsDynamicallyLoaded=false
      drawAs=NoDrawType
      imageType=NoImage
      mirroring=NoMirror
      resourceName=None
      resourceObject=null
      tiling=NoTile
      imageSize=<0 0>
      margin=<ltrb 0 0 0 0>
      outlineSettings {
       bUseBrushTransparency=false
       roundingType=HalfHeightRadius
       width=0
       color {colorUseRule=UseColor_Specified, specifiedColor=<rgba 0 0 0 0>}
       cornerRadii=<0 0 0 0>
      }
      tintColor {colorUseRule=UseColor_Specified, specifiedColor=<rgba 1 1 1 1>}
      uVRegion {bIsValid=false, max=<0 0>, min=<0 0>}
     }
     evenRowBackgroundHoveredBrush {
      bIsDynamicallyLoaded=false
      drawAs=Image
      imageType=FullColor
      mirroring=NoMirror
      resourceName=../../../Engine/Content/Slate/Common/Selection.png
      resourceObject=null
      tiling=NoTile
      imageSize=<8 8>
      margin=<ltrb 0 0 0 0>
      outlineSettings {
       bUseBrushTransparency=false
       roundingType=HalfHeightRadius
       width=0
       color {colorUseRule=UseColor_Specified, specifiedColor=<rgba 0 0 0 0>}
       cornerRadii=<0 0 0 0>
      }
      tintColor {colorUseRule=UseColor_Specified, specifiedColor=<rgba 1 1 1 0.1>}
      uVRegion {bIsValid=false, max=<0 0>, min=<0 0>}
     }
     inactiveBrush {
      bIsDynamicallyLoaded=false
      drawAs=Image
      imageType=FullColor
      mirroring=NoMirror
      resourceName=../../../Engine/Content/Slate/Common/Selection.png
      resourceObject=null
      tiling=NoTile
      imageSize=<8 8>
      margin=<ltrb 0 0 0 0>
      outlineSettings {
       bUseBrushTransparency=false
       roundingType=HalfHeightRadius
       width=0
       color {colorUseRule=UseColor_Specified, specifiedColor=<rgba 0 0 0 0>}
       cornerRadii=<0 0 0 0>
      }
      tintColor {colorUseRule=UseColor_Specified, specifiedColor=<rgba 0.25 0.25 0.25 1>}
      uVRegion {bIsValid=false, max=<0 0>, min=<0 0>}
     }
     inactiveHighlightedBrush {
      bIsDynamicallyLoaded=false
      drawAs=Image
      imageType=FullColor
      mirroring=NoMirror
      resourceName=../../../Engine/Content/Slate/Common/Selection.png
      resourceObject=null
      tiling=NoTile
      imageSize=<8 8>
      margin=<ltrb 0 0 0 0>
      outlineSettings {
       bUseBrushTransparency=false
       roundingType=HalfHeightRadius
       width=0
       color {colorUseRule=UseColor_Specified, specifiedColor=<rgba 0 0 0 0>}
       cornerRadii=<0 0 0 0>
      }
      tintColor {colorUseRule=UseColor_Specified, specifiedColor=<rgba 0.1 0.1 0.1 1>}
      uVRegion {bIsValid=false, max=<0 0>, min=<0 0>}
     }
     inactiveHoveredBrush {
      bIsDynamicallyLoaded=false
      drawAs=Image
      imageType=FullColor
      mirroring=NoMirror
      resourceName=../../../Engine/Content/Slate/Common/Selection.png
      resourceObject=null
      tiling=NoTile
      imageSize=<8 8>
      margin=<ltrb 0 0 0 0>
      outlineSettings {
       bUseBrushTransparency=false
       roundingType=HalfHeightRadius
       width=0
       color {colorUseRule=UseColor_Specified, specifiedColor=<rgba 0 0 0 0>}
       cornerRadii=<0 0 0 0>
      }
      tintColor {colorUseRule=UseColor_Specified, specifiedColor=<rgba 0.25 0.25 0.25 1>}
      uVRegion {bIsValid=false, max=<0 0>, min=<0 0>}
     }
     oddRowBackgroundBrush {
      bIsDynamicallyLoaded=false
      drawAs=NoDrawType
      imageType=NoImage
      mirroring=NoMirror
      resourceName=None
      resourceObject=null
      tiling=NoTile
      imageSize=<0 0>
      margin=<ltrb 0 0 0 0>
      outlineSettings {
       bUseBrushTransparency=false
       roundingType=HalfHeightRadius
       width=0
       color {colorUseRule=UseColor_Specified, specifiedColor=<rgba 0 0 0 0>}
       cornerRadii=<0 0 0 0>
      }
      tintColor {colorUseRule=UseColor_Specified, specifiedColor=<rgba 1 1 1 1>}
      uVRegion {bIsValid=false, max=<0 0>, min=<0 0>}
     }
     oddRowBackgroundHoveredBrush {
      bIsDynamicallyLoaded=false
      drawAs=Image
      imageType=FullColor
      mirroring=NoMirror
      resourceName=../../../Engine/Content/Slate/Common/Selection.png
      resourceObject=null
      tiling=NoTile
      imageSize=<8 8>
      margin=<ltrb 0 0 0 0>
      outlineSettings {
       bUseBrushTransparency=false
       roundingType=HalfHeightRadius
       width=0
       color {colorUseRule=UseColor_Specified, specifiedColor=<rgba 0 0 0 0>}
       cornerRadii=<0 0 0 0>
      }
      tintColor {colorUseRule=UseColor_Specified, specifiedColor=<rgba 1 1 1 0.1>}
      uVRegion {bIsValid=false, max=<0 0>, min=<0 0>}
     }
     parentRowBackgroundBrush {
      bIsDynamicallyLoaded=false
      drawAs=Image
      imageType=NoImage
      mirroring=NoMirror
      resourceName=None
      resourceObject=null
      tiling=NoTile
      imageSize=<32 32>
      margin=<ltrb 0 0 0 0>
      outlineSettings {
       bUseBrushTransparency=false
       roundingType=HalfHeightRadius
       width=0
       color {colorUseRule=UseColor_Specified, specifiedColor=<rgba 0 0 0 0>}
       cornerRadii=<0 0 0 0>
      }
      tintColor {colorUseRule=UseColor_Specified, specifiedColor=<rgba 1 1 1 1>}
      uVRegion {bIsValid=false, max=<0 0>, min=<0 0>}
     }
     parentRowBackgroundHoveredBrush {
      bIsDynamicallyLoaded=false
      drawAs=Image
      imageType=NoImage
      mirroring=NoMirror
      resourceName=None
      resourceObject=null
      tiling=NoTile
      imageSize=<32 32>
      margin=<ltrb 0 0 0 0>
      outlineSettings {
       bUseBrushTransparency=false
       roundingType=HalfHeightRadius
       width=0
       color {colorUseRule=UseColor_Specified, specifiedColor=<rgba 0 0 0 0>}
       cornerRadii=<0 0 0 0>
      }
      tintColor {colorUseRule=UseColor_Specified, specifiedColor=<rgba 1 1 1 1>}
      uVRegion {bIsValid=false, max=<0 0>, min=<0 0>}
     }
     selectedTextColor {colorUseRule=UseColor_Specified, specifiedColor=<rgba 0 0.2231 0.4729 1>}
     selectorFocusedBrush {
      bIsDynamicallyLoaded=false
      drawAs=Border
      imageType=FullColor
      mirroring=NoMirror
      resourceName=../../../Engine/Content/Slate/Common/Selector.png
      resourceObject=null
      tiling=Both
      imageSize=<0 0>
      margin=<ltrb 0.25 0.25 0.25 0.25>
      outlineSettings {
       bUseBrushTransparency=false
       roundingType=HalfHeightRadius
       width=0
       color {colorUseRule=UseColor_Specified, specifiedColor=<rgba 0 0 0 0>}
       cornerRadii=<0 0 0 0>
      }
      tintColor {colorUseRule=UseColor_Specified, specifiedColor=<rgba 0.701 0.225 0.003 1>}
      uVRegion {bIsValid=false, max=<0 0>, min=<0 0>}
     }
     textColor {colorUseRule=UseColor_Specified, specifiedColor=<rgba 0.7726 0.7726 0.7726 1>}
    }
   }
   slot {class=HorizontalBoxSlot}
  }
  CommonTextBlock {
   widgetClassPath=&CommonTextBlock
   parent=HorizontalBox
   widget=CommonTextBlock
   properties {
    style=&TextStyle-Header-S-outline_C
    text=Camera Presets
    font {
     bForceMonospaced=false
     bMaterialIsStencil=false
     fontMaterial=null
     fontObject=&Burbank
     letterSpacing=0
     monospacedWidth=1
     size=24
     skewAmount=0
     typefaceFontName=Burbank-Black
     outlineSettings {
      bApplyOutlineToDropShadows=false
      bMiteredCorners=false
      bSeparateFillAlpha=false
      outlineMaterial=null
      outlineSize=0
      outlineColor=<rgba 0 0 0 1>
     }
    }
    shadowColorAndOpacity=<rgba 0 0 0 1>
    shadowOffset=<1.6 1.6>
   }
   slot {class=HorizontalBoxSlot, properties {horizontalAlignment=HAlign_Right, size {sizeRule=Fill, value=1}}}
  }
 ]
}
