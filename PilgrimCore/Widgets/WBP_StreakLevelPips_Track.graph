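# CommonUserWidget that shows a player's streak level as a pip progress indicator in Pilgrim, the instrument-track gameplay mode.
# A single Image uses a streak meter material, and functions drive its dynamic material parameters for available pips, active pips, overdrive and precision border tint.
# On Construct it clears active pips and flags peripheral-instrument styling when the owning player's track, read from PilgrimCorePlayerStateComponent, is plastic guitar or bass.
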WidgetBlueprint /PilgrimCore/Widgets/WBP_StreakLevelPips_Track
graphs [
 EventGraph {nodes [
   Event Construct()
   > UpdateNumActivePips!(execute, self, NumActivePips:int=0) -> (then)
   n1 = GetOwningPlayer()
   n2 = GetTrackTypeForGameplay!(self:Pilgrim Core Player State Component Object Reference=GetComponentByClass!(
     self:Actor Object Reference=Get PlayerState!(self:Controller Object Reference=@n1) -> (PlayerState:Player State Object Reference)
     ComponentClass:Actor Component Class Reference=/Script/PilgrimCoreRuntime.PilgrimCorePlayerStateComponent
    ) -> (ReturnValue:Pilgrim Core Player State Component Object Reference)) -> (ReturnValue:EPilgrimTrackType Enum=TrackGuitar)
   > UpdateStreakLevelPipsPeripheralBorderColour!(
    execute
    self
    IsPeripheral:bool=OR Boolean(A="Equal (Enum)"(A:EPilgrimTrackType Enum=@n2, B:EPilgrimTrackType Enum=TrackPlasticGuitar), B="Equal (Enum)"(A=@n2, B=TrackPlasticBass))
   ) -> (then)
   Event PreConstruct()
  ]}
 Set Number Of Available Pips {nodes [
   n1 = GetDynamicMaterial()
   n2 = Set Number Of Available Pips!() -> (then, NumberOfPipsToShow:int) {is=title K2Node_FunctionEntry}
   > n3 = Set Max Pips(Max Pips:int=@n2) -> (Output_Get:int=0)
   n4 = "To Float (Integer)"(InInt=@n3)
   SetScalarParameterValue(execute=@n1.then, self=@n1, ParameterName=NumberofPips, Value=@n4)
   n5 = GetDynamicMaterial(execute=@n3.then, self=Get ProgressBar() -> (ProgressBar:Image Object Reference))
   > SetScalarParameterValue(self=@n5, ParameterName=NumberofPips, Value=@n4)
  ]}
 SetOverdriveColors {nodes [
   n1 = GetDynamicMaterial(self=Get ProgressBar())
   n2 = GetOwningPlayer()
   n3 = GetTrackTypeForGameplay(self=GetComponentByClass(self=Get PlayerState(self=@n2), ComponentClass=&PilgrimCorePlayerStateComponent)) -> (ReturnValue=TrackGuitar)
   n4 = "Equal (Enum)"(A=@n3, B=TrackPlasticBass)
   n5 = "Equal (Enum)"(A=@n3, B=TrackPlasticGuitar)
   > SetScalarParameterValue(self=@n1, ParameterName=IsPeripheralInstrument, Value="To Float (Boolean)"(InBool=OR Boolean(A=@n5, B=@n4)))
   n6 = SetOverdriveColors!() -> (then, IsOverdrive:bool=~) {is=title K2Node_FunctionEntry}
   > n7 = GetDynamicMaterial(self=Get ProgressBar())
   > SetScalarParameterValue(self=@n7, ParameterName=Overdrive, Value=Select(Option 0:float=0, Option 1:float=1, Index=@n6) -> (ReturnValue:float))
  ]}
 SetPrecisionColor {nodes [
   n1 = SetPrecisionColor!() -> (then, Value:Linear Color Structure) {is=title K2Node_FunctionEntry}
   > n2 = GetDynamicMaterial(self=Get ProgressBar())
   > SetVectorParameterValue(self=@n2, ParameterName=Border_Tint, Value=@n1)
  ]}
 UpdateNumActivePips {nodes [
   n1 = GetDynamicMaterial()
   n2 = UpdateNumActivePips!() -> (then, NumActivePips:int) {is=title K2Node_FunctionEntry}
   n3 = "To Float (Integer)"(InInt=@n2)
   SetScalarParameterValue(execute=@n1.then, self=@n1, ParameterName=NumberofPips, Value=@n3)
   n4 = GetDynamicMaterial(execute=@n2.then, self=Get ProgressBar())
   > SetScalarParameterValue(self=@n4, ParameterName=NumberofActivePips, Value=@n3)
  ]}
 UpdateStreakLevelPipsPeripheralBorderColour {nodes [
   n1 = UpdateStreakLevelPipsPeripheralBorderColour!() -> (then, IsPeripheral:bool=~) {is=title K2Node_FunctionEntry}
   > n2 = GetDynamicMaterial(self=Get ProgressBar())
   > SetScalarParameterValue(self=@n2, ParameterName=IsPeripheralInstrument, Value="To Float (Boolean)"(InBool=@n1))
  ]}
]
variables [Max Pips {type=integer}]
widgets {
 info {inheritedWidgetCount=0, parentClass=&CommonUserWidget, rootWidgetClass=&Overlay, widgetCount=3}
 widgets [
  Overlay_1 {widgetClassPath=&Overlay, widget=Overlay_1}
  SizeControl {widgetClassPath=&Spacer, parent=Overlay_1, widget=SizeControl, properties {size=<140 140>}, slot {class=OverlaySlot}}
  ProgressBar {
   widgetClassPath=&Image
   bIsVariable=true
   parent=Overlay_1
   widget=ProgressBar
   properties {brush {
     bIsDynamicallyLoaded=false
     drawAs=Image
     imageType=NoImage
     mirroring=NoMirror
     resourceName=None
     resourceObject=/PilgrimCore/Widgets/Materials/MI_UI_StreakMeter_ProgressIndicator
     tiling=NoTile
     imageSize=<32 32>
     margin=<ltrb 0 0 0 0>
     outlineSettings {
      bUseBrushTransparency=false
      roundingType=HalfHeightRadius
      width=0
      color {colorUseRule=UseColor_Specified, specifiedColor=<rgba 0 0 0 0>}
      cornerRadii=<0 0 0 0>
     }
     tintColor {colorUseRule=UseColor_Specified, specifiedColor=<rgba 1 1 1 1>}
     uVRegion {bIsValid=false, max=<0 0>, min=<0 0>}
    }}
   slot {class=OverlaySlot, properties {horizontalAlignment=HAlign_Fill, verticalAlignment=VAlign_Fill}}
  }
 ]
}
