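# Streak meter widget for the Pilgrim music game, showing the multiplier text, a material-driven progress ring with combo indicator, level pips and overdrive FX.
# Gameplay event router listeners set peripheral-instrument colors when a song plays and fade the widget with the track's secondary-elements phase.
# Functions drive material scalars and text colors for normal, max-streak and overdrive modes.
# Changing the multiplier plays a pop-and-flare animation.
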WidgetBlueprint /PilgrimCore/Widgets/WBP_StreakMeter
animations [A_StreakChange {
  binding_count=4
  length_seconds=0.4516
  object=&self:A_StreakChange
  sequence {
   bindings [
    Flare {
     id=#1
     possessed_class=Image
     tracks [
      {
       type=MovieSceneFloatTrack
       display_name=Render Opacity
       sections [{type=MovieSceneFloatSection, channels [None {type=MovieSceneScriptingFloatChannel, keys [[3,0],[3.983,1],[27.098,0]]}]}]
      }
      {
       type=MovieScene2DTransformTrack
       sections [{type=MovieScene2DTransformSection, channels [Scale.X {type=MovieSceneScriptingFloatChannel, keys [[2.902,1],[26.508,2]]}]}]
      }
     ]
    }
    StreakMultiplierText_Reflection {
     id=#2
     possessed_class=CommonTextBlock
     tracks [{
       type=MovieSceneFloatTrack
       display_name=Render Opacity
       sections [{type=MovieSceneFloatSection, channels [None {type=MovieSceneScriptingFloatChannel, keys [[3,0],[15,0.125]]}]}]
      }]
    }
    TextGlow {
     id=#3
     possessed_class=Image
     tracks [{
       type=MovieSceneFloatTrack
       display_name=Render Opacity
       sections [{type=MovieSceneFloatSection, end_frame=12, start_frame=0, channels [None {type=MovieSceneScriptingFloatChannel, keys [[0,0],[9,1]]}]}]
      }]
    }
    StreakMultiplierText {
     id=#4
     possessed_class=CommonTextBlock
     tracks [
      {
       type=MovieScene2DTransformTrack
       sections [{
         type=MovieScene2DTransformSection
         channels [
          Translation.X {type=MovieSceneScriptingFloatChannel}
          Translation.Y {type=MovieSceneScriptingFloatChannel, keys [[0,24],[6,0]]}
          Angle {type=MovieSceneScriptingFloatChannel}
          Scale.X {type=MovieSceneScriptingFloatChannel, keys [[0,4],[6,0.85],[15,1]]}
          Scale.Y {type=MovieSceneScriptingFloatChannel, keys [[0,2],[6,0.85],[15,1]]}
          Shear.X {type=MovieSceneScriptingFloatChannel}
          Shear.Y {type=MovieSceneScriptingFloatChannel}
         ]
        }]
      }
      {
       type=MovieSceneFloatTrack
       display_name=Render Opacity
       sections [{type=MovieSceneFloatSection, channels [None {type=MovieSceneScriptingFloatChannel, keys [[0,0],[6,1]]}]}]
      }
     ]
    }
   ]
   display_rate {numerator=60}
  }
 }]
graphs [
 EventGraph {nodes [
   Event on Set Owner() {is=title K2Node_CustomEvent}
   > n1 = GetEventRouter!(execute, Scope:Object Class Reference=/Script/FortniteGame.FortPlayspace, ContextActor:Actor Object Reference=GetOwningPlayer()) -> (then, ReturnValue:Gameplay Event Router Component Object Reference)
   > n2 = Sequence()
   n3 = GetOwningPlayer()
   > n4 = StartListeningToEvent!(
    execute:Exec
    Target:Gameplay Event Router Component Object Reference=@n1
    EventType:Script Struct Object Reference=/Script/PilgrimCoreRuntime.PilgrimEvent_PilgrimGame_PlayingSong
    Context:Object Reference=@n3
    EventBubblingRule:EEventBubblingRule Enum
   ) -> (
    then:Exec
    AsyncTaskProxy:Async Action Start Listening To Event Object Reference
    OnEventReceived:Exec
    Payload_PilgrimGame:Pilgrim Game Object Reference
   )
   > Add(
    TargetArray:Array of Async Action Start Listening To Event Object References=Get GameplayEventHandles() -> (GameplayEventHandles:Array of Async Action Start Listening To Event Object References)
    NewItem:"Async Action Start Listening To Event Object Reference (by ref)"=@n4
   ) {is=title K2Node_CallArrayFunction}
   UpdateStreakUIColors!(
    execute=@n4.OnEventReceived
    self
    IsPeripheral:bool=IsTrackUsingProColors!(self:Pilgrim Game Object Reference=@n4.Payload_PilgrimGame) -> (ReturnValue:bool)
   ) -> (then)
   n5 = StartListeningToEvent!(
    execute=@n2.then_1
    Target=@n1
    EventType=/Script/PilgrimCoreRuntime.PilgrimEvent_PilgrimTrack_SecondaryElementsPhaseChanged
    Context=@n3
    EventBubblingRule
   ) -> (then, AsyncTaskProxy, OnEventReceived, Payload_Track:Pilgrim Track Object Reference, Payload_SecondaryElementsPhase:float=~)
   > Add(TargetArray=Get GameplayEventHandles(), NewItem=@n5)
   SetRenderOpacity(execute=@n5.OnEventReceived, InOpacity=@n5.Payload_SecondaryElementsPhase)
   Event Construct()
   > SetRenderOpacity(InOpacity=0)
   Event OnInitialized()
   > Parent: OnInitialized() {is=title K2Node_CallParentFunction}
   > n6 = GetDynamicMaterial(self=Get StreakProgressMeter() -> (StreakProgressMeter:Image Object Reference))
   > Set StreakProgressMaterialInstance(StreakProgressMaterialInstance:Material Instance Dynamic Object Reference=@n6) -> (Output_Get:Material Instance Dynamic Object Reference)
   > n7 = GetDynamicMaterial(self=Get StreakProgress_Glow() -> (StreakProgress_Glow:Image Object Reference))
   > Set StreakGlow material Instance(StreakGlow material Instance:Material Instance Dynamic Object Reference=@n7) -> (Output_Get:Material Instance Dynamic Object Reference)
   > n8 = GetDynamicMaterial(self=Get StreakProgress_ComboIndicator() -> (StreakProgress_ComboIndicator:Image Object Reference))
   > Set Streak Combo Indicator Instance(Streak Combo Indicator Instance:Material Instance Dynamic Object Reference=@n8) -> (Output_Get:Material Instance Dynamic Object Reference)
   > SetScalarParameterValue(
    self=Get StreakProgressMaterialInstance() -> (StreakProgressMaterialInstance:Material Instance Dynamic Object Reference)
    ParameterName=Progress_Start
    Value=0
   )
   > SetScalarParameterValue(self=Get StreakProgressMaterialInstance(), ParameterName=Anim_ProgressLerp, Value=0)
  ]}
 GetAlphaFinalMaterialValue {nodes [
   n1 = GetAlphaFinalMaterialValue!() -> (then, Notes Past Last Multiplier:int, Notes To Next Multiplier:int, Is At Max Streak:bool=~) {is=title K2Node_FunctionEntry}
   > n2 = Branch(Condition=@n1.Is At Max Streak)
   > Return Node!(execute, TargetOvedrive:double=0.5)
   n3 = Branch(execute=@n2.else, Condition=AND Boolean(A="integer > integer"(A=@n1), B="integer > integer"(A=@n1.Notes To Next Multiplier)))
   > Return Node(TargetOvedrive=0.25)
   Return Node(execute=@n3.else, TargetOvedrive=0)
  ]}
 GetMultiplierMode {nodes [
   n1 = GetMultiplierMode!() -> (then, IsAtMaxStreak:bool=~, IsOverdriveActive:bool=~) {is=title K2Node_FunctionEntry}
   > n2 = Branch(Condition=@n1.IsOverdriveActive)
   > Return Node!(execute, MultiplierMode:E_Pilgrim_MultiplierMode Enum=NewEnumerator2)
   n3 = Branch(execute=@n2.else, Condition=@n1)
   > Return Node(MultiplierMode=NewEnumerator1)
   Return Node(execute=@n3.else, MultiplierMode=NewEnumerator0)
  ]}
 Set Multiplier Text {nodes [
   n1 = Set Multiplier Text!() -> (then, value:int) {is=title K2Node_FunctionEntry}
   > PlayAnimationForward(InAnimation=Get A_StreakChange() -> (A_StreakChange:Widget Animation Object Reference))
   n2 = Format Text!(Format="LOCTEXT(\"{value}x\")", value:int=@n1) -> (Result)
   > Set Text(Text=@n2, self=Get StreakMultiplierText() -> (StreakMultiplierText:Common Text Object Reference))
   > Set Text(Text=@n2, self=Get StreakMultiplierText_Reflection() -> (StreakMultiplierText_Reflection:Common Text Object Reference))
  ]}
 SetOverDriveColors {nodes [
   n1 = SetOverDriveColors!() -> (then, IsOverDrive:bool=~) {is=title K2Node_FunctionEntry}
   > n2 = GetDynamicMaterial(self=Get StreakProgressMeter())
   n3 = Select(Option 0:float=0, Option 1:float=1, Index=@n1) -> (ReturnValue:float)
   > SetScalarParameterValue(self=@n2, ParameterName=Overdrive, Value=@n3)
   > n4 = GetDynamicMaterial(self=Get StreakProgress_Glow())
   > SetScalarParameterValue(self=@n4, ParameterName=Overdrive, Value=@n3)
   > SetOverdriveColors!(
    execute
    self:WBP Streak Level Pips Track Object Reference=Get WBP_StreakLevelPips_Track_C_0() -> (WBP_StreakLevelPips_Track_C_0:WBP Streak Level Pips Track Object Reference)
    IsOverdrive:bool=@n1
   ) -> (then)
  ]}
 UpdateComboMaterials {nodes [
   n1 = UpdateComboMaterials!() -> (then, IsComboActive:bool=~) {is=title K2Node_FunctionEntry}
   > n2 = Branch(Condition=@n1)
   n3 = Get Streak Combo Indicator Instance() -> (Streak Combo Indicator Instance:Material Instance Dynamic Object Reference)
   > SetScalarParameterValue(self=@n3, ParameterName=IsCombo, Value=1)
   > SetScalarParameterValue(self=@n3, ParameterName=PerfectStreakActive, Value=1)
   n4 = Get Streak Combo Indicator Instance()
   SetScalarParameterValue(execute=@n2.else, self=@n4, ParameterName=IsCombo, Value=0)
   > SetScalarParameterValue(self=@n4, ParameterName=PerfectStreakActive, Value=0)
  ]}
 UpdateStreakMultiplier {nodes [
   n1 = UpdateStreakMultiplier!() -> (then, MultiplierMode:E_Pilgrim_MultiplierMode Enum=NewEnumerator0) {is=title K2Node_FunctionEntry}
   > n2 = Switch on E_Pilgrim_MultiplierMode!(execute, Selection:E_Pilgrim_MultiplierMode Enum=@n1) -> (NewEnumerator0:Exec, NewEnumerator1:Exec, NewEnumerator2:Exec) {is=title K2Node_SwitchEnum}
   > SetColorAndOpacity!(
    execute
    self=[Get StreakMultiplierText_Reflection(), Get StreakMultiplierText()]
    InColorAndOpacity_SpecifiedColor:Linear Color Structure=(R=0,G=0.8333,B=1,A=1)
    InColorAndOpacity_ColorUseRule:ESlateColorStylingMode Enum
   ) -> (then)
   > n3 = GetDynamicMaterial(self=Get TextGlow() -> (TextGlow:Image Object Reference))
   > SetScalarParameterValue(self=@n3, ParameterName=Alpha, Value=0)
   > n4 = GetDynamicMaterial(self=Get TextGlow())
   > SetScalarParameterValue(self=@n4, ParameterName=IsMax, Value=0)
   > n5 = GetDynamicMaterial(self=Get StreakProgress_ComboIndicator())
   > SetScalarParameterValue(self=@n5, ParameterName=IsMax, Value=0)
   > SetVisibility(self=Get OverDriveFX() -> (OverDriveFX:Overlay Object Reference), InVisibility=Collapsed)
   SetColorAndOpacity(
    execute=@n2.NewEnumerator1
    self=[Get StreakMultiplierText(), Get StreakMultiplierText_Reflection()]
    InColorAndOpacity_SpecifiedColor=(R=0.334,G=1,B=1,A=1)
   )
   > n6 = GetDynamicMaterial(self=Get TextGlow())
   > SetScalarParameterValue(self=@n6, ParameterName=Alpha, Value=1)
   > n7 = GetDynamicMaterial(self=Get TextGlow())
   > SetScalarParameterValue(self=@n7, ParameterName=IsMax, Value=0)
   > n8 = GetDynamicMaterial(self=Get StreakProgress_ComboIndicator())
   > SetScalarParameterValue(self=@n8, ParameterName=IsMax, Value=1)
   > SetVisibility(self=Get OverDriveFX(), InVisibility=Collapsed)
   SetColorAndOpacity(
    execute=@n2.NewEnumerator2
    self=[Get StreakMultiplierText(), Get StreakMultiplierText_Reflection()]
    InColorAndOpacity_SpecifiedColor=(R=1,G=1,B=1,A=1)
   )
   > n9 = GetDynamicMaterial(self=Get TextGlow())
   > SetScalarParameterValue(self=@n9, ParameterName=Alpha, Value=1)
   > n10 = GetDynamicMaterial(self=Get TextGlow())
   > SetScalarParameterValue(self=@n10, ParameterName=IsMax, Value=1)
   > n11 = GetDynamicMaterial(self=Get TextGlow())
   > SetScalarParameterValue(self=@n11, ParameterName=IsMax, Value=1)
   > SetVisibility(self=Get OverDriveFX(), InVisibility=SelfHitTestInvisible)
  ]}
 UpdateStreakUIColors {nodes [
   n1 = UpdateStreakUIColors!() -> (then, IsPeripheral:bool=~) {is=title K2Node_FunctionEntry}
   > UpdateStreakLevelPipsPeripheralBorderColour!(execute, self:WBP Streak Level Pips Track Object Reference=Get WBP_StreakLevelPips_Track_C_0(), IsPeripheral:bool=@n1) -> (then)
   > SetScalarParameterValue(self=Get StreakProgressMaterialInstance(), ParameterName=IsPeripheralInstrument, Value="To Float (Boolean)"(InBool=@n1))
   > n2 = GetDynamicMaterial(self=Get StreakProgressOrnament() -> (StreakProgressOrnament:Image Object Reference))
   > SetScalarParameterValue(self=@n2, ParameterName=IsPeripheralInstrument, Value="To Float (Boolean)"(InBool=@n1))
  ]}
]
variables [
 AlphaFinalMaterialValue {type=number}
 Streak Combo Indicator Instance {type=MaterialInstanceDynamic}
 StreakGlow material Instance {type=MaterialInstanceDynamic}
 StreakProgressMaterialInstance {type=MaterialInstanceDynamic}
]
widgets {
 info {inheritedWidgetCount=0, parentClass=/Script/PilgrimCoreUI.PilgrimGameWidgetBase, rootWidgetClass=&Overlay, widgetCount=23}
 widgets [
  All {widgetClassPath=&Overlay, widget=All}
  OverallSizeControl {widgetClassPath=&Spacer, parent=All, widget=OverallSizeControl, properties {size=<640 175>}, slot {class=OverlaySlot}}
  GridPanel_0 {
   widgetClassPath=&GridPanel
   parent=All
   widget=GridPanel_0
   slot {class=OverlaySlot, properties {horizontalAlignment=HAlign_Center, verticalAlignment=VAlign_Center, padding=<ltrb 0 0 0 4>}}
  }
  Text_Overlay {
   widgetClassPath=&Overlay
   parent=GridPanel_0
   widget=Text_Overlay
   slot {class=GridSlot, properties {horizontalAlignment=HAlign_Center, verticalAlignment=VAlign_Center}}
  }
  OverDriveFX {
   widgetClassPath=&Overlay
   bIsVariable=true
   parent=Text_Overlay
   widget=OverDriveFX
   properties {visibility=Collapsed}
   slot {class=OverlaySlot, properties {horizontalAlignment=HAlign_Fill, verticalAlignment=VAlign_Center, padding=<ltrb 0 4 0 0>}}
  }
  OverDriveCircles {
   widgetClassPath=&Image
   bIsVariable=true
   parent=OverDriveFX
   widget=OverDriveCircles
   properties {
    visibility=SelfHitTestInvisible
    brush {
     bIsDynamicallyLoaded=false
     drawAs=Image
     imageType=NoImage
     mirroring=NoMirror
     resourceName=None
     resourceObject=/PilgrimCore/Widgets/Materials/M_UI_StreakMeter_FX_Rings
     tiling=NoTile
     imageSize=<128 128>
     margin=<ltrb 0 0 0 0>
     outlineSettings {
      bUseBrushTransparency=false
      roundingType=HalfHeightRadius
      width=0
      color {colorUseRule=UseColor_Specified, specifiedColor=<rgba 0 0 0 0>}
      cornerRadii=<0 0 0 0>
     }
     tintColor {colorUseRule=UseColor_Specified, specifiedColor=<rgba 1 1 1 1>}
     uVRegion {bIsValid=false, max=<0 0>, min=<0 0>}
    }
   }
   slot {class=OverlaySlot, properties {horizontalAlignment=HAlign_Center, verticalAlignment=VAlign_Fill}}
  }
  OverDrive_Glow {
   widgetClassPath=&Image
   bIsVariable=true
   parent=OverDriveFX
   widget=OverDrive_Glow
   properties {
    visibility=SelfHitTestInvisible
    brush {
     bIsDynamicallyLoaded=false
     drawAs=Image
     imageType=NoImage
     mirroring=NoMirror
     resourceName=None
     resourceObject=/PilgrimCore/Widgets/Materials/M_UI_Pilgrim_OverDrive_Glow
     tiling=NoTile
     imageSize=<104 104>
     margin=<ltrb 0 0 0 0>
     outlineSettings {
      bUseBrushTransparency=false
      roundingType=HalfHeightRadius
      width=0
      color {colorUseRule=UseColor_Specified, specifiedColor=<rgba 0 0 0 0>}
      cornerRadii=<0 0 0 0>
     }
     tintColor {colorUseRule=UseColor_Specified, specifiedColor=<rgba 1 1 1 1>}
     uVRegion {bIsValid=false, max=<0 0>, min=<0 0>}
    }
   }
   slot {class=OverlaySlot, properties {horizontalAlignment=HAlign_Center, verticalAlignment=VAlign_Center}}
  }
  OverDriveFlare {
   widgetClassPath=&Image
   bIsVariable=true
   parent=OverDriveFX
   widget=OverDriveFlare
   properties {
    visibility=SelfHitTestInvisible
    brush {
     bIsDynamicallyLoaded=false
     drawAs=Image
     imageType=NoImage
     mirroring=NoMirror
     resourceName=None
     resourceObject=/PilgrimCore/Widgets/Materials/M_UI_Pilgrim_Flare_NoNoise
     tiling=NoTile
     imageSize=<32 64>
     margin=<ltrb 0 0 0 0>
     outlineSettings {
      bUseBrushTransparency=false
      roundingType=HalfHeightRadius
      width=0
      color {colorUseRule=UseColor_Specified, specifiedColor=<rgba 0 0 0 0>}
      cornerRadii=<0 0 0 0>
     }
     tintColor {colorUseRule=UseColor_Specified, specifiedColor=<rgba 1 1 1 1>}
     uVRegion {bIsValid=false, max=<0 0>, min=<0 0>}
    }
   }
   slot {class=OverlaySlot, properties {horizontalAlignment=HAlign_Fill, verticalAlignment=VAlign_Center, padding=<ltrb 0 4 0 0>}}
  }
  Particles {
   widgetClassPath=&Image
   parent=OverDriveFX
   widget=Particles
   properties {
    visibility=SelfHitTestInvisible
    brush {
     bIsDynamicallyLoaded=false
     drawAs=Image
     imageType=NoImage
     mirroring=NoMirror
     resourceName=None
     resourceObject=/PilgrimCore/Widgets/Materials/M_UI_StreakMeter_FX_Particles
     tiling=NoTile
     imageSize=<128 128>
     margin=<ltrb 0 0 0 0>
     outlineSettings {
      bUseBrushTransparency=false
      roundingType=HalfHeightRadius
      width=0
      color {colorUseRule=UseColor_Specified, specifiedColor=<rgba 0 0 0 0>}
      cornerRadii=<0 0 0 0>
     }
     tintColor {colorUseRule=UseColor_Specified, specifiedColor=<rgba 1 1 1 1>}
     uVRegion {bIsValid=false, max=<0 0>, min=<0 0>}
    }
    renderTransform {angle=0, scale=<2 2>, shear=<0 0>, translation=<0 0>}
   }
   slot {class=OverlaySlot, properties {horizontalAlignment=HAlign_Center, verticalAlignment=VAlign_Center}}
  }
  Flare {
   widgetClassPath=&Image
   parent=Text_Overlay
   widget=Flare
   properties {
    renderOpacity=0
    visibility=SelfHitTestInvisible
    brush {
     bIsDynamicallyLoaded=false
     drawAs=Image
     imageType=NoImage
     mirroring=NoMirror
     resourceName=None
     resourceObject=/PilgrimCore/Widgets/Materials/M_UI_Pilgrim_Flare
     tiling=NoTile
     imageSize=<512 16>
     margin=<ltrb 0 0 0 0>
     outlineSettings {
      bUseBrushTransparency=false
      roundingType=HalfHeightRadius
      width=0
      color {colorUseRule=UseColor_Specified, specifiedColor=<rgba 0 0 0 0>}
      cornerRadii=<0 0 0 0>
     }
     tintColor {colorUseRule=UseColor_Specified, specifiedColor=<rgba 1 1 1 1>}
     uVRegion {bIsValid=false, max=<0 0>, min=<0 0>}
    }
    renderTransform {angle=0, scale=<1 1>, shear=<0 0>, translation=<0 2>}
   }
   slot {class=OverlaySlot, properties {horizontalAlignment=HAlign_Fill, verticalAlignment=VAlign_Center}}
  }
  StreakMultiplierText_Reflection {
   widgetClassPath=&CommonTextBlock
   bIsVariable=true
   parent=Text_Overlay
   widget=StreakMultiplierText_Reflection
   properties {
    bIsScrollingEnabled=false
    justification=Center
    minDesiredWidth=0.5
    renderOpacity=0.125
    text=2x
    font {
     bForceMonospaced=false
     bMaterialIsStencil=false
     fontMaterial=null
     fontObject=&HeadingNow
     letterSpacing=0
     monospacedWidth=1
     size=25.5
     skewAmount=0
     typefaceFontName=86 Bold Italic
     outlineSettings {
      bApplyOutlineToDropShadows=false
      bMiteredCorners=false
      bSeparateFillAlpha=false
      outlineMaterial=&M-FontShadow
      outlineSize=0
      outlineColor=<rgba 1 1 1 0>
     }
    }
    renderTransform {angle=0, scale=<1 1>, shear=<0 0>, translation=<0 8>}
    shadowOffset=<0 0>
    strikeBrush {
     bIsDynamicallyLoaded=false
     drawAs=NoDrawType
     imageType=NoImage
     mirroring=NoMirror
     resourceName=None
     resourceObject=null
     tiling=NoTile
     imageSize=<32 32>
     margin=<ltrb 0 0 0 0>
     outlineSettings {
      bUseBrushTransparency=false
      roundingType=HalfHeightRadius
      width=0
      color {colorUseRule=UseColor_Specified, specifiedColor=<rgba 0 0 0 0>}
      cornerRadii=<0 0 0 0>
     }
     tintColor {colorUseRule=UseColor_Specified, specifiedColor=<rgba 1 1 1 1>}
     uVRegion {bIsValid=false, max=<0 0>, min=<0 0>}
    }
   }
   slot {class=OverlaySlot, properties {horizontalAlignment=HAlign_Center, verticalAlignment=VAlign_Center}}
  }
  Text {
   widgetClassPath=&Overlay
   parent=Text_Overlay
   widget=Text
   slot {class=OverlaySlot, properties {horizontalAlignment=HAlign_Center, verticalAlignment=VAlign_Center}}
  }
  TextGlow {
   widgetClassPath=&Image
   bIsVariable=true
   parent=Text
   widget=TextGlow
   properties {
    bIsVolatile=true
    brush {
     bIsDynamicallyLoaded=false
     drawAs=Box
     imageType=NoImage
     mirroring=NoMirror
     resourceName=None
     resourceObject=/PilgrimCore/Widgets/Materials/MI_UI_TextGlow
     tiling=NoTile
     imageSize=<64 64>
     margin=<ltrb 0.5 0.5 0.5 0.5>
     outlineSettings {
      bUseBrushTransparency=false
      roundingType=HalfHeightRadius
      width=0
      color {colorUseRule=UseColor_Specified, specifiedColor=<rgba 0 0 0 0>}
      cornerRadii=<0 0 0 0>
     }
     tintColor {colorUseRule=UseColor_Specified, specifiedColor=<rgba 1 1 1 1>}
     uVRegion {bIsValid=false, max=<0 0>, min=<0 0>}
    }
    renderTransform {angle=0, scale=<1 1>, shear=<-8 0>, translation=<0 0>}
   }
   slot {class=OverlaySlot, properties {horizontalAlignment=HAlign_Fill, verticalAlignment=VAlign_Center, padding=<ltrb -16 -4 -16 -8>}}
  }
  StreakMultiplierText {
   widgetClassPath=&CommonTextBlock
   bIsVariable=true
   parent=Text
   widget=StreakMultiplierText
   properties {
    bIsScrollingEnabled=false
    justification=Center
    minDesiredWidth=0.5
    text=2x
    colorAndOpacity {colorUseRule=UseColor_Specified, specifiedColor=<rgba 0 1 1 1>}
    font {
     bForceMonospaced=false
     bMaterialIsStencil=false
     fontMaterial=null
     fontObject=&HeadingNow
     letterSpacing=0
     monospacedWidth=1
     size=27
     skewAmount=0
     typefaceFontName=86 Bold Italic
     outlineSettings {
      bApplyOutlineToDropShadows=false
      bMiteredCorners=false
      bSeparateFillAlpha=false
      outlineMaterial=&M-FontShadow
      outlineSize=3
      outlineColor=<rgba 1 1 1 0.333>
     }
    }
    shadowOffset=<0 0>
    strikeBrush {
     bIsDynamicallyLoaded=false
     drawAs=NoDrawType
     imageType=NoImage
     mirroring=NoMirror
     resourceName=None
     resourceObject=null
     tiling=NoTile
     imageSize=<32 32>
     margin=<ltrb 0 0 0 0>
     outlineSettings {
      bUseBrushTransparency=false
      roundingType=HalfHeightRadius
      width=0
      color {colorUseRule=UseColor_Specified, specifiedColor=<rgba 0 0 0 0>}
      cornerRadii=<0 0 0 0>
     }
     tintColor {colorUseRule=UseColor_Specified, specifiedColor=<rgba 1 1 1 1>}
     uVRegion {bIsValid=false, max=<0 0>, min=<0 0>}
    }
   }
   slot {class=OverlaySlot, properties {horizontalAlignment=HAlign_Center, verticalAlignment=VAlign_Center}}
  }
  StreakMeter_Overlay {
   widgetClassPath=&Overlay
   parent=GridPanel_0
   widget=StreakMeter_Overlay
   slot {class=GridSlot, properties {horizontalAlignment=HAlign_Center, verticalAlignment=VAlign_Center}}
  }
  MeterSizeControl {
   widgetClassPath=&Spacer
   parent=StreakMeter_Overlay
   widget=MeterSizeControl
   properties {size=<128 128>}
   slot {class=OverlaySlot, properties {horizontalAlignment=HAlign_Center, verticalAlignment=VAlign_Center}}
  }
  StreakProgressOrnament {
   widgetClassPath=&Image
   bIsVariable=true
   parent=StreakMeter_Overlay
   widget=StreakProgressOrnament
   properties {
    bIsVolatile=true
    brush {
     bIsDynamicallyLoaded=false
     drawAs=Image
     imageType=NoImage
     mirroring=NoMirror
     resourceName=None
     resourceObject=/PilgrimCore/Widgets/Materials/M_UI_Pilgrim_StreakMeter_Ornament
     tiling=NoTile
     imageSize=<128 128>
     margin=<ltrb 0 0 0 0>
     outlineSettings {
      bUseBrushTransparency=false
      roundingType=HalfHeightRadius
      width=0
      color {colorUseRule=UseColor_Specified, specifiedColor=<rgba 0 0 0 0>}
      cornerRadii=<0 0 0 0>
     }
     tintColor {colorUseRule=UseColor_Specified, specifiedColor=<rgba 1 1 1 1>}
     uVRegion {bIsValid=false, max=<0 0>, min=<0 0>}
    }
   }
   slot {class=OverlaySlot, properties {horizontalAlignment=HAlign_Center, verticalAlignment=VAlign_Center}}
  }
  StreakProgress_Glow {
   widgetClassPath=&Image
   bIsVariable=true
   parent=StreakMeter_Overlay
   widget=StreakProgress_Glow
   properties {
    bIsVolatile=true
    visibility=SelfHitTestInvisible
    brush {
     bIsDynamicallyLoaded=false
     drawAs=Image
     imageType=NoImage
     mirroring=NoMirror
     resourceName=None
     resourceObject=/PilgrimCore/Widgets/Materials/M_UI_MeterGlow
     tiling=NoTile
     imageSize=<128 128>
     margin=<ltrb 0 0 0 0>
     outlineSettings {
      bUseBrushTransparency=false
      roundingType=HalfHeightRadius
      width=0
      color {colorUseRule=UseColor_Specified, specifiedColor=<rgba 0 0 0 0>}
      cornerRadii=<0 0 0 0>
     }
     tintColor {colorUseRule=UseColor_Specified, specifiedColor=<rgba 1 1 1 1>}
     uVRegion {bIsValid=false, max=<0 0>, min=<0 0>}
    }
    renderTransform {angle=0, scale=<1.5 1.5>, shear=<0 0>, translation=<0 0>}
   }
   slot {class=OverlaySlot, properties {horizontalAlignment=HAlign_Center, verticalAlignment=VAlign_Center}}
  }
  StreakProgressMeter {
   widgetClassPath=&Image
   bIsVariable=true
   parent=StreakMeter_Overlay
   widget=StreakProgressMeter
   properties {
    bIsVolatile=true
    brush {
     bIsDynamicallyLoaded=false
     drawAs=Image
     imageType=NoImage
     mirroring=NoMirror
     resourceName=None
     resourceObject=/PilgrimCore/Widgets/Materials/MI_UI_Pilgrim_StreakMeter
     tiling=NoTile
     imageSize=<128 128>
     margin=<ltrb 0 0 0 0>
     outlineSettings {
      bUseBrushTransparency=false
      roundingType=HalfHeightRadius
      width=0
      color {colorUseRule=UseColor_Specified, specifiedColor=<rgba 0 0 0 0>}
      cornerRadii=<0 0 0 0>
     }
     tintColor {colorUseRule=UseColor_Specified, specifiedColor=<rgba 1 1 1 1>}
     uVRegion {bIsValid=false, max=<0 0>, min=<0 0>}
    }
   }
   slot {class=OverlaySlot, properties {horizontalAlignment=HAlign_Center, verticalAlignment=VAlign_Center}}
  }
  StreakProgress_ComboIndicator {
   widgetClassPath=&Image
   bIsVariable=true
   parent=StreakMeter_Overlay
   widget=StreakProgress_ComboIndicator
   properties {
    bIsVolatile=true
    brush {
     bIsDynamicallyLoaded=false
     drawAs=Image
     imageType=NoImage
     mirroring=NoMirror
     resourceName=None
     resourceObject=/PilgrimCore/Widgets/Materials/M_UI_Pilgrim_StreakMeter_ComboIndicator
     tiling=NoTile
     imageSize=<128 128>
     margin=<ltrb 0 0 0 0>
     outlineSettings {
      bUseBrushTransparency=false
      roundingType=HalfHeightRadius
      width=0
      color {colorUseRule=UseColor_Specified, specifiedColor=<rgba 0 0 0 0>}
      cornerRadii=<0 0 0 0>
     }
     tintColor {colorUseRule=UseColor_Specified, specifiedColor=<rgba 1 1 1 1>}
     uVRegion {bIsValid=false, max=<0 0>, min=<0 0>}
    }
   }
   slot {class=OverlaySlot, properties {horizontalAlignment=HAlign_Center, verticalAlignment=VAlign_Center}}
  }
  Pips_Overlay {
   widgetClassPath=&Overlay
   parent=GridPanel_0
   widget=Pips_Overlay
   slot {class=GridSlot, properties {horizontalAlignment=HAlign_Center, verticalAlignment=VAlign_Center}}
  }
  PipsSizeControl {widgetClassPath=&Spacer, parent=Pips_Overlay, widget=PipsSizeControl, properties {size=<168 168>}, slot {class=OverlaySlot}}
  WBP_StreakLevelPips_Track_C_0 {
   widgetClassPath=&/WBP_StreakLevelPips_Track
   bIsVariable=true
   parent=Pips_Overlay
   widget=WBP_StreakLevelPips_Track_C_0
   slot {class=OverlaySlot, properties {horizontalAlignment=HAlign_Fill, verticalAlignment=VAlign_Fill}}
  }
 ]
}
