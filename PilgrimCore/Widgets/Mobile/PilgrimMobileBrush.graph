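# PilgrimMobileBrush is a Common UI widget showing the touch icon for a bound input action in Pilgrim's mobile UI.
# On Construct it requests the action's touch brush from the owning player's Pilgrim controller component and, if found, applies a resized copy to an image inside a ScaleBox.
# ResizeBrush scales the shorter side to a BrushSize-based target, keeping aspect ratio with rounding, and logs zero-size images.
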WidgetBlueprint /PilgrimCore/Widgets/Mobile/PilgrimMobileBrush
graphs [
 EventGraph {nodes [
   Event Construct()
   n1 = GetTouchBrushForInputAction!(
    self:Pilgrim Player Controller Component Object Reference=GetComponentByClass!(
     self:Actor Object Reference=GetOwningPlayer()
     ComponentClass:Actor Component Class Reference=/PilgrimCore/Player/Components/BP_PilgrimPlayerControllerComponent.BP_PilgrimPlayerControllerComponent_C
    ) -> (ReturnValue:BP Pilgrim Player Controller Component Object Reference)
    InputAction:Input Action Object Reference=Get BoundInputAction() -> (BoundInputAction:Input Action Object Reference)
   ) -> (OutBrush:Slate Brush Structure, ReturnValue:bool)
   > Branch(Condition=@n1.ReturnValue)
   > n2 = ResizeBrush!(execute, self, Brush:"Slate Brush Structure (by ref)"=@n1) -> (then, Resized Brush:Slate Brush Structure)
   > SetBrush(self=Get Image_Icon() -> (Image_Icon:Image Object Reference), InBrush=@n2)
  ]}
 ResizeBrush {nodes [
   n1 = ResizeBrush!() -> (then, Brush:"Slate Brush Structure (by ref)") {is=title K2Node_FunctionEntry}
   n2 = Break Slate Brush!(SlateBrush:"Slate Brush Structure (by ref)"=@n1) -> (
    TintColor:Slate Color Structure
    DrawAs:ESlateBrushDrawType Enum
    Tiling:ESlateBrushTileType Enum
    Mirroring:ESlateBrushMirrorType Enum
    ImageSize:Vector 2D Structure
    Margin:Margin Structure
    ResourceObject:Object Reference
    OutlineSettings:Slate Brush Outline Settings Structure
   ) {is=title K2Node_BreakStruct}
   n3 = GetMin2D!(A:Vector 2D Structure=@n2.ImageSize) -> (ReturnValue:double)
   > n4 = Branch(Condition="Not Equal (Float)"(A=@n3))
   n5 = vector2d / vector2d(
    A=Select!(
     VeryVerySmall:double=24
     VerySmall:double=24
     Small:double=24
     Medium:double=32
     Large:double=64
     VeryLarge:double=64
     Index:EFortBrushSize Enum=Get BrushSize() -> (BrushSize:EFortBrushSize Enum=VeryVerySmall)
    ) -> (ReturnValue:double)
    B=@n3
   ) -> (ReturnValue:Vector 2D Structure) {is=title K2Node_PromotableOperator}
   > n6 = Set members in Slate Brush!(
    execute:Exec
    StructRef:"Slate Brush Structure (by ref)"=@n1
    ImageSize:Vector 2D Structure="To Rounded (Vector2D)"!(A:Vector 2D Structure=vector2d * vector2d(A:Vector 2D Structure=@n2.ImageSize, B:Vector 2D Structure=@n5) -> (ReturnValue:Vector 2D Structure) {is=title K2Node_PromotableOperator}) -> (ReturnValue:Vector 2D Structure)
   ) -> (then:Exec, StructOut:"Slate Brush Structure (by ref)") {is=title K2Node_SetFieldsInStruct}
   > Return Node!(execute, Resized Brush:Slate Brush Structure=@n6)
   FortBlueprintLog!(execute=@n4.else, InString:String=Why is there an image with zero size?) -> (then)
   > Return Node()
  ]}
]
variables [
 BoundInputAction {type=InputAction, default=null}
 BrushSize {type=EFortBrushSize, default=Medium}
]
widgets {
 info {inheritedWidgetCount=0, parentClass=&CommonUserWidget, rootWidgetClass=&ScaleBox, widgetCount=2}
 widgets [
  ScaleBox_1 {widgetClassPath=&ScaleBox, widget=ScaleBox_1}
  Image_Icon {
   widgetClassPath=&Image
   bIsVariable=true
   parent=ScaleBox_1
   widget=Image_Icon
   properties {
    visibility=HitTestInvisible
    brush {
     bIsDynamicallyLoaded=false
     drawAs=Image
     imageType=NoImage
     mirroring=NoMirror
     resourceName=None
     resourceObject=null
     tiling=NoTile
     imageSize=<64 64>
     margin=<ltrb 0 0 0 0>
     outlineSettings {
      bUseBrushTransparency=false
      roundingType=HalfHeightRadius
      width=0
      color {colorUseRule=UseColor_Specified, specifiedColor=<rgba 0 0 0 0>}
      cornerRadii=<0 0 0 0>
     }
     tintColor {colorUseRule=UseColor_Specified, specifiedColor=<rgba 1 1 1 1>}
     uVRegion {bIsValid=false, max=<0 0>, min=<0 0>}
    }
   }
   slot {class=ScaleBoxSlot}
  }
 ]
}
