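# Pilgrim widget that shows an Early, Perfect or Late accuracy icon for gem hit feedback, built from a single lazy image.
# SetHitAccuracy maps the hit accuracy enum to the matching _Combo animation, rewinds and plays it, skipping this at design time.
# Those short animations drive the icon material's ScaleAnim and set IconType to pick the variant.
# Longer Early, Perfect and Late animations with fade and scale pop exist but are not referenced here.
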WidgetBlueprint /PilgrimCore/Widgets/WBP_Pilgrim_AccuracyIcon
animations [
 Early {
  binding_count=1
  length_seconds=0.75
  object=&self:Early
  sequence {bindings [AccuracyIcon {
     id=#1
     possessed_class=FortLazyImage
     tracks [
      {
       type=MovieSceneWidgetMaterialTrack
       sections [{
         type=MovieSceneParameterSection
         end_frame=15
         start_frame=0
         channels [
          ScaleAnim {type=MovieSceneScriptingFloatChannel, keys [[0,-2.4773],[6,1.1],[10.5,-0.5],[15,0]]}
          IconType {type=MovieSceneScriptingFloatChannel, const [0,1]}
         ]
        }]
      }
      {
       type=MovieSceneColorTrack
       display_name=Color and Opacity
       sections [{
         type=MovieSceneColorSection
         end_frame=4
         start_frame=0
         channels [
          Color.R {type=MovieSceneScriptingFloatChannel, const [1,1]}
          Color.G {type=MovieSceneScriptingFloatChannel, const [1,1]}
          Color.B {type=MovieSceneScriptingFloatChannel, const [1,2]}
          Color.A {type=MovieSceneScriptingFloatChannel, keys [[0,0],[4,1]]}
         ]
        }]
      }
      {
       type=MovieScene2DTransformTrack
       sections [{
         type=MovieScene2DTransformSection
         end_frame=7
         start_frame=0
         channels [
          Translation.X {type=MovieSceneScriptingFloatChannel, const [0,1]}
          Translation.Y {type=MovieSceneScriptingFloatChannel, const [0,1]}
          Angle {type=MovieSceneScriptingFloatChannel, const [0,1]}
          Scale.X {type=MovieSceneScriptingFloatChannel, keys [[0,0],[3,1.05],[7.5,1]]}
          Scale.Y {type=MovieSceneScriptingFloatChannel, keys [[0,0],[3,1.05],[7.5,1]]}
          Shear.X {type=MovieSceneScriptingFloatChannel, const [0,1]}
          Shear.Y {type=MovieSceneScriptingFloatChannel, const [0,1]}
         ]
        }]
      }
     ]
    }]}
 }
 Early_Combo {
  binding_count=1
  length_seconds=0.25
  object=&self:Early_Combo
  sequence {bindings [AccuracyIcon {
     id=#2
     possessed_class=FortLazyImage
     tracks [{
       type=MovieSceneWidgetMaterialTrack
       sections [{
         type=MovieSceneParameterSection
         end_frame=5
         start_frame=0
         channels [
          ScaleAnim {type=MovieSceneScriptingFloatChannel, keys [[0,0],[5,1]]}
          IconType {type=MovieSceneScriptingFloatChannel, const [0,1]}
         ]
        }]
      }]
    }]}
 }
 Perfect {
  binding_count=1
  length_seconds=0.75
  object=&self:Perfect
  sequence {bindings [AccuracyIcon {
     id=#3
     possessed_class=FortLazyImage
     tracks [
      {
       type=MovieSceneWidgetMaterialTrack
       sections [{
         type=MovieSceneParameterSection
         end_frame=15
         start_frame=0
         channels [ScaleAnim {type=MovieSceneScriptingFloatChannel, keys [[0,-2.4773],[6,1.1],[10.5,-0.5],[15,0]]}]
        }]
      }
      {
       type=MovieSceneColorTrack
       display_name=Color and Opacity
       sections [{
         type=MovieSceneColorSection
         end_frame=4
         start_frame=0
         channels [
          Color.R {type=MovieSceneScriptingFloatChannel, const [1,1]}
          Color.G {type=MovieSceneScriptingFloatChannel, const [1,1]}
          Color.B {type=MovieSceneScriptingFloatChannel, const [1,2]}
          Color.A {type=MovieSceneScriptingFloatChannel, keys [[0,0],[4.5,1]]}
         ]
        }]
      }
      {
       type=MovieScene2DTransformTrack
       sections [{
         type=MovieScene2DTransformSection
         end_frame=7
         start_frame=0
         channels [
          Translation.X {type=MovieSceneScriptingFloatChannel, const [0,1]}
          Translation.Y {type=MovieSceneScriptingFloatChannel, const [0,1]}
          Angle {type=MovieSceneScriptingFloatChannel, const [0,1]}
          Scale.X {type=MovieSceneScriptingFloatChannel, keys [[0,0],[3,1.05],[7.5,1]]}
          Scale.Y {type=MovieSceneScriptingFloatChannel, keys [[0,0],[3,1.05],[7.5,1]]}
          Shear.X {type=MovieSceneScriptingFloatChannel, const [0,1]}
          Shear.Y {type=MovieSceneScriptingFloatChannel, const [0,1]}
         ]
        }]
      }
     ]
    }]}
 }
 Perfect_Combo {
  binding_count=1
  length_seconds=0.25
  object=&self:Perfect_Combo
  sequence {bindings [AccuracyIcon {
     id=#4
     possessed_class=FortLazyImage
     tracks [{
       type=MovieSceneWidgetMaterialTrack
       sections [{
         type=MovieSceneParameterSection
         end_frame=5
         start_frame=0
         channels [
          ScaleAnim {type=MovieSceneScriptingFloatChannel, keys [[0,0],[5,1]]}
          IconType {type=MovieSceneScriptingFloatChannel, const [0.5,1]}
         ]
        }]
      }]
    }]}
 }
 Late {
  binding_count=1
  length_seconds=0.745
  object=&self:Late
  sequence {bindings [AccuracyIcon {
     id=#5
     possessed_class=FortLazyImage
     tracks [
      {
       type=MovieSceneColorTrack
       display_name=Color and Opacity
       sections [{
         type=MovieSceneColorSection
         end_frame=5
         start_frame=0
         channels [
          Color.R {type=MovieSceneScriptingFloatChannel, const [1,1]}
          Color.G {type=MovieSceneScriptingFloatChannel, const [1,1]}
          Color.B {type=MovieSceneScriptingFloatChannel, const [1,2]}
          Color.A {type=MovieSceneScriptingFloatChannel, keys [[0,0],[5.96,1]]}
         ]
        }]
      }
      {
       type=MovieScene2DTransformTrack
       sections [{
         type=MovieScene2DTransformSection
         end_frame=7
         start_frame=0
         channels [
          Translation.X {type=MovieSceneScriptingFloatChannel, const [0,1]}
          Translation.Y {type=MovieSceneScriptingFloatChannel, const [0,1]}
          Angle {type=MovieSceneScriptingFloatChannel, const [0,1]}
          Scale.X {type=MovieSceneScriptingFloatChannel, keys [[0,0],[2.98,1.05],[7.45,1]]}
          Scale.Y {type=MovieSceneScriptingFloatChannel, keys [[0,0],[2.98,1.05],[7.45,1]]}
          Shear.X {type=MovieSceneScriptingFloatChannel, const [0,1]}
          Shear.Y {type=MovieSceneScriptingFloatChannel, const [0,1]}
         ]
        }]
      }
      {
       type=MovieSceneWidgetMaterialTrack
       sections [{
         type=MovieSceneParameterSection
         end_frame=14
         start_frame=0
         channels [
          ScaleAnim {type=MovieSceneScriptingFloatChannel, keys [[0,-2.4773],[5.96,1.1],[10.43,-0.5],[14.9,0]]}
          IconType {type=MovieSceneScriptingFloatChannel, const [1,1]}
         ]
        }]
      }
     ]
    }]}
 }
 Late_Combo {
  binding_count=1
  length_seconds=0.25
  object=&self:Late_Combo
  sequence {bindings [AccuracyIcon {
     id=#6
     possessed_class=FortLazyImage
     tracks [{
       type=MovieSceneWidgetMaterialTrack
       sections [{
         type=MovieSceneParameterSection
         end_frame=5
         start_frame=0
         channels [
          ScaleAnim {type=MovieSceneScriptingFloatChannel, keys [[0,0],[5,1]]}
          IconType {type=MovieSceneScriptingFloatChannel, const [1,1]}
         ]
        }]
      }]
    }]}
 }
]
graphs [
 EventGraph {links [], nodes []}
 SetHitAccuracy {nodes [
   n1 = SetHitAccuracy!() -> (then, HitAccuracy:E_GemHitAccuracy Enum=NewEnumerator0, IsDesignTime:bool=~) {is=title K2Node_FunctionEntry}
   > n2 = Branch(Condition=@n1.IsDesignTime)
   n3 = Select!(
    NewEnumerator0:Widget Animation Object Reference=Get Early_Combo() -> (Early_Combo:Widget Animation Object Reference)
    NewEnumerator1:Widget Animation Object Reference=Get Perfect_Combo() -> (Perfect_Combo:Widget Animation Object Reference)
    NewEnumerator2:Widget Animation Object Reference=Get Late_Combo() -> (Late_Combo:Widget Animation Object Reference)
    Index:E_GemHitAccuracy Enum=@n1
   ) -> (ReturnValue:Widget Animation Object Reference)
   SetAnimationCurrentTime!(execute=@n2.else, self:User Widget Object Reference, InAnimation:Widget Animation Object Reference=@n3, InTime:float=0) -> (then)
   > PlayAnimation(InAnimation=@n3)
  ]}
]
variables []
widgets {
 info {inheritedWidgetCount=0, parentClass=&UserWidget, rootWidgetClass=&FortLazyImage, widgetCount=1}
 widgets [AccuracyIcon {
   widgetClassPath=&FortLazyImage
   bIsVariable=true
   widget=AccuracyIcon
   properties {
    visibility=Visible
    brush {
     bIsDynamicallyLoaded=false
     drawAs=Image
     imageType=FullColor
     mirroring=NoMirror
     resourceName=None
     resourceObject=/PilgrimCore/Widgets/Materials/GemHitFeedbackMaterials/MI_UI_Pilgrim_GemHitFeedback_Icons_Perfect
     tiling=NoTile
     imageSize=<45 45>
     margin=<ltrb 0 0 0 0>
     outlineSettings {
      bUseBrushTransparency=false
      roundingType=HalfHeightRadius
      width=0
      color {colorUseRule=UseColor_Specified, specifiedColor=<rgba 0 0 0 0>}
      cornerRadii=<0 0 0 0>
     }
     tintColor {colorUseRule=UseColor_Specified, specifiedColor=<rgba 1 1 1 1>}
     uVRegion {bIsValid=false, max=<0 0>, min=<0 0>}
    }
   }
  }]
}
