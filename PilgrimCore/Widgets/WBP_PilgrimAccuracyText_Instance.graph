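# Widget Blueprint for a single hit-accuracy popup in Pilgrim, with animations for Early, Perfect and Late timing results.
# Each animates the background, echo image, text and icons while the whole widget drifts and fades out.
# Early drifts up-left, Late down-left and Perfect straight left; material tracks flash the gradient, or pulse OverdriveActive for Perfect.
# Variants include horizontal Early, cascading and non-cascading FadeIn, and Late_old.
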WidgetBlueprint /PilgrimCore/Widgets/WBP_PilgrimAccuracyText_Instance
animations [
 FadeIn {
  binding_count=2
  length_seconds=1
  object=&self:FadeIn
  sequence {bindings [
    AccuracyBanner {
     id=#1
     possessed_class=Overlay
     tracks [{
       type=MovieScene2DTransformTrack
       sections [{
         type=MovieScene2DTransformSection
         end_frame=20
         start_frame=0
         channels [
          Angle {type=MovieSceneScriptingFloatChannel, const [0,1]}
          Scale.X {type=MovieSceneScriptingFloatChannel, keys [[0,0.1],[1,0.05],[2,0.1],[4,1]]}
          Scale.Y {type=MovieSceneScriptingFloatChannel, default=1}
          Shear.X {type=MovieSceneScriptingFloatChannel, const [0,1]}
          Shear.Y {type=MovieSceneScriptingFloatChannel, const [0,1]}
         ]
        }]
      }]
    }
    WBP_PilgrimAccuracyText_Instance_C_81 {
     id=#2
     possessed_class=WBP_PilgrimAccuracyText_Instance_C
     tracks [
      {
       type=MovieSceneFloatTrack
       display_name=Render Opacity
       sections [{
         type=MovieSceneFloatSection
         end_frame=20
         start_frame=0
         channels [None {type=MovieSceneScriptingFloatChannel, keys [[0,0],[1,1],[4,1],[20,0]]}]
        }]
      }
      {
       type=MovieScene2DTransformTrack
       sections [{
         type=MovieScene2DTransformSection
         end_frame=20
         start_frame=0
         channels [
          Translation.X {type=MovieSceneScriptingFloatChannel, keys [[5,0],[7,-40],[20,-60]]}
          Translation.Y {type=MovieSceneScriptingFloatChannel, keys [[5,0],[20,50]]}
          Angle {type=MovieSceneScriptingFloatChannel, const [0,1]}
          Scale.X {type=MovieSceneScriptingFloatChannel, const [1,1]}
          Scale.Y {type=MovieSceneScriptingFloatChannel, const [1,1]}
          Shear.X {type=MovieSceneScriptingFloatChannel, keys [[2,-16],[4,0]]}
          Shear.Y {type=MovieSceneScriptingFloatChannel, const [0,1]}
         ]
        }]
      }
      {
       type=MovieSceneByteTrack
       display_name=Visibility
       sections [{type=MovieSceneByteSection, end_frame=0, start_frame=0, channels [None {type=MovieSceneScriptingByteChannel, const [4,1]}]}]
      }
     ]
    }
   ]}
 }
 Early {
  binding_count=6
  length_seconds=1
  object=&self:Early
  sequence {
   bindings [
    Background {
     id=#3
     possessed_class=Image
     tracks [
      {
       type=MovieSceneColorTrack
       display_name=Color and Opacity
       sections [{
         type=MovieSceneColorSection
         end_frame=30
         start_frame=0
         channels [
          Color.R {type=MovieSceneScriptingFloatChannel, const [1,1]}
          Color.G {type=MovieSceneScriptingFloatChannel, const [1,1]}
          Color.B {type=MovieSceneScriptingFloatChannel, const [1,1]}
          Color.A {type=MovieSceneScriptingFloatChannel, keys [[0,0],[5,1]]}
         ]
        }]
      }
      {
       type=MovieSceneWidgetMaterialTrack
       sections [{
         type=MovieSceneParameterSection
         end_frame=100
         start_frame=0
         channels [
          Gradient_Color_BR {type=MovieSceneScriptingFloatChannel, keys [[0,0],[9,1],[40,0]]}
          Gradient_Color_BG {type=MovieSceneScriptingFloatChannel, keys [[0,0.5779],[8,1],[38,0.5779]]}
          Gradient_Color_BB {type=MovieSceneScriptingFloatChannel, keys [[0,0.8],[7,1],[36,0.8]]}
          Gradient_Color_BA {type=MovieSceneScriptingFloatChannel, const [1,2]}
         ]
        }]
      }
     ]
    }
    BG_Echo {
     id=#4
     possessed_class=Image
     tracks [
      {
       type=MovieSceneColorTrack
       display_name=Color and Opacity
       sections [{
         type=MovieSceneColorSection
         end_frame=100
         start_frame=0
         channels [
          Color.R {type=MovieSceneScriptingFloatChannel, const [1,1]}
          Color.G {type=MovieSceneScriptingFloatChannel, const [1,1]}
          Color.B {type=MovieSceneScriptingFloatChannel, const [1,1]}
          Color.A {type=MovieSceneScriptingFloatChannel, keys [[0,0],[6.6667,1]]}
         ]
        }]
      }
      {
       type=MovieScene2DTransformTrack
       sections [{
         type=MovieScene2DTransformSection
         end_frame=100
         start_frame=0
         channels [
          Scale.X {type=MovieSceneScriptingFloatChannel, const [-1,1]}
          Scale.Y {type=MovieSceneScriptingFloatChannel, keys [[0,1],[13.3333,1.2],[100,1]]}
         ]
        }]
      }
     ]
    }
    PerfectAccuracyText {
     id=#5
     possessed_class=CommonTextBlock
     tracks [
      {
       type=MovieSceneColorTrack
       display_name=Color and Opacity
       sections [{
         type=MovieSceneColorSection
         end_frame=50
         start_frame=0
         channels [
          Color.R {type=MovieSceneScriptingFloatChannel, keys [[0,0],[5,1]]}
          Color.G {type=MovieSceneScriptingFloatChannel, keys [[0,0.18],[5,1]]}
          Color.B {type=MovieSceneScriptingFloatChannel, const [1,2]}
          Color.A {type=MovieSceneScriptingFloatChannel, keys [[0,0],[5,1]]}
         ]
        }]
      }
      {
       type=MovieScene2DTransformTrack
       sections [{
         type=MovieScene2DTransformSection
         end_frame=100
         start_frame=0
         channels [
          Translation.X {type=MovieSceneScriptingFloatChannel, keys [[0,0],[5,0],[100,-10]]}
          Translation.Y {type=MovieSceneScriptingFloatChannel, const [0,1]}
          Angle {type=MovieSceneScriptingFloatChannel, const [0,1]}
          Scale.X {type=MovieSceneScriptingFloatChannel, const [1,1]}
          Scale.Y {type=MovieSceneScriptingFloatChannel, const [1,1]}
          Shear.X {type=MovieSceneScriptingFloatChannel, const [0,1]}
          Shear.Y {type=MovieSceneScriptingFloatChannel, const [0,1]}
         ]
        }]
      }
     ]
    }
    PerfectIcon {
     id=#6
     possessed_class=FortLazyImage
     tracks [
      {
       type=MovieSceneColorTrack
       display_name=Color and Opacity
       sections [{
         type=MovieSceneColorSection
         end_frame=100
         start_frame=0
         channels [
          Color.R {type=MovieSceneScriptingFloatChannel, const [1,1]}
          Color.G {type=MovieSceneScriptingFloatChannel, const [1,1]}
          Color.B {type=MovieSceneScriptingFloatChannel, const [1,1]}
          Color.A {type=MovieSceneScriptingFloatChannel, keys [[0,0],[10,1]]}
         ]
        }]
      }
      {
       type=MovieScene2DTransformTrack
       sections [{
         type=MovieScene2DTransformSection
         end_frame=100
         start_frame=0
         channels [
          Translation.X {type=MovieSceneScriptingFloatChannel, keys [[0,0],[30,0],[93,20]]}
          Translation.Y {type=MovieSceneScriptingFloatChannel, const [0,1]}
          Angle {type=MovieSceneScriptingFloatChannel, const [0,1]}
          Scale.X {type=MovieSceneScriptingFloatChannel, default=1}
          Scale.Y {type=MovieSceneScriptingFloatChannel, default=1}
          Shear.X {type=MovieSceneScriptingFloatChannel, const [0,1]}
          Shear.Y {type=MovieSceneScriptingFloatChannel, const [0,1]}
         ]
        }]
      }
     ]
    }
    WBP_Pilgrim_AccuracyIcon {
     id=#7
     possessed_class=WBP_Pilgrim_AccuracyIcon_C
     tracks [
      {
       type=MovieSceneFloatTrack
       display_name=Render Opacity
       sections [{type=MovieSceneFloatSection, end_frame=100, start_frame=0, channels [None {type=MovieSceneScriptingFloatChannel, keys [[0,0],[5,1]]}]}]
      }
      {
       type=MovieScene2DTransformTrack
       sections [{
         type=MovieScene2DTransformSection
         end_frame=100
         start_frame=0
         channels [
          Translation.X {type=MovieSceneScriptingFloatChannel, const [0,2]}
          Translation.Y {type=MovieSceneScriptingFloatChannel, const [0,1]}
          Angle {type=MovieSceneScriptingFloatChannel, const [0,1]}
          Scale.X {type=MovieSceneScriptingFloatChannel, const [1,1]}
          Scale.Y {type=MovieSceneScriptingFloatChannel, const [1,1]}
          Shear.X {type=MovieSceneScriptingFloatChannel, const [0,2]}
          Shear.Y {type=MovieSceneScriptingFloatChannel, const [0,2]}
         ]
        }]
      }
     ]
    }
    WBP_PilgrimAccuracyText_Instance_C_29 {
     id=#8
     possessed_class=WBP_PilgrimAccuracyText_Instance_C
     tracks [
      {
       type=MovieSceneFloatTrack
       display_name=Render Opacity
       sections [{
         type=MovieSceneFloatSection
         end_frame=100
         start_frame=0
         channels [None {type=MovieSceneScriptingFloatChannel, keys [[0,0],[5,1],[50,1],[91,0]]}]
        }]
      }
      {
       type=MovieScene2DTransformTrack
       sections [{
         type=MovieScene2DTransformSection
         end_frame=100
         start_frame=0
         channels [
          Translation.X {type=MovieSceneScriptingFloatChannel, keys [[0,0],[25,0],[100,-100]]}
          Translation.Y {type=MovieSceneScriptingFloatChannel, keys [[0,0],[25,0],[100,-100]]}
          Angle {type=MovieSceneScriptingFloatChannel, const [0,1]}
          Scale.X {type=MovieSceneScriptingFloatChannel, const [1,1]}
          Scale.Y {type=MovieSceneScriptingFloatChannel, const [1,1]}
          Shear.X {type=MovieSceneScriptingFloatChannel, const [0,1]}
          Shear.Y {type=MovieSceneScriptingFloatChannel, const [0,1]}
         ]
        }]
      }
      {
       type=MovieSceneByteTrack
       display_name=Visibility
       sections [{type=MovieSceneByteSection, end_frame=100, start_frame=0, channels [None {type=MovieSceneScriptingByteChannel, keys [[0,3],[100,2]]}]}]
      }
     ]
    }
   ]
   display_rate {numerator=100}
  }
 }
 Perfect {
  binding_count=6
  length_seconds=0.5
  object=&self:Perfect
  sequence {
   bindings [
    Background {
     id=#3
     possessed_class=Image
     tracks [
      {
       type=MovieSceneWidgetMaterialTrack
       sections [{
         type=MovieSceneParameterSection
         end_frame=50
         start_frame=0
         channels [OverdriveActive {type=MovieSceneScriptingFloatChannel, keys [[0,0],[4.04,6],[22.22,0],[50,1]]}]
        }]
      }
      {
       type=MovieSceneColorTrack
       display_name=Color and Opacity
       sections [{
         type=MovieSceneColorSection
         end_frame=14
         start_frame=0
         channels [
          Color.R {type=MovieSceneScriptingFloatChannel, const [1,1]}
          Color.G {type=MovieSceneScriptingFloatChannel, const [1,1]}
          Color.B {type=MovieSceneScriptingFloatChannel, const [1,1]}
          Color.A {type=MovieSceneScriptingFloatChannel, keys [[0,0],[2.525,1]]}
         ]
        }]
      }
     ]
    }
    BG_Echo {
     id=#4
     possessed_class=Image
     tracks [
      {
       type=MovieSceneColorTrack
       display_name=Color and Opacity
       sections [{
         type=MovieSceneColorSection
         end_frame=50
         start_frame=0
         channels [
          Color.R {type=MovieSceneScriptingFloatChannel, const [1,1]}
          Color.G {type=MovieSceneScriptingFloatChannel, const [1,1]}
          Color.B {type=MovieSceneScriptingFloatChannel, const [1,1]}
          Color.A {type=MovieSceneScriptingFloatChannel, keys [[0,0],[2.5,1]]}
         ]
        }]
      }
      {
       type=MovieScene2DTransformTrack
       sections [{
         type=MovieScene2DTransformSection
         end_frame=50
         start_frame=0
         channels [
          Scale.X {type=MovieSceneScriptingFloatChannel, keys [[0,0],[12.5,-1],[50,-1]]}
          Scale.Y {type=MovieSceneScriptingFloatChannel, const [1,2]}
         ]
        }]
      }
     ]
    }
    PerfectAccuracyText {
     id=#5
     possessed_class=CommonTextBlock
     tracks [
      {
       type=MovieSceneColorTrack
       display_name=Color and Opacity
       sections [{
         type=MovieSceneColorSection
         end_frame=25
         start_frame=0
         channels [
          Color.R {type=MovieSceneScriptingFloatChannel, const [1,2]}
          Color.G {type=MovieSceneScriptingFloatChannel, keys [[0,0.18],[2.525,1]]}
          Color.B {type=MovieSceneScriptingFloatChannel, keys [[0,-0.04],[2.525,1]]}
          Color.A {type=MovieSceneScriptingFloatChannel, keys [[0,0],[2.525,1]]}
         ]
        }]
      }
      {
       type=MovieScene2DTransformTrack
       sections [{
         type=MovieScene2DTransformSection
         end_frame=50
         start_frame=0
         channels [
          Translation.X {type=MovieSceneScriptingFloatChannel, keys [[0,0],[15.1517,0],[50,-10]]}
          Translation.Y {type=MovieSceneScriptingFloatChannel, const [0,1]}
          Angle {type=MovieSceneScriptingFloatChannel, const [0,1]}
          Scale.X {type=MovieSceneScriptingFloatChannel, const [1,1]}
          Scale.Y {type=MovieSceneScriptingFloatChannel, const [1,1]}
          Shear.X {type=MovieSceneScriptingFloatChannel, const [0,1]}
          Shear.Y {type=MovieSceneScriptingFloatChannel, const [0,1]}
         ]
        }]
      }
     ]
    }
    PerfectIcon {
     id=#6
     possessed_class=FortLazyImage
     tracks [
      {
       type=MovieScene2DTransformTrack
       sections [{
         type=MovieScene2DTransformSection
         end_frame=50
         start_frame=0
         channels [
          Translation.X {type=MovieSceneScriptingFloatChannel, keys [[0,0],[15,0],[46.5,20]]}
          Translation.Y {type=MovieSceneScriptingFloatChannel, const [0,1]}
          Angle {type=MovieSceneScriptingFloatChannel, const [0,1]}
          Scale.X {type=MovieSceneScriptingFloatChannel, keys [[0,1],[5,1.2],[12.5,1]]}
          Scale.Y {type=MovieSceneScriptingFloatChannel, keys [[0,1],[5,1.5],[15,1]]}
          Shear.X {type=MovieSceneScriptingFloatChannel, const [0,1]}
          Shear.Y {type=MovieSceneScriptingFloatChannel, const [0,1]}
         ]
        }]
      }
      {
       type=MovieSceneColorTrack
       display_name=Color and Opacity
       sections [{
         type=MovieSceneColorSection
         end_frame=37
         start_frame=0
         channels [
          Color.R {type=MovieSceneScriptingFloatChannel, const [1,1]}
          Color.G {type=MovieSceneScriptingFloatChannel, const [1,1]}
          Color.B {type=MovieSceneScriptingFloatChannel, const [1,1]}
          Color.A {type=MovieSceneScriptingFloatChannel, keys [[0,0],[2.5,1]]}
         ]
        }]
      }
     ]
    }
    WBP_Pilgrim_AccuracyIcon {
     id=#7
     possessed_class=WBP_Pilgrim_AccuracyIcon_C
     tracks [
      {
       type=MovieSceneFloatTrack
       display_name=Render Opacity
       sections [{type=MovieSceneFloatSection, end_frame=50, start_frame=0, channels [None {type=MovieSceneScriptingFloatChannel, keys [[0,0],[2.5,1]]}]}]
      }
      {
       type=MovieScene2DTransformTrack
       sections [{
         type=MovieScene2DTransformSection
         end_frame=50
         start_frame=0
         channels [
          Translation.X {type=MovieSceneScriptingFloatChannel, const [0,1]}
          Translation.Y {type=MovieSceneScriptingFloatChannel, const [0,2]}
          Angle {type=MovieSceneScriptingFloatChannel, const [0,2]}
          Scale.X {type=MovieSceneScriptingFloatChannel, default=1}
          Scale.Y {type=MovieSceneScriptingFloatChannel, default=1}
          Shear.X {type=MovieSceneScriptingFloatChannel, const [0,2]}
          Shear.Y {type=MovieSceneScriptingFloatChannel, const [0,2]}
         ]
        }]
      }
     ]
    }
    WBP_PilgrimAccuracyText_Instance_C_29 {
     id=#8
     possessed_class=WBP_PilgrimAccuracyText_Instance_C
     tracks [
      {
       type=MovieSceneFloatTrack
       display_name=Render Opacity
       sections [{
         type=MovieSceneFloatSection
         end_frame=50
         start_frame=0
         channels [None {type=MovieSceneScriptingFloatChannel, keys [[0,0],[2.5,1],[35.5,1],[40,0]]}]
        }]
      }
      {
       type=MovieScene2DTransformTrack
       sections [{
         type=MovieScene2DTransformSection
         end_frame=50
         start_frame=0
         channels [
          Translation.X {type=MovieSceneScriptingFloatChannel, keys [[0,0],[50,-75]]}
          Translation.Y {type=MovieSceneScriptingFloatChannel, const [0,2]}
          Angle {type=MovieSceneScriptingFloatChannel, const [0,1]}
          Scale.X {type=MovieSceneScriptingFloatChannel, const [1,1]}
          Scale.Y {type=MovieSceneScriptingFloatChannel, const [1,1]}
          Shear.X {type=MovieSceneScriptingFloatChannel, const [0,1]}
          Shear.Y {type=MovieSceneScriptingFloatChannel, const [0,1]}
         ]
        }]
      }
      {
       type=MovieSceneByteTrack
       display_name=Visibility
       sections [{type=MovieSceneByteSection, end_frame=50, start_frame=0, channels [None {type=MovieSceneScriptingByteChannel, keys [[0,3],[50,2]]}]}]
      }
     ]
    }
   ]
   display_rate {numerator=100}
  }
 }
 Late {
  binding_count=6
  length_seconds=1
  object=&self:Late
  sequence {
   bindings [
    Background {
     id=#3
     possessed_class=Image
     tracks [
      {
       type=MovieSceneColorTrack
       display_name=Color and Opacity
       sections [{
         type=MovieSceneColorSection
         end_frame=30
         start_frame=0
         channels [
          Color.R {type=MovieSceneScriptingFloatChannel, const [1,1]}
          Color.G {type=MovieSceneScriptingFloatChannel, const [1,1]}
          Color.B {type=MovieSceneScriptingFloatChannel, const [1,1]}
          Color.A {type=MovieSceneScriptingFloatChannel, keys [[0,0],[5,1]]}
         ]
        }]
      }
      {
       type=MovieSceneWidgetMaterialTrack
       sections [{
         type=MovieSceneParameterSection
         end_frame=100
         start_frame=0
         channels [
          Gradient_Color_BR {type=MovieSceneScriptingFloatChannel, keys [[0,0],[9,1],[40,0]]}
          Gradient_Color_BG {type=MovieSceneScriptingFloatChannel, keys [[0,0.5779],[8,1],[38,0.5779]]}
          Gradient_Color_BB {type=MovieSceneScriptingFloatChannel, keys [[0,0.8],[7,1],[36,0.8]]}
          Gradient_Color_BA {type=MovieSceneScriptingFloatChannel, const [1,2]}
         ]
        }]
      }
     ]
    }
    BG_Echo {
     id=#4
     possessed_class=Image
     tracks [
      {
       type=MovieSceneColorTrack
       display_name=Color and Opacity
       sections [{
         type=MovieSceneColorSection
         end_frame=100
         start_frame=0
         channels [
          Color.R {type=MovieSceneScriptingFloatChannel, const [1,1]}
          Color.G {type=MovieSceneScriptingFloatChannel, const [1,1]}
          Color.B {type=MovieSceneScriptingFloatChannel, const [1,1]}
          Color.A {type=MovieSceneScriptingFloatChannel, keys [[0,0],[6.6667,1]]}
         ]
        }]
      }
      {
       type=MovieScene2DTransformTrack
       sections [{
         type=MovieScene2DTransformSection
         end_frame=100
         start_frame=0
         channels [
          Scale.X {type=MovieSceneScriptingFloatChannel, const [-1,1]}
          Scale.Y {type=MovieSceneScriptingFloatChannel, keys [[0,1],[13.3333,1.2],[100,1]]}
         ]
        }]
      }
     ]
    }
    PerfectAccuracyText {
     id=#5
     possessed_class=CommonTextBlock
     tracks [
      {
       type=MovieSceneColorTrack
       display_name=Color and Opacity
       sections [{
         type=MovieSceneColorSection
         end_frame=50
         start_frame=0
         channels [
          Color.R {type=MovieSceneScriptingFloatChannel, keys [[0,0],[5,1]]}
          Color.G {type=MovieSceneScriptingFloatChannel, keys [[0,0.18],[5,1]]}
          Color.B {type=MovieSceneScriptingFloatChannel, const [1,2]}
          Color.A {type=MovieSceneScriptingFloatChannel, keys [[0,0],[5,1]]}
         ]
        }]
      }
      {
       type=MovieScene2DTransformTrack
       sections [{
         type=MovieScene2DTransformSection
         end_frame=100
         start_frame=0
         channels [
          Translation.X {type=MovieSceneScriptingFloatChannel, keys [[0,0],[5,0],[100,-10]]}
          Translation.Y {type=MovieSceneScriptingFloatChannel, const [0,1]}
          Angle {type=MovieSceneScriptingFloatChannel, const [0,1]}
          Scale.X {type=MovieSceneScriptingFloatChannel, const [1,1]}
          Scale.Y {type=MovieSceneScriptingFloatChannel, const [1,1]}
          Shear.X {type=MovieSceneScriptingFloatChannel, const [0,1]}
          Shear.Y {type=MovieSceneScriptingFloatChannel, const [0,1]}
         ]
        }]
      }
     ]
    }
    PerfectIcon {
     id=#6
     possessed_class=FortLazyImage
     tracks [
      {
       type=MovieSceneColorTrack
       display_name=Color and Opacity
       sections [{
         type=MovieSceneColorSection
         end_frame=100
         start_frame=0
         channels [
          Color.R {type=MovieSceneScriptingFloatChannel, const [1,1]}
          Color.G {type=MovieSceneScriptingFloatChannel, const [1,1]}
          Color.B {type=MovieSceneScriptingFloatChannel, const [1,1]}
          Color.A {type=MovieSceneScriptingFloatChannel, keys [[0,0],[10,1]]}
         ]
        }]
      }
      {
       type=MovieScene2DTransformTrack
       sections [{
         type=MovieScene2DTransformSection
         end_frame=100
         start_frame=0
         channels [
          Translation.X {type=MovieSceneScriptingFloatChannel, keys [[0,0],[30,0],[93,20]]}
          Translation.Y {type=MovieSceneScriptingFloatChannel, const [0,1]}
          Angle {type=MovieSceneScriptingFloatChannel, const [0,1]}
          Scale.X {type=MovieSceneScriptingFloatChannel, default=1}
          Scale.Y {type=MovieSceneScriptingFloatChannel, default=1}
          Shear.X {type=MovieSceneScriptingFloatChannel, const [0,1]}
          Shear.Y {type=MovieSceneScriptingFloatChannel, const [0,1]}
         ]
        }]
      }
     ]
    }
    WBP_Pilgrim_AccuracyIcon {
     id=#7
     possessed_class=WBP_Pilgrim_AccuracyIcon_C
     tracks [
      {
       type=MovieSceneFloatTrack
       display_name=Render Opacity
       sections [{type=MovieSceneFloatSection, end_frame=100, start_frame=0, channels [None {type=MovieSceneScriptingFloatChannel, keys [[0,0],[5,1]]}]}]
      }
      {
       type=MovieScene2DTransformTrack
       sections [{
         type=MovieScene2DTransformSection
         end_frame=100
         start_frame=0
         channels [
          Translation.X {type=MovieSceneScriptingFloatChannel, const [0,2]}
          Translation.Y {type=MovieSceneScriptingFloatChannel, const [0,1]}
          Angle {type=MovieSceneScriptingFloatChannel, const [0,1]}
          Scale.X {type=MovieSceneScriptingFloatChannel, const [1,1]}
          Scale.Y {type=MovieSceneScriptingFloatChannel, const [1,1]}
          Shear.X {type=MovieSceneScriptingFloatChannel, const [0,2]}
          Shear.Y {type=MovieSceneScriptingFloatChannel, const [0,2]}
         ]
        }]
      }
     ]
    }
    WBP_PilgrimAccuracyText_Instance_C_29 {
     id=#8
     possessed_class=WBP_PilgrimAccuracyText_Instance_C
     tracks [
      {
       type=MovieSceneFloatTrack
       display_name=Render Opacity
       sections [{
         type=MovieSceneFloatSection
         end_frame=100
         start_frame=0
         channels [None {type=MovieSceneScriptingFloatChannel, keys [[0,0],[5,1],[50,1],[91,0]]}]
        }]
      }
      {
       type=MovieScene2DTransformTrack
       sections [{
         type=MovieScene2DTransformSection
         end_frame=100
         start_frame=0
         channels [
          Translation.X {type=MovieSceneScriptingFloatChannel, keys [[0,0],[25,0],[100,-100]]}
          Translation.Y {type=MovieSceneScriptingFloatChannel, keys [[0,0],[25,0],[100,100]]}
          Angle {type=MovieSceneScriptingFloatChannel, const [0,1]}
          Scale.X {type=MovieSceneScriptingFloatChannel, const [1,1]}
          Scale.Y {type=MovieSceneScriptingFloatChannel, const [1,1]}
          Shear.X {type=MovieSceneScriptingFloatChannel, const [0,1]}
          Shear.Y {type=MovieSceneScriptingFloatChannel, const [0,1]}
         ]
        }]
      }
      {
       type=MovieSceneByteTrack
       display_name=Visibility
       sections [{type=MovieSceneByteSection, end_frame=100, start_frame=0, channels [None {type=MovieSceneScriptingByteChannel, keys [[0,3],[100,2]]}]}]
      }
     ]
    }
   ]
   display_rate {numerator=100}
  }
 }
 FadeIn_NoCascade {
  binding_count=1
  length_seconds=1
  object=&self:FadeIn_NoCascade
  sequence {bindings [WBP_PilgrimAccuracyText_Instance_C_81 {
     id=#2
     possessed_class=WBP_PilgrimAccuracyText_Instance_C
     tracks [
      {
       type=MovieSceneFloatTrack
       display_name=Render Opacity
       sections [{type=MovieSceneFloatSection, end_frame=20, start_frame=0, channels [None {type=MovieSceneScriptingFloatChannel, keys [[0,1],[20,0]]}]}]
      }
      {
       type=MovieSceneByteTrack
       display_name=Visibility
       sections [{type=MovieSceneByteSection, end_frame=0, start_frame=0, channels [None {type=MovieSceneScriptingByteChannel, const [4,1]}]}]
      }
      {
       type=MovieScene2DTransformTrack
       sections [{
         type=MovieScene2DTransformSection
         end_frame=20
         start_frame=0
         channels [
          Translation.X {type=MovieSceneScriptingFloatChannel, const [0,1]}
          Translation.Y {type=MovieSceneScriptingFloatChannel, const [0,1]}
          Angle {type=MovieSceneScriptingFloatChannel, const [0,1]}
          Scale.X {type=MovieSceneScriptingFloatChannel, keys [[0,1],[1,1.1],[20,1]]}
          Scale.Y {type=MovieSceneScriptingFloatChannel, keys [[0,1],[1,1.1],[20,1]]}
          Shear.X {type=MovieSceneScriptingFloatChannel, const [0,1]}
          Shear.Y {type=MovieSceneScriptingFloatChannel, const [0,1]}
         ]
        }]
      }
      {
       type=MovieSceneIntegerTrack
       display_name=Grid Slot Layer
       sections [{
         type=MovieSceneIntegerSection
         end_frame=20
         start_frame=0
         channels [None {type=MovieSceneScriptingIntegerChannel, thinned_from=11, keys [[0,10],[20,0]]}]
        }]
      }
     ]
    }]}
 }
 Late_old {
  binding_count=6
  length_seconds=1
  object=&self:Late_old
  sequence {
   bindings [
    Background {
     id=#3
     possessed_class=Image
     tracks [
      {
       type=MovieSceneWidgetMaterialTrack
       sections [{
         type=MovieSceneParameterSection
         end_frame=100
         start_frame=0
         channels [OverdriveActive {type=MovieSceneScriptingFloatChannel, keys [[26.23,6],[73.77,0.8],[100,1]]}]
        }]
      }
      {
       type=MovieSceneColorTrack
       display_name=Color and Opacity
       sections [{
         type=MovieSceneColorSection
         end_frame=29
         start_frame=0
         channels [
          Color.R {type=MovieSceneScriptingFloatChannel, const [1,1]}
          Color.G {type=MovieSceneScriptingFloatChannel, const [1,1]}
          Color.B {type=MovieSceneScriptingFloatChannel, const [1,1]}
          Color.A {type=MovieSceneScriptingFloatChannel, keys [[0,0],[29.5083,1]]}
         ]
        }]
      }
      {
       type=MovieScene2DTransformTrack
       sections [{
         type=MovieScene2DTransformSection
         end_frame=98
         start_frame=0
         channels [
          Translation.X {type=MovieSceneScriptingFloatChannel, const [0,1]}
          Translation.Y {type=MovieSceneScriptingFloatChannel, const [0,1]}
          Angle {type=MovieSceneScriptingFloatChannel, const [0,1]}
          Scale.X {type=MovieSceneScriptingFloatChannel, keys [[0,-0.2],[32.7867,-1]]}
          Scale.Y {type=MovieSceneScriptingFloatChannel, keys [[0,0],[32.7867,1.0948],[65.5733,1]]}
          Shear.X {type=MovieSceneScriptingFloatChannel, const [-15,1]}
          Shear.Y {type=MovieSceneScriptingFloatChannel, const [0,1]}
         ]
        }]
      }
     ]
    }
    BG_Echo {
     id=#4
     possessed_class=Image
     tracks [
      {
       type=MovieSceneColorTrack
       display_name=Color and Opacity
       sections [{
         type=MovieSceneColorSection
         end_frame=100
         start_frame=0
         channels [
          Color.R {type=MovieSceneScriptingFloatChannel, const [1,1]}
          Color.G {type=MovieSceneScriptingFloatChannel, const [1,1]}
          Color.B {type=MovieSceneScriptingFloatChannel, const [1,1]}
          Color.A {type=MovieSceneScriptingFloatChannel, keys [[0,0],[6.6667,1]]}
         ]
        }]
      }
      {
       type=MovieScene2DTransformTrack
       sections [{
         type=MovieScene2DTransformSection
         end_frame=100
         start_frame=0
         channels [
          Translation.X {type=MovieSceneScriptingFloatChannel, const [5,1]}
          Translation.Y {type=MovieSceneScriptingFloatChannel, const [0,1]}
          Angle {type=MovieSceneScriptingFloatChannel, const [0,1]}
          Scale.X {type=MovieSceneScriptingFloatChannel, const [-1,1]}
          Scale.Y {type=MovieSceneScriptingFloatChannel, keys [[0,1],[13.3333,1.2],[100,1]]}
          Shear.X {type=MovieSceneScriptingFloatChannel, const [-15,1]}
          Shear.Y {type=MovieSceneScriptingFloatChannel, const [0,1]}
         ]
        }]
      }
     ]
    }
    PerfectAccuracyText {
     id=#5
     possessed_class=CommonTextBlock
     tracks [
      {
       type=MovieSceneColorTrack
       display_name=Color and Opacity
       sections [{
         type=MovieSceneColorSection
         end_frame=50
         start_frame=0
         channels [
          Color.R {type=MovieSceneScriptingFloatChannel, const [1,2]}
          Color.G {type=MovieSceneScriptingFloatChannel, keys [[0,0.18],[38.8883,1]]}
          Color.B {type=MovieSceneScriptingFloatChannel, keys [[0,-0.04],[33.3333,1]]}
          Color.A {type=MovieSceneScriptingFloatChannel, keys [[0,0],[15,1]]}
         ]
        }]
      }
      {
       type=MovieSceneObjectPropertyTrack
       display_name="Font Material (Font)"
       sections [{
         type=MovieSceneObjectPropertySection
         end_frame=50
         start_frame=50
         channels [None {type=MovieSceneScriptingObjectPathChannel, const [null,1]}]
        }]
      }
     ]
    }
    PerfectIcon {
     id=#6
     possessed_class=FortLazyImage
     tracks [
      {
       type=MovieScene2DTransformTrack
       sections [{
         type=MovieScene2DTransformSection
         end_frame=75
         start_frame=0
         channels [
          Translation.X {type=MovieSceneScriptingFloatChannel, const [0,1]}
          Translation.Y {type=MovieSceneScriptingFloatChannel, const [0,1]}
          Angle {type=MovieSceneScriptingFloatChannel, const [0,1]}
          Scale.X {type=MovieSceneScriptingFloatChannel, keys [[0,1],[10,1.1],[25,1]]}
          Scale.Y {type=MovieSceneScriptingFloatChannel, keys [[0,1],[10,1.1],[30,1]]}
          Shear.X {type=MovieSceneScriptingFloatChannel, const [0,1]}
          Shear.Y {type=MovieSceneScriptingFloatChannel, const [0,1]}
         ]
        }]
      }
      {
       type=MovieSceneColorTrack
       display_name=Color and Opacity
       sections [{
         type=MovieSceneColorSection
         end_frame=75
         start_frame=0
         channels [
          Color.R {type=MovieSceneScriptingFloatChannel, const [1,1]}
          Color.G {type=MovieSceneScriptingFloatChannel, const [1,1]}
          Color.B {type=MovieSceneScriptingFloatChannel, const [1,1]}
          Color.A {type=MovieSceneScriptingFloatChannel, keys [[0,0],[30,1]]}
         ]
        }]
      }
     ]
    }
    WBP_Pilgrim_AccuracyIcon {
     id=#7
     possessed_class=WBP_Pilgrim_AccuracyIcon_C
     tracks [{
       type=MovieSceneFloatTrack
       display_name=Render Opacity
       sections [{type=MovieSceneFloatSection, end_frame=100, start_frame=0, channels [None {type=MovieSceneScriptingFloatChannel, keys [[0,0],[100,1]]}]}]
      }]
    }
    WBP_PilgrimAccuracyText_Instance_C_29 {
     id=#8
     possessed_class=WBP_PilgrimAccuracyText_Instance_C
     tracks [
      {
       type=MovieSceneFloatTrack
       display_name=Render Opacity
       sections [{
         type=MovieSceneFloatSection
         end_frame=100
         start_frame=0
         channels [None {type=MovieSceneScriptingFloatChannel, keys [[0,0],[19.8683,1],[33.1117,1],[100,0]]}]
        }]
      }
      {
       type=MovieScene2DTransformTrack
       sections [{
         type=MovieScene2DTransformSection
         end_frame=100
         start_frame=0
         channels [
          Translation.X {type=MovieSceneScriptingFloatChannel, default=0, keys [[0,-20],[34,-20],[100,-200]]}
          Translation.Y {type=MovieSceneScriptingFloatChannel, keys [[0,60],[34,60],[100,110]]}
          Angle {type=MovieSceneScriptingFloatChannel, const [0,1]}
          Scale.X {type=MovieSceneScriptingFloatChannel, const [1,1]}
          Scale.Y {type=MovieSceneScriptingFloatChannel, const [1,1]}
          Shear.X {type=MovieSceneScriptingFloatChannel, const [0,1]}
          Shear.Y {type=MovieSceneScriptingFloatChannel, const [0,1]}
         ]
        }]
      }
     ]
    }
   ]
   display_rate {numerator=100}
  }
 }
 Early_Horizontal {
  binding_count=6
  length_seconds=0.5
  object=&self:Early_Horizontal
  sequence {
   bindings [
    Background {
     id=#3
     possessed_class=Image
     tracks [
      {
       type=MovieSceneColorTrack
       display_name=Color and Opacity
       sections [{
         type=MovieSceneColorSection
         end_frame=15
         start_frame=0
         channels [
          Color.R {type=MovieSceneScriptingFloatChannel, const [1,1]}
          Color.G {type=MovieSceneScriptingFloatChannel, const [1,1]}
          Color.B {type=MovieSceneScriptingFloatChannel, const [1,1]}
          Color.A {type=MovieSceneScriptingFloatChannel, const [1,2]}
         ]
        }]
      }
      {
       type=MovieSceneWidgetMaterialTrack
       sections [{
         type=MovieSceneParameterSection
         end_frame=50
         start_frame=0
         channels [
          Gradient_Color_BR {type=MovieSceneScriptingFloatChannel, keys [[0,0],[4.5,1],[20,0]]}
          Gradient_Color_BG {type=MovieSceneScriptingFloatChannel, keys [[0,0.5779],[4,1],[19,0.5779]]}
          Gradient_Color_BB {type=MovieSceneScriptingFloatChannel, keys [[0,0.8],[3.5,1],[18,0.8]]}
          Gradient_Color_BA {type=MovieSceneScriptingFloatChannel, const [1,2]}
         ]
        }]
      }
     ]
    }
    BG_Echo {
     id=#4
     possessed_class=Image
     tracks [
      {
       type=MovieSceneColorTrack
       display_name=Color and Opacity
       sections [{
         type=MovieSceneColorSection
         end_frame=50
         start_frame=0
         channels [
          Color.R {type=MovieSceneScriptingFloatChannel, const [1,1]}
          Color.G {type=MovieSceneScriptingFloatChannel, const [1,1]}
          Color.B {type=MovieSceneScriptingFloatChannel, const [1,1]}
          Color.A {type=MovieSceneScriptingFloatChannel, const [1,2]}
         ]
        }]
      }
      {
       type=MovieScene2DTransformTrack
       sections [{
         type=MovieScene2DTransformSection
         end_frame=50
         start_frame=0
         channels [
          Scale.X {type=MovieSceneScriptingFloatChannel, const [-1,1]}
          Scale.Y {type=MovieSceneScriptingFloatChannel, keys [[0,1],[6.6667,1.2],[50,1]]}
         ]
        }]
      }
     ]
    }
    PerfectAccuracyText {
     id=#5
     possessed_class=CommonTextBlock
     tracks [
      {
       type=MovieSceneColorTrack
       display_name=Color and Opacity
       sections [{
         type=MovieSceneColorSection
         end_frame=25
         start_frame=0
         channels [
          Color.R {type=MovieSceneScriptingFloatChannel, const [1,2]}
          Color.G {type=MovieSceneScriptingFloatChannel, const [1,2]}
          Color.B {type=MovieSceneScriptingFloatChannel, const [1,2]}
          Color.A {type=MovieSceneScriptingFloatChannel, const [1,2]}
         ]
        }]
      }
      {
       type=MovieScene2DTransformTrack
       sections [{
         type=MovieScene2DTransformSection
         end_frame=50
         start_frame=0
         channels [
          Translation.X {type=MovieSceneScriptingFloatChannel, keys [[0,0],[2.5,0],[50,-10]]}
          Translation.Y {type=MovieSceneScriptingFloatChannel, const [0,1]}
          Angle {type=MovieSceneScriptingFloatChannel, const [0,1]}
          Scale.X {type=MovieSceneScriptingFloatChannel, const [1,1]}
          Scale.Y {type=MovieSceneScriptingFloatChannel, const [1,1]}
          Shear.X {type=MovieSceneScriptingFloatChannel, const [0,1]}
          Shear.Y {type=MovieSceneScriptingFloatChannel, const [0,1]}
         ]
        }]
      }
     ]
    }
    PerfectIcon {
     id=#6
     possessed_class=FortLazyImage
     tracks [
      {
       type=MovieSceneColorTrack
       display_name=Color and Opacity
       sections [{
         type=MovieSceneColorSection
         end_frame=50
         start_frame=0
         channels [
          Color.R {type=MovieSceneScriptingFloatChannel, const [1,1]}
          Color.G {type=MovieSceneScriptingFloatChannel, const [1,1]}
          Color.B {type=MovieSceneScriptingFloatChannel, const [1,1]}
          Color.A {type=MovieSceneScriptingFloatChannel, const [1,2]}
         ]
        }]
      }
      {
       type=MovieScene2DTransformTrack
       sections [{
         type=MovieScene2DTransformSection
         end_frame=50
         start_frame=0
         channels [
          Translation.X {type=MovieSceneScriptingFloatChannel, keys [[0,0],[15,0],[46.5,20]]}
          Translation.Y {type=MovieSceneScriptingFloatChannel, const [0,1]}
          Angle {type=MovieSceneScriptingFloatChannel, const [0,1]}
          Scale.X {type=MovieSceneScriptingFloatChannel, default=1}
          Scale.Y {type=MovieSceneScriptingFloatChannel, default=1}
          Shear.X {type=MovieSceneScriptingFloatChannel, const [0,1]}
          Shear.Y {type=MovieSceneScriptingFloatChannel, const [0,1]}
         ]
        }]
      }
     ]
    }
    WBP_Pilgrim_AccuracyIcon {
     id=#7
     possessed_class=WBP_Pilgrim_AccuracyIcon_C
     tracks [
      {
       type=MovieSceneFloatTrack
       display_name=Render Opacity
       sections [{type=MovieSceneFloatSection, end_frame=50, start_frame=0, channels [None {type=MovieSceneScriptingFloatChannel, const [1,1]}]}]
      }
      {
       type=MovieScene2DTransformTrack
       sections [{
         type=MovieScene2DTransformSection
         end_frame=50
         start_frame=0
         channels [
          Translation.X {type=MovieSceneScriptingFloatChannel, const [0,2]}
          Translation.Y {type=MovieSceneScriptingFloatChannel, const [0,1]}
          Angle {type=MovieSceneScriptingFloatChannel, const [0,1]}
          Scale.X {type=MovieSceneScriptingFloatChannel, const [1,1]}
          Scale.Y {type=MovieSceneScriptingFloatChannel, const [1,1]}
          Shear.X {type=MovieSceneScriptingFloatChannel, const [0,2]}
          Shear.Y {type=MovieSceneScriptingFloatChannel, const [0,2]}
         ]
        }]
      }
     ]
    }
    WBP_PilgrimAccuracyText_Instance_C_29 {
     id=#8
     possessed_class=WBP_PilgrimAccuracyText_Instance_C
     tracks [
      {
       type=MovieSceneFloatTrack
       display_name=Render Opacity
       sections [{
         type=MovieSceneFloatSection
         end_frame=50
         start_frame=0
         channels [None {type=MovieSceneScriptingFloatChannel, keys [[0,1],[25,1],[45.5,0]]}]
        }]
      }
      {
       type=MovieScene2DTransformTrack
       sections [{
         type=MovieScene2DTransformSection
         end_frame=50
         start_frame=0
         channels [
          Translation.X {type=MovieSceneScriptingFloatChannel, keys [[0,0],[50,-100]]}
          Translation.Y {type=MovieSceneScriptingFloatChannel, const [0,2]}
          Angle {type=MovieSceneScriptingFloatChannel, const [0,1]}
          Scale.X {type=MovieSceneScriptingFloatChannel, const [1,1]}
          Scale.Y {type=MovieSceneScriptingFloatChannel, const [1,1]}
          Shear.X {type=MovieSceneScriptingFloatChannel, const [0,1]}
          Shear.Y {type=MovieSceneScriptingFloatChannel, const [0,1]}
         ]
        }]
      }
      {
       type=MovieSceneByteTrack
       display_name=Visibility
       sections [{type=MovieSceneByteSection, end_frame=50, start_frame=0, channels [None {type=MovieSceneScriptingByteChannel, keys [[0,3],[50,2]]}]}]
      }
     ]
    }
   ]
   display_rate {numerator=100}
  }
 }
 Late_Horizontal {
  binding_count=6
  length_seconds=0.5
  object=&self:Late_Horizontal
  sequence {
   bindings [
    Background {
     id=#3
     possessed_class=Image
     tracks [
      {
       type=MovieSceneColorTrack
       display_name=Color and Opacity
       sections [{
         type=MovieSceneColorSection
         end_frame=15
         start_frame=0
         channels [
          Color.R {type=MovieSceneScriptingFloatChannel, const [1,1]}
          Color.G {type=MovieSceneScriptingFloatChannel, const [1,1]}
          Color.B {type=MovieSceneScriptingFloatChannel, const [1,1]}
          Color.A {type=MovieSceneScriptingFloatChannel, const [1,2]}
         ]
        }]
      }
      {
       type=MovieSceneWidgetMaterialTrack
       sections [{
         type=MovieSceneParameterSection
         end_frame=50
         start_frame=0
         channels [
          Gradient_Color_BR {type=MovieSceneScriptingFloatChannel, keys [[0,0],[4.5,1],[20,0]]}
          Gradient_Color_BG {type=MovieSceneScriptingFloatChannel, keys [[0,0.5779],[4,1],[19,0.5779]]}
          Gradient_Color_BB {type=MovieSceneScriptingFloatChannel, keys [[0,0.8],[3.5,1],[18,0.8]]}
          Gradient_Color_BA {type=MovieSceneScriptingFloatChannel, const [1,2]}
         ]
        }]
      }
     ]
    }
    BG_Echo {
     id=#4
     possessed_class=Image
     tracks [
      {
       type=MovieSceneColorTrack
       display_name=Color and Opacity
       sections [{
         type=MovieSceneColorSection
         end_frame=50
         start_frame=0
         channels [
          Color.R {type=MovieSceneScriptingFloatChannel, const [1,1]}
          Color.G {type=MovieSceneScriptingFloatChannel, const [1,1]}
          Color.B {type=MovieSceneScriptingFloatChannel, const [1,1]}
          Color.A {type=MovieSceneScriptingFloatChannel, const [1,2]}
         ]
        }]
      }
      {
       type=MovieScene2DTransformTrack
       sections [{
         type=MovieScene2DTransformSection
         end_frame=50
         start_frame=0
         channels [
          Scale.X {type=MovieSceneScriptingFloatChannel, const [-1,1]}
          Scale.Y {type=MovieSceneScriptingFloatChannel, keys [[0,1],[6.6667,1.2],[50,1]]}
         ]
        }]
      }
     ]
    }
    PerfectAccuracyText {
     id=#5
     possessed_class=CommonTextBlock
     tracks [
      {
       type=MovieSceneColorTrack
       display_name=Color and Opacity
       sections [{
         type=MovieSceneColorSection
         end_frame=25
         start_frame=0
         channels [
          Color.R {type=MovieSceneScriptingFloatChannel, const [1,2]}
          Color.G {type=MovieSceneScriptingFloatChannel, const [1,2]}
          Color.B {type=MovieSceneScriptingFloatChannel, const [1,2]}
          Color.A {type=MovieSceneScriptingFloatChannel, const [1,2]}
         ]
        }]
      }
      {
       type=MovieScene2DTransformTrack
       sections [{
         type=MovieScene2DTransformSection
         end_frame=50
         start_frame=0
         channels [
          Translation.X {type=MovieSceneScriptingFloatChannel, keys [[0,0],[2.5,0],[50,-10]]}
          Translation.Y {type=MovieSceneScriptingFloatChannel, const [0,1]}
          Angle {type=MovieSceneScriptingFloatChannel, const [0,1]}
          Scale.X {type=MovieSceneScriptingFloatChannel, const [1,1]}
          Scale.Y {type=MovieSceneScriptingFloatChannel, const [1,1]}
          Shear.X {type=MovieSceneScriptingFloatChannel, const [0,1]}
          Shear.Y {type=MovieSceneScriptingFloatChannel, const [0,1]}
         ]
        }]
      }
     ]
    }
    PerfectIcon {
     id=#6
     possessed_class=FortLazyImage
     tracks [
      {
       type=MovieSceneColorTrack
       display_name=Color and Opacity
       sections [{
         type=MovieSceneColorSection
         end_frame=50
         start_frame=0
         channels [
          Color.R {type=MovieSceneScriptingFloatChannel, const [1,1]}
          Color.G {type=MovieSceneScriptingFloatChannel, const [1,1]}
          Color.B {type=MovieSceneScriptingFloatChannel, const [1,1]}
          Color.A {type=MovieSceneScriptingFloatChannel, const [1,2]}
         ]
        }]
      }
      {
       type=MovieScene2DTransformTrack
       sections [{
         type=MovieScene2DTransformSection
         end_frame=50
         start_frame=0
         channels [
          Translation.X {type=MovieSceneScriptingFloatChannel, keys [[0,0],[15,0],[46.5,20]]}
          Translation.Y {type=MovieSceneScriptingFloatChannel, const [0,1]}
          Angle {type=MovieSceneScriptingFloatChannel, const [0,1]}
          Scale.X {type=MovieSceneScriptingFloatChannel, default=1}
          Scale.Y {type=MovieSceneScriptingFloatChannel, default=1}
          Shear.X {type=MovieSceneScriptingFloatChannel, const [0,1]}
          Shear.Y {type=MovieSceneScriptingFloatChannel, const [0,1]}
         ]
        }]
      }
     ]
    }
    WBP_Pilgrim_AccuracyIcon {
     id=#7
     possessed_class=WBP_Pilgrim_AccuracyIcon_C
     tracks [
      {
       type=MovieSceneFloatTrack
       display_name=Render Opacity
       sections [{type=MovieSceneFloatSection, end_frame=50, start_frame=0, channels [None {type=MovieSceneScriptingFloatChannel, const [1,2]}]}]
      }
      {
       type=MovieScene2DTransformTrack
       sections [{
         type=MovieScene2DTransformSection
         end_frame=50
         start_frame=0
         channels [
          Translation.X {type=MovieSceneScriptingFloatChannel, const [0,2]}
          Translation.Y {type=MovieSceneScriptingFloatChannel, const [0,1]}
          Angle {type=MovieSceneScriptingFloatChannel, const [0,1]}
          Scale.X {type=MovieSceneScriptingFloatChannel, const [1,1]}
          Scale.Y {type=MovieSceneScriptingFloatChannel, const [1,1]}
          Shear.X {type=MovieSceneScriptingFloatChannel, const [0,2]}
          Shear.Y {type=MovieSceneScriptingFloatChannel, const [0,2]}
         ]
        }]
      }
     ]
    }
    WBP_PilgrimAccuracyText_Instance_C_29 {
     id=#8
     possessed_class=WBP_PilgrimAccuracyText_Instance_C
     tracks [
      {
       type=MovieSceneFloatTrack
       display_name=Render Opacity
       sections [{
         type=MovieSceneFloatSection
         end_frame=50
         start_frame=0
         channels [None {type=MovieSceneScriptingFloatChannel, keys [[0,1],[25,1],[45.5,0]]}]
        }]
      }
      {
       type=MovieScene2DTransformTrack
       sections [{
         type=MovieScene2DTransformSection
         end_frame=50
         start_frame=0
         channels [
          Translation.X {type=MovieSceneScriptingFloatChannel, keys [[0,0],[50,-100]]}
          Translation.Y {type=MovieSceneScriptingFloatChannel, const [0,2]}
          Angle {type=MovieSceneScriptingFloatChannel, const [0,1]}
          Scale.X {type=MovieSceneScriptingFloatChannel, const [1,1]}
          Scale.Y {type=MovieSceneScriptingFloatChannel, const [1,1]}
          Shear.X {type=MovieSceneScriptingFloatChannel, const [0,1]}
          Shear.Y {type=MovieSceneScriptingFloatChannel, const [0,1]}
         ]
        }]
      }
      {
       type=MovieSceneByteTrack
       display_name=Visibility
       sections [{type=MovieSceneByteSection, end_frame=50, start_frame=0, channels [None {type=MovieSceneScriptingByteChannel, keys [[0,3],[50,2]]}]}]
      }
     ]
    }
   ]
   display_rate {numerator=100}
  }
 }
 Late_LiveMimic {
  binding_count=6
  length_seconds=1
  object=&self:Late_LiveMimic
  sequence {
   bindings [
    Background {
     id=#3
     possessed_class=Image
     tracks [
      {
       type=MovieSceneColorTrack
       display_name=Color and Opacity
       sections [{
         type=MovieSceneColorSection
         end_frame=9
         start_frame=0
         channels [
          Color.R {type=MovieSceneScriptingFloatChannel, const [1,1]}
          Color.G {type=MovieSceneScriptingFloatChannel, const [1,1]}
          Color.B {type=MovieSceneScriptingFloatChannel, const [1,1]}
          Color.A {type=MovieSceneScriptingFloatChannel, const [1,2]}
         ]
        }]
      }
      {
       type=MovieSceneWidgetMaterialTrack
       sections [{
         type=MovieSceneParameterSection
         end_frame=30
         start_frame=0
         channels [
          Gradient_Color_BR {type=MovieSceneScriptingFloatChannel, keys [[0,0],[2.7,1],[12,0]]}
          Gradient_Color_BG {type=MovieSceneScriptingFloatChannel, keys [[0,0.5779],[2.4,1],[11.4,0.5779]]}
          Gradient_Color_BB {type=MovieSceneScriptingFloatChannel, keys [[0,0.8],[2.1,1],[10.8,0.8]]}
          Gradient_Color_BA {type=MovieSceneScriptingFloatChannel, const [1,2]}
         ]
        }]
      }
     ]
    }
    BG_Echo {
     id=#4
     possessed_class=Image
     tracks [
      {
       type=MovieSceneColorTrack
       display_name=Color and Opacity
       sections [{
         type=MovieSceneColorSection
         end_frame=30
         start_frame=0
         channels [
          Color.R {type=MovieSceneScriptingFloatChannel, const [1,1]}
          Color.G {type=MovieSceneScriptingFloatChannel, const [1,1]}
          Color.B {type=MovieSceneScriptingFloatChannel, const [1,1]}
          Color.A {type=MovieSceneScriptingFloatChannel, const [1,2]}
         ]
        }]
      }
      {
       type=MovieScene2DTransformTrack
       sections [{
         type=MovieScene2DTransformSection
         end_frame=30
         start_frame=0
         channels [
          Scale.X {type=MovieSceneScriptingFloatChannel, const [-1,1]}
          Scale.Y {type=MovieSceneScriptingFloatChannel, keys [[0,1],[4,1.2],[30,1]]}
         ]
        }]
      }
     ]
    }
    PerfectAccuracyText {
     id=#5
     possessed_class=CommonTextBlock
     tracks [
      {
       type=MovieSceneColorTrack
       display_name=Color and Opacity
       sections [{
         type=MovieSceneColorSection
         end_frame=15
         start_frame=0
         channels [
          Color.R {type=MovieSceneScriptingFloatChannel, const [1,2]}
          Color.G {type=MovieSceneScriptingFloatChannel, const [1,2]}
          Color.B {type=MovieSceneScriptingFloatChannel, const [1,2]}
          Color.A {type=MovieSceneScriptingFloatChannel, const [1,2]}
         ]
        }]
      }
      {
       type=MovieScene2DTransformTrack
       sections [{
         type=MovieScene2DTransformSection
         end_frame=30
         start_frame=0
         channels [
          Translation.X {type=MovieSceneScriptingFloatChannel, const [0,1]}
          Translation.Y {type=MovieSceneScriptingFloatChannel, const [0,1]}
          Angle {type=MovieSceneScriptingFloatChannel, const [0,1]}
          Scale.X {type=MovieSceneScriptingFloatChannel, const [1,1]}
          Scale.Y {type=MovieSceneScriptingFloatChannel, const [1,1]}
          Shear.X {type=MovieSceneScriptingFloatChannel, const [0,1]}
          Shear.Y {type=MovieSceneScriptingFloatChannel, const [0,1]}
         ]
        }]
      }
     ]
    }
    PerfectIcon {
     id=#6
     possessed_class=FortLazyImage
     tracks [
      {
       type=MovieSceneColorTrack
       display_name=Color and Opacity
       sections [{
         type=MovieSceneColorSection
         end_frame=30
         start_frame=0
         channels [
          Color.R {type=MovieSceneScriptingFloatChannel, const [1,1]}
          Color.G {type=MovieSceneScriptingFloatChannel, const [1,1]}
          Color.B {type=MovieSceneScriptingFloatChannel, const [1,1]}
          Color.A {type=MovieSceneScriptingFloatChannel, const [1,2]}
         ]
        }]
      }
      {
       type=MovieScene2DTransformTrack
       sections [{
         type=MovieScene2DTransformSection
         end_frame=30
         start_frame=0
         channels [
          Translation.X {type=MovieSceneScriptingFloatChannel, const [0,1]}
          Translation.Y {type=MovieSceneScriptingFloatChannel, const [0,1]}
          Angle {type=MovieSceneScriptingFloatChannel, const [0,1]}
          Scale.X {type=MovieSceneScriptingFloatChannel, default=1}
          Scale.Y {type=MovieSceneScriptingFloatChannel, default=1}
          Shear.X {type=MovieSceneScriptingFloatChannel, const [0,1]}
          Shear.Y {type=MovieSceneScriptingFloatChannel, const [0,1]}
         ]
        }]
      }
     ]
    }
    WBP_Pilgrim_AccuracyIcon {
     id=#7
     possessed_class=WBP_Pilgrim_AccuracyIcon_C
     tracks [
      {
       type=MovieSceneFloatTrack
       display_name=Render Opacity
       sections [{type=MovieSceneFloatSection, end_frame=30, start_frame=0, channels [None {type=MovieSceneScriptingFloatChannel, const [1,2]}]}]
      }
      {
       type=MovieScene2DTransformTrack
       sections [{
         type=MovieScene2DTransformSection
         end_frame=30
         start_frame=0
         channels [
          Translation.X {type=MovieSceneScriptingFloatChannel, const [0,2]}
          Translation.Y {type=MovieSceneScriptingFloatChannel, const [0,1]}
          Angle {type=MovieSceneScriptingFloatChannel, const [0,1]}
          Scale.X {type=MovieSceneScriptingFloatChannel, const [1,1]}
          Scale.Y {type=MovieSceneScriptingFloatChannel, const [1,1]}
          Shear.X {type=MovieSceneScriptingFloatChannel, const [0,2]}
          Shear.Y {type=MovieSceneScriptingFloatChannel, const [0,2]}
         ]
        }]
      }
     ]
    }
    WBP_PilgrimAccuracyText_Instance_C_29 {
     id=#8
     possessed_class=WBP_PilgrimAccuracyText_Instance_C
     tracks [
      {
       type=MovieSceneFloatTrack
       display_name=Render Opacity
       sections [{
         type=MovieSceneFloatSection
         end_frame=30
         start_frame=0
         channels [None {type=MovieSceneScriptingFloatChannel, keys [[0,1],[15,1],[27.3,0]]}]
        }]
      }
      {
       type=MovieScene2DTransformTrack
       sections [{
         type=MovieScene2DTransformSection
         end_frame=30
         start_frame=0
         channels [
          Translation.X {type=MovieSceneScriptingFloatChannel, const [0,2]}
          Translation.Y {type=MovieSceneScriptingFloatChannel, const [0,2]}
          Angle {type=MovieSceneScriptingFloatChannel, const [0,1]}
          Scale.X {type=MovieSceneScriptingFloatChannel, keys [[0,1],[2,1.1],[30,1]]}
          Scale.Y {type=MovieSceneScriptingFloatChannel, keys [[0,1],[2,1.1],[30,1]]}
          Shear.X {type=MovieSceneScriptingFloatChannel, const [0,1]}
          Shear.Y {type=MovieSceneScriptingFloatChannel, const [0,1]}
         ]
        }]
      }
      {
       type=MovieSceneByteTrack
       display_name=Visibility
       sections [{type=MovieSceneByteSection, end_frame=30, start_frame=0, channels [None {type=MovieSceneScriptingByteChannel, keys [[0,3],[30,2]]}]}]
      }
     ]
    }
   ]
   display_rate {numerator=30}
  }
 }
 Early_LiveMimic {
  binding_count=6
  length_seconds=1
  object=&self:Early_LiveMimic
  sequence {
   bindings [
    Background {
     id=#3
     possessed_class=Image
     tracks [
      {
       type=MovieSceneColorTrack
       display_name=Color and Opacity
       sections [{
         type=MovieSceneColorSection
         end_frame=9
         start_frame=0
         channels [
          Color.R {type=MovieSceneScriptingFloatChannel, const [1,1]}
          Color.G {type=MovieSceneScriptingFloatChannel, const [1,1]}
          Color.B {type=MovieSceneScriptingFloatChannel, const [1,1]}
          Color.A {type=MovieSceneScriptingFloatChannel, const [1,2]}
         ]
        }]
      }
      {
       type=MovieSceneWidgetMaterialTrack
       sections [{
         type=MovieSceneParameterSection
         end_frame=30
         start_frame=0
         channels [
          Gradient_Color_BR {type=MovieSceneScriptingFloatChannel, keys [[0,0],[2.7,1],[12,0]]}
          Gradient_Color_BG {type=MovieSceneScriptingFloatChannel, keys [[0,0.5779],[2.4,1],[11.4,0.5779]]}
          Gradient_Color_BB {type=MovieSceneScriptingFloatChannel, keys [[0,0.8],[2.1,1],[10.8,0.8]]}
          Gradient_Color_BA {type=MovieSceneScriptingFloatChannel, const [1,2]}
         ]
        }]
      }
     ]
    }
    BG_Echo {
     id=#4
     possessed_class=Image
     tracks [
      {
       type=MovieSceneColorTrack
       display_name=Color and Opacity
       sections [{
         type=MovieSceneColorSection
         end_frame=30
         start_frame=0
         channels [
          Color.R {type=MovieSceneScriptingFloatChannel, const [1,1]}
          Color.G {type=MovieSceneScriptingFloatChannel, const [1,1]}
          Color.B {type=MovieSceneScriptingFloatChannel, const [1,1]}
          Color.A {type=MovieSceneScriptingFloatChannel, const [1,2]}
         ]
        }]
      }
      {
       type=MovieScene2DTransformTrack
       sections [{
         type=MovieScene2DTransformSection
         end_frame=30
         start_frame=0
         channels [
          Scale.X {type=MovieSceneScriptingFloatChannel, const [-1,1]}
          Scale.Y {type=MovieSceneScriptingFloatChannel, keys [[0,1],[4,1.2],[30,1]]}
         ]
        }]
      }
     ]
    }
    PerfectAccuracyText {
     id=#5
     possessed_class=CommonTextBlock
     tracks [
      {
       type=MovieSceneColorTrack
       display_name=Color and Opacity
       sections [{
         type=MovieSceneColorSection
         end_frame=15
         start_frame=0
         channels [
          Color.R {type=MovieSceneScriptingFloatChannel, const [1,2]}
          Color.G {type=MovieSceneScriptingFloatChannel, const [1,2]}
          Color.B {type=MovieSceneScriptingFloatChannel, const [1,2]}
          Color.A {type=MovieSceneScriptingFloatChannel, const [1,2]}
         ]
        }]
      }
      {
       type=MovieScene2DTransformTrack
       sections [{
         type=MovieScene2DTransformSection
         end_frame=30
         start_frame=0
         channels [
          Translation.X {type=MovieSceneScriptingFloatChannel, const [0,1]}
          Translation.Y {type=MovieSceneScriptingFloatChannel, const [0,1]}
          Angle {type=MovieSceneScriptingFloatChannel, const [0,1]}
          Scale.X {type=MovieSceneScriptingFloatChannel, const [1,1]}
          Scale.Y {type=MovieSceneScriptingFloatChannel, const [1,1]}
          Shear.X {type=MovieSceneScriptingFloatChannel, const [0,1]}
          Shear.Y {type=MovieSceneScriptingFloatChannel, const [0,1]}
         ]
        }]
      }
     ]
    }
    PerfectIcon {
     id=#6
     possessed_class=FortLazyImage
     tracks [
      {
       type=MovieSceneColorTrack
       display_name=Color and Opacity
       sections [{
         type=MovieSceneColorSection
         end_frame=30
         start_frame=0
         channels [
          Color.R {type=MovieSceneScriptingFloatChannel, const [1,1]}
          Color.G {type=MovieSceneScriptingFloatChannel, const [1,1]}
          Color.B {type=MovieSceneScriptingFloatChannel, const [1,1]}
          Color.A {type=MovieSceneScriptingFloatChannel, const [1,2]}
         ]
        }]
      }
      {
       type=MovieScene2DTransformTrack
       sections [{
         type=MovieScene2DTransformSection
         end_frame=30
         start_frame=0
         channels [
          Translation.X {type=MovieSceneScriptingFloatChannel, const [0,1]}
          Translation.Y {type=MovieSceneScriptingFloatChannel, const [0,1]}
          Angle {type=MovieSceneScriptingFloatChannel, const [0,1]}
          Scale.X {type=MovieSceneScriptingFloatChannel, default=1}
          Scale.Y {type=MovieSceneScriptingFloatChannel, default=1}
          Shear.X {type=MovieSceneScriptingFloatChannel, const [0,1]}
          Shear.Y {type=MovieSceneScriptingFloatChannel, const [0,1]}
         ]
        }]
      }
     ]
    }
    WBP_Pilgrim_AccuracyIcon {
     id=#7
     possessed_class=WBP_Pilgrim_AccuracyIcon_C
     tracks [
      {
       type=MovieSceneFloatTrack
       display_name=Render Opacity
       sections [{type=MovieSceneFloatSection, end_frame=30, start_frame=0, channels [None {type=MovieSceneScriptingFloatChannel, const [1,1]}]}]
      }
      {
       type=MovieScene2DTransformTrack
       sections [{
         type=MovieScene2DTransformSection
         end_frame=30
         start_frame=0
         channels [
          Translation.X {type=MovieSceneScriptingFloatChannel, const [0,2]}
          Translation.Y {type=MovieSceneScriptingFloatChannel, const [0,1]}
          Angle {type=MovieSceneScriptingFloatChannel, const [0,1]}
          Scale.X {type=MovieSceneScriptingFloatChannel, const [1,1]}
          Scale.Y {type=MovieSceneScriptingFloatChannel, const [1,1]}
          Shear.X {type=MovieSceneScriptingFloatChannel, const [0,2]}
          Shear.Y {type=MovieSceneScriptingFloatChannel, const [0,2]}
         ]
        }]
      }
     ]
    }
    WBP_PilgrimAccuracyText_Instance_C_29 {
     id=#8
     possessed_class=WBP_PilgrimAccuracyText_Instance_C
     tracks [
      {
       type=MovieSceneFloatTrack
       display_name=Render Opacity
       sections [{type=MovieSceneFloatSection, end_frame=30, start_frame=0, channels [None {type=MovieSceneScriptingFloatChannel, keys [[0,1],[30,0]]}]}]
      }
      {
       type=MovieScene2DTransformTrack
       sections [{
         type=MovieScene2DTransformSection
         end_frame=30
         start_frame=0
         channels [
          Translation.X {type=MovieSceneScriptingFloatChannel, const [0,1]}
          Translation.Y {type=MovieSceneScriptingFloatChannel, const [0,1]}
          Angle {type=MovieSceneScriptingFloatChannel, const [0,1]}
          Scale.X {type=MovieSceneScriptingFloatChannel, keys [[0,1],[2,1.1],[30,1]]}
          Scale.Y {type=MovieSceneScriptingFloatChannel, keys [[0,1],[2,1.1],[30,1]]}
          Shear.X {type=MovieSceneScriptingFloatChannel, const [0,1]}
          Shear.Y {type=MovieSceneScriptingFloatChannel, const [0,1]}
         ]
        }]
      }
      {
       type=MovieSceneByteTrack
       display_name=Visibility
       sections [{type=MovieSceneByteSection, end_frame=30, start_frame=0, channels [None {type=MovieSceneScriptingByteChannel, keys [[0,3],[30,2]]}]}]
      }
     ]
    }
   ]
   display_rate {numerator=30}
  }
 }
 Perfect_LiveMimic {
  binding_count=6
  length_seconds=1
  object=&self:Perfect_LiveMimic
  sequence {
   bindings [
    Background {
     id=#3
     possessed_class=Image
     tracks [
      {
       type=MovieSceneWidgetMaterialTrack
       sections [{
         type=MovieSceneParameterSection
         end_frame=30
         start_frame=0
         channels [OverdriveActive {type=MovieSceneScriptingFloatChannel, keys [[0,0],[2.424,6],[13.332,0],[30,1]]}]
        }]
      }
      {
       type=MovieSceneColorTrack
       display_name=Color and Opacity
       sections [{
         type=MovieSceneColorSection
         end_frame=8
         start_frame=0
         channels [
          Color.R {type=MovieSceneScriptingFloatChannel, const [1,1]}
          Color.G {type=MovieSceneScriptingFloatChannel, const [1,1]}
          Color.B {type=MovieSceneScriptingFloatChannel, const [1,1]}
          Color.A {type=MovieSceneScriptingFloatChannel, const [1,2]}
         ]
        }]
      }
     ]
    }
    BG_Echo {
     id=#4
     possessed_class=Image
     tracks [
      {
       type=MovieSceneColorTrack
       display_name=Color and Opacity
       sections [{
         type=MovieSceneColorSection
         end_frame=30
         start_frame=0
         channels [
          Color.R {type=MovieSceneScriptingFloatChannel, const [1,1]}
          Color.G {type=MovieSceneScriptingFloatChannel, const [1,1]}
          Color.B {type=MovieSceneScriptingFloatChannel, const [1,1]}
          Color.A {type=MovieSceneScriptingFloatChannel, const [1,2]}
         ]
        }]
      }
      {
       type=MovieScene2DTransformTrack
       sections [{
         type=MovieScene2DTransformSection
         end_frame=30
         start_frame=0
         channels [
          Scale.X {type=MovieSceneScriptingFloatChannel, keys [[0,0],[7.5,-1],[30,-1]]}
          Scale.Y {type=MovieSceneScriptingFloatChannel, const [1,2]}
         ]
        }]
      }
     ]
    }
    PerfectAccuracyText {
     id=#5
     possessed_class=CommonTextBlock
     tracks [
      {
       type=MovieSceneColorTrack
       display_name=Color and Opacity
       sections [{
         type=MovieSceneColorSection
         end_frame=15
         start_frame=0
         channels [
          Color.R {type=MovieSceneScriptingFloatChannel, const [1,2]}
          Color.G {type=MovieSceneScriptingFloatChannel, const [1,2]}
          Color.B {type=MovieSceneScriptingFloatChannel, const [1,2]}
          Color.A {type=MovieSceneScriptingFloatChannel, const [1,2]}
         ]
        }]
      }
      {
       type=MovieScene2DTransformTrack
       sections [{
         type=MovieScene2DTransformSection
         end_frame=30
         start_frame=0
         channels [
          Translation.X {type=MovieSceneScriptingFloatChannel, const [0,1]}
          Translation.Y {type=MovieSceneScriptingFloatChannel, const [0,1]}
          Angle {type=MovieSceneScriptingFloatChannel, const [0,1]}
          Scale.X {type=MovieSceneScriptingFloatChannel, const [1,1]}
          Scale.Y {type=MovieSceneScriptingFloatChannel, const [1,1]}
          Shear.X {type=MovieSceneScriptingFloatChannel, const [0,1]}
          Shear.Y {type=MovieSceneScriptingFloatChannel, const [0,1]}
         ]
        }]
      }
     ]
    }
    PerfectIcon {
     id=#6
     possessed_class=FortLazyImage
     tracks [
      {
       type=MovieScene2DTransformTrack
       sections [{
         type=MovieScene2DTransformSection
         end_frame=30
         start_frame=0
         channels [
          Translation.X {type=MovieSceneScriptingFloatChannel, const [0,1]}
          Translation.Y {type=MovieSceneScriptingFloatChannel, const [0,1]}
          Angle {type=MovieSceneScriptingFloatChannel, const [0,1]}
          Scale.X {type=MovieSceneScriptingFloatChannel, keys [[0,1],[3,1.2],[7.5,1]]}
          Scale.Y {type=MovieSceneScriptingFloatChannel, keys [[0,1],[3,1.5],[9,1]]}
          Shear.X {type=MovieSceneScriptingFloatChannel, const [0,1]}
          Shear.Y {type=MovieSceneScriptingFloatChannel, const [0,1]}
         ]
        }]
      }
      {
       type=MovieSceneColorTrack
       display_name=Color and Opacity
       sections [{
         type=MovieSceneColorSection
         end_frame=22
         start_frame=0
         channels [
          Color.R {type=MovieSceneScriptingFloatChannel, const [1,1]}
          Color.G {type=MovieSceneScriptingFloatChannel, const [1,1]}
          Color.B {type=MovieSceneScriptingFloatChannel, const [1,1]}
          Color.A {type=MovieSceneScriptingFloatChannel, const [1,2]}
         ]
        }]
      }
     ]
    }
    WBP_Pilgrim_AccuracyIcon {
     id=#7
     possessed_class=WBP_Pilgrim_AccuracyIcon_C
     tracks [
      {
       type=MovieSceneFloatTrack
       display_name=Render Opacity
       sections [{type=MovieSceneFloatSection, end_frame=30, start_frame=0, channels [None {type=MovieSceneScriptingFloatChannel, const [1,2]}]}]
      }
      {
       type=MovieScene2DTransformTrack
       sections [{
         type=MovieScene2DTransformSection
         end_frame=30
         start_frame=0
         channels [
          Translation.X {type=MovieSceneScriptingFloatChannel, const [0,1]}
          Translation.Y {type=MovieSceneScriptingFloatChannel, const [0,2]}
          Angle {type=MovieSceneScriptingFloatChannel, const [0,2]}
          Scale.X {type=MovieSceneScriptingFloatChannel, default=1}
          Scale.Y {type=MovieSceneScriptingFloatChannel, default=1}
          Shear.X {type=MovieSceneScriptingFloatChannel, const [0,2]}
          Shear.Y {type=MovieSceneScriptingFloatChannel, const [0,2]}
         ]
        }]
      }
     ]
    }
    WBP_PilgrimAccuracyText_Instance_C_29 {
     id=#8
     possessed_class=WBP_PilgrimAccuracyText_Instance_C
     tracks [
      {
       type=MovieSceneFloatTrack
       display_name=Render Opacity
       sections [{type=MovieSceneFloatSection, end_frame=30, start_frame=0, channels [None {type=MovieSceneScriptingFloatChannel, keys [[0,1],[30,0]]}]}]
      }
      {
       type=MovieScene2DTransformTrack
       sections [{
         type=MovieScene2DTransformSection
         end_frame=30
         start_frame=0
         channels [
          Translation.X {type=MovieSceneScriptingFloatChannel, const [0,1]}
          Translation.Y {type=MovieSceneScriptingFloatChannel, const [0,1]}
          Angle {type=MovieSceneScriptingFloatChannel, const [0,1]}
          Scale.X {type=MovieSceneScriptingFloatChannel, keys [[0,1],[2,1.1],[30,1]]}
          Scale.Y {type=MovieSceneScriptingFloatChannel, keys [[0,1],[2,1.1],[30,1]]}
          Shear.X {type=MovieSceneScriptingFloatChannel, const [0,1]}
          Shear.Y {type=MovieSceneScriptingFloatChannel, const [0,1]}
         ]
        }]
      }
      {
       type=MovieSceneByteTrack
       display_name=Visibility
       sections [{type=MovieSceneByteSection, end_frame=30, start_frame=0, channels [None {type=MovieSceneScriptingByteChannel, keys [[0,3],[30,2]]}]}]
      }
     ]
    }
   ]
   display_rate {numerator=30}
  }
 }
]
graphs [
 EventGraph {nodes [
   n1 = Event PreConstruct()
   > SetAccuracyVisuals!(
    execute
    self
    ShearX:double=Get ShearX() -> (ShearX:double)
    IsDesignTime:bool=@n1.IsDesignTime
    IsLocalMP:bool=false
    IsDetailedFeedback:bool=false
    InAccuracy:E_GemHitAccuracy Enum=NewEnumerator0
   ) -> (then, Accuracy:E_GemHitAccuracy Enum=NewEnumerator0)
   Event OnAnimationFinished!() -> (OutputDelegate, then, Animation:Widget Animation Object Reference) {is=title K2Node_Event}
   > SetVisibility(InVisibility=Hidden)
   Event Tick()
   > Branch(Condition=IsValid(Object=Get GridSlot() -> (GridSlot:Grid Slot Object Reference)))
   n2 = int + int(
    A=Get GridSlotLayer() -> (GridSlotLayer:int=0)
    B=Select!(
     NewEnumerator0:int=0
     NewEnumerator1:int=10
     NewEnumerator2:int=0
     Index:E_GemHitAccuracy Enum=Get HitAccuracy() -> (HitAccuracy:E_GemHitAccuracy Enum=NewEnumerator0)
    ) -> (ReturnValue:int)
   )
   > Branch(Condition="Not Equal (Integer)"(A=Get LastGridSlotLayer() -> (LastGridSlotLayer:int=0), B=@n2))
   > n3 = Set LastGridSlotLayer(LastGridSlotLayer:int=@n2) -> (Output_Get:int=0)
   > SetLayer!(execute, self:Grid Slot Object Reference=Get GridSlot(), InLayer:int=@n3) -> (then)
  ]}
 GetAccuracyEnum {nodes [
   n1 = GetAccuracyEnum!() -> (then, PilgrimGemHitAccuracy:"Pilgrim Gem Hit Accuracy Structure (by ref)") {is=title K2Node_FunctionEntry}
   n2 = Break Pilgrim Gem Hit Accuracy!(PilgrimGemHitAccuracy:"Pilgrim Gem Hit Accuracy Structure (by ref)"=@n1) -> (AccuracyMs:float=~, AccuracyTier:Pilgrim Accuracy Tier Structure, TierIndex:int) {is=title K2Node_BreakStruct}
   n3 = Break Pilgrim Accuracy Tier!(PilgrimAccuracyTier:"Pilgrim Accuracy Tier Structure (by ref)"=@n2.AccuracyTier) -> (ThresholdMs:float=~, ScoreMultiplier:float=~, Name:Text) {is=title K2Node_BreakStruct}
   n4 = Get PerfectText() -> (PerfectText:Text)
   n5 = "float < float"(A=@n2)
   > Return Node!(
    execute
    ReturnValue:E_GemHitAccuracy Enum=Select!(
     Option 0:E_GemHitAccuracy Enum=Select(Option 0=NewEnumerator2, Option 1=NewEnumerator0, Index=@n5)
     Option 1:E_GemHitAccuracy Enum=NewEnumerator1
     Index:bool="Equal Exactly (Text)"!(A:"Text (by ref)"=@n4, B:"Text (by ref)"=@n3.Name) -> (ReturnValue:bool)
    ) -> (ReturnValue:E_GemHitAccuracy Enum)
   )
  ]}
 SetAccuracyVisuals {nodes [
   n1 = SetAccuracyVisuals!() -> (then, ShearX:double=~, IsDesignTime:bool=~, IsLocalMP:bool=~, IsDetailedFeedback:bool=~, InAccuracy:E_GemHitAccuracy Enum=NewEnumerator0) {is=title K2Node_FunctionEntry}
   n2 = Get HitAccuracy() -> (HitAccuracy=NewEnumerator0)
   n3 = GetVisibility!(self:Widget Object Reference) -> (ReturnValue:ESlateVisibility Enum=Visible)
   n4 = Get GridSlotLayer() -> (GridSlotLayer=0)
   > n5 = Branch(Condition=AND Boolean!(
     A="Not Equal (Enum)"!(A:E_GemHitAccuracy Enum=@n1.InAccuracy, B:E_GemHitAccuracy Enum=@n2) -> (ReturnValue:bool)
     B="Not Equal (Enum)"(A=@n2, B=NewEnumerator1)
     C:bool="Not Equal (Enum)"(A:ESlateVisibility Enum=@n3, B:ESlateVisibility Enum=Hidden)
     D:bool="integer >= integer"(A=@n4, B=9)
    ) -> (ReturnValue))
   n6 = Sequence!(execute=@n5.else) -> (then_0, then_1, then_2:Exec, then_3:Exec, then_4:Exec)
   > Set IsDesign(IsDesign:bool=@n1.IsDesignTime) -> (Output_Get:bool)
   Set LocalMP(execute=@n6.then_1, LocalMP:bool=@n1.IsLocalMP) -> (Output_Get:bool)
   Set IsDetailed(execute=@n6.then_2, IsDetailed:bool=@n1.IsDetailedFeedback) -> (Output_Get:bool)
   Set HitAccuracy(execute=@n6.then_3, HitAccuracy:E_GemHitAccuracy Enum=@n1.InAccuracy) -> (Output_Get:E_GemHitAccuracy Enum=NewEnumerator0)
   n7 = Sequence!(execute=@n6.then_4) -> (then_0, then_1, then_2, then_3, then_4, then_5:Exec)
   Return Node!(execute=@n7.then_5, Accuracy:E_GemHitAccuracy Enum=Get HitAccuracy() -> (HitAccuracy=NewEnumerator0))
   > SetShear!(execute, self, ShearX:double=@n1) -> (then)
   SetHitAccuracy!(
    execute=@n7.then_3
    self:WBP Pilgrim Accuracy Icon Object Reference=Get WBP_Pilgrim_AccuracyIcon() -> (WBP_Pilgrim_AccuracyIcon:WBP Pilgrim Accuracy Icon Object Reference)
    HitAccuracy:E_GemHitAccuracy Enum=Get HitAccuracy() -> (HitAccuracy=NewEnumerator0)
    IsDesignTime:bool=Get IsDesign() -> (IsDesign:bool)
   ) -> (then)
   n8 = NOT Boolean(A=Get IsDetailed() -> (IsDetailed:bool))
   n9 = Get PreviewNoIcon() -> (PreviewNoIcon:bool)
   n10 = Get IsDesign()
   > SetVisibility(
    self=Get WBP_Pilgrim_AccuracyIcon()
    InVisibility=Select(
     Option 0:ESlateVisibility Enum=SelfHitTestInvisible
     Option 1:ESlateVisibility Enum=Collapsed
     Index=Select(Option 0:bool=@n8, Option 1:bool=@n9, Index=@n10) -> (ReturnValue:bool)
    ) -> (ReturnValue:ESlateVisibility Enum)
   )
   n11 = Branch(execute=@n7.then_4, Condition=Get IsDesign())
   n12 = Get FadeIn_NoCascade() -> (FadeIn_NoCascade:Widget Animation Object Reference)
   SetAnimationCurrentTime!(execute=@n11.else, self:User Widget Object Reference, InAnimation:Widget Animation Object Reference=@n12, InTime:float=0) -> (then)
   > PlayAnimation(InAnimation=@n12, bRestoreState=true)
   SetBrushFromMaterial(
    execute=@n7.then_2
    self=Get Background() -> (Background:Image Object Reference)
    Material=Select!(
     NewEnumerator0:Material Interface Object Reference=Get EarlyMaterial_New() -> (EarlyMaterial_New:Material Instance Object Reference)
     NewEnumerator1:Material Interface Object Reference=Get PerfectMaterial_New() -> (PerfectMaterial_New:Material Instance Object Reference)
     NewEnumerator2:Material Interface Object Reference=Get LateMaterial_New() -> (LateMaterial_New:Material Instance Object Reference)
     Index:E_GemHitAccuracy Enum=Get HitAccuracy() -> (HitAccuracy=NewEnumerator0)
    ) -> (ReturnValue:Material Interface Object Reference)
   )
   n13 = Sequence!(execute=@n7.then_1) -> (then_0, then_1)
   n14 = Get HitAccuracy() -> (HitAccuracy=NewEnumerator0)
   > SetActiveWidgetIndex(
    self:Common Visibility Switcher Object Reference=Get TextSwitcher() -> (TextSwitcher:Common Visibility Switcher Object Reference)
    Index=Select!(Option 0:int=1, Option 1:int=0, Index:bool="Equal (Enum)"(A:E_GemHitAccuracy Enum=@n14, B:E_GemHitAccuracy Enum=NewEnumerator1)) -> (ReturnValue:int)
   )
   n15 = Get PreviewNoText() -> (PreviewNoText:bool)
   n16 = Get IsDesign()
   n17 = Get LocalMP() -> (LocalMP:bool)
   > SetVisibility(
    self=[
     Get GoodAccuracyText() -> (GoodAccuracyText:Common Text Object Reference)
     Get PerfectAccuracyText() -> (PerfectAccuracyText:Common Text Object Reference)
    ]
    InVisibility=Select(
     Option 0:ESlateVisibility Enum=SelfHitTestInvisible
     Option 1:ESlateVisibility Enum=Collapsed
     Index=Select(Option 0:bool=@n17, Option 1:bool=@n15, Index=@n16) -> (ReturnValue:bool)
    ) -> (ReturnValue:ESlateVisibility Enum)
   )
   n18 = "Equal (Enum)"(A=Get HitAccuracy() -> (HitAccuracy=NewEnumerator0), B=NewEnumerator1)
   n19 = Select(Option 0:ESlateVisibility Enum=Hidden, Option 1:ESlateVisibility Enum=SelfHitTestInvisible, Index=@n18) -> (ReturnValue:ESlateVisibility Enum)
   SetVisibility(execute=@n13.then_1, self=Get PerfectIcon() -> (PerfectIcon:Fort Lazy Image Object Reference), InVisibility=@n19)
   > SetVisibility(self=Get BG_Echo() -> (BG_Echo:Image Object Reference), InVisibility=@n19)
   n20 = Get IsDesign()
   n21 = Get IsDetailed()
   n22 = Get PreviewNewVisuals() -> (PreviewNewVisuals:bool)
   Select(
    Option 0=Hidden
    Option 1=SelfHitTestInvisible
    Index=AND Boolean!(A=@n18, B=Select(Option 0:bool=@n21, Option 1:bool=@n22, Index=@n20) -> (ReturnValue:bool)) -> (ReturnValue)
   )
   Select(
    Option 0:Widget Animation Object Reference=Get FadeIn_NoCascade()
    Option 1:Widget Animation Object Reference=Select!(
     NewEnumerator0:Widget Animation Object Reference=Get Early_LiveMimic() -> (Early_LiveMimic:Widget Animation Object Reference)
     NewEnumerator1:Widget Animation Object Reference=Get Perfect_LiveMimic() -> (Perfect_LiveMimic:Widget Animation Object Reference)
     NewEnumerator2:Widget Animation Object Reference=Get Late_LiveMimic() -> (Late_LiveMimic:Widget Animation Object Reference)
     Index:E_GemHitAccuracy Enum=Get HitAccuracy() -> (HitAccuracy=NewEnumerator0)
    ) -> (ReturnValue)
    Index=GetUseHitFeedbackCascade!() -> (ReturnValue:bool)
   ) -> (ReturnValue:Widget Animation Object Reference)
  ]}
 SetGridSlot {nodes [
   n1 = SetGridSlot!() -> (then, GridSlot:Grid Slot Object Reference) {is=title K2Node_FunctionEntry}
   > Set GridSlot(GridSlot:Grid Slot Object Reference=@n1) -> (Output_Get:Grid Slot Object Reference)
  ]}
 SetShear {nodes [
   n1 = SetShear!() -> (then, ShearX:double=~) {is=title K2Node_FunctionEntry}
   > n2 = Set ShearX(ShearX:double=@n1) -> (Output_Get:double)
   > SetRenderShear!(execute, self:Widget Object Reference=[Get Background(), Get BG_Echo()], Shear_X:double=@n2, Shear_Y:double=0) -> (then)
  ]}
]
variables [
 EarlyMaterial_New {type=MaterialInstance, default=/PilgrimCore/Widgets/Materials/GemHitFeedbackMaterials/MI_UI_Pilgrim_GoodFeedback_Early}
 GridSlot {type=GridSlot, default=null, class=~}
 GridSlotLayer {type=integer}
 HitAccuracy {type=E_GemHitAccuracy, default=Perfect}
 LastGridSlotLayer {type=integer}
 LateMaterial_New {type=MaterialInstance, default=/PilgrimCore/Widgets/Materials/GemHitFeedbackMaterials/MI_UI_Pilgrim_GoodFeedback_Late}
 PerfectMaterial_New {type=MaterialInstance, default=/PilgrimCore/Widgets/Materials/GemHitFeedbackMaterials/MI_UI_Pilgrim_PerfectFeedback}
 PreviewNewVisuals {type=boolean}
 PreviewNoIcon {type=boolean}
 PreviewNoText {type=boolean}
 ShearX {type=number}
]
widgets {
 info {inheritedWidgetCount=0, parentClass=&UserWidget, rootWidgetClass=&GridPanel, widgetCount=10}
 widgets [
  AccuracyBanner {
   widgetClassPath=&GridPanel
   widget=AccuracyBanner
   properties {bOverride_Cursor=true, pixelSnapping=Disabled, visibility=HitTestInvisible, renderTransformPivot=<1 0.5>}
  }
  BG_Echo {
   widgetClassPath=&Image
   bIsVariable=true
   parent=AccuracyBanner
   widget=BG_Echo
   properties {
    brush {
     bIsDynamicallyLoaded=false
     drawAs=Image
     imageType=FullColor
     mirroring=NoMirror
     resourceName=None
     resourceObject=/PilgrimCore/Widgets/Materials/MI_UI_Pilgrim_PerfectFeedback_Echo
     tiling=NoTile
     imageSize=<0 0>
     margin=<ltrb 0 0 0 0>
     outlineSettings {
      bUseBrushTransparency=false
      roundingType=HalfHeightRadius
      width=0
      color {colorUseRule=UseColor_Specified, specifiedColor=<rgba 0 0 0 0>}
      cornerRadii=<0 0 0 0>
     }
     tintColor {colorUseRule=UseColor_Specified, specifiedColor=<rgba 1 1 1 1>}
     uVRegion {bIsValid=false, max=<0 0>, min=<0 0>}
    }
    renderTransform {angle=0, scale=<-1 1>, shear=<-24 0>, translation=<5 0>}
   }
   slot {class=GridSlot}
  }
  Background {
   widgetClassPath=&Image
   bIsVariable=true
   parent=AccuracyBanner
   widget=Background
   properties {
    visibility=SelfHitTestInvisible
    brush {
     bIsDynamicallyLoaded=false
     drawAs=Image
     imageType=FullColor
     mirroring=NoMirror
     resourceName=None
     resourceObject=&MI_UI_Pilgrim_PerfectFeedback
     tiling=NoTile
     imageSize=<0 0>
     margin=<ltrb 0 0 0 0>
     outlineSettings {
      bUseBrushTransparency=false
      roundingType=HalfHeightRadius
      width=0
      color {colorUseRule=UseColor_Specified, specifiedColor=<rgba 0 0 0 0>}
      cornerRadii=<0 0 0 0>
     }
     tintColor {colorUseRule=UseColor_Specified, specifiedColor=<rgba 1 1 1 1>}
     uVRegion {bIsValid=false, max=<0 0>, min=<0 0>}
    }
    renderTransform {angle=0, scale=<-1 1>, shear=<-24 0>, translation=<0 0>}
   }
   slot {class=GridSlot}
  }
  Contents {
   widgetClassPath=&HorizontalBox
   parent=AccuracyBanner
   widget=Contents
   slot {class=GridSlot, properties {horizontalAlignment=HAlign_Right, padding=<ltrb 15 5 20 5>}}
  }
  TextSwitcher {
   widgetClassPath=&CommonVisibilitySwitcher
   bIsVariable=true
   parent=Contents
   widget=TextSwitcher
   properties {bOverride_Cursor=true, shownVisibility=HitTestInvisible}
   slot {class=HorizontalBoxSlot, properties {padding=<ltrb 0 9 0 9>}}
  }
  PerfectAccuracyText {
   widgetClassPath=&CommonTextBlock
   bIsVariable=true
   parent=TextSwitcher
   widget=PerfectAccuracyText
   properties {
    justification=Right
    text=Perfect
    textTransformPolicy=ToUpper
    font {
     bForceMonospaced=false
     bMaterialIsStencil=false
     fontMaterial=null
     fontObject=&HeadingNow
     letterSpacing=-40
     monospacedWidth=1
     size=15
     skewAmount=0
     typefaceFontName=86 Bold Italic
     outlineSettings {
      bApplyOutlineToDropShadows=false
      bMiteredCorners=false
      bSeparateFillAlpha=false
      outlineMaterial=null
      outlineSize=0
      outlineColor=<rgba 0 0 0 1>
     }
    }
    renderTransformPivot=<1 0.5>
    shadowOffset=<0 0>
    strikeBrush {
     bIsDynamicallyLoaded=false
     drawAs=NoDrawType
     imageType=NoImage
     mirroring=NoMirror
     resourceName=None
     resourceObject=null
     tiling=NoTile
     imageSize=<32 32>
     margin=<ltrb 0 0 0 0>
     outlineSettings {
      bUseBrushTransparency=false
      roundingType=HalfHeightRadius
      width=0
      color {colorUseRule=UseColor_Specified, specifiedColor=<rgba 0 0 0 0>}
      cornerRadii=<0 0 0 0>
     }
     tintColor {colorUseRule=UseColor_Specified, specifiedColor=<rgba 1 1 1 1>}
     uVRegion {bIsValid=false, max=<0 0>, min=<0 0>}
    }
   }
   slot {class=CommonVisibilitySwitcherSlot}
  }
  GoodAccuracyText {
   widgetClassPath=&CommonTextBlock
   bIsVariable=true
   parent=TextSwitcher
   widget=GoodAccuracyText
   properties {
    justification=Right
    text=Good
    textTransformPolicy=ToUpper
    visibility=Hidden
    font {
     bForceMonospaced=false
     bMaterialIsStencil=false
     fontMaterial=null
     fontObject=&HeadingNow
     letterSpacing=-40
     monospacedWidth=1
     size=15
     skewAmount=0
     typefaceFontName=86 Bold Italic
     outlineSettings {
      bApplyOutlineToDropShadows=false
      bMiteredCorners=false
      bSeparateFillAlpha=false
      outlineMaterial=null
      outlineSize=0
      outlineColor=<rgba 0 0 0 1>
     }
    }
    renderTransformPivot=<1 0.5>
    shadowOffset=<0 0>
    strikeBrush {
     bIsDynamicallyLoaded=false
     drawAs=NoDrawType
     imageType=NoImage
     mirroring=NoMirror
     resourceName=None
     resourceObject=null
     tiling=NoTile
     imageSize=<32 32>
     margin=<ltrb 0 0 0 0>
     outlineSettings {
      bUseBrushTransparency=false
      roundingType=HalfHeightRadius
      width=0
      color {colorUseRule=UseColor_Specified, specifiedColor=<rgba 0 0 0 0>}
      cornerRadii=<0 0 0 0>
     }
     tintColor {colorUseRule=UseColor_Specified, specifiedColor=<rgba 1 1 1 1>}
     uVRegion {bIsValid=false, max=<0 0>, min=<0 0>}
    }
   }
   slot {class=CommonVisibilitySwitcherSlot}
  }
  Spacer_133 {widgetClassPath=&Spacer, parent=Contents, widget=Spacer_133, properties {size=<5 0>}, slot {class=HorizontalBoxSlot}}
  WBP_Pilgrim_AccuracyIcon {
   widgetClassPath=&/WBP_Pilgrim_AccuracyIcon
   bIsVariable=true
   parent=Contents
   widget=WBP_Pilgrim_AccuracyIcon
   slot {class=HorizontalBoxSlot}
  }
  PerfectIcon {
   widgetClassPath=&FortLazyImage
   bIsVariable=true
   parent=AccuracyBanner
   widget=PerfectIcon
   properties {
    visibility=Hidden
    brush {
     bIsDynamicallyLoaded=false
     drawAs=Image
     imageType=FullColor
     mirroring=NoMirror
     resourceName=None
     resourceObject=/PilgrimCore/Widgets/Materials/M_UI_Icon_Perfect
     tiling=NoTile
     imageSize=<45 45>
     margin=<ltrb 0 0 0 0>
     outlineSettings {
      bUseBrushTransparency=false
      roundingType=HalfHeightRadius
      width=0
      color {colorUseRule=UseColor_Specified, specifiedColor=<rgba 0 0 0 0>}
      cornerRadii=<0 0 0 0>
     }
     tintColor {colorUseRule=UseColor_Specified, specifiedColor=<rgba 1 1 1 1>}
     uVRegion {bIsValid=false, max=<0 0>, min=<0 0>}
    }
   }
   slot {class=GridSlot, properties {horizontalAlignment=HAlign_Right, verticalAlignment=VAlign_Top, nudge=<0 -20>}}
  }
 ]
}
