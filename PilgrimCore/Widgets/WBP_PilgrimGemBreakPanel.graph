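# WBP_PilgrimGemBreakPanel is a widget wrapper for Pilgrim that fades named-slot content out when a gem break starts and back in when it ends.
# BreakStart and BreakEnd guard on InBreak, then play configurable name-selected curve fades through a SimpleAnimationsPanel, which broadcasts on finish.
# Content starts transparent and in break.
# With local-multiplayer exceptions in HamSandwich sessions, breaks start only without a playing song or during ending breaks, and LMPSongStart fades content in.
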WidgetBlueprint /PilgrimCore/Widgets/WBP_PilgrimGemBreakPanel
graphs [
 BreakEnd {nodes [
   n1 = BreakEnd!() -> (then, CanFadeIn:bool=~) {is=title K2Node_FunctionEntry}
   > Branch(Condition=AND Boolean(A=@n1, B=Get InBreak() -> (InBreak:bool)))
   > Set InBreak(InBreak:bool=false) -> (Output_Get:bool)
   > PlayAnimationByName!(
    execute
    self:Fort Simple Widget Animations Panel Object Reference=Get SimpleAnimationsPanel() -> (SimpleAnimationsPanel:Fort Simple Widget Animations Panel Object Reference)
    Name:"Name (by ref)"=Get BreakEndAnimation() -> (BreakEndAnimation:Name=None)
    Speed:float=1
    StartAtTime:float=0
   ) -> (then, ReturnValue:bool)
  ]}
 BreakStart {nodes [
   n1 = BreakStart!() -> (then, CanFadeOut:bool=~) {is=title K2Node_FunctionEntry}
   n2 = Get InBreak()
   > Branch(Condition=AND Boolean(A=@n1, B=NOT Boolean(A=@n2)))
   > n3 = Branch(Condition=AND Boolean(A=Get ApplyLocalMultiplayerExceptions() -> (ApplyLocalMultiplayerExceptions:bool), B=IsHamSandwichSessionActive!() -> (ReturnValue:bool)))
   n4 = Get SongPlaying!(self:Pilgrim Currently Playing Song View Model Object Reference=Get PilgrimCurrentlyPlayingSongVM() -> (PilgrimCurrentlyPlayingSongVM:Pilgrim Currently Playing Song View Model Object Reference)) -> (SongPlaying:bool)
   n5 = Get PilgrimGemBreakVM() -> (PilgrimGemBreakVM:Pilgrim Gem Break View Model Object Reference)
   > n6 = Branch(Condition=OR Boolean(A=NOT Boolean(A=@n4), B=Get bIsInEndingGemBreak!(self:Pilgrim Gem Break View Model Object Reference=@n5) -> (bIsInEndingGemBreak:bool)))
   Set InBreak(execute=[@n6.then, @n3.else], InBreak=true)
   > PlayAnimationByName(
    self=Get SimpleAnimationsPanel()
    Name=Get BreakStartAnimation() -> (BreakStartAnimation:Name=None)
    Speed=Get FadeOutSpeed() -> (FadeOutSpeed:double)
    StartAtTime=0
   )
  ]}
 EventGraph {nodes [
   n1 = "On Animations Finished (SimpleAnimationsPanel)"!() -> (OutputDelegate, then, Name:Name=None) {is=title K2Node_ComponentBoundEvent}
   > Call On Break Animation Finished!(execute:Exec, self:WBP Pilgrim Gem Break Panel Object Reference, Name:Name=@n1.Name) -> (then:Exec) {is=title K2Node_CallDelegate}
   Event Construct()
   > SetRenderOpacity(self=Get Slot_GemBreakContent() -> (Slot_GemBreakContent:Named Slot Object Reference), InOpacity=0)
  ]}
 LMPSongStart {nodes [
   n1 = LMPSongStart!() -> (then, SongPlaying:bool=~) {is=title K2Node_FunctionEntry}
   > Branch(Condition=AND Boolean!(A=@n1, B=Get InBreak(), C:bool=Get ApplyLocalMultiplayerExceptions(), D:bool=IsHamSandwichSessionActive()) -> (ReturnValue))
   > Set InBreak(InBreak=false)
   > PlayAnimationByName(self=Get SimpleAnimationsPanel(), Name=Get BreakEndAnimation() -> (BreakEndAnimation=None), Speed=1, StartAtTime=0)
  ]}
 MakeFadeAnimationOptions {nodes [
   MakeFadeAnimationOptions() {is=title K2Node_FunctionEntry}
   > Return Node!(
    execute
    AnimOptions:Array of Names=Make Array!(
     "[0]":Name=None
     "[1]":Name=0.5sFadeIn
     "[2]":Name=1sFadeIn
     "[3]":Name=1sDelay_1sFadeIn
     "[4]":Name=0.3sFadeIn_1.7sTail
     "[5]":Name=0.5sFadeOut
     "[6]":Name=1sFadeOut
     "[7]":Name=1sFadeOut_1sTail
     "[8]":Name=1.7sDelay_0.3sFadeOut
    ) -> (Array:Array of Names)
   )
  ]}
 OnBreakAnimationFinished {nodes [OnBreakAnimationFinished!() -> (then, Name:Name) {is=title K2Node_FunctionEntry}]}
]
variables [
 ApplyLocalMultiplayerExceptions {type=boolean}
 BreakEndAnimation {type=string, default=1sDelay_1sFadeIn}
 BreakStartAnimation {type=string, default=1sFadeOut_1sTail}
 FadeOutSpeed {type=number}
 InBreak {type=boolean, default=true}
 OnBreakAnimationFinished {}
]
widgets {
 info {
  inheritedWidgetCount=0
  parentClass=&UserWidget
  rootWidgetClass=/Script/FortniteUI.FortSimpleWidgetAnimationsPanel
  widgetCount=2
 }
 widgets [
  SimpleAnimationsPanel {
   widgetClassPath=&FortSimpleWidgetAnimationsPanel
   bIsVariable=true
   widget=SimpleAnimationsPanel
   properties {animations [
     0.5sFadeIn {
      bIsEnabled=true
      onAnimationsFinished=()
      targets [{
        alpha=/PilgrimCore/Widgets/Curves/CV_0_5sFadeIn
        bResetOnFinish=false
        bShouldAnimateAlpha=true
        bShouldAnimateScale=false
        bShouldAnimateTranslation=false
        scale=null
        targetName=Slot_GemBreakContent
        translation=null
        materialScalarAnimations []
       }]
     }
     1sFadeIn {
      bIsEnabled=true
      onAnimationsFinished=()
      targets [{
        alpha=/PilgrimCore/Widgets/Curves/CV_1sFadeIn
        bResetOnFinish=false
        bShouldAnimateAlpha=true
        bShouldAnimateScale=false
        bShouldAnimateTranslation=false
        scale=null
        targetName=Slot_GemBreakContent
        translation=null
        materialScalarAnimations []
       }]
     }
     1sDelay_1sFadeIn {
      bIsEnabled=true
      onAnimationsFinished=()
      targets [{
        alpha=/PilgrimCore/Widgets/Curves/CV_1sDelay_1sFadeIn
        bResetOnFinish=false
        bShouldAnimateAlpha=true
        bShouldAnimateScale=false
        bShouldAnimateTranslation=false
        scale=null
        targetName=Slot_GemBreakContent
        translation=null
        materialScalarAnimations []
       }]
     }
     0.3sFadeIn_1.7sTail {
      bIsEnabled=true
      onAnimationsFinished=()
      targets [{
        alpha=/PilgrimCore/Widgets/Curves/CV_0_3sFadeIn_1_7sTail
        bResetOnFinish=false
        bShouldAnimateAlpha=true
        bShouldAnimateScale=false
        bShouldAnimateTranslation=false
        scale=null
        targetName=Slot_GemBreakContent
        translation=null
        materialScalarAnimations []
       }]
     }
     0.5sFadeOut {
      bIsEnabled=true
      onAnimationsFinished=()
      targets [{
        alpha=/PilgrimCore/Widgets/Curves/CV_0_5sFadeOut
        bResetOnFinish=false
        bShouldAnimateAlpha=true
        bShouldAnimateScale=false
        bShouldAnimateTranslation=false
        scale=null
        targetName=Slot_GemBreakContent
        translation=null
        materialScalarAnimations []
       }]
     }
     1sFadeOut {
      bIsEnabled=true
      onAnimationsFinished=()
      targets [{
        alpha=/PilgrimCore/Widgets/Curves/CV_1sFadeOut
        bResetOnFinish=false
        bShouldAnimateAlpha=true
        bShouldAnimateScale=false
        bShouldAnimateTranslation=false
        scale=null
        targetName=Slot_GemBreakContent
        translation=null
        materialScalarAnimations []
       }]
     }
     1sFadeOut_1sTail {
      bIsEnabled=true
      onAnimationsFinished=()
      targets [{
        alpha=/PilgrimCore/Widgets/Curves/CV_1sFadeOut_1sTail
        bResetOnFinish=false
        bShouldAnimateAlpha=true
        bShouldAnimateScale=false
        bShouldAnimateTranslation=false
        scale=null
        targetName=Slot_GemBreakContent
        translation=null
        materialScalarAnimations []
       }]
     }
     1.7sDelay_0.3sFadeOut {
      bIsEnabled=true
      onAnimationsFinished=()
      targets [{
        alpha=/PilgrimCore/Widgets/Curves/CV_1_7sDelay_0_3sFadeOut
        bResetOnFinish=false
        bShouldAnimateAlpha=true
        bShouldAnimateScale=false
        bShouldAnimateTranslation=false
        scale=null
        targetName=Slot_GemBreakContent
        translation=null
        materialScalarAnimations []
       }]
     }
    ]}
  }
  Slot_GemBreakContent {
   widgetClassPath=&NamedSlot
   bIsVariable=true
   parent=SimpleAnimationsPanel
   widget=Slot_GemBreakContent
   slot {class=FortSimpleWidgetAnimationsSlot}
  }
 ]
}
