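# Pilgrim HUD widget showing the local player's note streak as glowing text over a flame background.
# Score updates set the text and pulse a toast animation at thresholds built from the streak start, then every 100 notes; streak drops reset the threshold index.
# Streak-derived intensity drives outline color, glow and flame material parameters, and layer opacity, refreshed every tick.
# An event-router listener sets widget opacity from the track's secondary-elements phase.
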WidgetBlueprint /PilgrimCore/Widgets/WBP_PllgrimCore_StreakCounter
animations [
 A_NewToast {
  binding_count=1
  length_seconds=0.5035
  object=&self:A_NewToast
  sequence {
   bindings [Overlay_20 {
     id=#1
     possessed_class=Overlay
     tracks [{
       type=MovieScene2DTransformTrack
       sections [{
         type=MovieScene2DTransformSection
         end_frame=60
         start_frame=0
         channels [
          Scale.X {type=MovieSceneScriptingFloatChannel, keys [[0,1],[3.624,1.2],[24,1.2],[60.414,1]]}
          Scale.Y {type=MovieSceneScriptingFloatChannel, keys [[0,1],[3.624,1.2],[24,1.2],[60.414,1]]}
         ]
        }]
      }]
    }]
   display_rate {numerator=120}
  }
 }
 Hit {
  binding_count=1
  length_seconds=0.3
  object=&self:Hit
  sequence {
   bindings [Overlay_20 {
     id=#1
     possessed_class=Overlay
     tracks [{
       type=MovieScene2DTransformTrack
       sections [{
         type=MovieScene2DTransformSection
         end_frame=36
         start_frame=0
         channels [
          Scale.X {type=MovieSceneScriptingFloatChannel, keys [[0,1],[3.6,1.2],[36,1]]}
          Scale.Y {type=MovieSceneScriptingFloatChannel, keys [[0,1],[3.6,1.2],[36,1]]}
         ]
        }]
      }]
    }]
   display_rate {numerator=120}
  }
 }
]
graphs [
 EventGraph {nodes [
   ResetStreakThresholds() {is=title K2Node_CustomEvent}
   > Set CurrentToastThresholdInc(CurrentToastThresholdInc:int=0) -> (Output_Get:int=0)
   ShowNewStreakToast!() -> (OutputDelegate, then, StreakValue:int) {is=title K2Node_CustomEvent}
   > "Play Animation with Finished event (legacy)"!(
    execute:Exec
    Widget:User Widget Object Reference=Self-Reference()
    InAnimation:Widget Animation Object Reference=Get A_NewToast() -> (A_NewToast:Widget Animation Object Reference)
    StartAtTime:float=0
    NumLoopsToPlay:int=1
    PlayMode:EUMGSequencePlayMode Enum
    PlaybackSpeed:float=1
   ) -> (then:Exec, Result:UMGSequence Player Object Reference, Finished:Exec) {is=title K2Node_PlayAnimation}
   Event OnInitialized()
   > Parent: OnInitialized() {is=title K2Node_CallParentFunction}
   > n1 = PilgrimWaitForEventRouter!(execute:Exec, ContextActor:Actor Object Reference=GetOwningPlayer()) -> (then:Exec, OnEventRouterAvailable:Exec, EventRouter:Gameplay Event Router Component Object Reference) {is=title K2Node_AsyncAction}
   Is Valid(exec=@n1.OnEventRouterAvailable, InputObject=@n1)
   > n2 = StartListeningToEvent!(
    execute:Exec
    Target:Gameplay Event Router Component Object Reference=@n1
    EventType:Script Struct Object Reference=/Script/PilgrimCoreRuntime.PilgrimEvent_PilgrimTrack_SecondaryElementsPhaseChanged
    Context:Object Reference=GetOwningPlayer()
    EventBubblingRule:EEventBubblingRule Enum
   ) -> (
    then:Exec
    AsyncTaskProxy:Async Action Start Listening To Event Object Reference
    OnEventReceived:Exec
    Payload:Pilgrim Event Pilgrim Track Secondary Elements Phase Changed Structure
   )
   > Add(
    TargetArray:Array of Async Action Start Listening To Event Object References=Get GameplayEventHandles() -> (GameplayEventHandles:Array of Async Action Start Listening To Event Object References)
    NewItem:"Async Action Start Listening To Event Object Reference (by ref)"=@n2
   ) {is=title K2Node_CallArrayFunction}
   SetRenderOpacity(
    execute=@n2.OnEventReceived
    InOpacity=Break Pilgrim Event Pilgrim Track Secondary Elements Phase Changed!(PilgrimEvent_PilgrimTrack_SecondaryElementsPhaseChanged:"Pilgrim Event Pilgrim Track Secondary Elements Phase Changed Structure (by ref)"=@n2.Payload) -> (Track:Pilgrim Track Object Reference, SecondaryElementsPhase:float=~) {is=title K2Node_BreakStruct}.SecondaryElementsPhase
   )
   Event Construct()
   > Parent: Construct()
   > SetupVisuals()
   > Set Text(Text="LOCTEXT(\"0\")", self=Get TextGlow() -> (TextGlow:Common Text Object Reference))
   > SetupInitialThreshold()
   > Set Final Repeating Increment(Final Repeating Increment:int=100) -> (Output_Get:int=0)
   n3 = Event Tick()
   > Parent: Tick!(execute, MyGeometry:Geometry Structure=@n3.MyGeometry, InDeltaTime:float=@n3.InDeltaTime) -> (then) {is=title K2Node_CallParentFunction}
   > UpdateVisuals()
   Event Destruct()
   > n4 = For Each Loop(Array:Array of Async Action Start Listening To Event Object References=Get GameplayEventHandles()) -> (Array Element:Async Action Start Listening To Event Object Reference)
   > StopListeningToEvent(self:Async Action Start Listening To Event Object Reference=@n4)
   n5 = Event OnScoreDataReceivedLocalPlayer!() -> (OutputDelegate, then, GameProgress:"Pilgrim Game Progress Structure (by ref)") {is=title K2Node_Event}
   > n6 = Sequence()
   n7 = Get ToastThresholds() -> (ToastThresholds:Array of Integers)
   > n8 = Branch(Condition="integer <= integer"(A=Get CurrentToastThresholdInc() -> (CurrentToastThresholdInc:int=0), B=Last Index!(TargetArray:Array of Integers=@n7) -> (ReturnValue=0)))
   n9 = Break Pilgrim Game Progress!(PilgrimGameProgress:"Pilgrim Game Progress Structure (by ref)"=@n5.GameProgress) -> (
    NumHits:int
    NumMisses:int
    NumStreakBreakingMisses:int
    NumPasses:int
    bHasFullCombo:bool=~
    bHasPerfectCombo:bool=~
    CurrentTick:int
    NoteStreak:int
    CurrentScore:float=~
    RawCurrentScore:float=~
    CurrentModifierBonus:float=~
    CurrentStar:int
    NextStarProgress:float=~
    StreakScoreMultiplier:int
    NotesPastLastMultiplier:int
    NotesToNextMultiplier:int
    CurrentHealth:float=~
    IsOverdriveActive:bool=~
    CurrentOverdrive:float=~
    CurrentOverdriveSectionNum:int
    BandScoreContribution:float=~
    BandModifierBonusContribution:float=~
    OverdriveScoringState:EOverdriveScoringState Enum
    MaxScore:float=~
    RawMaxScore:float=~
    MaxStreakLength:int
    AverageStreakMultiplierPerHitOrPass:float=~
    bIsAtMaxStreak:bool=~
    CurrentSectionCompletionPercentage:float=~
   ) {is=title K2Node_BreakStruct}
   > Update Streak!(
    execute
    self
    CurrentNoteStreak:int=@n9.NoteStreak
    NoteThreshold:"Integer (by ref)"="Get (a ref)"(Array=Get ToastThresholds(), Dimension 1=Get CurrentToastThresholdInc() -> (CurrentToastThresholdInc=0))
   ) -> (then)
   n10 = Get ToastThresholds()
   n11 = Last Index(TargetArray=@n10) -> (ReturnValue=0)
   n12 = int - int(A=Get CurrentToastThresholdInc() -> (CurrentToastThresholdInc=0), B=@n11)
   n13 = Get Final Repeating Increment() -> (Final Repeating Increment:int=0)
   Update Streak(
    execute=@n8.else
    CurrentNoteStreak=@n9.NoteStreak
    NoteThreshold=int + int(A="Get (a ref)"(Array=@n10, Dimension 1=@n11), B=int * int(A=@n13, B=@n12))
   )
   Set Last Note Streak(execute=@n6.then_1, Last Note Streak:int=@n9.NoteStreak) -> (Output_Get:int=0)
  ]}
 SetFinalRepeatIncrement {nodes [
   SetFinalRepeatIncrement() {is=title K2Node_FunctionEntry}
   n1 = Get ToastThresholds()
   n2 = Last Index(TargetArray=@n1) -> (ReturnValue=0)
   n3 = int - int(A=@n2)
   > Set Final Repeating Increment(Final Repeating Increment=int - int(A="Get (a ref)"(Array=@n1, Dimension 1=@n2), B="Get (a ref)"(Array=@n1, Dimension 1=@n3))) -> (Output_Get=0)
  ]}
 SetupInitialThreshold {nodes [
   SetupInitialThreshold() {is=title K2Node_FunctionEntry}
   n1 = GetStreakCountStart!() -> (ReturnValue:int=0)
   > Branch(Condition="Not Equal (Integer)"(A="Absolute (Integer)"!(A:int=@n1) -> (ReturnValue:int=0), B=Get LastStartThreshold() -> (LastStartThreshold:int=0)))
   > Set LastStartThreshold(LastStartThreshold:int="Absolute (Integer)"(A=GetStreakCountStart() -> (ReturnValue=0)) -> (ReturnValue=0)) -> (Output_Get:int=0)
   > Clear(TargetArray:Array of Integers=Get ToastThresholds())
   > n2 = Sequence()
   n3 = Get LastStartThreshold() -> (LastStartThreshold=0)
   n4 = Select(Option 0:int=@n3, Option 1:int=0, Index="integer <= integer"(A=@n3)) -> (ReturnValue:int)
   n5 = Get NewThresholdsList() -> (NewThresholdsList:Set of Integers)
   > Add!(execute, TargetSet:Set of Integers=@n5, NewItem:"Integer (by ref)"=@n4) -> (then) {is=title K2Node_CallFunction}
   n6 = float / float(A=float + float(A="To Float (Integer)"(InInt=@n4), B=1), B=100)
   > Add(TargetSet=@n5, NewItem=int * int(A=Ceil(A=@n6) {is=title K2Node_CallFunction}, B=100))
   n7 = To Array!(execute=@n2.then_1, A:Set of Integers=Get NewThresholdsList()) -> (then, Result:Array of Integers)
   > Set ToastThresholds(ToastThresholds:Array of Integers=@n7) -> (Output_Get:Array of Integers)
  ]}
 SetupVisuals {nodes [
   SetupVisuals() {is=title K2Node_FunctionEntry}
   > n1 = Sequence!(execute) -> (then_0, then_1, then_2:Exec)
   n2 = Get Background() -> (Background:Image Object Reference)
   > Set BackgroundDynamicMaterial(BackgroundDynamicMaterial:Material Instance Dynamic Object Reference=GetDynamicMaterial!(Brush:"Slate Brush Structure (by ref)"=Get Brush!(self:Image Object Reference=@n2) -> (Brush:Slate Brush Structure)) -> (ReturnValue)) -> (Output_Get:Material Instance Dynamic Object Reference)
   > SetRenderOpacity(self=@n2, InOpacity=0)
   n3 = GetDynamicOutlineMaterial!(execute=@n1.then_1, self:Text Object Reference=Get TextGlow()) -> (then, ReturnValue:Material Instance Dynamic Object Reference)
   > Set TextGlowDynamicMaterial(TextGlowDynamicMaterial:Material Instance Dynamic Object Reference=@n3) -> (Output_Get:Material Instance Dynamic Object Reference)
  ]}
 Update Streak {nodes [
   n1 = Update Streak!() -> (then, CurrentNoteStreak:int, NoteThreshold:"Integer (by ref)") {is=title K2Node_FunctionEntry}
   > Branch(Condition="Not Equal (Integer)"(A=@n1, B=Get CurrentStreakValue() -> (CurrentStreakValue:int=0)))
   > Set CurrentStreakValue(CurrentStreakValue:int=@n1) -> (Output_Get:int=0)
   > Set NoteStreakThreshold(NoteStreakThreshold:int=@n1.NoteThreshold) -> (Output_Get:int=0)
   > n2 = Sequence()
   n3 = Branch(
    execute=@n2.then_2
    Condition="integer >= integer"(A=Get CurrentStreakValue() -> (CurrentStreakValue=0), B=Get NoteStreakThreshold() -> (NoteStreakThreshold:int=0))
   )
   > ShowNewStreakToast!(execute, self, StreakValue:int=Get NoteStreakThreshold() -> (NoteStreakThreshold=0)) -> (then) {is=title K2Node_CallFunction}
   > Increment Int(Value=Get CurrentToastThresholdInc() -> (CurrentToastThresholdInc=0))
   Branch(
    execute=@n3.else
    Condition="integer > integer"(A=Get Last Note Streak() -> (Last Note Streak:int=0), B=Get CurrentStreakValue() -> (CurrentStreakValue=0))
   )
   > ResetStreakThresholds() {is=title K2Node_CallFunction}
   n4 = GetStreakCountStart() -> (ReturnValue=0)
   n5 = "To Float (Integer)"(InInt=int - int(A=Get CurrentStreakValue() -> (CurrentStreakValue=0), B=int - int(A=@n4)))
   Set VisualIntensity(execute=@n2.then_1, VisualIntensity:double="Clamp (Float)"(Value=float / float(A=@n5, B=100), Min=0, Max=1)) -> (Output_Get:double)
   > UpdateVisuals()
   Update Streak Text!(execute=@n2.then_0, self, Streak:int=Get CurrentStreakValue() -> (CurrentStreakValue=0)) -> (then)
  ]}
 Update Streak Text {nodes [
   n1 = Update Streak Text!() -> (then, Streak:int) {is=title K2Node_FunctionEntry}
   > n2 = Branch(Condition="Not Equal (Integer)"(A=@n1))
   > n3 = Set Text!(execute, Text:String="To String (Text)"(InText="To Text (Integer)"(Value=@n1, bUseGrouping=false, MaximumIntegralDigits=10))) -> (then, Output_Get:String)
   Set Text!(execute=[@n2.else, @n3.then], Text:Text="To Text (String)"(InString=@n3), self:Text Object Reference=Get TextGlow()) -> (then, Output_Get:Text)
  ]}
 UpdateVisuals {nodes [
   UpdateVisuals() {is=title K2Node_FunctionEntry}
   > n1 = Sequence!(execute) -> (then_0, then_1, then_2, then_3:Exec, then_4:Exec)
   SetScalarParameterValue(
    execute=@n1.then_4
    self=Get BackgroundDynamicMaterial() -> (BackgroundDynamicMaterial:Material Instance Dynamic Object Reference)
    ParameterName=GlowToFire
    Value=Get TextGlowOpacity() -> (TextGlowOpacity:double)
   )
   n2 = Get TextMin() -> (TextMin:double)
   n3 = float - float(A=Get VisualIntensity() -> (VisualIntensity:double), B=@n2)
   n4 = float - float(A=1, B=@n2)
   Set TextGlowOpacity(execute=@n1.then_0, TextGlowOpacity:double="Clamp (Float)"(Value=float / float(A=@n3, B=@n4), Min=0, Max=1)) -> (Output_Get:double)
   n5 = Select(Option 0:float=0, Option 1:float=1, Index="float > float"(A=Get VisualIntensity())) -> (ReturnValue:float)
   n6 = Get BaseOpacity() -> (BaseOpacity:double)
   > SetRenderOpacity(self=Get TextGlow(), InOpacity="Clamp (Float)"(Value=float + float(A=@n5, B=@n6), Min=0, Max=1))
   n7 = Get BackgroundMin() -> (BackgroundMin:double)
   n8 = float - float(A=Get VisualIntensity(), B=@n7)
   n9 = float - float(A=1, B=@n7)
   Set BackgroundOpacity(execute=@n1.then_1, BackgroundOpacity:double="Clamp (Float)"(Value=float / float(A=@n8, B=@n9), Min=0, Max=1)) -> (Output_Get:double)
   n10 = Get VisualIntensity()
   n11 = Get BackgroundMin()
   > SetRenderOpacity(self=Get Background(), InOpacity=Select(Option 0=0, Option 1=1, Index="float >= float"(A=@n10, B=@n11)))
   Set Outline Color!(
    In:Exec=@n1.then_3
    CommonTextWidget:Common Text Object Reference=Get TextGlow()
    Color:Linear Color Structure="Lerp (LinearColor)"(
     A=Get StartColor() -> (StartColor:Linear Color Structure)
     B=Get EndColor() -> (EndColor:Linear Color Structure)
     Alpha=Get TextGlowOpacity()
    )
   ) -> (Out:Exec) {is=title K2Node_MacroInstance}
   SetScalarParameterValue(
    execute=@n1.then_2
    self=Get TextGlowDynamicMaterial() -> (TextGlowDynamicMaterial:Material Instance Dynamic Object Reference)
    ParameterName=Range
    Value=Get TextGlowOpacity()
   )
  ]}
]
variables [
 AsyncEventHandles {type="array<CancellableAsyncAction>", default []}
 BackgroundDynamicMaterial {type=MaterialInstanceDynamic}
 BaseOpacity {type=number}
 CurrentStreakValue {type=integer}
 CurrentToastThresholdInc {type=integer}
 DebugStreakCountPositionIndex {type=integer}
 EndColor {type=LinearColor, default=<rgba 0.6308 0.0395 0 1>}
 Event Router {type=GameplayEventRouterComponent, default=null}
 Final Repeating Increment {type=integer}
 IsSetup {type=boolean}
 Last Note Streak {type=integer}
 LastStartThreshold {type=integer, default=-1}
 StartColor {type=LinearColor, default=<rgba 0.6308 0.1384 0 1>}
 StreakCountDebugContext {type=integer}
 TextGlowDynamicMaterial {type=MaterialInstanceDynamic}
 ToastThresholds {type="array<integer>", default [50,100]}
 VisualIntensity {type=number}
]
widgets {
 info {inheritedWidgetCount=0, parentClass=/Script/PilgrimCoreUI.PilgrimGameWidgetBase, rootWidgetClass=&Overlay, widgetCount=3}
 widgets [
  Overlay_20 {
   widgetClassPath=&Overlay
   widget=Overlay_20
   properties {renderTransform {angle=0, scale=<0.777 0.777>, shear=<0 0>, translation=<0 0>}}
  }
  Background {
   widgetClassPath=&Image
   bIsVariable=true
   parent=Overlay_20
   widget=Background
   properties {
    visibility=SelfHitTestInvisible
    brush {
     bIsDynamicallyLoaded=false
     drawAs=Image
     imageType=FullColor
     mirroring=NoMirror
     resourceName=None
     resourceObject=/PilgrimCore/Widgets/Materials/MI_UI_Pilgrim_GlowFlame
     tiling=NoTile
     imageSize=<100 100>
     margin=<ltrb 0 0 0 0>
     outlineSettings {
      bUseBrushTransparency=false
      roundingType=HalfHeightRadius
      width=0
      color {colorUseRule=UseColor_Specified, specifiedColor=<rgba 0 0 0 0>}
      cornerRadii=<0 0 0 0>
     }
     tintColor {colorUseRule=UseColor_Specified, specifiedColor=<rgba 1 1 1 1>}
     uVRegion {bIsValid=false, max=<0 0>, min=<0 0>}
    }
   }
   slot {class=OverlaySlot, properties {horizontalAlignment=HAlign_Center, verticalAlignment=VAlign_Center, padding=<ltrb -40 0 -40 0>}}
  }
  TextGlow {
   widgetClassPath=&CommonTextBlock
   bIsVariable=true
   parent=Overlay_20
   widget=TextGlow
   properties {
    bIsScrollingEnabled=false
    justification=Center
    text=500
    font {
     bForceMonospaced=true
     bMaterialIsStencil=false
     fontMaterial=null
     fontObject=&HeadingNow
     letterSpacing=0
     monospacedWidth=1.22
     size=21
     skewAmount=0
     typefaceFontName=86 Bold Italic
     outlineSettings {
      bApplyOutlineToDropShadows=false
      bMiteredCorners=false
      bSeparateFillAlpha=false
      outlineMaterial=/PilgrimCore/Widgets/Materials/MI_Pilgrim_FontRangeGlow
      outlineSize=30
      outlineColor=<rgba 0.6308 0.1384 0 1>
     }
    }
    shadowOffset=<0 0>
    strikeBrush {
     bIsDynamicallyLoaded=false
     drawAs=Image
     imageType=FullColor
     mirroring=NoMirror
     resourceName=None
     resourceObject=null
     tiling=NoTile
     imageSize=<32 32>
     margin=<ltrb 0 0 0 0>
     outlineSettings {
      bUseBrushTransparency=false
      roundingType=HalfHeightRadius
      width=0
      color {colorUseRule=UseColor_Specified, specifiedColor=<rgba 0 0 0 0>}
      cornerRadii=<0 0 0 0>
     }
     tintColor {colorUseRule=UseColor_Specified, specifiedColor=<rgba 1 1 1 1>}
     uVRegion {bIsValid=false, max=<0 0>, min=<0 0>}
    }
   }
   slot {class=OverlaySlot, properties {horizontalAlignment=HAlign_Center, verticalAlignment=VAlign_Center}}
  }
 ]
}
